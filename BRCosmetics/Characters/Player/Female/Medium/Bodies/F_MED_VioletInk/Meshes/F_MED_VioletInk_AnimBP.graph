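# Animation Blueprint for the F_MED_VioletInk female player body, adding secondary motion to skirt, chest flaps, straps and chains.
# It copies the parent mesh pose into a Control Rig driven by movement states like crouch, slide, skydive and DBNO, then caches it.
# Chained AnimDynamics nodes simulate that pose with angular springs, gravity overrides and some spherical limits.
# The output applies additive skirt bone offsets to cached pose CRCache_SkirtDyn.
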AnimBlueprint /BRCosmetics/Characters/Player/Female/Medium/Bodies/F_MED_VioletInk/Meshes/F_MED_VioletInk_AnimBP
target_skeleton=&/F_MED_VioletInk_Skeleton
graphs [
 AnimGraph {nodes [
   n1 = Anim Dynamics - Bone: dyn_braclate_l_01!(ComponentPose, Alpha=1, GravityOverride:Vector="0,0,-980", ExternalForce="0,0,0") -> (Pose) {
    angularDampingOverride=1
    angularSpringConstant=10000
    bAngularSpring=true
    bOverrideAngularDamping=true
    bUseGravityOverride=true
    gravityScale=0.35
    simulationSpace=RootRelative
    boundBone {boneName=dyn_braclate_l_01}
    componentAppliedLinearAccClamp=<2000 2000 2000>
    componentLinearAccScale=<1 1 1>
    componentLinearVelScale=<1 1 1>
    gravityOverride {z=-980}
    physicsBodyDefinitions [{
      boundBone {boneName=dyn_braclate_l_01}
      boxExtents=<30 30 30>
      constraintSetup {angularLimitsMax=<0 0 1>, angularLimitsMin=<0 0 -1>, angularTarget=<1 0 0>, linearAxesMax=<0.5 0 0>, linearAxesMin=<-0.5 0 0>}
      localJointOffset=<4 0 0>
     }]
   }
   n2 = Anim Dynamics - Bone: base_braclate_l_02!(ComponentPose=@n1, Alpha=1, GravityOverride:Vector="0,0,-980", ExternalForce="0,0,0") -> (Pose) {
    angularDampingOverride=1
    angularSpringConstant=5000
    bAngularSpring=true
    bOverrideAngularDamping=true
    bUseGravityOverride=true
    gravityScale=0.35
    simulationSpace=RootRelative
    boundBone {boneName=base_braclate_l_02}
    componentAppliedLinearAccClamp=<2000 2000 2000>
    componentLinearAccScale=<1 1 1>
    componentLinearVelScale=<1 1 1>
    gravityOverride {z=-980}
    physicsBodyDefinitions [{
      boundBone {boneName=base_braclate_l_02}
      boxExtents=<20 30 30>
      constraintSetup {angularLimitsMax=<0 0 1>, angularLimitsMin=<0 0 -1>, angularTarget=<1 0 0>, linearAxesMax=<0.2 0 0>, linearAxesMin=<-0.2 0 0>}
      localJointOffset=<4 0 0>
     }]
   }
   n3 = Anim Dynamics - Bone: dyn_braclate_r_01!(ComponentPose=@n2, Alpha=1, GravityOverride:Vector="0,0,-980", ExternalForce="0,0,0") -> (Pose) {
    angularDampingOverride=1
    angularSpringConstant=10000
    bAngularSpring=true
    bOverrideAngularDamping=true
    bUseGravityOverride=true
    gravityScale=0.35
    simulationSpace=RootRelative
    boundBone {boneName=dyn_braclate_r_01}
    componentAppliedLinearAccClamp=<2000 2000 2000>
    componentLinearAccScale=<1 1 1>
    componentLinearVelScale=<1 1 1>
    gravityOverride {z=-980}
    physicsBodyDefinitions [{
      boundBone {boneName=dyn_braclate_r_01}
      boxExtents=<30 30 30>
      constraintSetup {angularLimitsMax=<0 0 1>, angularLimitsMin=<0 0 -1>, angularTarget=<1 0 0>, linearAxesMax=<0.5 0 0>, linearAxesMin=<-0.5 0 0>}
      localJointOffset=<4 0 0>
     }]
   }
   Anim Dynamics - Bone: dyn_braclate_r_02!(ComponentPose=@n3, Alpha=1, GravityOverride:Vector="0,0,-980", ExternalForce="0,0,0") -> (Pose) {
    angularDampingOverride=1
    angularSpringConstant=5000
    bAngularSpring=true
    bOverrideAngularDamping=true
    bUseGravityOverride=true
    gravityScale=0.35
    simulationSpace=RootRelative
    boundBone {boneName=dyn_braclate_r_02}
    componentAppliedLinearAccClamp=<2000 2000 2000>
    componentLinearAccScale=<1 1 1>
    componentLinearVelScale=<1 1 1>
    gravityOverride {z=-980}
    physicsBodyDefinitions [{
      boundBone {boneName=dyn_braclate_r_02}
      boxExtents=<20 30 30>
      constraintSetup {angularLimitsMax=<0 0 1>, angularLimitsMin=<0 0 -1>, angularTarget=<1 0 0>, linearAxesMax=<0.2 0 0>, linearAxesMin=<-0.2 0 0>}
      localJointOffset=<4 0 0>
     }]
   }
   n4 = "Constraint - base_strap_loop_01 (PPPP)"!(
    ComponentPose=Local To Component()
    Alpha=1
    ConstraintWeights_0=0.322
    ConstraintWeights_1:float=0.2635
    ConstraintWeights_2:float=0.2279
    ConstraintWeights_3:float=0.1865
   ) -> (Pose) {
    boneToModify {boneName=base_strap_loop_01}
    constraintSetup [
     {targetBone {boneName=dyn_SideSkirt_l_02}}
     {targetBone {boneName=dyn_SideSkirt_l_01}}
     {targetBone {boneName=dyn_frontSkirt_l_02}}
     {targetBone {boneName=dyn_frontSkirt_l_01}}
    ]
    constraintWeights [1,1,1,1]
   }
   Component To Local(ComponentPose=@n4)
   "Transform (Modify) Bone - Bone: belt_deform_ALTSkin"(Alpha=1, Translation="47.464,-44.349,46.646", Rotation="0,0,0", Scale="1,1,1") {
    rotationMode=BMM_Additive
    rotationSpace=BCS_BoneSpace
    translationMode=BMM_Additive
    translationSpace=BCS_BoneSpace
    alphaBoolBlend {blendOption=Linear}
    alphaScaleBiasClamp {bInterpResult=true}
    boneToModify {boneName=belt_deform_ALTSkin}
    translation=<47.464 -44.349 46.646>
   }
   n5 = "Transform (Modify) Bone - Bone: dyn_frontSkirt_m_01"!(
    ComponentPose=Local To Component(LocalPose=Use cached pose 'CRCache_SkirtDyn'())
    bAlphaBoolEnabled:bool=True
    Translation="0,0,0"
    Rotation="-6.5,0,0"
    Scale="1,1,1"
   ) -> (Pose) {
    alphaInputType=Bool
    rotationMode=BMM_Additive
    rotationSpace=BCS_BoneSpace
    translationMode=BMM_Additive
    translationSpace=BCS_BoneSpace
    alphaBoolBlend {blendInTime=0.2, blendOption=Linear, blendOutTime=0.2}
    boneToModify {boneName=dyn_frontSkirt_m_01}
    rotation {pitch=-6.5}
   }
   n6 = "Transform (Modify) Bone - Bone: dyn_BackSkirt_m_01"!(ComponentPose=@n5, bAlphaBoolEnabled:bool=True, Translation="0,0,1", Rotation="15,0,0", Scale="1,1,1") -> (Pose) {
    alphaInputType=Bool
    rotationMode=BMM_Additive
    rotationSpace=BCS_BoneSpace
    translationMode=BMM_Additive
    translationSpace=BCS_BoneSpace
    alphaBoolBlend {blendInTime=0.2, blendOption=Linear, blendOutTime=0.2}
    boneToModify {boneName=dyn_BackSkirt_m_01}
    rotation {pitch=15}
    translation {z=1}
   }
   n7 = "Transform (Modify) Bone - Bone: dyn_frontSkirt_l_01"(ComponentPose=@n6, Alpha=1, Translation="0,0,0", Rotation="-6.209,15.688,-0.558", Scale="1,1,1") {
    rotationMode=BMM_Additive
    rotationSpace=BCS_BoneSpace
    translationMode=BMM_Additive
    translationSpace=BCS_BoneSpace
    alphaBoolBlend {blendOption=Linear}
    alphaScaleBiasClamp {bInterpResult=true}
    boneToModify {boneName=dyn_frontSkirt_l_01}
    rotation=<rot -6.209 15.688 -0.558>
   }
   Output Pose(Result=Component To Local(ComponentPose=@n7)) {layerGroup=None}
   n8 = Get bIsCrouching()
   n9 = Control Rig!(
    Source=Copy Pose From Mesh() {bUseAttachedParent=true}
    Alpha
    Crouching:bool=@n8
    DBNO:bool=Get bIsDBNO()
    KneeSliding:double=Get KneeSliding() -> (KneeSliding:double)
    BaseballSlide:double=Get BaseballSlide() -> (BaseballSlide:double)
    isCrouchStrafe:double=Get isCrouchStrafe float() -> (isCrouchStrafe float:double)
    Crouching_Alpha:double=Get Couching_Alpha() -> (Couching_Alpha:double)
    Crouching_Alpha_Bag:double=0
    Idle:double=Get isIdle() -> (isIdle:double)
    Crouching_Not:bool=NOT Boolean(A=@n8)
    Forward_Move:double=Get moveForward() -> (moveForward:double)
    Backward_Move:double=Get moveBackward() -> (moveBackward:double)
    SkyDive_Float:double=Get SkyDive_Float() -> (SkyDive_Float:double)
    SkyDive_Left:double=Get SkyDive_Left() -> (SkyDive_Left:double)
    IsTacticaljumping:double=Get IsTacticaljumping() -> (IsTacticaljumping:double)
    ArmsVolume:bool=Get ArmsVolume() -> (ArmsVolume:bool)
    FrontEnd:bool=Get bIsInFrontEnd()
    TacticalSprinting:double=Get RunningAlpha() -> (RunningAlpha:double)
   ) -> (Pose) {
    alphaBoolBlend {blendOption=Linear}
    controlRigAssetReference {blueprintRigClass=&/F_MED_VioletInk_CtrlRig1.F_MED_VioletInk_CtrlRig1_C}
   }
   n10 = "Transform (Modify) Bone - Bone: dyn_frontSkirt_m_01"!(ComponentPose=Local To Component(LocalPose=@n9), Alpha:float=1, Translation="0,0,0", Rotation="0,0,0", Scale="1,1,1") -> (Pose) {
    rotationMode=BMM_Additive
    rotationSpace=BCS_BoneSpace
    translationMode=BMM_Additive
    translationSpace=BCS_BoneSpace
    alphaBoolBlend {blendOption=Linear}
    alphaScaleBiasClamp {bInterpResult=true}
    boneToModify {boneName=dyn_frontSkirt_m_01}
   }
   n11 = "Transform (Modify) Bone - Bone: dyn_spine_belt_02"(ComponentPose=@n10, Alpha=1, Translation="0,0,0", Rotation="0,0,0", Scale="1,1,1") {
    rotationMode=BMM_Additive
    rotationSpace=BCS_BoneSpace
    translationMode=BMM_Additive
    translationSpace=BCS_BoneSpace
    alphaBoolBlend {blendOption=Linear}
    alphaScaleBiasClamp {bInterpResult=true}
    boneToModify {boneName=dyn_spine_belt_02}
   }
   Save cached pose 'CRCache_Input'(Pose=Component To Local(ComponentPose=@n11)) {cacheName=CRCache_Input}
   n12 = Anim Dynamics - Bone: dyn_l_frontChest_flap_01!(
    ComponentPose=Local To Component(LocalPose=Use cached pose 'CRCache_Input'())
    Alpha=1
    GravityOverride:Vector="0,0,-980"
    ExternalForce="0,0,0"
   ) -> (Pose) {
    angularDampingOverride=1
    angularSpringConstant=10000
    bAngularSpring=true
    bOverrideAngularDamping=true
    bUseGravityOverride=true
    bUseSphericalLimits=true
    gravityScale=0.35
    simulationSpace=RootRelative
    boundBone {boneName=dyn_l_frontChest_flap_01}
    componentLinearAccScale=<1 1 1>
    componentLinearVelScale=<1 1 1>
    gravityOverride {z=-980}
    physicsBodyDefinitions [{
      boundBone {boneName=dyn_l_frontChest_flap_01}
      boxExtents=<30 30 30>
      constraintSetup {angularLimitsMax=<0 0 5>, angularLimitsMin=<0 -30 -5>, angularTarget=<1 0 0>}
      localJointOffset=<4 0 0>
     }]
    sphericalLimits [{limitRadius=8, drivingBone {boneName=spine_05}, sphereLocalOffset=<-7.6238 4.766 -7.6725>}]
   }
   n13 = Anim Dynamics - Bone: dyn_l_frontChest_flap_02!(ComponentPose=@n12, Alpha=1, GravityOverride:Vector="0,0,-980", ExternalForce="0,0,0") -> (Pose) {
    angularDampingOverride=1
    angularSpringConstant=10000
    bAngularSpring=true
    bOverrideAngularDamping=true
    bUseGravityOverride=true
    gravityScale=0.35
    simulationSpace=RootRelative
    boundBone {boneName=dyn_l_frontChest_flap_02}
    componentLinearAccScale=<1 1 1>
    componentLinearVelScale=<1 1 1>
    gravityOverride {z=-980}
    physicsBodyDefinitions [{
      boundBone {boneName=dyn_l_frontChest_flap_02}
      boxExtents=<30 30 30>
      constraintSetup {angularLimitsMax=<0 0 5>, angularLimitsMin=<0 -10 -5>, angularTarget=<1 0 0>}
      localJointOffset=<4 0 0>
     }]
   }
   n14 = Anim Dynamics - Bone: dyn_r_frontChest_flap_01!(ComponentPose=@n13, Alpha=1, GravityOverride:Vector="0,0,-980", ExternalForce="0,0,0") -> (Pose) {
    angularDampingOverride=1
    angularSpringConstant=10000
    bAngularSpring=true
    bOverrideAngularDamping=true
    bUseGravityOverride=true
    bUseSphericalLimits=true
    gravityScale=0.35
    simulationSpace=RootRelative
    boundBone {boneName=dyn_r_frontChest_flap_01}
    componentLinearAccScale=<1 1 1>
    componentLinearVelScale=<1 1 1>
    gravityOverride {z=-980}
    physicsBodyDefinitions [{
      boundBone {boneName=dyn_r_frontChest_flap_01}
      boxExtents=<30 30 30>
      constraintSetup {angularLimitsMax=<0 30 5>, angularLimitsMin=<0 0 -5>, angularTarget=<1 0 0>}
      localJointOffset=<4 0 0>
     }]
    sphericalLimits [{limitRadius=8, drivingBone {boneName=spine_05}, sphereLocalOffset=<-9.253 5.2143 7.6725>}]
   }
   n15 = Anim Dynamics - Bone: dyn_r_frontChest_flap_02!(ComponentPose=@n14, Alpha=1, GravityOverride:Vector="0,0,-980", ExternalForce="0,0,0") -> (Pose) {
    angularDampingOverride=1
    angularSpringConstant=10000
    bAngularSpring=true
    bOverrideAngularDamping=true
    bUseGravityOverride=true
    gravityScale=0.35
    simulationSpace=RootRelative
    boundBone {boneName=dyn_r_frontChest_flap_02}
    componentLinearAccScale=<1 1 1>
    componentLinearVelScale=<1 1 1>
    gravityOverride {z=-980}
    physicsBodyDefinitions [{
      boundBone {boneName=dyn_r_frontChest_flap_02}
      boxExtents=<30 30 30>
      constraintSetup {angularLimitsMax=<0 -10 5>, angularLimitsMin=<0 0 -5>, angularTarget=<1 0 0>}
      localJointOffset=<4 0 0>
     }]
   }
   n16 = Anim Dynamics - Bone: dyn_frontSkirt_m_01!(SimulationSpace:AnimPhysSimSpaceType Enum=BoneRelative, ComponentPose=@n15, Alpha=1, GravityOverride:Vector="0,0,500") -> (Pose) {
    angularDampingOverride=0.9999
    angularSpringConstant=200
    bAngularSpring=true
    bOverrideAngularDamping=true
    bUseGravityOverride=true
    bUsePlanarLimit=false
    gravityScale=0.3
    simulationSpace=BoneRelative
    alphaBoolBlend {blendOption=Linear}
    alphaScaleBiasClamp {bInterpResult=true}
    boundBone {boneName=dyn_frontSkirt_m_01}
    componentAppliedLinearAccClamp=<1000 1000 1000>
    componentLinearAccScale=<1 1 1>
    componentLinearVelScale=<1 1 1>
    gravityOverride {z=500}
    physicsBodyDefinitions [{
      sphereCollisionRadius=5
      boundBone {boneName=dyn_frontSkirt_m_01}
      boxExtents=<40 40 40>
      constraintSetup {angularLimitsMax=<0 -5 5>, angularLimitsMin=<0 -15 -5>, angularTarget=<0 0 0>}
      localJointOffset=<10 0 0>
     }]
    relativeSpaceBone {boneName=root}
   }
   n17 = Anim Dynamics - Bone: dyn_frontSkirt_m_02!(SimulationSpace:AnimPhysSimSpaceType Enum=BoneRelative, ComponentPose=@n16, Alpha=1, GravityOverride:Vector="0,0,500") -> (Pose) {
    angularDampingOverride=0.9999
    angularSpringConstant=200
    bAngularSpring=true
    bOverrideAngularDamping=true
    bUseGravityOverride=true
    bUsePlanarLimit=false
    gravityScale=0.3
    simulationSpace=BoneRelative
    alphaBoolBlend {blendOption=Linear}
    alphaScaleBiasClamp {bInterpResult=true}
    boundBone {boneName=dyn_frontSkirt_m_02}
    componentAppliedLinearAccClamp=<1000 1000 1000>
    componentLinearAccScale=<1 1 1>
    componentLinearVelScale=<1 1 1>
    gravityOverride {z=500}
    physicsBodyDefinitions [{
      sphereCollisionRadius=5
      boundBone {boneName=dyn_frontSkirt_m_02}
      boxExtents=<40 40 40>
      constraintSetup {angularLimitsMax=<0 -5 5>, angularLimitsMin=<0 -25 -5>, angularTarget=<0 0 0>}
      localJointOffset=<10 0 0>
     }]
    relativeSpaceBone {boneName=root}
   }
   n18 = Anim Dynamics - Bone: dyn_frontSkirt_m_03!(SimulationSpace:AnimPhysSimSpaceType Enum=BoneRelative, ComponentPose=@n17, Alpha=1, GravityOverride:Vector="0,0,500") -> (Pose) {
    angularDampingOverride=0.9999
    angularSpringConstant=200
    bAngularSpring=true
    bOverrideAngularDamping=true
    bUseGravityOverride=true
    bUsePlanarLimit=false
    gravityScale=0.3
    simulationSpace=BoneRelative
    alphaBoolBlend {blendOption=Linear}
    alphaScaleBiasClamp {bInterpResult=true}
    boundBone {boneName=dyn_frontSkirt_m_03}
    componentAppliedLinearAccClamp=<1000 1000 1000>
    componentLinearAccScale=<1 1 1>
    componentLinearVelScale=<1 1 1>
    gravityOverride {z=500}
    physicsBodyDefinitions [{
      sphereCollisionRadius=5
      boundBone {boneName=dyn_frontSkirt_m_03}
      boxExtents=<40 40 40>
      constraintSetup {angularLimitsMax=<0 -15 5>, angularLimitsMin=<0 -30 -5>, angularTarget=<0 0 0>}
      localJointOffset=<10 0 0>
     }]
    relativeSpaceBone {boneName=root}
   }
   n19 = Anim Dynamics - Bone: dyn_BackSkirt_m_01!(SimulationSpace:AnimPhysSimSpaceType Enum=BoneRelative, ComponentPose=@n18, Alpha=1, GravityOverride:Vector="0,0,-980") -> (Pose) {
    angularDampingOverride=0.9999
    angularSpringConstant=200
    bAngularSpring=true
    bOverrideAngularDamping=true
    bUseGravityOverride=true
    bUsePlanarLimit=false
    bUseSphericalLimits=true
    gravityScale=0.3
    simulationSpace=BoneRelative
    alphaBoolBlend {blendOption=Linear}
    alphaScaleBiasClamp {bInterpResult=true}
    boundBone {boneName=dyn_BackSkirt_m_01}
    componentAppliedLinearAccClamp=<1000 1000 1000>
    componentLinearAccScale=<1 1 1>
    componentLinearVelScale=<1 1 1>
    gravityOverride {z=-980}
    physicsBodyDefinitions [{
      sphereCollisionRadius=5
      boundBone {boneName=dyn_BackSkirt_m_01}
      boxExtents=<40 40 40>
      constraintSetup {angularLimitsMax=<0 0 5>, angularLimitsMin=<0 -15 -5>, angularTarget=<0 0 0>}
      localJointOffset=<10 0 0>
     }]
    relativeSpaceBone {boneName=root}
   }
   n20 = Anim Dynamics - Bone: dyn_BackSkirt_m_02!(SimulationSpace:AnimPhysSimSpaceType Enum=BoneRelative, ComponentPose=@n19, Alpha=1, GravityOverride:Vector="0,0,-980") -> (Pose) {
    angularDampingOverride=0.9999
    angularSpringConstant=200
    bAngularSpring=true
    bOverrideAngularDamping=true
    bUseGravityOverride=true
    bUsePlanarLimit=false
    bUseSphericalLimits=true
    gravityScale=0.3
    simulationSpace=BoneRelative
    alphaBoolBlend {blendOption=Linear}
    alphaScaleBiasClamp {bInterpResult=true}
    boundBone {boneName=dyn_BackSkirt_m_02}
    componentAppliedLinearAccClamp=<1000 1000 1000>
    componentLinearAccScale=<1 1 1>
    componentLinearVelScale=<1 1 1>
    gravityOverride {z=-980}
    physicsBodyDefinitions [{
      sphereCollisionRadius=5
      boundBone {boneName=dyn_BackSkirt_m_02}
      boxExtents=<40 40 40>
      constraintSetup {angularLimitsMax=<0 0 5>, angularLimitsMin=<0 -25 -5>, angularTarget=<0 0 0>}
      localJointOffset=<10 0 0>
     }]
    relativeSpaceBone {boneName=root}
   }
   n21 = Anim Dynamics - Bone: dyn_BackSkirt_m_03!(SimulationSpace:AnimPhysSimSpaceType Enum=BoneRelative, ComponentPose=@n20, Alpha=1, GravityOverride:Vector="0,0,-980") -> (Pose) {
    angularDampingOverride=0.9999
    angularSpringConstant=200
    bAngularSpring=true
    bOverrideAngularDamping=true
    bUseGravityOverride=true
    bUsePlanarLimit=false
    bUseSphericalLimits=true
    gravityScale=0.3
    simulationSpace=BoneRelative
    alphaBoolBlend {blendOption=Linear}
    alphaScaleBiasClamp {bInterpResult=true}
    boundBone {boneName=dyn_BackSkirt_m_03}
    componentAppliedLinearAccClamp=<1000 1000 1000>
    componentLinearAccScale=<1 1 1>
    componentLinearVelScale=<1 1 1>
    gravityOverride {z=-980}
    physicsBodyDefinitions [{
      sphereCollisionRadius=5
      boundBone {boneName=dyn_BackSkirt_m_03}
      boxExtents=<40 40 40>
      constraintSetup {angularLimitsMax=<0 0 5>, angularLimitsMin=<0 -30 -5>, angularTarget=<0 0 0>}
      localJointOffset=<10 0 0>
     }]
    relativeSpaceBone {boneName=root}
   }
   n22 = Anim Dynamics - Bone: dyn_frontSkirt_l_01!(SimulationSpace:AnimPhysSimSpaceType Enum=BoneRelative, ComponentPose=@n21, Alpha=1, GravityOverride:Vector="0,0,500") -> (Pose) {
    angularDampingOverride=0.9999
    angularSpringConstant=200
    bAngularSpring=true
    bOverrideAngularDamping=true
    bUseGravityOverride=true
    bUsePlanarLimit=false
    gravityScale=0.3
    simulationSpace=BoneRelative
    alphaBoolBlend {blendOption=Linear}
    alphaScaleBiasClamp {bInterpResult=true}
    boundBone {boneName=dyn_frontSkirt_l_01}
    componentAppliedLinearAccClamp=<1000 1000 1000>
    componentLinearAccScale=<1 1 1>
    componentLinearVelScale=<1 1 1>
    gravityOverride {z=500}
    physicsBodyDefinitions [{
      sphereCollisionRadius=5
      boundBone {boneName=dyn_frontSkirt_l_01}
      boxExtents=<40 40 40>
      constraintSetup {angularLimitsMax=<0 0 4>, angularLimitsMin=<0 -15 -4>, angularTarget=<0 0 0>}
      localJointOffset=<10 0 0>
     }]
    relativeSpaceBone {boneName=root}
   }
   n23 = Anim Dynamics - Bone: dyn_frontSkirt_l_02!(SimulationSpace:AnimPhysSimSpaceType Enum=BoneRelative, ComponentPose=@n22, Alpha=1, GravityOverride:Vector="0,0,500") -> (Pose) {
    angularDampingOverride=0.9999
    angularSpringConstant=200
    bAngularSpring=true
    bOverrideAngularDamping=true
    bUseGravityOverride=true
    bUsePlanarLimit=false
    bUseSphericalLimits=true
    gravityScale=0.3
    simulationSpace=BoneRelative
    alphaBoolBlend {blendOption=Linear}
    alphaScaleBiasClamp {bInterpResult=true}
    boundBone {boneName=dyn_frontSkirt_l_02}
    componentAppliedLinearAccClamp=<1000 1000 1000>
    componentLinearAccScale=<1 1 1>
    componentLinearVelScale=<1 1 1>
    gravityOverride {z=500}
    physicsBodyDefinitions [{
      sphereCollisionRadius=5
      boundBone {boneName=dyn_frontSkirt_l_02}
      boxExtents=<40 40 40>
      constraintSetup {angularLimitsMax=<0 -5 4>, angularLimitsMin=<0 -25 -4>, angularTarget=<0 0 0>}
      localJointOffset=<10 0 0>
     }]
    relativeSpaceBone {boneName=root}
    sphericalLimits [{limitRadius=10, drivingBone {boneName=thigh_twist_01_l}, sphereLocalOffset=<8.5501 -1.3573 0>}]
   }
   n24 = Anim Dynamics - Bone: dyn_frontSkirt_l_03!(SimulationSpace:AnimPhysSimSpaceType Enum=BoneRelative, ComponentPose=@n23, Alpha=1, GravityOverride:Vector="0,0,500") -> (Pose) {
    angularDampingOverride=0.9999
    angularSpringConstant=200
    bAngularSpring=true
    bOverrideAngularDamping=true
    bUseGravityOverride=true
    bUsePlanarLimit=false
    bUseSphericalLimits=true
    gravityScale=0.3
    simulationSpace=BoneRelative
    alphaBoolBlend {blendOption=Linear}
    alphaScaleBiasClamp {bInterpResult=true}
    boundBone {boneName=dyn_frontSkirt_l_03}
    componentAppliedLinearAccClamp=<1000 1000 1000>
    componentLinearAccScale=<1 1 1>
    componentLinearVelScale=<1 1 1>
    gravityOverride {z=500}
    physicsBodyDefinitions [{
      boundBone {boneName=dyn_frontSkirt_l_03}
      boxExtents=<40 40 40>
      constraintSetup {angularLimitsMax=<0 0 4>, angularLimitsMin=<0 -30 -4>, angularTarget=<0 0 0>}
      localJointOffset=<10 0 0>
     }]
    relativeSpaceBone {boneName=root}
    sphericalLimits [{limitRadius=10, drivingBone {boneName=thigh_twist_01_l}, sphereLocalOffset=<8.0757 3 -3.7201>}]
   }
   n25 = Anim Dynamics - Bone: dyn_SideSkirt_l_01!(SimulationSpace:AnimPhysSimSpaceType Enum=BoneRelative, ComponentPose=@n24, Alpha=1, GravityOverride:Vector="0,0,500") -> (Pose) {
    angularDampingOverride=0.9999
    angularSpringConstant=200
    bAngularSpring=true
    bOverrideAngularDamping=true
    bUseGravityOverride=true
    bUsePlanarLimit=false
    gravityScale=0.3
    simulationSpace=BoneRelative
    alphaBoolBlend {blendOption=Linear}
    alphaScaleBiasClamp {bInterpResult=true}
    boundBone {boneName=dyn_SideSkirt_l_01}
    componentAppliedLinearAccClamp=<1000 1000 1000>
    componentLinearAccScale=<1 1 1>
    componentLinearVelScale=<1 1 1>
    gravityOverride {z=500}
    physicsBodyDefinitions [{
      sphereCollisionRadius=5
      boundBone {boneName=dyn_SideSkirt_l_01}
      boxExtents=<40 40 40>
      constraintSetup {angularLimitsMax=<0 0 4>, angularLimitsMin=<0 -15 -4>, angularTarget=<0 0 0>}
      localJointOffset=<10 0 0>
     }]
    relativeSpaceBone {boneName=root}
   }
   n26 = Anim Dynamics - Bone: dyn_SideSkirt_l_02!(SimulationSpace:AnimPhysSimSpaceType Enum=BoneRelative, ComponentPose=@n25, Alpha=1, GravityOverride:Vector="0,0,500") -> (Pose) {
    angularDampingOverride=0.9999
    angularSpringConstant=200
    bAngularSpring=true
    bOverrideAngularDamping=true
    bUseGravityOverride=true
    bUsePlanarLimit=false
    gravityScale=0.3
    simulationSpace=BoneRelative
    alphaBoolBlend {blendOption=Linear}
    alphaScaleBiasClamp {bInterpResult=true}
    boundBone {boneName=dyn_SideSkirt_l_02}
    componentAppliedLinearAccClamp=<1000 1000 1000>
    componentLinearAccScale=<1 1 1>
    componentLinearVelScale=<1 1 1>
    gravityOverride {z=500}
    physicsBodyDefinitions [{
      sphereCollisionRadius=5
      boundBone {boneName=dyn_SideSkirt_l_02}
      boxExtents=<40 40 40>
      constraintSetup {angularLimitsMax=<0 0 4>, angularLimitsMin=<0 -25 -4>, angularTarget=<0 0 0>}
      localJointOffset=<10 0 0>
     }]
    relativeSpaceBone {boneName=root}
   }
   n27 = Anim Dynamics - Bone: dyn_SideSkirt_l_03!(SimulationSpace:AnimPhysSimSpaceType Enum=BoneRelative, ComponentPose=@n26, Alpha=1, GravityOverride:Vector="0,0,500") -> (Pose) {
    angularDampingOverride=0.9999
    angularSpringConstant=200
    bAngularSpring=true
    bOverrideAngularDamping=true
    bUseGravityOverride=true
    bUsePlanarLimit=false
    bUseSphericalLimits=true
    gravityScale=0.3
    simulationSpace=BoneRelative
    alphaBoolBlend {blendOption=Linear}
    alphaScaleBiasClamp {bInterpResult=true}
    boundBone {boneName=dyn_SideSkirt_l_03}
    componentAppliedLinearAccClamp=<1000 1000 1000>
    componentLinearAccScale=<1 1 1>
    componentLinearVelScale=<1 1 1>
    gravityOverride {z=500}
    physicsBodyDefinitions [{
      boundBone {boneName=dyn_SideSkirt_l_03}
      boxExtents=<40 40 40>
      constraintSetup {angularLimitsMax=<0 0 4>, angularLimitsMin=<0 -30 -4>, angularTarget=<0 0 0>}
      localJointOffset=<10 0 0>
     }]
    relativeSpaceBone {boneName=root}
    sphericalLimits [{limitRadius=10, drivingBone {boneName=thigh_twist_01_r}, sphereLocalOffset=<0 0 0>}]
   }
   n28 = Anim Dynamics - Bone: dyn_BackSkirt_l_01!(SimulationSpace:AnimPhysSimSpaceType Enum=BoneRelative, ComponentPose=@n27, Alpha=1, GravityOverride:Vector="0,0,500") -> (Pose) {
    angularDampingOverride=0.9999
    angularSpringConstant=200
    bAngularSpring=true
    bOverrideAngularDamping=true
    bUseGravityOverride=true
    bUsePlanarLimit=false
    bUseSphericalLimits=true
    gravityScale=0.3
    simulationSpace=BoneRelative
    alphaBoolBlend {blendOption=Linear}
    alphaScaleBiasClamp {bInterpResult=true}
    boundBone {boneName=dyn_BackSkirt_l_01}
    componentAppliedLinearAccClamp=<1000 1000 1000>
    componentLinearAccScale=<1 1 1>
    componentLinearVelScale=<1 1 1>
    gravityOverride {z=500}
    physicsBodyDefinitions [{
      sphereCollisionRadius=5
      boundBone {boneName=dyn_BackSkirt_l_01}
      boxExtents=<40 40 40>
      constraintSetup {angularLimitsMax=<0 0 4>, angularLimitsMin=<0 -15 -4>, angularTarget=<0 0 0>}
      localJointOffset=<10 0 0>
     }]
    relativeSpaceBone {boneName=root}
   }
   n29 = Anim Dynamics - Bone: dyn_BackSkirt_l_02!(SimulationSpace:AnimPhysSimSpaceType Enum=BoneRelative, ComponentPose=@n28, Alpha=1, GravityOverride:Vector="0,0,500") -> (Pose) {
    angularDampingOverride=0.9999
    angularSpringConstant=200
    bAngularSpring=true
    bOverrideAngularDamping=true
    bUseGravityOverride=true
    bUsePlanarLimit=false
    bUseSphericalLimits=true
    gravityScale=0.3
    simulationSpace=BoneRelative
    alphaBoolBlend {blendOption=Linear}
    alphaScaleBiasClamp {bInterpResult=true}
    boundBone {boneName=dyn_BackSkirt_l_02}
    componentAppliedLinearAccClamp=<1000 1000 1000>
    componentLinearAccScale=<1 1 1>
    componentLinearVelScale=<1 1 1>
    gravityOverride {z=500}
    physicsBodyDefinitions [{
      sphereCollisionRadius=5
      boundBone {boneName=dyn_BackSkirt_l_02}
      boxExtents=<40 40 40>
      constraintSetup {angularLimitsMax=<0 0 4>, angularLimitsMin=<0 -25 -4>, angularTarget=<0 0 0>}
      localJointOffset=<10 0 0>
     }]
    relativeSpaceBone {boneName=root}
   }
   n30 = Anim Dynamics - Bone: dyn_BackSkirt_l_03!(SimulationSpace:AnimPhysSimSpaceType Enum=BoneRelative, ComponentPose=@n29, Alpha=1, GravityOverride:Vector="0,0,500") -> (Pose) {
    angularDampingOverride=0.9999
    angularSpringConstant=200
    bAngularSpring=true
    bOverrideAngularDamping=true
    bUseGravityOverride=true
    bUsePlanarLimit=false
    bUseSphericalLimits=true
    gravityScale=0.3
    simulationSpace=BoneRelative
    alphaBoolBlend {blendOption=Linear}
    alphaScaleBiasClamp {bInterpResult=true}
    boundBone {boneName=dyn_BackSkirt_l_03}
    componentAppliedLinearAccClamp=<1000 1000 1000>
    componentLinearAccScale=<1 1 1>
    componentLinearVelScale=<1 1 1>
    gravityOverride {z=500}
    physicsBodyDefinitions [{
      sphereCollisionRadius=5
      boundBone {boneName=dyn_BackSkirt_l_03}
      boxExtents=<40 40 40>
      constraintSetup {angularLimitsMax=<0 0 4>, angularLimitsMin=<0 -30 -4>, angularTarget=<0 0 0>}
      localJointOffset=<10 0 0>
     }]
    relativeSpaceBone {boneName=root}
   }
   n31 = Anim Dynamics - Bone: dyn_frontSkirt_r_01!(SimulationSpace:AnimPhysSimSpaceType Enum=BoneRelative, ComponentPose=@n30, Alpha=1, GravityOverride:Vector="0,0,500") -> (Pose) {
    angularDampingOverride=0.9999
    angularSpringConstant=200
    bAngularSpring=true
    bOverrideAngularDamping=true
    bUseGravityOverride=true
    bUsePlanarLimit=false
    gravityScale=0.3
    simulationSpace=BoneRelative
    alphaBoolBlend {blendOption=Linear}
    boundBone {boneName=dyn_frontSkirt_r_01}
    componentAppliedLinearAccClamp=<1000 1000 1000>
    componentLinearAccScale=<1 1 1>
    componentLinearVelScale=<1 1 1>
    gravityOverride {z=500}
    physicsBodyDefinitions [{
      sphereCollisionRadius=5
      boundBone {boneName=dyn_frontSkirt_r_01}
      boxExtents=<40 40 40>
      constraintSetup {angularLimitsMax=<0 0 4>, angularLimitsMin=<0 -15 -4>, angularTarget=<0 0 0>}
      localJointOffset=<10 0 0>
     }]
    relativeSpaceBone {boneName=root}
   }
   n32 = Anim Dynamics - Bone: dyn_frontSkirt_r_02!(SimulationSpace:AnimPhysSimSpaceType Enum=BoneRelative, ComponentPose=@n31, Alpha=1, GravityOverride:Vector="0,0,500") -> (Pose) {
    angularDampingOverride=0.9999
    angularSpringConstant=200
    bAngularSpring=true
    bOverrideAngularDamping=true
    bUseGravityOverride=true
    bUsePlanarLimit=false
    bUseSphericalLimits=true
    gravityScale=0.3
    simulationSpace=BoneRelative
    alphaBoolBlend {blendOption=Linear}
    boundBone {boneName=dyn_frontSkirt_r_02}
    componentAppliedLinearAccClamp=<1000 1000 1000>
    componentLinearAccScale=<1 1 1>
    componentLinearVelScale=<1 1 1>
    gravityOverride {z=500}
    physicsBodyDefinitions [{
      sphereCollisionRadius=5
      boundBone {boneName=dyn_frontSkirt_r_02}
      boxExtents=<40 40 40>
      constraintSetup {angularLimitsMax=<0 -5 4>, angularLimitsMin=<0 -25 -4>, angularTarget=<0 0 0>}
      localJointOffset=<10 0 0>
     }]
    relativeSpaceBone {boneName=root}
    sphericalLimits [{limitRadius=12, drivingBone {boneName=thigh_twist_01_r}, sphereLocalOffset=<-6.3401 0 -3.1637>}]
   }
   n33 = Anim Dynamics - Bone: dyn_frontSkirt_r_03!(SimulationSpace:AnimPhysSimSpaceType Enum=BoneRelative, ComponentPose=@n32, Alpha=1, GravityOverride:Vector="0,0,500") -> (Pose) {
    angularDampingOverride=0.9999
    angularSpringConstant=200
    bAngularSpring=true
    bOverrideAngularDamping=true
    bUseGravityOverride=true
    bUsePlanarLimit=false
    bUseSphericalLimits=true
    gravityScale=0.3
    simulationSpace=BoneRelative
    alphaBoolBlend {blendOption=Linear}
    boundBone {boneName=dyn_frontSkirt_r_03}
    componentAppliedLinearAccClamp=<1000 1000 1000>
    componentLinearAccScale=<1 1 1>
    componentLinearVelScale=<1 1 1>
    gravityOverride {z=500}
    physicsBodyDefinitions [{
      sphereCollisionRadius=5
      boundBone {boneName=dyn_frontSkirt_r_03}
      boxExtents=<40 40 40>
      constraintSetup {angularLimitsMax=<0 -5 4>, angularLimitsMin=<0 -30 -4>, angularTarget=<0 0 0>}
      localJointOffset=<10 0 0>
     }]
    relativeSpaceBone {boneName=root}
    sphericalLimits [{limitRadius=12, drivingBone {boneName=thigh_twist_01_r}, sphereLocalOffset=<-6.764 0 -2.7922>}]
   }
   n34 = Anim Dynamics - Bone: dyn_SideSkirt_r_01!(SimulationSpace:AnimPhysSimSpaceType Enum=BoneRelative, ComponentPose=@n33, Alpha=1, GravityOverride:Vector="0,0,500") -> (Pose) {
    angularDampingOverride=0.9999
    angularSpringConstant=200
    bAngularSpring=true
    bOverrideAngularDamping=true
    bUseGravityOverride=true
    bUsePlanarLimit=false
    gravityScale=0.3
    simulationSpace=BoneRelative
    alphaBoolBlend {blendOption=Linear}
    boundBone {boneName=dyn_SideSkirt_r_01}
    componentAppliedLinearAccClamp=<1000 1000 1000>
    componentLinearAccScale=<1 1 1>
    componentLinearVelScale=<1 1 1>
    gravityOverride {z=500}
    physicsBodyDefinitions [{
      sphereCollisionRadius=5
      boundBone {boneName=dyn_SideSkirt_r_01}
      boxExtents=<40 40 40>
      constraintSetup {angularLimitsMax=<0 0 4>, angularLimitsMin=<0 -15 -4>, angularTarget=<0 0 0>}
      localJointOffset=<10 0 0>
     }]
    relativeSpaceBone {boneName=root}
   }
   n35 = Anim Dynamics - Bone: dyn_SideSkirt_r_02!(SimulationSpace:AnimPhysSimSpaceType Enum=BoneRelative, ComponentPose=@n34, Alpha=1, GravityOverride:Vector="0,0,500") -> (Pose) {
    angularDampingOverride=0.9999
    angularSpringConstant=200
    bAngularSpring=true
    bOverrideAngularDamping=true
    bUseGravityOverride=true
    bUsePlanarLimit=false
    gravityScale=0.3
    simulationSpace=BoneRelative
    alphaBoolBlend {blendOption=Linear}
    boundBone {boneName=dyn_SideSkirt_r_02}
    componentAppliedLinearAccClamp=<1000 1000 1000>
    componentLinearAccScale=<1 1 1>
    componentLinearVelScale=<1 1 1>
    gravityOverride {z=500}
    physicsBodyDefinitions [{
      sphereCollisionRadius=5
      boundBone {boneName=dyn_SideSkirt_r_02}
      boxExtents=<40 40 40>
      constraintSetup {angularLimitsMax=<0 0 4>, angularLimitsMin=<0 -25 -4>, angularTarget=<0 0 0>}
      localJointOffset=<10 0 0>
     }]
    relativeSpaceBone {boneName=root}
   }
   n36 = Anim Dynamics - Bone: dyn_SideSkirt_r_03!(SimulationSpace:AnimPhysSimSpaceType Enum=BoneRelative, ComponentPose=@n35, Alpha=1, GravityOverride:Vector="0,0,500") -> (Pose) {
    angularDampingOverride=0.9999
    angularSpringConstant=200
    bAngularSpring=true
    bOverrideAngularDamping=true
    bUseGravityOverride=true
    bUsePlanarLimit=false
    gravityScale=0.3
    simulationSpace=BoneRelative
    alphaBoolBlend {blendOption=Linear}
    boundBone {boneName=dyn_SideSkirt_r_03}
    componentAppliedLinearAccClamp=<1000 1000 1000>
    componentLinearAccScale=<1 1 1>
    componentLinearVelScale=<1 1 1>
    gravityOverride {z=500}
    physicsBodyDefinitions [{
      sphereCollisionRadius=5
      boundBone {boneName=dyn_SideSkirt_r_03}
      boxExtents=<40 40 40>
      constraintSetup {angularLimitsMax=<0 0 4>, angularLimitsMin=<0 -30 -4>, angularTarget=<0 0 0>}
      localJointOffset=<10 0 0>
     }]
    relativeSpaceBone {boneName=root}
   }
   n37 = Anim Dynamics - Bone: dyn_BackSkirt_r_01!(SimulationSpace:AnimPhysSimSpaceType Enum=BoneRelative, ComponentPose=@n36, Alpha=1, GravityOverride:Vector="0,0,500") -> (Pose) {
    angularDampingOverride=0.9999
    angularSpringConstant=200
    bAngularSpring=true
    bOverrideAngularDamping=true
    bUseGravityOverride=true
    bUsePlanarLimit=false
    bUseSphericalLimits=true
    gravityScale=0.3
    simulationSpace=BoneRelative
    alphaBoolBlend {blendOption=Linear}
    boundBone {boneName=dyn_BackSkirt_r_01}
    componentAppliedLinearAccClamp=<1000 1000 1000>
    componentLinearAccScale=<1 1 1>
    componentLinearVelScale=<1 1 1>
    gravityOverride {z=500}
    physicsBodyDefinitions [{
      sphereCollisionRadius=5
      boundBone {boneName=dyn_BackSkirt_r_01}
      boxExtents=<40 40 40>
      constraintSetup {angularLimitsMax=<0 0 4>, angularLimitsMin=<0 -15 -4>, angularTarget=<0 0 0>}
      localJointOffset=<10 0 0>
     }]
    relativeSpaceBone {boneName=root}
   }
   n38 = Anim Dynamics - Bone: dyn_BackSkirt_r_02!(SimulationSpace:AnimPhysSimSpaceType Enum=BoneRelative, ComponentPose=@n37, Alpha=1, GravityOverride:Vector="0,0,500") -> (Pose) {
    angularDampingOverride=0.9999
    angularSpringConstant=200
    bAngularSpring=true
    bOverrideAngularDamping=true
    bUseGravityOverride=true
    bUsePlanarLimit=false
    bUseSphericalLimits=true
    gravityScale=0.3
    simulationSpace=BoneRelative
    alphaBoolBlend {blendOption=Linear}
    boundBone {boneName=dyn_BackSkirt_r_02}
    componentAppliedLinearAccClamp=<1000 1000 1000>
    componentLinearAccScale=<1 1 1>
    componentLinearVelScale=<1 1 1>
    gravityOverride {z=500}
    physicsBodyDefinitions [{
      sphereCollisionRadius=5
      boundBone {boneName=dyn_BackSkirt_r_02}
      boxExtents=<40 40 40>
      constraintSetup {angularLimitsMax=<0 0 4>, angularLimitsMin=<0 -25 -4>, angularTarget=<0 0 0>}
      localJointOffset=<10 0 0>
     }]
    relativeSpaceBone {boneName=root}
   }
   n39 = Anim Dynamics - Bone: dyn_BackSkirt_r_03!(SimulationSpace:AnimPhysSimSpaceType Enum=BoneRelative, ComponentPose=@n38, Alpha=1, GravityOverride:Vector="0,0,500") -> (Pose) {
    angularDampingOverride=0.9999
    angularSpringConstant=200
    bAngularSpring=true
    bOverrideAngularDamping=true
    bUseGravityOverride=true
    bUsePlanarLimit=false
    bUseSphericalLimits=true
    gravityScale=0.3
    simulationSpace=BoneRelative
    alphaBoolBlend {blendOption=Linear}
    boundBone {boneName=dyn_BackSkirt_r_03}
    componentAppliedLinearAccClamp=<1000 1000 1000>
    componentLinearAccScale=<1 1 1>
    componentLinearVelScale=<1 1 1>
    gravityOverride {z=500}
    physicsBodyDefinitions [{
      sphereCollisionRadius=5
      boundBone {boneName=dyn_BackSkirt_r_03}
      boxExtents=<40 40 40>
      constraintSetup {angularLimitsMax=<0 0 4>, angularLimitsMin=<0 -30 -4>, angularTarget=<0 0 0>}
      localJointOffset=<10 0 0>
     }]
    relativeSpaceBone {boneName=root}
   }
   n40 = Anim Dynamics - Bone: dyn_strap_loop_01!(ComponentPose=@n39, Alpha=1, GravityOverride:Vector="0,0,500") -> (Pose) {
    angularDampingOverride=1
    angularSpringConstant=15000
    bAngularSpring=true
    bOverrideAngularDamping=true
    bUseGravityOverride=true
    bUsePlanarLimit=false
    gravityScale=0.3
    simulationSpace=RootRelative
    alphaBoolBlend {blendOption=Linear}
    boundBone {boneName=dyn_strap_loop_01}
    componentAppliedLinearAccClamp=<5000 5000 5000>
    componentLinearAccScale=<1 1 1>
    componentLinearVelScale=<1 1 1>
    gravityOverride {z=500}
    physicsBodyDefinitions [{
      sphereCollisionRadius=5
      boundBone {boneName=dyn_strap_loop_01}
      boxExtents=<20 20 20>
      constraintSetup {angularLimitsMax=<-5 -10 45>, angularLimitsMin=<0 -45 -45>, angularTarget=<0 0 0>}
      localJointOffset=<4 0 0>
     }]
    relativeSpaceBone {boneName=pelvis}
   }
   n41 = "Transform (Modify) Bone - Bone: dyn_spine_belt_02"(ComponentPose=@n40, Alpha=1, Translation="-0.245,1,-3.75", Rotation="-26.038,6.549,37.007", Scale="1,1,1") {
    rotationMode=BMM_Additive
    rotationSpace=BCS_BoneSpace
    translationMode=BMM_Additive
    translationSpace=BCS_BoneSpace
    alphaBoolBlend {blendOption=Linear}
    alphaScaleBiasClamp {bInterpResult=true}
    boneToModify {boneName=dyn_spine_belt_02}
    rotation=<rot -26.038 6.549 37.0071>
    translation=<-0.245 1 -3.75>
   }
   n42 = "Transform (Modify) Bone - Bone: dyn_strap_loop_01"(ComponentPose=@n41, Alpha=1, Translation="0,0,0", Rotation="15.000,0.030,0.062", Scale="1,1,1") {
    rotationMode=BMM_Additive
    rotationSpace=BCS_BoneSpace
    translationMode=BMM_Additive
    translationSpace=BCS_BoneSpace
    alphaBoolBlend {blendOption=Linear}
    alphaScaleBiasClamp {bInterpResult=true}
    boneToModify {boneName=dyn_strap_loop_01}
    rotation=<rot 15 0.03 0.062>
   }
   n43 = Anim Dynamics - Bone: dyn_l_leg_bag_chain_01!(ComponentPose=@n42, Alpha=1, GravityOverride:Vector="0,0,500") -> (Pose) {
    angularDampingOverride=0.9999
    angularSpringConstant=10000
    bAngularSpring=true
    bOverrideAngularDamping=true
    bUseGravityOverride=true
    bUsePlanarLimit=false
    gravityScale=0.3
    simulationSpace=RootRelative
    alphaBoolBlend {blendOption=Linear}
    boundBone {boneName=dyn_l_leg_bag_chain_01}
    componentAppliedLinearAccClamp=<1000 1000 1000>
    componentLinearAccScale=<1 1 1>
    componentLinearVelScale=<1 1 1>
    gravityOverride {z=500}
    physicsBodyDefinitions [{
      sphereCollisionRadius=5
      boundBone {boneName=dyn_l_leg_bag_chain_01}
      boxExtents=<20 20 20>
      constraintSetup {angularLimitsMax=<0 20 10>, angularLimitsMin=<0 -30 -10>, angularTarget=<0 0 0>}
      localJointOffset=<4 0 0>
     }]
    relativeSpaceBone {boneName=dyn_l_leg_bag_01}
   }
   n44 = Anim Dynamics - Bone: dyn_l_leg_shoe_chain_01!(ComponentPose=@n43, Alpha=1, GravityOverride:Vector="0,0,500") -> (Pose) {
    angularDampingOverride=0.9999
    angularSpringConstant=10000
    bAngularSpring=true
    bOverrideAngularDamping=true
    bUseGravityOverride=true
    bUsePlanarLimit=false
    gravityScale=0.3
    simulationSpace=RootRelative
    alphaBoolBlend {blendOption=Linear}
    boundBone {boneName=dyn_l_leg_shoe_chain_01}
    componentAppliedLinearAccClamp=<1000 1000 1000>
    componentLinearAccScale=<1 1 1>
    componentLinearVelScale=<1 1 1>
    gravityOverride {z=500}
    physicsBodyDefinitions [{
      sphereCollisionRadius=5
      boundBone {boneName=dyn_l_leg_shoe_chain_01}
      boxExtents=<20 20 20>
      constraintSetup {angularLimitsMax=<0 20 10>, angularLimitsMin=<0 -30 -10>, angularTarget=<0 0 0>}
      localJointOffset=<4 0 0>
     }]
    relativeSpaceBone {boneName=dyn_l_leg_bag_01}
   }
   n45 = Anim Dynamics - Bone: dyn_r_leg_shoe_chain_01!(ComponentPose=@n44, Alpha=1, GravityOverride:Vector="0,0,500") -> (Pose) {
    angularDampingOverride=0.9999
    angularSpringConstant=10000
    bAngularSpring=true
    bOverrideAngularDamping=true
    bUseGravityOverride=true
    bUsePlanarLimit=false
    gravityScale=0.3
    simulationSpace=RootRelative
    alphaBoolBlend {blendOption=Linear}
    boundBone {boneName=dyn_r_leg_shoe_chain_01}
    componentAppliedLinearAccClamp=<1000 1000 1000>
    componentLinearAccScale=<1 1 1>
    componentLinearVelScale=<1 1 1>
    gravityOverride {z=500}
    physicsBodyDefinitions [{
      sphereCollisionRadius=5
      boundBone {boneName=dyn_r_leg_shoe_chain_01}
      boxExtents=<20 20 20>
      constraintSetup {angularLimitsMax=<0 20 10>, angularLimitsMin=<0 -30 -10>, angularTarget=<0 0 0>}
      localJointOffset=<4 0 0>
     }]
    relativeSpaceBone {boneName=dyn_l_leg_bag_01}
   }
   n46 = "Transform (Modify) Bone - Bone: dyn_frontSkirt_l_01"(ComponentPose=@n45, Alpha=1, Translation="0,0,0", Rotation="5,0,0", Scale="1,1,1") {
    rotationMode=BMM_Additive
    rotationSpace=BCS_BoneSpace
    translationMode=BMM_Additive
    translationSpace=BCS_BoneSpace
    alphaBoolBlend {blendOption=Linear}
    alphaScaleBiasClamp {bInterpResult=true, interpSpeedDecreasing=5, interpSpeedIncreasing=5}
    boneToModify {boneName=dyn_frontSkirt_l_01}
    rotation {pitch=5}
   }
   n47 = "Transform (Modify) Bone - Bone: dyn_frontSkirt_m_01"(ComponentPose=@n46, Alpha=1, Translation="0,0,0", Rotation="10,0,0", Scale="1,1,1") {
    rotationMode=BMM_Additive
    rotationSpace=BCS_BoneSpace
    translationMode=BMM_Additive
    translationSpace=BCS_BoneSpace
    alphaBoolBlend {blendOption=Linear}
    alphaScaleBiasClamp {bInterpResult=true, interpSpeedDecreasing=5, interpSpeedIncreasing=5}
    boneToModify {boneName=dyn_frontSkirt_m_01}
    rotation {pitch=10}
   }
   Save cached pose 'CRCache_SkirtDyn'(Pose=Component To Local(ComponentPose=@n47)) {cacheName=CRCache_SkirtDyn}
  ]}
 EventGraph {nodes [
   Event OnEnterGrindRail() {is=title K2Node_Event}
   > Set GrindRail(GrindRail:double=1) -> (Output_Get:double)
   Event OnParachuteOpened()
   > Switch on EGliderType(Selection=Get PlayerGliderType() -> (PlayerGliderType=HangGlider))
   > Set isHangGliding(isHangGliding:bool=true) -> (Output_Get:bool)
   Event OnBeginKneeSliding() {is=title K2Node_Event}
   > Set KneeSliding(KneeSliding:double=1) -> (Output_Get:double)
   Event OnEndKneeSliding() {is=title K2Node_Event}
   > Set KneeSliding(KneeSliding=0)
   Event OnBeginJumping()
   > Set Jumping(Jumping:bool=true) -> (Output_Get:bool)
   Event OnEndFalling()
   > Set Jumping(Jumping=false)
   Event OnCharacterCustomizationLoaded()
   > n1 = Sequence()
   > Set isWearingBackbling(isWearingBackbling=CheckForBackbling())
   n2 = Branch(execute=@n1.then_1, Condition=Get bIsInFrontEnd())
   > Set Sim Space Skirt(Sim Space Skirt:AnimPhysSimSpaceType Enum=World) -> (Output_Get:AnimPhysSimSpaceType Enum)
   > Set SkirtDynAlpha(SkirtDynAlpha:double=0.2) -> (Output_Get:double)
   Set Sim Space Skirt(execute=@n2.else)
   > Set SkirtDynAlpha(SkirtDynAlpha=1)
   n3 = Event BlueprintUpdateAnimation()
   > Set Delta Time X(Delta Time X=@n3.DeltaTimeX)
   > n4 = Sequence!(execute) -> (then_0, then_1, then_2:Exec, then_3:Exec, then_4:Exec, then_5:Exec, then_6:Exec, then_7:Exec, then_8:Exec, then_9:Exec)
   > n5 = "Gravity Override State Function (Parameter Asset) (Nativized)"(
    CharacterPartAnimInstance=Self-Reference()
    GameTimeInSeconds=GetGameTimeInSeconds()
    Parameters=&/M_MED_VioletInk_Parameters
   )
   > Set GravityOverride(GravityOverride=@n5)
   > Set RBAN Settings Out(RBAN Settings Out=@n5.OutRigidBodyAnimNodeParameters)
   n6 = Get bIsCrouching()
   > Set Couching_Alpha(Couching_Alpha:double="To Float (Boolean)"(InBool=@n6)) -> (Output_Get:double)
   > Set RigidBody_off(RigidBody_off:double="To Float (Boolean)"(InBool=NOT Boolean(A=@n6))) -> (Output_Get:double)
   n7 = Get bIsSkydiveFloating()
   > Set SkyDive_Float(SkyDive_Float:double="To Float (Boolean)"(InBool=@n7)) -> (Output_Get:double)
   n8 = Get bIsParachuteOpened()
   n9 = Get bIsSkydiveDiveMode()
   > Set SkyDive_Left(SkyDive_Left:double=SelectFloat(A="Get Move Left "() -> ("Move Left ":double), B=0, bPickA=OR Boolean!(A=@n7, B=@n9, C:bool=@n8) -> (ReturnValue))) -> (Output_Get:double)
   n10 = OR Boolean!(A=Get bRigidBodyEnabled(), B=Get bAnimDynamicsEnabled() -> (bAnimDynamicsEnabled:bool)) -> (ReturnValue)
   > Set ALT_Skin_ADJ(ALT_Skin_ADJ:double="To Float (Boolean)"(InBool=NOT Boolean(A=@n10))) -> (Output_Get:double)
   n11 = OR Boolean!(A=@n7, B=Get bIsSkyDivingFromBus() -> (bIsSkyDivingFromBus:bool), C:bool=@n9, D:bool=Get bIsParachuteOpened()) -> (ReturnValue)
   > Set SkyDive_Float_Off(SkyDive_Float_Off:double="To Float (Boolean)"(InBool=@n11)) -> (Output_Get:double)
   > n12 = Set SkyDive(SkyDive:bool=@n11) -> (Output_Get:bool)
   > Set SkyDive_Float_Flap(SkyDive_Float_Flap:double="To Float (Boolean)"(InBool=@n12)) -> (Output_Get:double)
   n13 = NOT Boolean(A=Get bIsCrouching())
   n14 = Get bIsSprinting()
   n15 = Get bIsTacticalSprinting()
   > Set RunningAlpha(RunningAlpha:double="To Float (Boolean)"(InBool=OR Boolean!(A=@n14, B=@n15, C=@n13) -> (ReturnValue))) -> (Output_Get:double)
   > Set Emote(Emote:double="To Float (Boolean)"(InBool=Get bIsPlayingEmote())) -> (Output_Get:double)
   n16 = Get bIsCrouching()
   n17 = Get bIsMoving2D()
   > Set Crouch_Move(Crouch_Move:double="To Float (Boolean)"(InBool=AND Boolean(A=@n16, B=@n17))) -> (Output_Get:double)
   n18 = NOT Boolean(A=Get bIsMoving2D())
   n19 = Get bIsCrouching()
   > Set Crouch_Only(Crouch_Only:double="To Float (Boolean)"(InBool=AND Boolean(A=@n19, B=@n18))) -> (Output_Get:double)
   > n20 = PlayerMovement(AnimInstance=Self-Reference(), XDotProductCoefficient=1, YDotProductCoefficient=1, ZDotProductCoefficient=1)
   n21 = BreakVector(InVec=MakeVector(X=@n20, Y=@n20.YMovement))
   > Set strafeLeft float(strafeLeft float:double="Absolute (Float)"(A="Clamp (Float)"(Value=@n21, Min=-1, Max=0))) -> (Output_Get:double)
   > Set strafeRight float(strafeRight float:double="Clamp (Float)"(Value=@n21, Min=0, Max=1)) -> (Output_Get:double)
   > Set moveForward(moveForward:double="Clamp (Float)"(Value=@n21.Y, Min=0, Max=1)) -> (Output_Get:double)
   > Set moveBackward(moveBackward:double="Absolute (Float)"(A="Clamp (Float)"(Value=@n21.Y, Min=-1, Max=0))) -> (Output_Get:double)
   n22 = Get bIsMoving2D()
   > Set IsIdle(IsIdle:double="To Float (Boolean)"(InBool=NOT Boolean(A=@n22))) -> (Output_Get:double)
   > n23 = Branch(Condition=OR Boolean!(A=Get bIsSkydiveFloating(), B=Get bIsSkyDiving(), C=Get bIsParachuteOpened(), D:bool=Get bIsInFrontEnd()) -> (ReturnValue))
   > Set SkirtDynAlpha(SkirtDynAlpha=0.5)
   Set SkirtDynAlpha(execute=@n23.else, SkirtDynAlpha=1)
   n24 = Self-Reference()
   n25 = PlayerMovement(execute=@n4.then_6, AnimInstance=@n24, XDotProductCoefficient=1, YDotProductCoefficient=1, ZDotProductCoefficient=1)
   n26 = BreakVector(InVec=MakeVector(X=@n25, Y=@n25.YMovement))
   > "Set Move Left "("Move Left ":double="Absolute (Float)"(A="Clamp (Float)"(Value=@n26, Min=-1, Max=0))) -> (Output_Get:double)
   > Set Move Right(Move Right:double="Clamp (Float)"(Value=@n26, Min=0, Max=1)) -> (Output_Get:double)
   > Set moveForward(moveForward="Clamp (Float)"(Value=@n26.Y, Min=0, Max=1))
   > Set moveBackward(moveBackward="Absolute (Float)"(A="Clamp (Float)"(Value=@n26.Y, Min=-1, Max=0)))
   n27 = OR Boolean!(A=Get bIsJumping(), B=Get bIsFalling()) -> (ReturnValue)
   n28 = Get bIsTacticalSprinting()
   > Set IsTacticaljumping(IsTacticaljumping:double="To Float (Boolean)"(InBool=AND Boolean(A=@n28, B=@n27))) -> (Output_Get:double)
   n29 = Branch(
    execute=@n4.then_8
    Condition=OR Boolean!(
     A=Get bIsSkyDiving()
     B=Get bIsSkydiveDiveMode()
     C:bool=Get bIsSkydiveFloating()
     D:bool=Get bIsInFrontEnd()
     E:bool=Get bIsTacticalSprinting()
     F:bool=Get bIsSprinting()
    ) -> (ReturnValue)
   )
   > Set SkyDive_ON(SkyDive_ON:double=0.2) -> (Output_Get:double)
   Set SkyDive_ON(execute=@n29.else, SkyDive_ON=1)
   n30 = Get bIsSprinting()
   n31 = Get bIsFalling()
   n32 = Get bIsJumping()
   n33 = Branch(execute=@n4.then_1, Condition=AND Boolean(A=Get bIsTacticalSprinting(), B=OR Boolean!(A=@n31, B=@n32, C=@n30) -> (ReturnValue)))
   > Set IsTacticalandJump(IsTacticalandJump:double=1) -> (Output_Get:double)
   Set IsTacticalandJump(execute=@n33.else, IsTacticalandJump=0)
   n34 = PlayerMovement(execute=@n4.then_4, AnimInstance=@n24, XDotProductCoefficient=100, YDotProductCoefficient=100, ZDotProductCoefficient=100)
   n35 = BreakVector(InVec=MakeVector(X=@n34, Y=@n34.YMovement))
   > Set moveLeft bool(moveLeft bool:bool="float < float"(A=@n35, B=-50)) -> (Output_Get:bool)
   > Set moveRight bool(moveRight bool:bool="float > float"(A=@n35, B=50)) -> (Output_Get:bool)
   n36 = AND Boolean!(
    A=Get bIsCrouching()
    B=Get bIsMoving2D()
    C:bool=OR Boolean!(A=Get moveLeft bool() -> (moveLeft bool:bool), B=Get moveRight bool() -> (moveRight bool:bool)) -> (ReturnValue)
   ) -> (ReturnValue)
   > Set isCrouchStrafe float(isCrouchStrafe float:double=FInterpTo(Current=Get isCrouchStrafe float(), Target="To Float (Boolean)"(InBool=@n36), DeltaTime=Get Delta Time X(), InterpSpeed=10)) -> (Output_Get:double)
   > n37 = Branch(Condition=OR Boolean(A=Get bIsSkydiveFloating(), B=Get bIsParachuteOpened()))
   > Set ArmsVolume(ArmsVolume:bool=true) -> (Output_Get:bool)
   Set ArmsVolume(execute=@n37.else, ArmsVolume=false)
   n38 = Branch(execute=@n4.then_3, Condition=Get bIsCrouching())
   > Set SkirtDynAlpha(SkirtDynAlpha=0.5)
   Set SkirtDynAlpha(execute=@n38.else, SkirtDynAlpha=1)
   n39 = Branch(execute=@n4.then_5, Condition=Get bIsCrouching())
   > Set Alpha_Crouching(Alpha_Crouching:double=1) -> (Output_Get:double)
   Set Alpha_Crouching(execute=@n39.else, Alpha_Crouching=0)
   n40 = PlayerMovement(execute=@n4.then_2, AnimInstance=Self-Reference(), XDotProductCoefficient=100, YDotProductCoefficient=100, ZDotProductCoefficient=100)
   n41 = Get bIsKneeSliding()
   n42 = Get bIsSliding()
   n43 = Get bIsBaseballSliding()
   > Branch(Condition=AND Boolean!(A=Get bIsCrouching(), B=NOT Boolean(A=@n42), C=NOT Boolean(A=@n43), D:bool=NOT Boolean(A=@n41)) -> (ReturnValue))
   n44 = BreakVector(InVec=MakeVector(X=@n40, Y=@n40.YMovement))
   > Set CrouchLeft(CrouchLeft:double="To Float (Boolean)"(InBool="float < float"(A=@n44, B=-50))) -> (Output_Get:double)
   > Set CrouchRight(CrouchRight:double="To Float (Boolean)"(InBool="float > float"(A=@n44, B=50))) -> (Output_Get:double)
   > Set CrouchForward(CrouchForward:double="To Float (Boolean)"(InBool="float > float"(A=@n44.Y, B=40))) -> (Output_Get:double)
   > Set CrouchBackward(CrouchBackward:double="To Float (Boolean)"(InBool="float < float"(A=@n44.Y, B=-40))) -> (Output_Get:double)
   n45 = Branch(
    execute=@n4.then_7
    Condition=OR Boolean!(A=Get bIsSkyDiving(), B=Get bIsCrouching(), C:bool=Get bIsPlayingEmote(), D:bool=Get bIsJumping(), E:bool=Get bIsFalling()) -> (ReturnValue)
   )
   > n46 = Set isIdle(isIdle:double=0) -> (Output_Get:double)
   > Set isIdleInFE(isIdleInFE:double=0) -> (Output_Get:double)
   > Set isIdle_Bool(isIdle_Bool:bool=false) -> (Output_Get:bool)
   > Set isNotIdle(isNotIdle:double=@n46) -> (Output_Get:double)
   n47 = float + float!(A="Get Move Left "(), B=Get Move Right() -> (Move Right:double), C:double=Get moveForward(), D:double=Get moveBackward()) -> (ReturnValue)
   Set isIdle(execute=@n45.else, isIdle=Select(Option 0:double=1, Option 1:double=0, Index="float > float"(A=@n47)) -> (ReturnValue:double))
   > Set isIdle_Bool(isIdle_Bool=true)
   Event OnParachuteClosed()
   > Set isHangGliding(isHangGliding=false)
   Event OnExitGrindRail() {is=title K2Node_Event}
   > Set GrindRail(GrindRail=0)
   Event OnBeginBaseballSliding() {is=title K2Node_Event}
   > Set BaseballSlide(BaseballSlide:double=1) -> (Output_Get:double)
   Event OnEndBaseballSliding() {is=title K2Node_Event}
   > Set BaseballSlide(BaseballSlide=0)
   "To Float (Boolean)"(InBool=@n15)
   NOT Boolean(A=@n11)
   OR Boolean!(A=Get bIsSprinting(), B=Get bIsTacticalSprinting()) -> (ReturnValue)
   Reroute Node(InputPin:Wildcard) -> (OutputPin:Wildcard)
  ]}
]
variables [
 Alpha_Crouching {type=number}
 ALT_Skin_ADJ {type=number}
 ArmsVolume {type=boolean}
 BaseballSlide {type=number}
 Couching_Alpha {type=number}
 Crouch_Move {type=number}
 Crouch_Only {type=number}
 CrouchBackward {type=number}
 CrouchForward {type=number}
 CrouchLeft {type=number}
 CrouchRight {type=number}
 Delta Time X {type=number}
 Emote {type=number}
 GravityOverride {type=Vector}
 GrindRail {type=number}
 Idile and Moving {type=number}
 isCrouchStrafe float {type=number}
 isHangGliding {type=boolean}
 isIdle {type=number}
 isIdle_Bool {type=boolean}
 isIdleInFE {type=number}
 isNotIdle {type=number}
 IsTacticalandJump {type=number}
 IsTacticaljumping {type=number}
 isWearingBackbling {type=boolean}
 Jumping {type=boolean}
 KneeSliding {type=number}
 "Move Left " {type=number}
 Move Right {type=number}
 moveBackward {type=number}
 moveForward {type=number}
 moveLeft bool {type=boolean}
 moveRight bool {type=boolean}
 RBAN Settings Out {type=FortRigidBodyAnimNodeParameters}
 RigidBody_off {type=number}
 RunningAlpha {type=number}
 Sim Space Skirt {type=AnimPhysSimSpaceType}
 SkirtDynAlpha {type=number}
 SkyDive {type=boolean}
 SkyDive_Float {type=number}
 SkyDive_Float_Flap {type=number}
 SkyDive_Float_Off {type=number}
 SkyDive_Left {type=number}
 SkyDive_ON {type=number}
 strafeLeft float {type=number}
 strafeRight float {type=number}
]
