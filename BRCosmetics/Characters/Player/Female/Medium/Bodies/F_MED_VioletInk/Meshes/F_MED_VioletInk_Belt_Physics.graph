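# Physics asset for the belt mesh of the female medium VioletInk character cosmetic, covering a dangling strap loop at the hips.
# Two capsule bodies on the dynamic strap bones form a chain from the pelvis, with narrow swing limits and a zero twist limit that keep motion tight.
# Pelvis and left thigh capsules supply body volumes around the strap, and the thigh-to-pelvis constraint leaves all angular motion free.
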PhysicsAsset /BRCosmetics/Characters/Player/Female/Medium/Bodies/F_MED_VioletInk/Meshes/F_MED_VioletInk_Belt_Physics
bodies {
 dyn_strap_loop_01 [{
   shapeType=Capsule
   length=10.9555
   radius=4.4442
   shapeName=dyn_strap_loop_01_capsule
   center=<3.6125 -0.0247 2.0699>
   rotation=<rot 0 174.281 -79.3043>
  }]
 dyn_strap_loop_02 [{
   shapeType=Capsule
   length=6.5805
   radius=4.5692
   shapeName=dyn_strap_loop_01_capsule
   center=<1.8991 -0.1969 3.8811>
   rotation=<rot 0 174.281 -79.3043>
  }]
 pelvis [{shapeType=Capsule, length=6.75, radius=7.9978, shapeName=pelvis_capsule}]
 thigh_l [{
   shapeType=Capsule
   length=23.995
   radius=9.1176
   shapeName=thigh_l_capsule
   center=<-15.4554 0.1757 2.079>
   rotation=<rot 0.0925 93.7714 -88.9422>
  }]
}
constraints [
 {bone1Name=dyn_strap_loop_02, bone2Name=dyn_strap_loop_01, swing1LimitDegrees=5, swing2LimitDegrees=20, twistLimitDegrees=0}
 {bone1Name=dyn_strap_loop_01, bone2Name=pelvis, swing1LimitDegrees=5, swing2LimitDegrees=15, twistLimitDegrees=0}
 {
  bone1Name=thigh_l
  bone2Name=pelvis
  swing1LimitDegrees=45
  swing1Motion=Free
  swing2LimitDegrees=45
  swing2Motion=Free
  twistLimitDegrees=45
  twistMotion=Free
 }
]
