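# Physics asset for the belt of the VioletInk medium female player cosmetic, covering its dangling bag and strap loops.
# Single capsules cover the bag, two strap loops, pelvis and both thighs.
# The strap loops chain to the pelvis and the bag attaches to it directly, all with narrow swing limits and the bag's twist locked.
# Both thighs join the pelvis with free swing and twist.
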PhysicsAsset /BRCosmetics/Characters/Player/Female/Medium/Bodies/F_MED_VioletInk/Meshes/F_MED_VioletInk_Belt_Physics1
bodies {
 dyn_bag_01 [{
   shapeType=Capsule
   length=10.1397
   radius=6.3718
   shapeName=dyn_bag_01_capsule
   center=<6.4416 -0.8717 -0.62>
   rotation=<rot 0.0003 -5.2407 -88.0337>
  }]
 dyn_strap_loop_01 [{
   shapeType=Capsule
   length=10.9555
   radius=4.4442
   shapeName=dyn_strap_loop_01_capsule
   center=<0.6241 -0.9387 8.462>
   rotation=<rot 0 174.281 -79.3043>
  }]
 dyn_strap_loop_02 [{
   shapeType=Capsule
   length=10.0805
   radius=4.5692
   shapeName=dyn_strap_loop_01_capsule
   center=<1.9339 0.1502 2.0343>
   rotation=<rot -1.7651 178.96 -69.3764>
  }]
 pelvis [{shapeType=Capsule, length=6.75, radius=7.9978, shapeName=pelvis_capsule}]
 thigh_l [{shapeType=Capsule, length=36.995, radius=8.3676, shapeName=thigh_l_capsule, center=<-15.4554 0.1757 0.1493>, rotation=<rot -90 -0.0004 0>}]
 thigh_r [{
   shapeType=Capsule
   length=45.745
   radius=8.8676
   shapeName=thigh_r_capsule
   center=<18.7982 8.4438 -1.9934>
   rotation=<rot 0.1843 88.772 91.0457>
  }]
}
constraints [
 {bone1Name=dyn_strap_loop_02, bone2Name=dyn_strap_loop_01, swing1LimitDegrees=5, swing2LimitDegrees=20, twistLimitDegrees=0}
 {bone1Name=dyn_strap_loop_01, bone2Name=pelvis, swing1LimitDegrees=5, swing2LimitDegrees=15, twistLimitDegrees=0}
 {
  bone1Name=thigh_l
  bone2Name=pelvis
  swing1LimitDegrees=45
  swing1Motion=Free
  swing2LimitDegrees=45
  swing2Motion=Free
  twistLimitDegrees=45
  twistMotion=Free
 }
 {bone1Name=dyn_bag_01, bone2Name=pelvis, swing1LimitDegrees=5, swing2LimitDegrees=10, twistLimitDegrees=45, twistMotion=Locked}
 {
  bone1Name=thigh_r
  bone2Name=pelvis
  swing1LimitDegrees=45
  swing1Motion=Free
  swing2LimitDegrees=45
  swing2Motion=Free
  twistLimitDegrees=45
  twistMotion=Free
 }
]
