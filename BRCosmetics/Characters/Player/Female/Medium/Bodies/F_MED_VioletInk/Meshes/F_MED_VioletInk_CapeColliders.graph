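# Physics asset of cape colliders for the F_MED_VioletInk player cosmetic body.
# Capsules cover the pelvis, spine, clavicles, head and feet, while arm, leg and neck bodies have no shapes.
# Three back-skirt chains (left, middle, right) each link three capsule bodies in series from the pelvis.
# Their constraints leave swing and twist free, so the stored angle limits are not enforced.
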PhysicsAsset /BRCosmetics/Characters/Player/Female/Medium/Bodies/F_MED_VioletInk/Meshes/F_MED_VioletInk_CapeColliders
bodies {
 calf_l []
 calf_r []
 clavicle_l [{shapeType=Capsule, length=6.5623, radius=7.5194, center=<3.9566 -1.0862 -2.1127>, rotation=<rot -77.5566 177.3915 7.7314>}]
 clavicle_r [{shapeType=Capsule, length=6.5623, radius=7.5194, center=<-3.9566 1.0862 2.1127>, rotation=<rot -77.5566 177.3915 -172.2686>}]
 dyn_BackSkirt_l_02 [{shapeType=Capsule, length=14.5118, radius=5.9442, shapeName=dyn_BackSkirt_m_02_capsule, center=<6.0146 0 0>, rotation=<rot -0.0002 0 90>}]
 dyn_BackSkirt_l_03 [{
   shapeType=Capsule
   length=15.0118
   radius=5.4442
   shapeName=dyn_BackSkirt_m_02_capsule
   center=<1.1283 0 -1.1407>
   rotation=<rot -0.0002 0 90>
  }]
 dyn_BackSkirt_l_04 [{
   shapeType=Capsule
   length=15.0118
   radius=5.4442
   shapeName=dyn_BackSkirt_m_02_capsule
   center=<1.1283 0 -1.1407>
   rotation=<rot -0.0002 0 90>
  }]
 dyn_BackSkirt_m_02 [{shapeType=Capsule, length=14.5118, radius=5.9442, shapeName=dyn_BackSkirt_m_02_capsule, center=<6.0146 0 0>, rotation=<rot -0.0002 0 90>}]
 dyn_BackSkirt_m_03 [{
   shapeType=Capsule
   length=15.0118
   radius=5.4442
   shapeName=dyn_BackSkirt_m_02_capsule
   center=<1.1283 0 -1.1407>
   rotation=<rot -0.0002 0 90>
  }]
 dyn_BackSkirt_m_04 [{
   shapeType=Capsule
   length=15.0118
   radius=5.4442
   shapeName=dyn_BackSkirt_m_02_capsule
   center=<1.1283 0 -1.1407>
   rotation=<rot -0.0002 0 90>
  }]
 dyn_BackSkirt_r_02 [{shapeType=Capsule, length=14.5118, radius=5.9442, shapeName=dyn_BackSkirt_m_02_capsule, center=<6.0146 0 0>, rotation=<rot -0.0002 0 90>}]
 dyn_BackSkirt_r_03 [{
   shapeType=Capsule
   length=15.0118
   radius=5.4442
   shapeName=dyn_BackSkirt_m_02_capsule
   center=<1.1283 0 -1.1407>
   rotation=<rot -0.0002 0 90>
  }]
 dyn_BackSkirt_r_04 [{
   shapeType=Capsule
   length=15.0118
   radius=5.4442
   shapeName=dyn_BackSkirt_m_02_capsule
   center=<1.1283 0 -1.1407>
   rotation=<rot -0.0002 0 90>
  }]
 foot_l [{shapeType=Capsule, length=11.0765, radius=8.0344, center=<-4.1811 6.0027 -0.338>, rotation=<rot 0.0388 1.4701 -89.3982>}]
 foot_r [{shapeType=Capsule, length=11.0765, radius=8.0344, center=<4.1811 -6.0027 0.338>, rotation=<rot 0.0388 1.4701 90.6018>}]
 head [{shapeType=Capsule, length=0, radius=18, center=<9.0697 0.0746 0>, rotation=<rot 0 -90 0>}]
 lowerarm_l []
 lowerarm_r []
 neck_02 []
 pelvis [{shapeType=Capsule, length=15.8542, radius=15.8941, center=<-2.4471 -2.5289 -0.1981>}]
 spine_01 [{shapeType=Capsule, length=9.8888, radius=23.9888, center=<-13.1743 -0.9448 0>}]
 spine_02 [
  {shapeType=Capsule, length=2.5219, radius=9.7317, center=<-2.6885 -1.2641 0>}
  {shapeType=Capsule, length=20.6875, radius=9.3125, center=<-1.8317 1.9362 0>, rotation=<rot 90 0 0>}
 ]
 spine_03 [
  {shapeType=Capsule, length=19.875, radius=19.7354, center=<-2.1696 -0.7665 0>}
  {shapeType=Capsule, length=20.7188, radius=9.2812, center=<-1.2163 0.549 0>, rotation=<rot 90 180 -175.0732>}
 ]
 spine_04 [
  {shapeType=Capsule, length=5.6326, radius=13.7666, center=<2.789 -3.3911 0>}
  {shapeType=Capsule, length=10.0312, radius=9.2812, center=<-1.2326 1.308 0>, rotation=<rot 90 0 0>}
 ]
 spine_05 [{shapeType=Capsule, length=22.25, radius=17.4821, center=<-10.533 -2.0933 0>, rotation=<rot 90 9.9995 0>}]
 thigh_l []
 thigh_r []
 upperarm_l []
 upperarm_r []
}
constraints [
 {
  bone1Name=dyn_BackSkirt_l_02
  bone2Name=pelvis
  swing1LimitDegrees=45
  swing1Motion=Free
  swing2LimitDegrees=45
  swing2Motion=Free
  twistLimitDegrees=45
  twistMotion=Free
 }
 {
  bone1Name=dyn_BackSkirt_l_03
  bone2Name=dyn_BackSkirt_l_02
  swing1LimitDegrees=45
  swing1Motion=Free
  swing2LimitDegrees=45
  swing2Motion=Free
  twistLimitDegrees=45
  twistMotion=Free
 }
 {
  bone1Name=dyn_BackSkirt_l_04
  bone2Name=dyn_BackSkirt_l_03
  swing1LimitDegrees=45
  swing1Motion=Free
  swing2LimitDegrees=45
  swing2Motion=Free
  twistLimitDegrees=45
  twistMotion=Free
 }
 {
  bone1Name=dyn_BackSkirt_m_02
  bone2Name=pelvis
  swing1LimitDegrees=45
  swing1Motion=Free
  swing2LimitDegrees=45
  swing2Motion=Free
  twistLimitDegrees=45
  twistMotion=Free
 }
 {
  bone1Name=dyn_BackSkirt_m_03
  bone2Name=dyn_BackSkirt_m_02
  swing1LimitDegrees=45
  swing1Motion=Free
  swing2LimitDegrees=45
  swing2Motion=Free
  twistLimitDegrees=45
  twistMotion=Free
 }
 {
  bone1Name=dyn_BackSkirt_m_04
  bone2Name=dyn_BackSkirt_m_03
  swing1LimitDegrees=45
  swing1Motion=Free
  swing2LimitDegrees=45
  swing2Motion=Free
  twistLimitDegrees=45
  twistMotion=Free
 }
 {
  bone1Name=dyn_BackSkirt_r_02
  bone2Name=pelvis
  swing1LimitDegrees=45
  swing1Motion=Free
  swing2LimitDegrees=45
  swing2Motion=Free
  twistLimitDegrees=45
  twistMotion=Free
 }
 {
  bone1Name=dyn_BackSkirt_r_03
  bone2Name=dyn_BackSkirt_r_02
  swing1LimitDegrees=45
  swing1Motion=Free
  swing2LimitDegrees=45
  swing2Motion=Free
  twistLimitDegrees=45
  twistMotion=Free
 }
 {
  bone1Name=dyn_BackSkirt_r_04
  bone2Name=dyn_BackSkirt_r_03
  swing1LimitDegrees=45
  swing1Motion=Free
  swing2LimitDegrees=45
  swing2Motion=Free
  twistLimitDegrees=45
  twistMotion=Free
 }
]
