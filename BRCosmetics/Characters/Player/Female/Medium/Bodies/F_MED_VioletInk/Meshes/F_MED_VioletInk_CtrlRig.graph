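# Control Rig for the F_MED_VioletInk female body cosmetic that procedurally poses its skirt and strap bones to follow body motion.
# DeltaRotation functions measure each thigh against a null, weighting corrective bone transforms, with multiplied left/right weights for combined-leg poses.
# Movement-state variables such as crouching, sliding, DBNO and idle drive further offsets and constraints, many smoothed with AlphaInterp.
# Crouch correctives also use pelvis-height curves and spherical pose readers on the thighs.
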ControlRigBlueprint /BRCosmetics/Characters/Player/Female/Medium/Bodies/F_MED_VioletInk/Meshes/F_MED_VioletInk_CtrlRig
graphs [
 CollapseNode_ContainedGraph {nodes [
   n1 = RigVMFunctionEntryNode()
   &RigUnit_ModifyTransforms(
    ItemToModify="((Item=(Type=Bone,Name=\"dyn_frontSkirt_m_01\"),Transform=(Rotation=(X=-0.0807,Y=-0.1337,Z=0.0317,W=0.9872),Translation=(X=0,Y=0,Z=0),Scale3D=(X=1,Y=1,Z=1))),(Item=(Type=Bone,Name=\"dyn_frontSkirt_l_01\"),Transform=(Rotation=(X=0,Y=0,Z=-0.1305,W=0.9914),Transl ...(truncated)"
    WeightMinimum=0
    ExecutePin=@n1.ExecutePin
    Weight=@n1.Value
   )
   > &RigUnit_ModifyTransforms(
    ItemToModify="((Item=(Type=Bone,Name=\"dyn_SideSkirt_l_02\"),Transform=(Rotation=(X=-0.1806,Y=0.0109,Z=-0.3449,W=0.921),Translation=(X=0,Y=0,Z=0),Scale3D=(X=1,Y=1,Z=1))),(Item=(Type=Bone,Name=\"dyn_SideSkirt_r_02\"),Transform=(Rotation=(X=0.1806,Y=0.0109,Z=0.3449,W=0.921),T ...(truncated)"
    WeightMinimum=0
    Weight=@n1.Value
   )
   > &RigUnit_ModifyTransforms(
    ItemToModify="((Item=(Type=Bone,Name=\"dyn_BackSkirt_r_02\"),Transform=(Rotation=(X=0,Y=0,Z=0,W=1),Translation=(X=0,Y=0,Z=-1),Scale3D=(X=1,Y=1,Z=1))),(Item=(Type=Bone,Name=\"dyn_BackSkirt_r_03\"),Transform=(Rotation=(X=0,Y=0.1736,Z=0,W=0.9848),Translation=(X=0,Y=0,Z=0),Scal ...(truncated)"
    WeightMinimum=0
    Weight=@n1.Value
   )
   n2 = RigVMVariableNode() {variable=Moving2D}
   n3 = &RigVMFunction_AlphaInterp(
    InterpSpeedDecreasing=10
    InterpSpeedIncreasing=10
    Scale=1
    bClampResult=true
    Value=&RigVMFunction_MathDoubleMul(B=&RigVMFunction_MathBoolToFloat(Value=@n2.Value).Result, A=@n1.Value).Result
   )
   > n4 = &RigUnit_ModifyTransforms(
    ItemToModify=((Item=(Type=Bone,Name="dyn_frontSkirt_l_01"),Transform=(Rotation=(X=0.0057,Y=-0.1304,Z=0.0432,W=0.9905),Translation=(X=0,Y=0,Z=0),Scale3D=(X=1,Y=1,Z=1))))
    WeightMinimum=0
    Weight=@n3.Result
   )
   RigVMFunctionReturnNode(ExecutePin=@n4.ExecutePin)
  ]}
 New Function_ContainedGraph {nodes [
   RigVMFunctionEntryNode()
   > RigVMFunction_Sequence = RigVMAggregateNode()
   n1 = &RigVMFunction_AlphaInterp(InterpSpeedDecreasing=10, InterpSpeedIncreasing=10, Scale=1, bClampResult=true, Value=RigVMVariableNode() {variable=KneeSliding}.Value)
   &RigUnit_ModifyTransforms(
    ItemToModify="((Item=(Type=Bone,Name=\"dyn_BackSkirt_m_01\"),Transform=(Rotation=(X=0,Y=0.6156,Z=0,W=-0.7881),Translation=(X=0,Y=0,Z=0),Scale3D=(X=1,Y=1,Z=1))),(Item=(Type=Bone,Name=\"dyn_BackSkirt_r_01\"),Transform=(Rotation=(X=0,Y=0.4151,Z=0,W=-0.9098),Translation=(X=0,Y= ...(truncated)"
    WeightMinimum=0
    Weight=@n1.Result
    ExecutePin=@RigVMFunction_Sequence.A
   )
   > &RigUnit_ModifyTransforms(
    ItemToModify="((Item=(Type=Bone,Name=\"dyn_SideSkirt_l_01\"),Transform=(Rotation=(X=-0.3204,Y=-0.2265,Z=-0.3216,W=0.8617),Translation=(X=0,Y=0,Z=0),Scale3D=(X=1,Y=1,Z=1))),(Item=(Type=Bone,Name=\"dyn_SideSkirt_r_01\"),Transform=(Rotation=(X=-0.2665,Y=-0.1384,Z=0.0824,W=0.95 ...(truncated)"
    WeightMinimum=0
    Weight=@n1.Result
   )
   > &RigUnit_ModifyTransforms(
    ItemToModify="((Item=(Type=Bone,Name=\"dyn_frontSkirt_l_01\"),Transform=(Rotation=(X=0,Y=0.0008,Z=0,W=-1),Translation=(X=0,Y=0,Z=0),Scale3D=(X=1,Y=1,Z=1))),(Item=(Type=Bone,Name=\"dyn_frontSkirt_r_01\"),Transform=(Rotation=(X=0.1762,Y=-0.076,Z=-0.0395,W=0.9806),Translation= ...(truncated)"
    WeightMinimum=0
    Weight=@n1.Result
   )
   n2 = &RigVMFunction_AlphaInterp(InterpSpeedDecreasing=10, InterpSpeedIncreasing=10, Scale=1, bClampResult=true, Value=RigVMVariableNode() {variable=BaseballSlide}.Value)
   > &RigUnit_ModifyTransforms(
    ItemToModify="((Item=(Type=Bone,Name=\"dyn_BackSkirt_m_01\"),Transform=(Rotation=(X=0,Y=0.6156,Z=0,W=-0.7881),Translation=(X=0,Y=0,Z=0),Scale3D=(X=1,Y=1,Z=1))),(Item=(Type=Bone,Name=\"dyn_BackSkirt_r_01\"),Transform=(Rotation=(X=-0.0073,Y=-0.5884,Z=-0.1035,W=0.8019),Transla ...(truncated)"
    WeightMinimum=0
    Weight=@n2.Result
   )
   > &RigUnit_ModifyTransforms(
    ItemToModify="((Item=(Type=Bone,Name=\"dyn_SideSkirt_l_01\"),Transform=(Rotation=(X=-0.3204,Y=-0.2265,Z=-0.3216,W=0.8617),Translation=(X=0,Y=0,Z=0),Scale3D=(X=1,Y=1,Z=1))),(Item=(Type=Bone,Name=\"dyn_SideSkirt_r_01\"),Transform=(Rotation=(X=-0.2833,Y=0.1459,Z=-0.0113,W=-0.9 ...(truncated)"
    WeightMinimum=0
    Weight=@n2.Result
   )
   > &RigUnit_ModifyTransforms(
    ItemToModify="((Item=(Type=Bone,Name=\"dyn_frontSkirt_l_01\"),Transform=(Rotation=(X=-0.0902,Y=0.1354,Z=0.0378,W=0.9859),Translation=(X=0,Y=0,Z=0),Scale3D=(X=1,Y=1,Z=1))),(Item=(Type=Bone,Name=\"dyn_frontSkirt_r_01\"),Transform=(Rotation=(X=-0.5587,Y=0.1871,Z=0.0441,W=-0.80 ...(truncated)"
    WeightMinimum=0
    Weight=@n2.Result
   )
   n3 = &RigVMFunction_AlphaInterp(
    InterpSpeedDecreasing=10
    InterpSpeedIncreasing=10
    Scale=1
    bClampResult=true
    Value=&RigVMFunction_MathBoolToFloat(Value=RigVMVariableNode() {variable=DBNO}.Value).Result
   )
   > &RigUnit_ParentConstraint(
    AdvancedSettings=(InterpolationType=Average,RotationOrderForFilter=XZY)
    Child=(Type=Bone,Name="dyn_frontSkirt_l_01")
    Filter=(TranslationFilter=(bX=false,bY=false,bZ=false),RotationFilter=(bX=True,bY=True,bZ=True),ScaleFilter=(bX=false,bY=false,bZ=false))
    Parents=((Item=(Type=Bone,Name="thigh_twist_01_l"),Weight=1))
    Weight=1
    Parents.0.Weight=@n3.Result
   )
   > &RigUnit_ParentConstraint(
    AdvancedSettings=(InterpolationType=Average,RotationOrderForFilter=XZY)
    Child=(Type=Bone,Name="dyn_frontSkirt_r_01")
    Filter=(TranslationFilter=(bX=false,bY=false,bZ=false),RotationFilter=(bX=True,bY=True,bZ=True),ScaleFilter=(bX=false,bY=false,bZ=false))
    Parents=((Item=(Type=Bone,Name="thigh_twist_01_r"),Weight=1))
    Weight=1
    Parents.0.Weight=@n3.Result
   )
   > &RigUnit_ParentConstraint(
    AdvancedSettings=(InterpolationType=Average,RotationOrderForFilter=XZY)
    Child=(Type=Bone,Name="dyn_SideSkirt_l_01")
    Filter=(TranslationFilter=(bX=false,bY=false,bZ=false),RotationFilter=(bX=True,bY=True,bZ=True),ScaleFilter=(bX=false,bY=false,bZ=false))
    Parents=((Item=(Type=Bone,Name="thigh_twist_01_r"),Weight=1))
    Weight=1
    Parents.0.Weight=@n3.Result
   )
   > n4 = &RigUnit_ParentConstraint(
    AdvancedSettings=(InterpolationType=Average,RotationOrderForFilter=XZY)
    Child=(Type=Bone,Name="dyn_SideSkirt_r_01")
    Filter=(TranslationFilter=(bX=false,bY=false,bZ=false),RotationFilter=(bX=True,bY=True,bZ=True),ScaleFilter=(bX=false,bY=false,bZ=false))
    Parents=((Item=(Type=Bone,Name="thigh_twist_01_r"),Weight=1))
    Weight=1
    Parents.0.Weight=@n3.Result
   )
   DBNO_DEF = RigVMCollapseNode(ExecutePin=@n4.ExecutePin, Value=@n3.Result)
   n5 = &RigVMFunction_AlphaInterp(
    InterpSpeedDecreasing=10
    InterpSpeedIncreasing=10
    Scale=1
    bClampResult=true
    Value=&RigVMFunction_MathBoolToFloat(Value=RigVMVariableNode() {variable=Crouching}.Value).Result
   )
   &RigUnit_ModifyTransforms(
    ItemToModify="((Item=(Type=Bone,Name=\"dyn_frontSkirt_m_01\"),Transform=(Rotation=(X=0,Y=0.1736,Z=0,W=0.9848),Translation=(X=0,Y=0,Z=0),Scale3D=(X=1,Y=1,Z=1))),(Item=(Type=Bone,Name=\"dyn_SideSkirt_r_01\"),Transform=(Rotation=(X=0,Y=0,Z=0,W=1),Translation=(X=0,Y=0,Z=0),Scal ...(truncated)"
    WeightMinimum=0
    Weight=@n5.Result
    ExecutePin=@RigVMFunction_Sequence.B
   )
   n6 = &RigVMFunction_AnimEvalRichCurve(
    Curve=(EditorCurveData=(Keys=((),(Time=0.5,Value=1),(Time=0.8))))
    SourceMaximum=2
    SourceMinimum=0
    TargetMinimum=0
    Value=&RigVMFunction_MathDoubleRemap(
     SourceMaximum=75
     SourceMinimum=45
     TargetMaximum=1
     TargetMinimum=0
     bClamp=false
     Value=&RigUnit_GetRelativeTransformForItem(Child=(Type=Bone,Name="pelvis"), Parent=(Type=Bone,Name="root"), bChildInitial=False, bParentInitial=False).RelativeTransform.Translation.Z
    ).Result
   )
   n7 = RigVMVariableNode() {variable=Crouching}
   > n8 = &RigUnit_ModifyTransforms(
    ItemToModify="((Item=(Type=Bone,Name=\"dyn_frontSkirt_m_01\"),Transform=(Rotation=(X=0.0878,Y=-0.1743,Z=0.0002,W=0.9808),Translation=(X=0,Y=0,Z=0),Scale3D=(X=1,Y=1,Z=1))),(Item=(Type=Bone,Name=\"dyn_SideSkirt_r_01\"),Transform=(Rotation=(X=0,Y=0,Z=0,W=1),Translation=(X=0,Y= ...(truncated)"
    WeightMinimum=0
    Weight=&RigVMFunction_MathDoubleMul(B=@n6.Result, A=&RigVMFunction_MathBoolToFloat(Value=@n7.Value).Result).Result
   )
   RigVMFunction_ControlFlowBranch = &RigVMFunction_ControlFlowBranch(ExecuteContext=@n8.ExecutePin, Condition=RigVMVariableNode() {variable=Crouching}.Value)
   n9 = &RigUnit_SphericalPoseReader(
    ActiveRegionScaleFactors=(PositiveWidth=1.5,NegativeWidth=1.5,PositiveHeight=1.5,NegativeHeight=1.5)
    ActiveRegionSize=0.1
    Debug=(bDrawDebug=false,bDraw2D=False,bDrawLocalAxes=False,DebugScale=25,DebugSegments=20,DebugThickness=0.25)
    DriverAxis=(X=1,Y=0,Z=0)
    DriverItem=(Type=Bone,Name="thigh_r")
    FalloffRegionScaleFactors=(PositiveWidth=1,NegativeWidth=1,PositiveHeight=1.5,NegativeHeight=1.5)
    FalloffSize=0.2
    FlipHeightScaling=False
    FlipWidthScaling=False
    OptionalParentItem=(Type=Bone,Name="None")
    RotationOffset=(X=-40,Y=220,Z=0)
    ExecutePin=@RigVMFunction_ControlFlowBranch.True
   )
   > &RigUnit_ModifyTransforms(
    ItemToModify="((Item=(Type=Bone,Name=\"dyn_frontSkirt_l_01\"),Transform=(Rotation=(X=0.0038,Y=-0.0871,Z=0.0435,W=0.9952),Translation=(X=0,Y=0,Z=0),Scale3D=(X=1,Y=1,Z=1))),(Item=(Type=Bone,Name=\"dyn_SideSkirt_l_01\"),Transform=(Rotation=(X=0,Y=-0.0872,Z=0,W=0.9962),Translat ...(truncated)"
    WeightMinimum=0
    Weight=@n9.OutputParam
   )
   RigVMFunctionReturnNode(ExecuteContext=@RigVMFunction_Sequence.C)
   n10 = &RigUnit_SphericalPoseReader(
    ActiveRegionScaleFactors=(PositiveWidth=1,NegativeWidth=1,PositiveHeight=1,NegativeHeight=1)
    ActiveRegionSize=0.1
    Debug=(bDrawDebug=false,bDraw2D=False,bDrawLocalAxes=False,DebugScale=25,DebugSegments=20,DebugThickness=0.25)
    DriverAxis=(X=1,Y=0,Z=0)
    DriverItem=(Type=Bone,Name="thigh_r")
    FalloffRegionScaleFactors=(PositiveWidth=0.6,NegativeWidth=1,PositiveHeight=0.7,NegativeHeight=1)
    FalloffSize=0.3
    FlipHeightScaling=False
    FlipWidthScaling=False
    OptionalParentItem=(Type=Bone,Name="None")
    RotationOffset=(X=-65,Y=-170,Z=0)
    ExecutePin=@RigVMFunction_Sequence.D
   )
   n11 = &RigVMFunction_AnimEvalRichCurve(
    Curve="(EditorCurveData=(Keys=((),(InterpMode=RCIM_Cubic,Time=0.3217,Value=0.0979,ArriveTangent=0.6194,LeaveTangent=0.6194),(InterpMode=RCIM_Cubic,Time=0.6504,Value=0.4028,ArriveTangent=1.3299,LeaveTangent=1.3299),(InterpMode=RCIM_Cubic,TangentMode=RCTM_User,Time ...(truncated)"
    SourceMinimum=0
    TargetMinimum=0
    Value=@n10.OutputParam
   )
   > &RigUnit_ModifyTransforms(
    ItemToModify="((Item=(Type=Bone,Name=\"dyn_frontSkirt_r_01\"),Transform=(Rotation=(X=0.0112,Y=0.0691,Z=-0.1598,W=0.9847),Translation=(X=0,Y=0,Z=0),Scale3D=(X=1,Y=1,Z=1))),(Item=(Type=Bone,Name=\"dyn_frontSkirt_m_01\"),Transform=(Rotation=(X=0.0102,Y=0.1108,Z=0.0508,W=-0.992 ...(truncated)"
    WeightMinimum=0
    Weight=&RigVMFunction_MathDoubleMul(A=RigVMVariableNode() {variable=isCrouchStrafe}.Value, B=@n11.Result).Result
   )
   Remap = &RigVMFunction_MathDoubleRemap(SourceMaximum=1, SourceMinimum=0, TargetMaximum=0, TargetMinimum=1, bClamp=False, Value=@n11.Result)
   > &RigUnit_ModifyTransforms(
    ItemToModify=((Item=(Type=Bone,Name="None"),Transform=(Rotation=(X=0,Y=0,Z=0,W=1),Translation=(X=0,Y=0,Z=0),Scale3D=(X=1,Y=1,Z=1))))
    WeightMinimum=0
    Weight=&RigVMFunction_MathDoubleMul(B=@Remap.Result, A=RigVMVariableNode() {variable=isCrouchStrafe}.Value).Result
   )
   n12 = &RigUnit_SphericalPoseReader(
    ActiveRegionScaleFactors=(PositiveWidth=1,NegativeWidth=1,PositiveHeight=1,NegativeHeight=1)
    ActiveRegionSize=0.1
    Debug=(bDrawDebug=false,bDraw2D=False,bDrawLocalAxes=False,DebugScale=25,DebugSegments=20,DebugThickness=0.25)
    DriverAxis=(X=-6.9408,Y=0,Z=0)
    DriverItem=(Type=Bone,Name="thigh_l")
    FalloffRegionScaleFactors=(PositiveWidth=0.6,NegativeWidth=1,PositiveHeight=0.7,NegativeHeight=1)
    FalloffSize=0.3
    FlipHeightScaling=true
    FlipWidthScaling=true
    OptionalParentItem=(Type=Bone,Name="None")
    RotationOffset=(X=90,Y=220,Z=0)
    ExecutePin=@RigVMFunction_Sequence.E
   )
   n13 = &RigVMFunction_AnimEvalRichCurve(
    Curve="(EditorCurveData=(Keys=((),(InterpMode=RCIM_Cubic,Time=0.3217,Value=0.0979,ArriveTangent=0.6194,LeaveTangent=0.6194),(InterpMode=RCIM_Cubic,Time=0.6504,Value=0.4028,ArriveTangent=1.3299,LeaveTangent=1.3299),(InterpMode=RCIM_Cubic,TangentMode=RCTM_User,Time ...(truncated)"
    SourceMinimum=0
    TargetMinimum=0
    Value=@n12.OutputParam
   )
   > &RigUnit_ModifyTransforms(
    ItemToModify="((Item=(Type=Bone,Name=\"dyn_frontSkirt_l_01\"),Transform=(Rotation=(X=0.0064,Y=-0.0418,Z=-0.1503,W=-0.9877),Translation=(X=0,Y=0,Z=0),Scale3D=(X=1,Y=1,Z=1))),(Item=(Type=Bone,Name=\"dyn_frontSkirt_m_01\"),Transform=(Rotation=(X=0.0042,Y=0.0923,Z=-0.0453,W=-0. ...(truncated)"
    WeightMinimum=0
    Weight=&RigVMFunction_MathDoubleMul(A=RigVMVariableNode() {variable=isCrouchStrafe}.Value, B=@n13.Result).Result
   )
   > &RigUnit_ModifyTransforms(
    ItemToModify=((Item=(Type=Bone,Name="None"),Transform=(Rotation=(X=0,Y=0,Z=0,W=1),Translation=(X=0,Y=0,Z=0),Scale3D=(X=1,Y=1,Z=1))))
    WeightMinimum=0
    Weight=&RigVMFunction_MathDoubleMul(
     B=&RigVMFunction_MathDoubleRemap(SourceMaximum=1, SourceMinimum=0, TargetMaximum=0, TargetMinimum=1, bClamp=False, Value=@n13.Result).Result
     A=RigVMVariableNode() {variable=isCrouchStrafe}.Value
    ).Result
   )
   &RigUnit_ModifyTransforms(
    ItemToModify="((Item=(Type=Bone,Name=\"dyn_frontSkirt_l_01\"),Transform=(Rotation=(X=-0.13,Y=0.0114,Z=-0.0864,W=0.9877),Translation=(X=0,Y=0,Z=0),Scale3D=(X=1,Y=1,Z=1))),(Item=(Type=Bone,Name=\"dyn_SideSkirt_l_01\"),Transform=(Rotation=(X=0,Y=-0.0436,Z=0,W=0.999),Translatio ...(truncated)"
    WeightMinimum=0
    ExecutePin=@RigVMFunction_Sequence.F
    Weight=RigVMVariableNode() {variable=Idle}.Value
   )
   &RigUnit_ParentConstraint(
    AdvancedSettings=(InterpolationType=Average,RotationOrderForFilter=XZY)
    Child=(Type=Bone,Name="dyn_strap_loop_01")
    Filter=(TranslationFilter=(bX=false,bY=false,bZ=false),RotationFilter=(bX=True,bY=True,bZ=True),ScaleFilter=(bX=True,bY=True,bZ=True))
    Parents=((Item=(Type=Bone,Name="dyn_frontSkirt_l_03"),Weight=0.2),(Item=(Type=Bone,Name="dyn_SideSkirt_l_03"),Weight=0.2),(Item=(Type=Bone,Name="thigh_l"),Weight=0.3),(Item=(Type=Bone,Name="pelvis"),Weight=0.3))
    Weight=1
    ExecutePin=@RigVMFunction_Sequence.G
   )
   n14 = &RigVMFunction_AlphaInterp(
    InterpSpeedDecreasing=10
    InterpSpeedIncreasing=10
    Scale=1
    bClampResult=true
    Value=&RigVMFunction_MathBoolToFloat(Value=RigVMVariableNode() {variable=Crouching}.Value).Result
   )
   &RigUnit_ModifyTransforms(
    ExecutePin=()
    ItemToModify="((Item=(Type=Bone,Name=\"dyn_frontSkirt_l_01\"),Transform=(Rotation=(X=-0.2407,Y=0.142,Z=-0.0962,W=0.9553),Translation=(X=-2,Y=0,Z=0),Scale3D=(X=1,Y=1,Z=1))),(Item=(Type=Bone,Name=\"dyn_SideSkirt_l_01\"),Transform=(Rotation=(X=0.0841,Y=0.0286,Z=0.1285,W=0.9877 ...(truncated)"
    WeightMinimum=0
    Weight=@n14.Result
   )
   > &RigUnit_ModifyTransforms(
    ItemToModify="((Item=(Type=Bone,Name=\"dyn_frontSkirt_r_01\"),Transform=(Rotation=(X=0,Y=0.1736,Z=0,W=0.9848),Translation=(X=-2,Y=-2,Z=0),Scale3D=(X=1,Y=1,Z=1))),(Item=(Type=Bone,Name=\"dyn_SideSkirt_r_01\"),Transform=(Rotation=(X=0,Y=0.1305,Z=0,W=0.9914),Translation=(X=-2, ...(truncated)"
    WeightMinimum=0
    Weight=@n14.Result
   )
   > &RigUnit_ModifyTransforms(
    ItemToModify="((Item=(Type=Bone,Name=\"dyn_frontSkirt_m_01\"),Transform=(Rotation=(X=-0.2888,Y=0.1771,Z=0.0151,W=0.9407),Translation=(X=-1.9999,Y=0,Z=0),Scale3D=(X=1,Y=1,Z=1))),(Item=(Type=Bone,Name=\"dyn_BackSkirt_m_01\"),Transform=(Rotation=(X=0,Y=0,Z=0,W=1),Translation=( ...(truncated)"
    WeightMinimum=0
    Weight=@n14.Result
   )
   n15 = &RigVMFunction_AnimEvalRichCurve(
    Curve=(EditorCurveData=(Keys=((),(Time=0.5,Value=1),(Time=0.8))))
    SourceMaximum=2
    SourceMinimum=0
    TargetMinimum=0
    Value=&RigVMFunction_MathDoubleRemap(
     SourceMaximum=75
     SourceMinimum=45
     TargetMaximum=1
     TargetMinimum=0
     bClamp=false
     Value=&RigUnit_GetRelativeTransformForItem(Child=(Type=Bone,Name="pelvis"), Parent=(Type=Bone,Name="root"), bChildInitial=False, bParentInitial=False).RelativeTransform.Translation.Z
    ).Result
   )
   Multiply = &RigVMFunction_MathDoubleMul(B=@n15.Result, A=&RigVMFunction_MathBoolToFloat(Value=RigVMVariableNode() {variable=Crouching}.Value).Result)
   > &RigUnit_ModifyTransforms(
    ItemToModify="((Item=(Type=Bone,Name=\"dyn_frontSkirt_l_01\"),Transform=(Rotation=(X=-0.0726,Y=-0.1126,Z=0.0981,W=0.9861),Translation=(X=0,Y=0,Z=0),Scale3D=(X=1,Y=1,Z=1))),(Item=(Type=Bone,Name=\"dyn_SideSkirt_l_01\"),Transform=(Rotation=(X=0,Y=-0.1736,Z=0,W=0.9848),Transla ...(truncated)"
    WeightMinimum=0
    Weight=@Multiply.Result
   )
   > &RigUnit_ModifyTransforms(
    ItemToModify="((Item=(Type=Bone,Name=\"dyn_frontSkirt_r_01\"),Transform=(Rotation=(X=0.014,Y=-0.2199,Z=-0.118,W=0.9683),Translation=(X=-0.4323,Y=0,Z=1.4626),Scale3D=(X=1,Y=1,Z=1))),(Item=(Type=Bone,Name=\"dyn_SideSkirt_r_01\"),Transform=(Rotation=(X=-0.0134,Y=-0.0861,Z=-0.1 ...(truncated)"
    WeightMinimum=0
    Weight=@Multiply.Result
   )
   > &RigUnit_ModifyTransforms(
    ItemToModify="((Item=(Type=Bone,Name=\"dyn_BackSkirt_m_01\"),Transform=(Rotation=(X=0,Y=0,Z=0,W=1),Translation=(X=0,Y=0,Z=0),Scale3D=(X=1,Y=1,Z=1))),(Item=(Type=Bone,Name=\"dyn_BackSkirt_r_01\"),Transform=(Rotation=(X=0,Y=-0.0872,Z=0,W=0.9962),Translation=(X=0,Y=0,Z=0),Scal ...(truncated)"
    WeightMinimum=0
    Weight=@Multiply.Result
   )
   &RigUnit_ModifyTransforms(
    ExecutePin=()
    ItemToModify="((Item=(Type=Bone,Name=\"dyn_frontSkirt_m_01\"),Transform=(Rotation=(X=0,Y=0.0872,Z=0,W=0.9962),Translation=(X=0,Y=0,Z=0),Scale3D=(X=1,Y=1,Z=1))),(Item=(Type=Bone,Name=\"dyn_SideSkirt_r_01\"),Transform=(Rotation=(X=0,Y=0,Z=0,W=1),Translation=(X=0,Y=0,Z=0),Scal ...(truncated)"
    Weight=1
    WeightMinimum=0
   )
   &RigUnit_ModifyTransforms(
    ExecutePin=()
    ItemToModify=((Item=(Type=Bone,Name="dyn_strap_loop_01"),Transform=(Rotation=(X=0,Y=-0.0436,Z=0,W=0.999),Translation=(X=0,Y=0,Z=0),Scale3D=(X=1,Y=1,Z=1))))
    Weight=1
    WeightMinimum=0
   )
   &RigUnit_ParentConstraint(
    AdvancedSettings=(InterpolationType=Average,RotationOrderForFilter=XZY)
    Child=(Type=Bone,Name="dyn_strap_loop_02")
    ExecutePin=()
    Filter=(TranslationFilter=(bX=true,bY=true,bZ=true),RotationFilter=(bX=True,bY=True,bZ=True),ScaleFilter=(bX=True,bY=True,bZ=True))
    Parents=((Item=(Type=Bone,Name="dyn_frontSkirt_l_03"),Weight=0.3),(Item=(Type=Bone,Name="dyn_SideSkirt_l_03"),Weight=0.3),(Item=(Type=Bone,Name="pelvis"),Weight=0.4))
    Weight=1
   )
  ]}
 New Function_ContainedGraph {nodes [
   DeltaRotation v4_11 = RigVMFunctionReferenceNode(
    ExecuteContext=()
    Rotation Order=ZYX
    child=(Type=Bone,Name="thigh_l")
    outputMinMax=(X=0,Y=1)
    parent=(Type=Null,Name="thigh_l_null")
    premult=1
    rotationMinMax=(X=0,Y=12)
    triggerAxis=Z
    twistAxis=(X=1,Y=0,Z=0)
   )
   > &RigUnit_ModifyTransforms(
    ItemToModify="((Item=(Type=Bone,Name=\"dyn_BackSkirt_l_01\"),Transform=(Rotation=(X=-0.0055,Y=0.1044,Z=0.052,W=0.9932),Translation=(X=0,Y=0,Z=0),Scale3D=(X=1,Y=1,Z=1))),(Item=(Type=Bone,Name=\"dyn_SideSkirt_l_01\"),Transform=(Rotation=(X=0,Y=0,Z=0.0523,W=0.9986),Translation ...(truncated)"
    WeightMinimum=0
    Weight=@"DeltaRotation v4_11".outputWeight
   )
   > DeltaRotation v4_12 = RigVMFunctionReferenceNode(
    Rotation Order=ZYX
    child=(Type=Bone,Name="thigh_r")
    outputMinMax=(X=0,Y=1)
    parent=(Type=Null,Name="thigh_r_null")
    premult=1
    rotationMinMax=(X=0,Y=12)
    triggerAxis=Z
    twistAxis=(X=1,Y=0,Z=0)
   )
   > &RigUnit_ModifyTransforms(
    ItemToModify="((Item=(Type=Bone,Name=\"dyn_BackSkirt_r_01\"),Transform=(Rotation=(X=0.0039,Y=-0.0512,Z=0.0361,W=-0.998),Translation=(X=0,Y=0,Z=0),Scale3D=(X=1,Y=1,Z=1))),(Item=(Type=Bone,Name=\"dyn_SideSkirt_r_01\"),Transform=(Rotation=(X=0,Y=0,Z=-0.0523,W=0.9986),Translati ...(truncated)"
    WeightMinimum=0
    Weight=@"DeltaRotation v4_12".outputWeight
   )
   > &RigUnit_ParentConstraint(
    AdvancedSettings=(InterpolationType=Average,RotationOrderForFilter=XZY)
    Child=(Type=Bone,Name="dyn_BackSkirt_m_01")
    Parents=((Item=(Type=Bone,Name="dyn_BackSkirt_l_01"),Weight=0.5),(Item=(Type=Bone,Name="dyn_BackSkirt_r_01"),Weight=0.5))
    Weight=1
   )
   DeltaRotation v4_2 = RigVMFunctionReferenceNode(
    ExecuteContext=()
    Rotation Order=ZYX
    child=(Type=Bone,Name="thigh_l")
    outputMinMax=(X=0,Y=1)
    parent=(Type=Null,Name="thigh_l_null")
    premult=1
    rotationMinMax=(X=0,Y=-105)
    triggerAxis=Y
    twistAxis=(X=1,Y=0,Z=0)
   )
   > &RigUnit_ModifyTransforms(
    ItemToModify="((Item=(Type=Bone,Name=\"dyn_frontSkirt_l_01\"),Transform=(Rotation=(X=0.0045,Y=0.0568,Z=0.0792,W=0.9952),Translation=(X=0,Y=0,Z=0),Scale3D=(X=1,Y=1,Z=1))),(Item=(Type=Bone,Name=\"dyn_SideSkirt_l_01\"),Transform=(Rotation=(X=0,Y=-0.2792,Z=0,W=0.9602),Translati ...(truncated)"
    WeightMinimum=0
    Weight=@"DeltaRotation v4_2".outputWeight
   )
   > DeltaRotation v4_3 = RigVMFunctionReferenceNode(
    Rotation Order=ZYX
    child=(Type=Bone,Name="thigh_r")
    outputMinMax=(X=0,Y=1)
    parent=(Type=Null,Name="thigh_r_null")
    premult=1
    rotationMinMax=(X=0,Y=-90)
    triggerAxis=Y
    twistAxis=(X=1,Y=0,Z=0)
   )
   > &RigUnit_ModifyTransforms(
    ItemToModify="((Item=(Type=Bone,Name=\"dyn_frontSkirt_r_01\"),Transform=(Rotation=(X=0.0042,Y=-0.0292,Z=-0.1416,W=-0.9895),Translation=(X=0,Y=0,Z=0),Scale3D=(X=1,Y=1,Z=1))),(Item=(Type=Bone,Name=\"dyn_SideSkirt_r_01\"),Transform=(Rotation=(X=0.0077,Y=-0.2605,Z=-0.0339,W=0.9 ...(truncated)"
    WeightMinimum=0
    Weight=@"DeltaRotation v4_3".outputWeight
   )
   > &RigUnit_ModifyTransforms(
    ItemToModify="((Item=(Type=Bone,Name=\"dyn_frontSkirt_r_01\"),Transform=(Rotation=(X=0,Y=0,Z=0.1983,W=0.9801),Translation=(X=0,Y=0,Z=0),Scale3D=(X=1,Y=1,Z=1))),(Item=(Type=Bone,Name=\"dyn_frontSkirt_m_01\"),Transform=(Rotation=(X=0,Y=0,Z=-0.1132,W=0.9936),Translation=(X=0,Y ...(truncated)"
    WeightMinimum=0
    Weight=&RigVMFunction_MathDoubleMul(A=@"DeltaRotation v4_3".outputWeight, B=@"DeltaRotation v4_2".outputWeight).Result
   )
   > DeltaRotation v4_4 = RigVMFunctionReferenceNode(
    Rotation Order=ZYX
    child=(Type=Bone,Name="thigh_l")
    outputMinMax=(X=0,Y=1)
    parent=(Type=Null,Name="thigh_l_null")
    premult=1
    rotationMinMax=(X=0,Y=30)
    triggerAxis=Y
    twistAxis=(X=1,Y=0,Z=0)
   )
   > &RigUnit_ModifyTransforms(
    ItemToModify="((Item=(Type=Bone,Name=\"dyn_frontSkirt_l_01\"),Transform=(Rotation=(X=-0.0338,Y=-0.1261,Z=-0.2566,W=0.9577),Translation=(X=0,Y=0,Z=0),Scale3D=(X=1,Y=1,Z=1))),(Item=(Type=Bone,Name=\"dyn_SideSkirt_l_01\"),Transform=(Rotation=(X=0,Y=-0.1305,Z=0,W=0.9914),Transl ...(truncated)"
    WeightMinimum=0
    Weight=@"DeltaRotation v4_4".outputWeight
   )
   > DeltaRotation v4_5 = RigVMFunctionReferenceNode(
    Rotation Order=ZYX
    child=(Type=Bone,Name="thigh_r")
    outputMinMax=(X=0,Y=1)
    parent=(Type=Null,Name="thigh_r_null")
    premult=1
    rotationMinMax=(X=0,Y=30)
    triggerAxis=Y
    twistAxis=(X=1,Y=0,Z=0)
   )
   > &RigUnit_ModifyTransforms(
    ItemToModify="((Item=(Type=Bone,Name=\"dyn_frontSkirt_r_01\"),Transform=(Rotation=(X=0.0338,Y=-0.1261,Z=0.2566,W=0.9577),Translation=(X=0,Y=0,Z=0),Scale3D=(X=1,Y=1,Z=1))),(Item=(Type=Bone,Name=\"dyn_SideSkirt_r_01\"),Transform=(Rotation=(X=0,Y=-0.1305,Z=0,W=0.9914),Translat ...(truncated)"
    WeightMinimum=0
    Weight=@"DeltaRotation v4_5".outputWeight
   )
   DeltaRotation v4_21 = RigVMFunctionReferenceNode(
    ExecuteContext=()
    Rotation Order=ZYX
    child=(Type=Bone,Name="thigh_l")
    outputMinMax=(X=0,Y=1)
    parent=(Type=Null,Name="thigh_l_null")
    premult=1
    rotationMinMax=(X=0,Y=12)
    triggerAxis=Z
    twistAxis=(X=1,Y=0,Z=0)
   )
   > &RigUnit_ModifyTransforms(
    ItemToModify="((Item=(Type=Bone,Name=\"dyn_BackSkirt_l_01\"),Transform=(Rotation=(X=-0.0055,Y=0.1044,Z=0.052,W=0.9932),Translation=(X=0,Y=0,Z=0),Scale3D=(X=1,Y=1,Z=1))),(Item=(Type=Bone,Name=\"dyn_SideSkirt_l_01\"),Transform=(Rotation=(X=0,Y=0,Z=0.0523,W=0.9986),Translation ...(truncated)"
    WeightMinimum=0
    Weight=@"DeltaRotation v4_21".outputWeight
   )
   > DeltaRotation v4_22 = RigVMFunctionReferenceNode(
    Rotation Order=ZYX
    child=(Type=Bone,Name="thigh_r")
    outputMinMax=(X=0,Y=1)
    parent=(Type=Null,Name="thigh_r_null")
    premult=1
    rotationMinMax=(X=0,Y=12)
    triggerAxis=Z
    twistAxis=(X=1,Y=0,Z=0)
   )
   > &RigUnit_ModifyTransforms(
    ItemToModify="((Item=(Type=Bone,Name=\"dyn_BackSkirt_r_01\"),Transform=(Rotation=(X=0.0039,Y=-0.0512,Z=0.0361,W=-0.998),Translation=(X=0,Y=0,Z=0),Scale3D=(X=1,Y=1,Z=1))),(Item=(Type=Bone,Name=\"dyn_SideSkirt_r_01\"),Transform=(Rotation=(X=0,Y=0,Z=-0.0523,W=0.9986),Translati ...(truncated)"
    WeightMinimum=0
    Weight=@"DeltaRotation v4_22".outputWeight
   )
   > &RigUnit_ParentConstraint(
    AdvancedSettings=(InterpolationType=Average,RotationOrderForFilter=XZY)
    Child=(Type=Bone,Name="dyn_BackSkirt_m_01")
    Parents=((Item=(Type=Bone,Name="dyn_BackSkirt_l_01"),Weight=0.5),(Item=(Type=Bone,Name="dyn_BackSkirt_r_01"),Weight=0.5))
    Weight=1
   )
   n1 = RigVMFunctionEntryNode()
   RigVMFunction_Sequence = &RigVMFunction_Sequence(ExecuteContext=@n1.ExecuteContext)
   &RigUnit_ModifyTransforms(
    ItemToModify=((Item=(Type=Bone,Name="upperarm_r"),Transform=(Rotation=(X=0,Y=0,Z=0,W=1),Translation=(X=0,Y=0,Z=0),Scale3D=(X=1,Y=1,Z=1))),(Item=(Type=Bone,Name="upperarm_l"),Transform=(Rotation=(X=0,Y=0,Z=0,W=1),Translation=(X=0,Y=0,Z=0),Scale3D=(X=1,Y=1,Z=1))))
    Weight=1
    WeightMinimum=0
    ExecutePin=@RigVMFunction_Sequence.A
   )
   > &RigUnit_ModifyTransforms(
    ItemToModify=((Item=(Type=Bone,Name="thigh_l"),Transform=(Rotation=(X=0,Y=0,Z=0,W=1),Translation=(X=0,Y=0,Z=0),Scale3D=(X=1,Y=1,Z=1))))
    Weight=1
    WeightMinimum=0
   )
   > &RigUnit_ModifyTransforms(
    ItemToModify=((Item=(Type=Bone,Name="thigh_r"),Transform=(Rotation=(X=0,Y=0,Z=0,W=1),Translation=(X=0,Y=0,Z=0),Scale3D=(X=1,Y=1,Z=1))))
    Weight=1
    WeightMinimum=0
   )
   > Constraints = RigVMFunctionReferenceNode()
   > RigVMFunction_Sequence_2 = RigVMAggregateNode()
   DeltaRotation v4_13 = RigVMFunctionReferenceNode(
    Rotation Order=ZYX
    child=(Type=Bone,Name="thigh_l")
    outputMinMax=(X=0,Y=1)
    parent=(Type=Null,Name="thigh_l_null")
    premult=1
    rotationMinMax=(X=-20,Y=-100)
    triggerAxis=Z
    twistAxis=(X=1,Y=0,Z=0)
    ExecuteContext=@RigVMFunction_Sequence_2.A
   )
   > &RigUnit_ModifyTransforms(
    ItemToModify="((Item=(Type=Bone,Name=\"dyn_frontSkirt_l_01\"),Transform=(Rotation=(X=-0.0262,Y=-0.2996,Z=-0.0831,W=0.9501),Translation=(X=0,Y=0,Z=0),Scale3D=(X=1,Y=1,Z=1))),(Item=(Type=Bone,Name=\"dyn_SideSkirt_l_01\"),Transform=(Rotation=(X=-0.1299,Y=-0.204,Z=0.1805,W=0.95 ...(truncated)"
    WeightMinimum=0
    Weight=@"DeltaRotation v4_13".outputWeight
   )
   > DeltaRotation v4_14 = RigVMFunctionReferenceNode(
    Rotation Order=ZYX
    child=(Type=Bone,Name="thigh_r")
    outputMinMax=(X=0,Y=1)
    parent=(Type=Null,Name="thigh_r_null")
    premult=1
    rotationMinMax=(X=-10,Y=-100)
    triggerAxis=Z
    twistAxis=(X=1,Y=0,Z=0)
   )
   > &RigUnit_ModifyTransforms(
    ItemToModify="((Item=(Type=Bone,Name=\"MSTR_frontSkirt_r\"),Transform=(Rotation=(X=0,Y=0,Z=0,W=1),Translation=(X=4.5917,Y=-2.8644,Z=10.3059),Scale3D=(X=1,Y=1,Z=1))),(Item=(Type=Bone,Name=\"dyn_SideSkirt_r_01\"),Transform=(Rotation=(X=0.0659,Y=-0.0882,Z=-0.3121,W=0.9437),Tra ...(truncated)"
    WeightMinimum=0
    Weight=@"DeltaRotation v4_14".outputWeight
   )
   > &RigUnit_ModifyTransforms(
    ItemToModify="((Item=(Type=Bone,Name=\"dyn_frontSkirt_m_01\"),Transform=(Rotation=(X=0,Y=0.5501,Z=0,W=0.8351),Translation=(X=0,Y=0,Z=0),Scale3D=(X=1,Y=1,Z=1))),(Item=(Type=Bone,Name=\"dyn_BackSkirt_r_01\"),Transform=(Rotation=(X=0.0026,Y=0.0825,Z=-0.011,W=0.9965),Translatio ...(truncated)"
    WeightMinimum=0
    Weight=&RigVMFunction_MathDoubleMul(A=@"DeltaRotation v4_13".outputWeight, B=@"DeltaRotation v4_14".outputWeight).Result
   )
   DeltaRotation v4_19 = RigVMFunctionReferenceNode(
    Rotation Order=ZYX
    child=(Type=Bone,Name="thigh_l")
    outputMinMax=(X=0,Y=1)
    parent=(Type=Null,Name="thigh_l_null")
    premult=1
    rotationMinMax=(X=0,Y=90)
    triggerAxis=Z
    twistAxis=(X=1,Y=0,Z=0)
    ExecuteContext=@RigVMFunction_Sequence_2.B
   )
   > &RigUnit_ModifyTransforms(
    ItemToModify="((Item=(Type=Bone,Name=\"dyn_BackSkirt_m_01\"),Transform=(Rotation=(X=-0.0082,Y=-0.4923,Z=0.0182,W=0.8702),Translation=(X=0,Y=0,Z=0),Scale3D=(X=1,Y=1,Z=1))),(Item=(Type=Bone,Name=\"dyn_BackSkirt_l_01\"),Transform=(Rotation=(X=-0.1209,Y=0.4647,Z=0.1891,W=-0.856 ...(truncated)"
    WeightMinimum=0
    Weight=@"DeltaRotation v4_19".outputWeight
   )
   > DeltaRotation v4_20 = RigVMFunctionReferenceNode(
    Rotation Order=ZYX
    child=(Type=Bone,Name="thigh_r")
    debugRotation=false
    outputMinMax=(X=0,Y=1)
    parent=(Type=Null,Name="thigh_r_null")
    premult=1
    rotationMinMax=(X=0,Y=90)
    triggerAxis=Z
    twistAxis=(X=1,Y=0,Z=0)
   )
   > &RigUnit_ModifyTransforms(
    ItemToModify="((Item=(Type=Bone,Name=\"dyn_BackSkirt_m_01\"),Transform=(Rotation=(X=0,Y=0.475,Z=0,W=-0.88),Translation=(X=0,Y=0,Z=0),Scale3D=(X=1,Y=1,Z=1))),(Item=(Type=Bone,Name=\"dyn_BackSkirt_r_01\"),Transform=(Rotation=(X=-0.1147,Y=-0.4706,Z=0.227,W=0.8449),Translation= ...(truncated)"
    WeightMinimum=0
    Weight=@"DeltaRotation v4_20".outputWeight
   )
   > &RigUnit_ModifyTransforms(
    ItemToModify="((Item=(Type=Bone,Name=\"dyn_BackSkirt_m_01\"),Transform=(Rotation=(X=0,Y=0.4583,Z=0,W=0.8888),Translation=(X=0,Y=0,Z=0),Scale3D=(X=1,Y=1,Z=1))),(Item=(Type=Bone,Name=\"dyn_BackSkirt_l_01\"),Transform=(Rotation=(X=0.0027,Y=0.0537,Z=0.0509,W=0.9973),Translation ...(truncated)"
    WeightMinimum=0
    Weight=&RigVMFunction_MathDoubleMul(A=@"DeltaRotation v4_19".outputWeight, B=@"DeltaRotation v4_20".outputWeight).Result
   )
   DeltaRotation v4_15 = RigVMFunctionReferenceNode(
    Rotation Order=ZYX
    child=(Type=Bone,Name="thigh_l")
    outputMinMax=(X=0,Y=1)
    parent=(Type=Null,Name="thigh_l_null")
    premult=1
    rotationMinMax=(X=0,Y=-90)
    triggerAxis=Y
    twistAxis=(X=1,Y=0,Z=0)
    ExecuteContext=@RigVMFunction_Sequence_2.C
   )
   > &RigUnit_ModifyTransforms(
    ItemToModify="((Item=(Type=Bone,Name=\"dyn_frontSkirt_l_01\"),Transform=(Rotation=(X=0.0045,Y=0.0568,Z=0.0792,W=0.9952),Translation=(X=0,Y=0,Z=0),Scale3D=(X=1,Y=1,Z=1))),(Item=(Type=Bone,Name=\"dyn_SideSkirt_l_01\"),Transform=(Rotation=(X=0,Y=-0.2792,Z=0,W=0.9602),Translati ...(truncated)"
    WeightMinimum=0
    Weight=@"DeltaRotation v4_15".outputWeight
   )
   > DeltaRotation v4_16 = RigVMFunctionReferenceNode(
    Rotation Order=ZYX
    child=(Type=Bone,Name="thigh_r")
    outputMinMax=(X=0,Y=1)
    parent=(Type=Null,Name="thigh_r_null")
    premult=1
    rotationMinMax=(X=0,Y=-90)
    triggerAxis=Y
    twistAxis=(X=1,Y=0,Z=0)
   )
   > &RigUnit_ModifyTransforms(
    ItemToModify="((Item=(Type=Bone,Name=\"dyn_frontSkirt_r_01\"),Transform=(Rotation=(X=0.0042,Y=-0.0292,Z=-0.1416,W=-0.9895),Translation=(X=0,Y=0,Z=0),Scale3D=(X=1,Y=1,Z=1))),(Item=(Type=Bone,Name=\"dyn_SideSkirt_r_01\"),Transform=(Rotation=(X=0.0077,Y=-0.2605,Z=-0.0339,W=0.9 ...(truncated)"
    WeightMinimum=0
    Weight=@"DeltaRotation v4_16".outputWeight
   )
   > &RigUnit_ModifyTransforms(
    ItemToModify="((Item=(Type=Bone,Name=\"dyn_frontSkirt_r_01\"),Transform=(Rotation=(X=0,Y=0,Z=0.1983,W=0.9801),Translation=(X=0,Y=0,Z=0),Scale3D=(X=1,Y=1,Z=1))),(Item=(Type=Bone,Name=\"dyn_frontSkirt_m_01\"),Transform=(Rotation=(X=0,Y=0,Z=-0.1132,W=0.9936),Translation=(X=0,Y ...(truncated)"
    WeightMinimum=0
    Weight=&RigVMFunction_MathDoubleMul(A=@"DeltaRotation v4_16".outputWeight, B=@"DeltaRotation v4_15".outputWeight).Result
   )
   > DeltaRotation v4_17 = RigVMFunctionReferenceNode(
    Rotation Order=ZYX
    child=(Type=Bone,Name="thigh_l")
    outputMinMax=(X=0,Y=1)
    parent=(Type=Null,Name="thigh_l_null")
    premult=1
    rotationMinMax=(X=0,Y=30)
    triggerAxis=Y
    twistAxis=(X=1,Y=0,Z=0)
   )
   > &RigUnit_ModifyTransforms(
    ItemToModify="((Item=(Type=Bone,Name=\"dyn_frontSkirt_l_01\"),Transform=(Rotation=(X=-0.0338,Y=-0.1261,Z=-0.2566,W=0.9577),Translation=(X=0,Y=0,Z=0),Scale3D=(X=1,Y=1,Z=1))),(Item=(Type=Bone,Name=\"dyn_SideSkirt_l_01\"),Transform=(Rotation=(X=0,Y=-0.1305,Z=0,W=0.9914),Transl ...(truncated)"
    WeightMinimum=0
    Weight=@"DeltaRotation v4_17".outputWeight
   )
   > DeltaRotation v4_18 = RigVMFunctionReferenceNode(
    Rotation Order=ZYX
    child=(Type=Bone,Name="thigh_r")
    outputMinMax=(X=0,Y=1)
    parent=(Type=Null,Name="thigh_r_null")
    premult=1
    rotationMinMax=(X=0,Y=30)
    triggerAxis=Y
    twistAxis=(X=1,Y=0,Z=0)
   )
   > &RigUnit_ModifyTransforms(
    ItemToModify="((Item=(Type=Bone,Name=\"dyn_frontSkirt_r_01\"),Transform=(Rotation=(X=0.0338,Y=-0.1261,Z=0.2566,W=0.9577),Translation=(X=0,Y=0,Z=0),Scale3D=(X=1,Y=1,Z=1))),(Item=(Type=Bone,Name=\"dyn_SideSkirt_r_01\"),Transform=(Rotation=(X=0,Y=-0.1305,Z=0,W=0.9914),Translat ...(truncated)"
    WeightMinimum=0
    Weight=@"DeltaRotation v4_18".outputWeight
   )
   RigVMFunctionReturnNode(ExecuteContext=@RigVMFunction_Sequence.B)
   &RigUnit_ParentConstraint(
    AdvancedSettings=(InterpolationType=Average,RotationOrderForFilter=XZY)
    Child=(Type=Bone,Name="MSTR_frontSkirt_l")
    ExecutePin=()
    Filter=(TranslationFilter=(bX=false,bY=false,bZ=false),RotationFilter=(bX=True,bY=True,bZ=True),ScaleFilter=(bX=false,bY=false,bZ=false))
    Parents=((Item=(Type=Bone,Name="thigh_twist_01_l"),Weight=0.9),(Item=(Type=Bone,Name="pelvis"),Weight=0.1))
    Weight=1
   )
   > &RigUnit_ParentConstraint(
    AdvancedSettings=(InterpolationType=Average,RotationOrderForFilter=XZY)
    Child=(Type=Bone,Name="MSTR_SideSkirt_l")
    Filter=(TranslationFilter=(bX=false,bY=false,bZ=false),RotationFilter=(bX=True,bY=True,bZ=True),ScaleFilter=(bX=false,bY=false,bZ=false))
    Parents=((Item=(Type=Bone,Name="thigh_twist_01_l"),Weight=0.5),(Item=(Type=Bone,Name="pelvis"),Weight=0.5))
    Weight=1
   )
   > &RigUnit_ParentConstraint(
    AdvancedSettings=(InterpolationType=Average,RotationOrderForFilter=XZY)
    Child=(Type=Bone,Name="MSTR_BackSkirt_l")
    Filter=(TranslationFilter=(bX=false,bY=false,bZ=false),RotationFilter=(bX=True,bY=True,bZ=True),ScaleFilter=(bX=false,bY=false,bZ=false))
    Parents=((Item=(Type=Bone,Name="thigh_twist_01_l"),Weight=0.1),(Item=(Type=Bone,Name="pelvis"),Weight=0.9))
    Weight=1
   )
   > &RigUnit_ParentConstraint(
    AdvancedSettings=(InterpolationType=Average,RotationOrderForFilter=XZY)
    Child=(Type=Bone,Name="MSTR_frontSkirt_r")
    Filter=(TranslationFilter=(bX=false,bY=false,bZ=false),RotationFilter=(bX=True,bY=True,bZ=True),ScaleFilter=(bX=false,bY=false,bZ=false))
    Parents=((Item=(Type=Bone,Name="thigh_twist_01_r"),Weight=0.9),(Item=(Type=Bone,Name="pelvis"),Weight=0.1))
    Weight=1
   )
   > &RigUnit_ParentConstraint(
    AdvancedSettings=(InterpolationType=Average,RotationOrderForFilter=XZY)
    Child=(Type=Bone,Name="MSTR_SideSkirt_r")
    Filter=(TranslationFilter=(bX=false,bY=false,bZ=false),RotationFilter=(bX=True,bY=True,bZ=True),ScaleFilter=(bX=false,bY=false,bZ=false))
    Parents=((Item=(Type=Bone,Name="thigh_twist_01_r"),Weight=0.5),(Item=(Type=Bone,Name="pelvis"),Weight=0.5))
    Weight=1
   )
   > &RigUnit_ParentConstraint(
    AdvancedSettings=(InterpolationType=Average,RotationOrderForFilter=XZY)
    Child=(Type=Bone,Name="MSTR_BackSkirt_r")
    Filter=(TranslationFilter=(bX=false,bY=false,bZ=false),RotationFilter=(bX=True,bY=True,bZ=True),ScaleFilter=(bX=false,bY=false,bZ=false))
    Parents=((Item=(Type=Bone,Name="thigh_twist_01_r"),Weight=0.1),(Item=(Type=Bone,Name="pelvis"),Weight=0.9))
    Weight=1
   )
   > &RigUnit_ParentConstraint(
    AdvancedSettings=(InterpolationType=Average,RotationOrderForFilter=XZY)
    Child=(Type=Bone,Name="MSTR_frontSkirt_m")
    Filter=(TranslationFilter=(bX=true,bY=true,bZ=true),RotationFilter=(bX=True,bY=True,bZ=True),ScaleFilter=(bX=false,bY=false,bZ=false))
    Parents=((Item=(Type=Bone,Name="MSTR_frontSkirt_l"),Weight=0.5),(Item=(Type=Bone,Name="MSTR_frontSkirt_r"),Weight=0.5))
    Weight=1
   )
   > &RigUnit_ParentConstraint(
    AdvancedSettings=(InterpolationType=Average,RotationOrderForFilter=XZY)
    Child=(Type=Bone,Name="MSTR_BackSkirt_m")
    Filter=(TranslationFilter=(bX=true,bY=true,bZ=true),RotationFilter=(bX=True,bY=True,bZ=True),ScaleFilter=(bX=false,bY=false,bZ=false))
    Parents=((Item=(Type=Bone,Name="MSTR_BackSkirt_l"),Weight=0.5),(Item=(Type=Bone,Name="MSTR_BackSkirt_r"),Weight=0.5))
    Weight=1
   )
   > RigVMFunction_Sequence_1 = RigVMAggregateNode()
   DeltaRotation v4_8 = RigVMFunctionReferenceNode(
    Rotation Order=ZYX
    child=(Type=Bone,Name="thigh_l")
    outputMinMax=(X=0,Y=1)
    parent=(Type=Null,Name="thigh_l_null")
    premult=1
    rotationMinMax=(X=-20,Y=-120)
    triggerAxis=Z
    twistAxis=(X=1,Y=0,Z=0)
    ExecuteContext=@RigVMFunction_Sequence_1.A
   )
   > &RigUnit_ModifyTransforms(
    ItemToModify="((Item=(Type=Bone,Name=\"dyn_frontSkirt_l_01\"),Transform=(Rotation=(X=-0.0047,Y=0.054,Z=-0.0527,W=-0.9971),Translation=(X=0,Y=2,Z=2),Scale3D=(X=1,Y=1,Z=1))),(Item=(Type=Bone,Name=\"dyn_SideSkirt_l_01\"),Transform=(Rotation=(X=0.0302,Y=0.0957,Z=-0.2189,W=-0.97 ...(truncated)"
    WeightMinimum=0
    Weight=@"DeltaRotation v4_8".outputWeight
   )
   > DeltaRotation v4_9 = RigVMFunctionReferenceNode(
    Rotation Order=ZYX
    child=(Type=Bone,Name="thigh_r")
    outputMinMax=(X=0,Y=1)
    parent=(Type=Null,Name="thigh_r_null")
    premult=1
    rotationMinMax=(X=-20,Y=-120)
    triggerAxis=Z
    twistAxis=(X=1,Y=0,Z=0)
   )
   > &RigUnit_ModifyTransforms(
    ItemToModify="((Item=(Type=Bone,Name=\"dyn_frontSkirt_r_01\"),Transform=(Rotation=(X=0.007,Y=0.0452,Z=-0.1006,W=0.9939),Translation=(X=0.143,Y=2.0045,Z=-6.8034),Scale3D=(X=1,Y=1,Z=1))),(Item=(Type=Bone,Name=\"dyn_SideSkirt_r_01\"),Transform=(Rotation=(X=0.0266,Y=-0.2541,Z=0 ...(truncated)"
    WeightMinimum=0
    Weight=@"DeltaRotation v4_9".outputWeight
   )
   Multiply = &RigVMFunction_MathDoubleMul(A=@"DeltaRotation v4_9".outputWeight, B=@"DeltaRotation v4_8".outputWeight)
   > &RigUnit_ModifyTransforms(
    ItemToModify="((Item=(Type=Bone,Name=\"dyn_frontSkirt_r_01\"),Transform=(Rotation=(X=-0.0151,Y=-0.173,Z=-0.0858,W=0.9811),Translation=(X=0,Y=0,Z=0),Scale3D=(X=1,Y=1,Z=1))),(Item=(Type=Bone,Name=\"dyn_frontSkirt_l_01\"),Transform=(Rotation=(X=0,Y=0.1305,Z=0,W=0.9914),Transla ...(truncated)"
    WeightMinimum=0
    Weight=@Multiply.Result
   )
   > DeltaRotation v4_6 = RigVMFunctionReferenceNode(
    Rotation Order=ZYX
    child=(Type=Bone,Name="thigh_l")
    debugRotation=false
    outputMinMax=(X=0,Y=1)
    parent=(Type=Null,Name="thigh_l_null")
    premult=1
    rotationMinMax=(X=0,Y=-12)
    triggerAxis=Z
    twistAxis=(X=1,Y=0,Z=0)
   )
   > &RigUnit_ModifyTransforms(
    ItemToModify="((Item=(Type=Bone,Name=\"dyn_frontSkirt_l_01\"),Transform=(Rotation=(X=-0.0889,Y=0.0396,Z=0.0472,W=0.9941),Translation=(X=0,Y=0,Z=0),Scale3D=(X=1,Y=1,Z=1))),(Item=(Type=Bone,Name=\"dyn_SideSkirt_l_01\"),Transform=(Rotation=(X=0.0872,Y=0,Z=0,W=0.9962),Translati ...(truncated)"
    WeightMinimum=0
    Weight=@"DeltaRotation v4_6".outputWeight
   )
   > DeltaRotation v4_7 = RigVMFunctionReferenceNode(
    Rotation Order=ZYX
    child=(Type=Bone,Name="thigh_r")
    outputMinMax=(X=0,Y=1)
    parent=(Type=Null,Name="thigh_r_null")
    premult=1
    rotationMinMax=(X=0,Y=-12)
    triggerAxis=Z
    twistAxis=(X=1,Y=0,Z=0)
   )
   > &RigUnit_ModifyTransforms(
    ItemToModify="((Item=(Type=Bone,Name=\"dyn_frontSkirt_r_01\"),Transform=(Rotation=(X=0.0871,Y=0.0038,Z=0.0435,W=0.9952),Translation=(X=0,Y=0,Z=0),Scale3D=(X=1,Y=1,Z=1))),(Item=(Type=Bone,Name=\"dyn_SideSkirt_r_01\"),Transform=(Rotation=(X=0,Y=0,Z=0.1305,W=0.9914),Translatio ...(truncated)"
    WeightMinimum=0
    Weight=@"DeltaRotation v4_7".outputWeight
   )
   > &RigUnit_ModifyTransforms(
    ItemToModify=((Item=(Type=Bone,Name="dyn_frontSkirt_m_01"),Transform=(Rotation=(X=0,Y=-0.048,Z=0,W=0.9988),Translation=(X=0,Y=0,Z=0),Scale3D=(X=1,Y=1,Z=1))))
    WeightMinimum=0
    Weight=&RigVMFunction_MathDoubleMul(B=@"DeltaRotation v4_7".outputWeight, A=@"DeltaRotation v4_6".outputWeight).Result
   )
   DeltaRotation v4_1 = RigVMFunctionReferenceNode(
    Rotation Order=ZYX
    child=(Type=Bone,Name="thigh_l")
    outputMinMax=(X=0,Y=1)
    parent=(Type=Null,Name="thigh_l_null")
    premult=1
    rotationMinMax=(X=0,Y=75)
    triggerAxis=Z
    twistAxis=(X=1,Y=0,Z=0)
    ExecuteContext=@RigVMFunction_Sequence_1.B
   )
   > &RigUnit_ModifyTransforms(
    ItemToModify="((Item=(Type=Bone,Name=\"dyn_BackSkirt_m_01\"),Transform=(Rotation=(X=0,Y=-0.4217,Z=0,W=0.9067),Translation=(X=0,Y=0,Z=0),Scale3D=(X=1,Y=1,Z=1))),(Item=(Type=Bone,Name=\"dyn_BackSkirt_l_01\"),Transform=(Rotation=(X=0,Y=0,Z=0,W=1),Translation=(X=0,Y=0,Z=2),Scal ...(truncated)"
    WeightMinimum=0
    Weight=@"DeltaRotation v4_1".outputWeight
   )
   > DeltaRotation v4_10 = RigVMFunctionReferenceNode(
    Rotation Order=ZYX
    child=(Type=Bone,Name="thigh_r")
    debugRotation=false
    outputMinMax=(X=0,Y=1)
    parent=(Type=Null,Name="thigh_r_null")
    premult=1
    rotationMinMax=(X=0,Y=75)
    triggerAxis=Z
    twistAxis=(X=1,Y=0,Z=0)
   )
   > &RigUnit_ModifyTransforms(
    ItemToModify="((Item=(Type=Bone,Name=\"dyn_BackSkirt_m_02\"),Transform=(Rotation=(X=-0.1992,Y=0.0526,Z=0.1037,W=0.973),Translation=(X=0,Y=0,Z=0),Scale3D=(X=1,Y=1,Z=1))),(Item=(Type=Bone,Name=\"dyn_BackSkirt_r_01\"),Transform=(Rotation=(X=0.2046,Y=-0.3538,Z=0.298,W=0.8626),T ...(truncated)"
    WeightMinimum=0
    Weight=@"DeltaRotation v4_10".outputWeight
   )
   &RigUnit_ModifyTransforms(
    ExecutePin=()
    ItemToModify="((Item=(Type=Bone,Name=\"dyn_frontSkirt_l_01\"),Transform=(Rotation=(X=0.0121,Y=-0.1732,Z=0.0687,W=0.9824),Translation=(X=0,Y=2,Z=2),Scale3D=(X=1,Y=1,Z=1))),(Item=(Type=Bone,Name=\"dyn_SideSkirt_l_01\"),Transform=(Rotation=(X=-0.0055,Y=-0.2189,Z=0.1199,W=0.968 ...(truncated)"
    Weight=1
    WeightMinimum=0
   )
   &RigUnit_ModifyTransforms(
    ExecutePin=()
    ItemToModify="((Item=(Type=Bone,Name=\"dyn_frontSkirt_r_01\"),Transform=(Rotation=(X=-0.0076,Y=-0.0868,Z=-0.0868,W=0.9924),Translation=(X=0,Y=2,Z=0),Scale3D=(X=1,Y=1,Z=1))),(Item=(Type=Bone,Name=\"dyn_SideSkirt_r_01\"),Transform=(Rotation=(X=0.0334,Y=-0.3812,Z=0.0805,W=0.92 ...(truncated)"
    Weight=1
    WeightMinimum=0
   )
   &RigUnit_ModifyTransforms(
    ExecutePin=()
    ItemToModify="((Item=(Type=Bone,Name=\"dyn_frontSkirt_r_01\"),Transform=(Rotation=(X=0.0411,Y=-0.0157,Z=-0.1789,W=-0.9829),Translation=(X=2,Y=2,Z=0),Scale3D=(X=1,Y=1,Z=1))),(Item=(Type=Bone,Name=\"dyn_SideSkirt_r_01\"),Transform=(Rotation=(X=0.0113,Y=-0.2586,Z=0.0421,W=0.96 ...(truncated)"
    Weight=1
    WeightMinimum=0
   )
   &RigUnit_ModifyTransforms(
    ExecutePin=()
    ItemToModify="((Item=(Type=Bone,Name=\"dyn_frontSkirt_l_01\"),Transform=(Rotation=(X=0.025,Y=-0.0364,Z=0.4251,W=0.9041),Translation=(X=2,Y=-2,Z=0),Scale3D=(X=1,Y=1,Z=1))),(Item=(Type=Bone,Name=\"dyn_SideSkirt_l_01\"),Transform=(Rotation=(X=-0.0113,Y=-0.2586,Z=-0.0421,W=0.96 ...(truncated)"
    Weight=1
    WeightMinimum=0
   )
   &RigUnit_ModifyTransforms(
    ExecutePin=()
    ItemToModify="((Item=(Type=Bone,Name=\"dyn_BackSkirt_m_02\"),Transform=(Rotation=(X=0.1724,Y=-0.1015,Z=-0.0233,W=0.9795),Translation=(X=0,Y=0,Z=0),Scale3D=(X=1,Y=1,Z=1))),(Item=(Type=Bone,Name=\"dyn_BackSkirt_l_01\"),Transform=(Rotation=(X=-0.0325,Y=0.3914,Z=0.2615,W=-0.881 ...(truncated)"
    Weight=1
    WeightMinimum=0
   )
   &RigUnit_ModifyTransforms(
    ExecutePin=()
    ItemToModify="((Item=(Type=Bone,Name=\"dyn_BackSkirt_m_02\"),Transform=(Rotation=(X=-0.1992,Y=0.0526,Z=0.1037,W=0.973),Translation=(X=0,Y=0,Z=0),Scale3D=(X=1,Y=1,Z=1))),(Item=(Type=Bone,Name=\"dyn_BackSkirt_r_01\"),Transform=(Rotation=(X=0.2046,Y=-0.3538,Z=0.298,W=0.8626),T ...(truncated)"
    Weight=1
    WeightMinimum=0
   )
   &RigUnit_ModifyTransforms(
    ExecutePin=()
    ItemToModify="((Item=(Type=Bone,Name=\"dyn_frontSkirt_l_01\"),Transform=(Rotation=(X=0,Y=0.1357,Z=0,W=-0.9907),Translation=(X=0,Y=0,Z=0),Scale3D=(X=1,Y=1,Z=1))),(Item=(Type=Bone,Name=\"dyn_SideSkirt_l_01\"),Transform=(Rotation=(X=0.0295,Y=0.0627,Z=-0.2834,W=-0.9565),Transla ...(truncated)"
    Weight=1
    WeightMinimum=0
   )
   &RigUnit_ModifyTransforms(
    ExecutePin=()
    ItemToModify="((Item=(Type=Bone,Name=\"dyn_frontSkirt_r_01\"),Transform=(Rotation=(X=0,Y=-0.1079,Z=0,W=0.9942),Translation=(X=0,Y=0,Z=0),Scale3D=(X=1,Y=1,Z=1))),(Item=(Type=Bone,Name=\"dyn_SideSkirt_r_01\"),Transform=(Rotation=(X=0.0659,Y=-0.0882,Z=-0.3121,W=0.9437),Transla ...(truncated)"
    Weight=1
    WeightMinimum=0
   )
   n2 = &RigVMFunction_AnimEvalRichCurve(
    Curve=(EditorCurveData=(Keys=((),(Time=0.5,Value=1),(Time=0.8))))
    SourceMaximum=2
    SourceMinimum=0
    TargetMinimum=0
    Value=&RigVMFunction_MathDoubleRemap(
     SourceMaximum=75
     SourceMinimum=45
     TargetMaximum=1
     TargetMinimum=0
     bClamp=false
     Value=&RigUnit_GetRelativeTransformForItem(Child=(Type=Bone,Name="pelvis"), Parent=(Type=Bone,Name="root"), bChildInitial=False, bParentInitial=False).RelativeTransform.Translation.Z
    ).Result
   )
   n3 = RigVMVariableNode() {variable=Crouching}
   &RigUnit_ModifyTransforms(
    ExecutePin=()
    ItemToModify="((Item=(Type=Bone,Name=\"dyn_frontSkirt_m_01\"),Transform=(Rotation=(X=0.0868,Y=-0.0868,Z=0.0076,W=0.9924),Translation=(X=0,Y=0,Z=0),Scale3D=(X=1,Y=1,Z=1))),(Item=(Type=Bone,Name=\"dyn_frontSkirt_r_01\"),Transform=(Rotation=(X=0,Y=0,Z=0,W=1),Translation=(X=0,Y ...(truncated)"
    WeightMinimum=0
    Weight=&RigVMFunction_MathDoubleMul(B=@n2.Result, A=&RigVMFunction_MathBoolToFloat(Value=@n3.Value).Result).Result
   )
  ]}
 New Function_ContainedGraph {nodes [
   n1 = RigVMFunctionEntryNode()
   RigVMFunction_ControlFlowBranch = &RigVMFunction_ControlFlowBranch(Condition=@n1.Debug, ExecuteContext=@n1.ExecuteContext)
   RigVMFunctionReturnNode(ExecuteContext=@RigVMFunction_ControlFlowBranch.Completed)
   DISPATCH_RigVMDispatch_ArrayIterator = RigVMDispatchNode(
    Array=/Script/ControlRig.RigUnit_CollectionGetAll(TypeToSearch=Control) {is=struct RigVMUnitNode}.Items
    ExecuteContext=@RigVMFunction_ControlFlowBranch.True
   ) {template="DISPATCH_RigVMDispatch_ArrayIterator(in Array,out Element,out Index,out Count,out Ratio)"}
   Replace = /Script/RigVM.RigVMFunction_NameReplace(Old=_ctrl, Name=@DISPATCH_RigVMDispatch_ArrayIterator.Element.Name) {is=struct RigVMUnitNode, template="Replace::Execute(in Name,in Old,in New,out Result)"}
   > &RigUnit_SetTransform(
    Item=(Type=Bone,Name="spine_01")
    Weight=1
    bInitial=False
    Item.Name=@Replace.Result
    Value=&RigUnit_GetTransform(Item=(Type=Control,Name="spine_01_ctrl"), bInitial=False, Item.Name=@DISPATCH_RigVMDispatch_ArrayIterator.Element.Name).Transform
   )
   Constraints = RigVMFunctionReferenceNode(ExecuteContext=())
  ]}
 New Function_ContainedGraph {nodes [
   n1 = RigVMFunctionEntryNode()
   RigVMFunction_MathBoolToInteger = &RigVMFunction_MathBoolToInteger(Value=@n1.onlyTwist)
   RigVMFunction_MathQuaternionSwingTwist = &RigVMFunction_MathQuaternionSwingTwist(
    TwistAxis=@n1.twistAxis
    Input=&RigUnit_GetRelativeTransformForItem(bChildInitial=False, bParentInitial=False, Child=@n1.child, Parent=@n1.parent).RelativeTransform.Rotation
   )
   n2 = RigVMDispatchNode(
    Values=((),())
    Index=@RigVMFunction_MathBoolToInteger.Result
    Values.0=@RigVMFunction_MathQuaternionSwingTwist.Swing
    Values.1=@RigVMFunction_MathQuaternionSwingTwist.Twist
   ) {template="DISPATCH_RigVMDispatch_SelectInt32(in Index,in Values,out Result)"}
   RigVMFunction_MathQuaternionToEuler = &RigVMFunction_MathQuaternionToEuler(Value=@n2.Result, RotationOrder=@n1.Rotation Order)
   > DISPATCH_RigVMDispatch_Print = RigVMDispatchNode(ScreenColor=(R=1,G=1,B=1,A=1), ScreenDuration=1.5, Enabled=@n1.debugRotation, Value=@RigVMFunction_MathQuaternionToEuler.Result) {template="DISPATCH_RigVMDispatch_Print(in Prefix,in Value,in Enabled,in ScreenDuration,in ScreenColor)"}
   Cast = RigVMDispatchNode(Value=@n1.triggerAxis) {template="DISPATCH_RigVMDispatch_CastEnumToInt(in Value,out Result)"}
   DISPATCH_RigVMDispatch_SelectInt32 = RigVMDispatchNode(
    Values=(0,0,0)
    Index=@Cast.Result
    Values.0=@RigVMFunction_MathQuaternionToEuler.Result.X
    Values.1=@RigVMFunction_MathQuaternionToEuler.Result.Y
    Values.2=@RigVMFunction_MathQuaternionToEuler.Result.Z
   ) {template="DISPATCH_RigVMDispatch_SelectInt32(in Index,in Values,out Result)"}
   Remap = &RigVMFunction_MathDoubleRemap(
    bClamp=False
    Value=@DISPATCH_RigVMDispatch_SelectInt32.Result
    TargetMinimum=@n1.outputMinMax.X
    TargetMaximum=@n1.outputMinMax.Y
    SourceMinimum=@n1.rotationMinMax.X
    SourceMaximum=@n1.rotationMinMax.Y
   )
   Absolute = &RigVMFunction_MathDoubleAbs(Value=@Remap.Result)
   Multiply = &RigVMFunction_MathDoubleMul(
    A=&RigVMFunction_MathDoubleClamp(Value=RigVMDispatchNode(True=@Absolute.Result, Condition=@n1.absoluteValue, False=@Remap.Result) {template="DISPATCH_RigVMDispatch_If(in Condition,in True,in False,out Result)"}.Result).Result
    B=@n1.premult
   )
   > RigVMFunctionReturnNode(outputQuat=@n2.Result, outputWeight=@Multiply.Result)
  ]}
 New Function_ContainedGraph {nodes [
   n1 = RigVMFunctionEntryNode()
   RigVMFunction_Sequence = &RigVMFunction_Sequence(ExecuteContext=@n1.ExecuteContext)
   &RigUnit_ParentConstraint(
    AdvancedSettings=(InterpolationType=Average,RotationOrderForFilter=XZY)
    Child=(Type=Bone,Name="MSTR_frontSkirt_l")
    Filter=(TranslationFilter=(bX=false,bY=false,bZ=false),RotationFilter=(bX=True,bY=True,bZ=True),ScaleFilter=(bX=True,bY=True,bZ=True))
    Parents=((Item=(Type=Bone,Name="pelvis"),Weight=0.2),(Item=(Type=Bone,Name="thigh_twist_01_l"),Weight=0.8))
    Weight=1
    ExecutePin=@RigVMFunction_Sequence.A
   )
   > &RigUnit_ParentConstraint(
    AdvancedSettings=(InterpolationType=Average,RotationOrderForFilter=XZY)
    Child=(Type=Bone,Name="MSTR_SideSkirt_l")
    Filter=(TranslationFilter=(bX=false,bY=false,bZ=false),RotationFilter=(bX=True,bY=True,bZ=True),ScaleFilter=(bX=True,bY=True,bZ=True))
    Parents=((Item=(Type=Bone,Name="pelvis"),Weight=0.5),(Item=(Type=Bone,Name="thigh_twist_01_l"),Weight=0.5))
    Weight=1
   )
   > &RigUnit_ParentConstraint(
    AdvancedSettings=(InterpolationType=Average,RotationOrderForFilter=XZY)
    Child=(Type=Bone,Name="MSTR_BackSkirt_l")
    Filter=(TranslationFilter=(bX=false,bY=false,bZ=false),RotationFilter=(bX=True,bY=True,bZ=True),ScaleFilter=(bX=True,bY=True,bZ=True))
    Parents=((Item=(Type=Bone,Name="pelvis"),Weight=0.9),(Item=(Type=Bone,Name="thigh_twist_01_l"),Weight=0.1))
    Weight=1
   )
   > &RigUnit_ParentConstraint(
    AdvancedSettings=(InterpolationType=Average,RotationOrderForFilter=XZY)
    Child=(Type=Bone,Name="MSTR_frontSkirt_r")
    Filter=(TranslationFilter=(bX=false,bY=false,bZ=false),RotationFilter=(bX=True,bY=True,bZ=True),ScaleFilter=(bX=True,bY=True,bZ=True))
    Parents=((Item=(Type=Bone,Name="pelvis"),Weight=0.2),(Item=(Type=Bone,Name="thigh_twist_01_r"),Weight=0.8))
    Weight=1
   )
   > &RigUnit_ParentConstraint(
    AdvancedSettings=(InterpolationType=Average,RotationOrderForFilter=XZY)
    Child=(Type=Bone,Name="MSTR_SideSkirt_r")
    Filter=(TranslationFilter=(bX=false,bY=false,bZ=false),RotationFilter=(bX=True,bY=True,bZ=True),ScaleFilter=(bX=True,bY=True,bZ=True))
    Parents=((Item=(Type=Bone,Name="pelvis"),Weight=0.5),(Item=(Type=Bone,Name="thigh_twist_01_r"),Weight=0.5))
    Weight=1
   )
   > n2 = &RigUnit_ParentConstraint(
    AdvancedSettings=(InterpolationType=Average,RotationOrderForFilter=XZY)
    Child=(Type=Bone,Name="MSTR_BackSkirt_r")
    Filter=(TranslationFilter=(bX=false,bY=false,bZ=false),RotationFilter=(bX=True,bY=True,bZ=True),ScaleFilter=(bX=True,bY=True,bZ=True))
    Parents=((Item=(Type=Bone,Name="pelvis"),Weight=0.9),(Item=(Type=Bone,Name="thigh_twist_01_r"),Weight=0.1))
    Weight=1
   )
   RigVMFunction_ControlFlowBranch = &RigVMFunction_ControlFlowBranch(ExecuteContext=@n2.ExecutePin, Condition=RigVMVariableNode() {variable=Crouching}.Value)
   &RigUnit_ParentConstraint(
    AdvancedSettings=(InterpolationType=Average,RotationOrderForFilter=XZY)
    Child=(Type=Bone,Name="MSTR_frontSkirt_m")
    Filter=(TranslationFilter=(bX=false,bY=false,bZ=false),RotationFilter=(bX=True,bY=True,bZ=True),ScaleFilter=(bX=True,bY=True,bZ=True))
    Parents=((Item=(Type=Bone,Name="MSTR_SideSkirt_l"),Weight=0.5),(Item=(Type=Bone,Name="MSTR_frontSkirt_r"),Weight=0.5))
    Weight=1
    ExecutePin=@RigVMFunction_ControlFlowBranch.False
   )
   > &RigUnit_ParentConstraint(
    AdvancedSettings=(InterpolationType=Average,RotationOrderForFilter=XZY)
    Child=(Type=Bone,Name="MSTR_BackSkirt_m")
    Filter=(TranslationFilter=(bX=false,bY=false,bZ=false),RotationFilter=(bX=True,bY=True,bZ=True),ScaleFilter=(bX=True,bY=True,bZ=True))
    Parents=((Item=(Type=Bone,Name="MSTR_BackSkirt_l"),Weight=0.5),(Item=(Type=Bone,Name="MSTR_BackSkirt_r"),Weight=0.5))
    Weight=1
   )
   &RigUnit_ParentConstraint(
    AdvancedSettings=(InterpolationType=Average,RotationOrderForFilter=XZY)
    Child=(Type=Bone,Name="MSTR_frontSkirt_m")
    Filter=(TranslationFilter=(bX=false,bY=false,bZ=false),RotationFilter=(bX=True,bY=True,bZ=True),ScaleFilter=(bX=True,bY=True,bZ=True))
    Parents=((Item=(Type=Bone,Name="MSTR_SideSkirt_l"),Weight=0.3),(Item=(Type=Bone,Name="MSTR_frontSkirt_r"),Weight=0.7))
    Weight=1
    ExecutePin=@RigVMFunction_ControlFlowBranch.True
   )
   > &RigUnit_ParentConstraint(
    AdvancedSettings=(InterpolationType=Average,RotationOrderForFilter=XZY)
    Child=(Type=Bone,Name="MSTR_BackSkirt_m")
    Filter=(TranslationFilter=(bX=false,bY=false,bZ=false),RotationFilter=(bX=True,bY=True,bZ=True),ScaleFilter=(bX=True,bY=True,bZ=True))
    Parents=((Item=(Type=Bone,Name="MSTR_BackSkirt_l"),Weight=0.5),(Item=(Type=Bone,Name="MSTR_BackSkirt_r"),Weight=0.5))
    Weight=1
   )
   RigVMFunctionReturnNode(ExecuteContext=@RigVMFunction_Sequence.B)
   &RigUnit_ParentConstraint(
    AdvancedSettings=(InterpolationType=Average,RotationOrderForFilter=XZY)
    Child=(Type=Bone,Name="MSTR_BackSkirt_m")
    ExecutePin=()
    Filter=(TranslationFilter=(bX=false,bY=false,bZ=false),RotationFilter=(bX=True,bY=True,bZ=True),ScaleFilter=(bX=True,bY=True,bZ=True))
    Parents=((Item=(Type=Bone,Name="pelvis"),Weight=0.8),(Item=(Type=Bone,Name="thigh_l"),Weight=0.2))
    Weight=1
   )
  ]}
 New Function_ContainedGraph {nodes [
   RigVMFunctionEntryNode()
   > RigVMFunction_Sequence = RigVMAggregateNode()
   RigVMFunction_ControlFlowBranch = &RigVMFunction_ControlFlowBranch(ExecuteContext=@RigVMFunction_Sequence.A, Condition=RigVMVariableNode() {variable=Crouching}.Value)
   DeltaRotation v4_14 = RigVMFunctionReferenceNode(
    Rotation Order=ZYX
    child=(Type=Bone,Name="thigh_r")
    outputMinMax=(X=0,Y=1)
    parent=(Type=Null,Name="thigh_r_null")
    premult=1
    rotationMinMax=(X=-10,Y=-90)
    triggerAxis=Z
    twistAxis=(X=1,Y=0,Z=0)
    ExecuteContext=@RigVMFunction_ControlFlowBranch.True
   )
   > &RigUnit_ModifyTransforms(
    ItemToModify=((Item=(Type=Bone,Name="base_bag_01"),Transform=(Rotation=(X=0,Y=-0.5,Z=0,W=0.866),Translation=(X=7,Y=0,Z=7),Scale3D=(X=1,Y=1,Z=1))))
    WeightMinimum=0
    Weight=@"DeltaRotation v4_14".outputWeight
   )
   n1 = &RigVMFunction_AnimEvalRichCurve(
    Curve=(EditorCurveData=(Keys=((),(Time=0.5,Value=1),(Time=0.8))))
    SourceMaximum=2
    SourceMinimum=0
    TargetMinimum=0
    Value=&RigVMFunction_MathDoubleRemap(
     SourceMaximum=75
     SourceMinimum=45
     TargetMaximum=1
     TargetMinimum=0
     bClamp=false
     Value=&RigUnit_GetRelativeTransformForItem(Child=(Type=Bone,Name="pelvis"), Parent=(Type=Bone,Name="root"), bChildInitial=False, bParentInitial=False).RelativeTransform.Translation.Z
    ).Result
   )
   > &RigUnit_ModifyTransforms(
    ItemToModify=((Item=(Type=Bone,Name="base_bag_01"),Transform=(Rotation=(X=0,Y=-0.3007,Z=0,W=0.9537),Translation=(X=2,Y=0,Z=0),Scale3D=(X=1,Y=1,Z=1))))
    WeightMinimum=0
    Weight=@n1.Result
   )
   RigVMFunctionReturnNode(ExecuteContext=@RigVMFunction_Sequence.B)
   &RigUnit_ParentConstraint(
    AdvancedSettings=(InterpolationType=Average,RotationOrderForFilter=XZY)
    Child=(Type=Bone,Name="base_strap_loop_01")
    Parents=((Item=(Type=Bone,Name="pelvis"),Weight=0.2),(Item=(Type=Bone,Name="dyn_SideSkirt_l_01"),Weight=0.8))
    Weight=1
    ExecutePin=@RigVMFunction_Sequence.C
   )
   &RigUnit_ParentConstraint(
    AdvancedSettings=(InterpolationType=Average,RotationOrderForFilter=XZY)
    Child=(Type=Bone,Name="base_bag_01")
    ExecutePin=()
    Filter=(TranslationFilter=(bX=True,bY=True,bZ=True),RotationFilter=(bX=false,bY=false,bZ=false),ScaleFilter=(bX=True,bY=True,bZ=True))
    Parents=((Item=(Type=Bone,Name="pelvis"),Weight=0.3),(Item=(Type=Bone,Name="thigh_l"),Weight=0.7))
    Weight=1
   )
   &RigUnit_SphericalPoseReader(
    ActiveRegionScaleFactors=(PositiveWidth=1.2,NegativeWidth=1.2,PositiveHeight=1.2,NegativeHeight=1.2)
    ActiveRegionSize=0.1
    Debug=(bDrawDebug=True,bDraw2D=False,bDrawLocalAxes=False,DebugScale=25,DebugSegments=20,DebugThickness=0.25)
    DriverAxis=(X=1,Y=0,Z=0)
    DriverItem=(Type=Bone,Name="thigh_r")
    ExecutePin=()
    FalloffRegionScaleFactors=(PositiveWidth=1,NegativeWidth=1,PositiveHeight=1,NegativeHeight=1)
    FalloffSize=0.3
    FlipHeightScaling=False
    FlipWidthScaling=False
    OptionalParentItem=(Type=Bone,Name="pelvis")
    RotationOffset=(X=-83.8085,Y=0,Z=4.3348)
   )
   RigVMVariableNode() {variable=Crouching_Alpha}
  ]}
 RigVMFunctionLibrary {nodes [
   BagSetup = RigVMCollapseNode(ExecuteContext=())
   Constraints = RigVMCollapseNode(ExecuteContext=())
   ControlDebug = RigVMCollapseNode(Debug=False, ExecuteContext=())
   DeltaRotation v4 = RigVMCollapseNode(
    ExecuteContext=()
    Rotation Order=ZYX
    child=(Type=Bone,Name="None")
    outputMinMax=(X=0,Y=1)
    parent=(Type=Bone,Name="None")
    premult=1
    rotationMinMax=(X=0,Y=0)
    twistAxis=(X=1,Y=0,Z=0)
   )
   SkirtHelper = RigVMCollapseNode(ExecuteContext=())
   SkirtSetup_DEF = RigVMCollapseNode(ExecuteContext=())
  ]}
 RigVMFunction_Sequence_1_ContainedGraph {nodes [
   n1 = RigVMFunctionEntryNode()
   RigVMFunction_Sequence = &RigVMFunction_Sequence(ExecuteContext=@n1.ExecuteContext)
   n2 = &RigVMFunction_Sequence(ExecuteContext=@RigVMFunction_Sequence.B)
   n3 = &RigVMFunction_Sequence(ExecuteContext=@n2.B)
   n4 = &RigVMFunction_Sequence(ExecuteContext=@n3.B)
   n5 = &RigVMFunction_Sequence(ExecuteContext=@n4.B)
   n6 = &RigVMFunction_Sequence(ExecuteContext=@n5.B)
   RigVMFunctionReturnNode(A=@RigVMFunction_Sequence.A, B=@n2.A, C=@n3.A, D=@n4.A, E=@n5.A, F=@n6.A, G=@n6.B)
  ]}
 RigVMFunction_Sequence_1_ContainedGraph {nodes [
   n1 = RigVMFunctionEntryNode()
   RigVMFunction_Sequence = &RigVMFunction_Sequence(ExecuteContext=@n1.ExecuteContext)
   n2 = &RigVMFunction_Sequence(ExecuteContext=@RigVMFunction_Sequence.B)
   RigVMFunctionReturnNode(A=@RigVMFunction_Sequence.A, B=@n2.A, C=@n2.B)
  ]}
 RigVMFunction_Sequence_2_ContainedGraph {nodes [
   n1 = RigVMFunctionEntryNode()
   n2 = &RigVMFunction_Sequence(ExecuteContext=@n1.ExecuteContext)
   n3 = &RigVMFunction_Sequence(ExecuteContext=@n2.B)
   RigVMFunctionReturnNode(A=@n2.A, B=@n3.A, C=@n3.B)
  ]}
 RigVMFunction_Sequence_2_ContainedGraph {nodes [
   n1 = RigVMFunctionEntryNode()
   n2 = &RigVMFunction_Sequence(ExecuteContext=@n1.ExecuteContext)
   n3 = &RigVMFunction_Sequence(ExecuteContext=@n2.B)
   n4 = &RigVMFunction_Sequence(ExecuteContext=@n3.B)
   RigVMFunctionReturnNode(A=@n2.A, B=@n3.A, C=@n4.A, D=@n4.B)
  ]}
 RigVMModel {nodes [
   &RigUnit_ParentConstraint(
    AdvancedSettings=(InterpolationType=Average,RotationOrderForFilter=XZY)
    Child=(Type=Bone,Name="base_back_cloth_l")
    ExecutePin=()
    Filter=(TranslationFilter=(bX=false,bY=false,bZ=true),RotationFilter=(bX=True,bY=false,bZ=True),ScaleFilter=(bX=True,bY=True,bZ=True))
    Parents=((Item=(Type=Bone,Name="clavicle_l"),Weight=0.8),(Item=(Type=Bone,Name="spine_05"),Weight=0.2))
    Weight=1
   )
   > &RigUnit_ParentConstraint(
    AdvancedSettings=(InterpolationType=Average,RotationOrderForFilter=XZY)
    Child=(Type=Bone,Name="base_back_cloth_r")
    Filter=(TranslationFilter=(bX=false,bY=false,bZ=true),RotationFilter=(bX=True,bY=false,bZ=True),ScaleFilter=(bX=True,bY=True,bZ=True))
    Parents=((Item=(Type=Bone,Name="clavicle_r"),Weight=0.8),(Item=(Type=Bone,Name="spine_05"),Weight=0.2))
    Weight=1
   )
   > &RigUnit_ModifyTransforms(
    ItemToModify="((Item=(Type=Bone,Name=\"upperarm_twist_02_l\"),Transform=(Rotation=(X=0,Y=0,Z=0,W=1),Translation=(X=0,Y=-2,Z=0),Scale3D=(X=1,Y=1,Z=1))),(Item=(Type=Bone,Name=\"upperarm_twist_01_l\"),Transform=(Rotation=(X=0,Y=0.0436,Z=0,W=0.999),Translation=(X=0,Y=-1,Z=0),Sc ...(truncated)"
    Weight=1
    WeightMinimum=0
   )
   RigUnit_BeginExecution = &RigUnit_BeginExecution()
   ControlDebug = RigVMFunctionReferenceNode(Debug=false, ExecuteContext=@RigUnit_BeginExecution.ExecutePin)
   > FN_DeformRig_FNC = RigVMFunctionReferenceNode(
    DeformRig_ Curve="(deformRig_l_upperarm_up_alpha_72_#1=False,deformRig_r_upperarm_up_alpha_73_#2=False,deformRig_l_upperarm_fr_alpha_74_#3=False,deformRig_r_upperarm_fr_alpha_75_CB51B8 ...(truncated)"
    DeformRig_Bool="(upperarm_1_#4=True,elbow_3_#5=True,wrist_7_#6=True,groin_8_#7=True,glute_17_#8=false,knee_9_40C408324CA26 ...(truncated)"
    DeformRig_Debug=(upperarm_debug_18_#4=False,wrist_debug_20_#6=False,lat_debug_26_#9=False,groinGlute_debug_27_#7=False)
    DeformRig_Value="(upperarm_up_18_#10=(Rotation=(X=0.0,Y=-0.0,Z=0.0,W=1.0),Translation=(X=0.0,Y=0.0,Z=8.0),Scale3D=(X=0.70,Y=1.0,Z=1.0)),upperarm_fwd_fr_19_#11=(Rotation=(X=0.0,Y=-0.0,Z=0.0,W=1.0),Translation=(X=0.0, ...(truncated)"
   )
   > SkirtSetup_DEF = RigVMFunctionReferenceNode()
   > SkirtHelper = RigVMFunctionReferenceNode()
   > BagSetup = RigVMFunctionReferenceNode()
  ]}
]
hierarchy [
 root {loc=<0 0 0>, rot=<0 0 0 1>}
 pelvis {parent=root, loc=<0 -0.5301 98.0571>, rot=<0.0127 0.707 -0.0127 -0.707>}
 spine_01 {parent=pelvis, loc=<6.1208 -0.9328 0>, rot=<0 0 0.034 -0.9994>}
 spine_02 {parent=spine_01, loc=<6.5997 -0.2095 0>, rot=<0 0 0.0238 -0.9997>}
 spine_03 {parent=spine_02, loc=<6.4914 -0.1237 0>, rot=<0 0 0.0149 -0.9999>}
 spine_04 {parent=spine_03, loc=<6.9598 0.0794 0>, rot=<0 0 -0.0132 -0.9999>}
 spine_05 {parent=spine_04, loc=<6.9987 0 0>, rot=<0 0 -0.0015 -1>}
 neck_01 {parent=spine_05, loc=<11.3138 0.1529 0>, rot=<0 0 -0.2003 -0.9797>}
 neck_02 {parent=neck_01, loc=<4.5627 0 0>, rot=<0 0 0.0504 -0.9987>}
 head {parent=neck_02, loc=<5.3914 0 0>, rot=<0 0 0.115 -0.9934>}
 base_Hair_Front_01 {parent=head, loc=<9.6449 9.4177 3.2702>, rot=<0.6384 -0.2519 0.7269 0.0251>}
 dyn_Hair_Front_01 {parent=base_Hair_Front_01, loc=<0 0 0>, rot=<0 0 0 1>}
 dyn_Hair_Front_02 {parent=dyn_Hair_Front_01, loc=<6.1091 0 0>, rot=<0.0845 0.1617 -0.035 -0.9826>}
 base_Hair_Back_01 {parent=head, loc=<0.7046 -7.3333 1.5201>, rot=<0.191 -0.6808 0.6884 -0.1615>}
 dyn_Hair_Back_01 {parent=base_Hair_Back_01, loc=<0 0 0>, rot=<0 0 0 1>}
 dyn_Hair_Back_02 {parent=dyn_Hair_Back_01, loc=<9.4144 0 -0.4972>, rot=<0 -0.06 0 -0.9982>}
 dyn_Hair_Back_03 {parent=dyn_Hair_Back_02, loc=<9.1453 -0.1102 -0.653>, rot=<0.0895 0 0 -0.996>}
 base_Hair_Back_a_01 {parent=dyn_Hair_Back_03, loc=<11.2843 3.9032 3.8351>, rot=<0.0085 -0.1906 -0.2672 -0.9446>}
 dyn_Hair_Back_a_01 {parent=base_Hair_Back_a_01, loc=<0 0 0>, rot=<0 0 0 1>}
 dyn_Hair_Back_a_02 {parent=dyn_Hair_Back_a_01, loc=<10.8891 0 0>, rot=<0 -0.2714 0.2714 -0.9234>}
 dyn_Hair_Back_a_03 {parent=dyn_Hair_Back_a_02, loc=<10.9042 0 0>, rot=<0 0 0 1>}
 base_Hair_Back_b_01 {parent=dyn_Hair_Back_03, loc=<14.4441 -1.2879 -3.7162>, rot=<-0.1142 -0.0846 -0.1356 -0.9805>}
 dyn_Hair_Back_b_01 {parent=base_Hair_Back_b_01, loc=<0 0 0>, rot=<0 0 0 1>}
 dyn_Hair_Back_b_02 {parent=dyn_Hair_Back_b_01, loc=<9.8933 0 0>, rot=<-0.062 0.1847 -0.3123 -0.9298>}
 base_Hair_Back_r_01 {parent=head, loc=<2.2127 -4.6859 7.4284>, rot=<0.1763 -0.6848 0.6848 -0.1763>}
 dyn_Hair_Back_r_01 {parent=base_Hair_Back_r_01, loc=<0 0 0>, rot=<0 0 0 1>}
 dyn_Hair_Back_r_02 {parent=dyn_Hair_Back_r_01, loc=<9.4144 0 -0.4972>, rot=<0 0.0321 0 -0.9995>}
 dyn_Hair_Back_r_03 {parent=dyn_Hair_Back_r_02, loc=<9.1453 -0.1102 -0.653>, rot=<0.0895 0 0 -0.996>}
 dyn_Hair_Back_r_04 {parent=dyn_Hair_Back_r_03, loc=<6.0746 0.6349 -4.3148>, rot=<0 0 0.0243 -0.9997>}
 base_Hair_Back_a_r_01 {parent=dyn_Hair_Back_r_04, loc=<4.0186 -2.0884 -3.3267>, rot=<-0.1328 -0.3363 0.3165 -0.877>}
 dyn_Hair_Back_a_r_01 {parent=base_Hair_Back_a_r_01, loc=<0 0 0>, rot=<0 0 0 1>}
 dyn_Hair_Back_c_r_01 {parent=dyn_Hair_Back_a_r_01, loc=<5.5613 -4.0655 0.6853>, rot=<-0.0525 0.1224 0.3879 -0.912>}
 dyn_Hair_Back_b_r_01 {parent=dyn_Hair_Back_a_r_01, loc=<12.2052 5.2039 0.3365>, rot=<-0.0509 0.1187 -0.5998 -0.7897>}
 dyn_Hair_Back_d_r_01 {parent=dyn_Hair_Back_r_02, loc=<4.4045 -4.7776 -2.8023>, rot=<-0.2872 0.6598 0.4785 -0.5032>}
 base_Hair_Back_l_01 {parent=head, loc=<0.9888 -5.2206 -5.2919>, rot=<0.1155 -0.7013 0.6865 -0.1537>}
 dyn_Hair_Back_l_01 {parent=base_Hair_Back_l_01, loc=<0 0 0>, rot=<0 0 0 1>}
 dyn_Hair_Back_l_02 {parent=dyn_Hair_Back_l_01, loc=<7.7051 -0.0387 -0.5557>, rot=<0.0002 0.0171 -0.0113 -0.9998>}
 dyn_Hair_Back_l_03 {parent=dyn_Hair_Back_l_02, loc=<9.1453 -0.1102 -0.653>, rot=<0.0894 -0.0519 -0.0047 -0.9946>}
 base_Hair_Back_b_l_01 {parent=dyn_Hair_Back_l_03, loc=<3.2802 1.5407 -0.1252>, rot=<0.4605 0.326 -0.3396 -0.7526>}
 dyn_Hair_Back_b_l_01 {parent=base_Hair_Back_b_l_01, loc=<0 0 0>, rot=<0 0 0 1>}
 dyn_Hair_Back_b_l_02 {parent=dyn_Hair_Back_b_l_01, loc=<7.752 0 0>, rot=<0 0.2356 0 -0.9718>}
 base_Hair_Back_c_l_01 {parent=dyn_Hair_Back_l_03, loc=<9.7094 0.2216 -0.3218>, rot=<0.1594 0.0527 -0.3635 -0.9164>}
 dyn_Hair_Back_c_l_01 {parent=base_Hair_Back_c_l_01, loc=<0 0 0>, rot=<0 0 0 1>}
 dyn_Hair_Back_c_l_02 {parent=dyn_Hair_Back_c_l_01, loc=<9.2457 0 0>, rot=<0 0 0.3864 -0.9223>}
 dyn_Hair_Back_c_l_03 {parent=dyn_Hair_Back_c_l_02, loc=<9.0351 0 0>, rot=<0 0 0 1>}
 base_Hair_Back_a_l_01 {parent=dyn_Hair_Back_l_01, loc=<1.71 2.061 -2.42>, rot=<0.3224 -0.1253 0.061 -0.9363>}
 dyn_Hair_Back_a_l_01 {parent=base_Hair_Back_a_l_01, loc=<0 0 0>, rot=<0 0 0 1>}
 dyn_Hair_Back_a_l_02 {parent=dyn_Hair_Back_a_l_01, loc=<6.2675 0.3161 0.5473>, rot=<0 0.4592 0 -0.8883>}
 dyn_Hair_Back_a_l_03 {parent=dyn_Hair_Back_a_l_02, loc=<6.3158 0 0>, rot=<0.0946 0.4714 -0.1205 -0.8685>}
 base_hat_01 {parent=head, loc=<10.7632 -12.1789 0>, rot=<0.0566 -0.7048 0.7048 -0.0566>}
 dyn_hat_01 {parent=base_hat_01, loc=<0 0 0>, rot=<0 0 0 1>}
 dyn_hat_02 {parent=dyn_hat_01, loc=<4.4116 0 0>, rot=<0 0 0 1>}
 base_hat_Back_01 {parent=head, loc=<3.4753 -9.4287 0>, rot=<0.0029 -0.7071 -0.0029 -0.7071>}
 dyn_hat_Back_01 {parent=base_hat_Back_01, loc=<0 0 0>, rot=<0 0 0 1>}
 clavicle_l {parent=spine_05, loc=<5.4831 -0.1053 -3.2946>, rot=<0.0465 -0.7289 -0.009 -0.6829>}
 upperarm_l {parent=clavicle_l, loc=<9.8658 -1.2136 -0.4105>, rot=<0.0164 -0.3245 -0.0269 -0.9454>}
 lowerarm_l {parent=upperarm_l, loc=<26.465 -0.8345 0.259>, rot=<0.0036 0.0161 -0.2298 -0.9731>}
 hand_l {parent=lowerarm_l, loc=<22.7355 -0.3203 -0.6105>, rot=<-0.6784 -0.0223 0.0735 -0.7307>}
 thumb_01_l {parent=hand_l, loc=<1.3118 -0.3191 -2.459>, rot=<0.6546 -0.4042 0.0664 -0.6354>}
 thumb_02_l {parent=thumb_01_l, loc=<5.9336 -0.0448 0.0401>, rot=<-0.0037 0.0213 0.298 -0.9543>}
 thumb_03_l {parent=thumb_02_l, loc=<2.2223 0.3682 -0.0907>, rot=<0.0118 -0.013 0.0066 -0.9998>}
 index_metacarpal_l {parent=hand_l, loc=<1.6535 0.0183 -2.96>, rot=<0.0158 -0.0124 -0.0173 -0.9996>}
 index_01_l {parent=index_metacarpal_l, loc=<8.3008 0.0357 -0.0524>, rot=<-0.0044 0.0485 0.0924 -0.9945>}
 index_02_l {parent=index_01_l, loc=<3.3587 -0.3004 0.0452>, rot=<-0.002 0.0027 0.3225 -0.9466>}
 index_03_l {parent=index_02_l, loc=<2.5578 -0.123 -0.0252>, rot=<-0.002 0.0067 0.2558 -0.9667>}
 middle_metacarpal_l {parent=hand_l, loc=<2.1856 -0.1725 -0.5984>, rot=<0.006 0.0168 -0.0329 -0.9993>}
 middle_01_l {parent=middle_metacarpal_l, loc=<7.3275 -0.2447 0.0129>, rot=<-0.0208 0.0042 0.0858 -0.9961>}
 middle_02_l {parent=middle_01_l, loc=<4.0123 0.1401 0.1175>, rot=<-0.0218 0.0258 0.3512 -0.9357>}
 middle_03_l {parent=middle_02_l, loc=<2.6077 0.0534 0.1373>, rot=<0.0082 0.0188 0.1708 -0.9851>}
 ring_metacarpal_l {parent=hand_l, loc=<2.2074 -0.3625 1.3638>, rot=<-0.0284 0.0722 -0.0458 -0.9959>}
 ring_01_l {parent=ring_metacarpal_l, loc=<7.0295 -0.6197 -0.1446>, rot=<-0.0489 -0.0252 0.1892 -0.9804>}
 ring_02_l {parent=ring_01_l, loc=<3.6174 0.4647 -0.1027>, rot=<0 0 0.2806 -0.9598>}
 ring_03_l {parent=ring_02_l, loc=<2.4152 0.114 0.0199>, rot=<0 0 0.1189 -0.9929>}
 pinky_metacarpal_l {parent=hand_l, loc=<1.7616 -0.9753 3.3814>, rot=<-0.0397 0.0814 -0.069 -0.9935>}
 pinky_01_l {parent=pinky_metacarpal_l, loc=<7.1709 -0.4831 -0.1094>, rot=<-0.0654 -0.0035 0.2242 -0.9723>}
 pinky_02_l {parent=pinky_01_l, loc=<2.8274 0.0893 -0.1043>, rot=<0.0037 0.0011 0.2854 -0.9584>}
 pinky_03_l {parent=pinky_02_l, loc=<2.2046 0.0307 -0.0078>, rot=<-0.0002 0.0098 0.1445 -0.9895>}
 hand_attach_l {parent=hand_l, loc=<8.5 -3 0>, rot=<0.7071 0.7071 0 0>}
 weapon_l {parent=hand_l, loc=<-0.1974 -3.4111 0.3811>, rot=<0 0 0 1>}
 deform_wrist_dn_base_l {parent=hand_l, loc=<0 -1.5326 0>, rot=<0 0 0 -1>}
 deform_wrist_dn_l {parent=deform_wrist_dn_base_l, loc=<0 0 0>, rot=<0 0 0 1>}
 deform_wrist_up_base_l {parent=hand_l, loc=<0 2.3009 0>, rot=<0 0 0 -1>}
 deform_wrist_up_l {parent=deform_wrist_up_base_l, loc=<0 0 0>, rot=<0 0 0 1>}
 lowerarm_twist_02_l {parent=lowerarm_l, loc=<10.6707 0 0>, rot=<0 0 0 1>}
 lowerarm_twist_01_l {parent=lowerarm_l, loc=<18.316 0 0>, rot=<0 0 0 1>}
 deform_elbow_in_base_l {parent=lowerarm_l, loc=<0.1834 3.6062 -0.2572>, rot=<0 0 0 -1>}
 deform_elbow_in_l {parent=deform_elbow_in_base_l, loc=<0 0 0>, rot=<0 0 0 1>}
 deform_elbow_out_base_l {parent=lowerarm_l, loc=<0 -4 0>, rot=<0 0 0 1>}
 deform_elbow_out_l {parent=deform_elbow_out_base_l, loc=<0 0 0>, rot=<0 0 0 1>}
 upperarm_twist_01_l {parent=upperarm_l, loc=<7.2135 -0.1134 0>, rot=<0 0 0 1>}
 deform_upperarm_bk_base_l {parent=upperarm_twist_01_l, loc=<-6.4053 -4.3746 0.314>, rot=<0 0 0 -1>}
 deform_upperarm_bk_l {parent=deform_upperarm_bk_base_l, loc=<0 0 0>, rot=<0 0 0 1>}
 deform_upperarm_fr_base_l {parent=upperarm_twist_01_l, loc=<-5.6312 5.092 0.3354>, rot=<0 0 0 -1>}
 deform_upperarm_fr_l {parent=deform_upperarm_fr_base_l, loc=<0 0 0>, rot=<0 0 0 1>}
 deform_upperarm_up_base_l {parent=upperarm_twist_01_l, loc=<-8.891 0.4825 3.9517>, rot=<0 0 0 -1>}
 deform_upperarm_up_l {parent=deform_upperarm_up_base_l, loc=<0 0 0>, rot=<0 0 0 1>}
 upperarm_twist_02_l {parent=upperarm_l, loc=<19.7998 -0.1134 0>, rot=<0 0 0 1>}
 clavicle_r {parent=spine_05, loc=<5.4831 -0.1053 3.2946>, rot=<0.7289 0.0465 -0.6829 0.009>}
 upperarm_r {parent=clavicle_r, loc=<-9.8658 1.2136 0.4105>, rot=<0.0164 -0.3245 -0.0269 -0.9454>}
 lowerarm_r {parent=upperarm_r, loc=<-26.465 0.8345 -0.259>, rot=<0.0036 0.0161 -0.2298 -0.9731>}
 hand_r {parent=lowerarm_r, loc=<-22.7355 0.3203 0.6106>, rot=<-0.6784 -0.0223 0.0735 -0.7307>}
 thumb_01_r {parent=hand_r, loc=<-1.3118 0.3191 2.459>, rot=<0.6546 -0.4042 0.0664 -0.6354>}
 thumb_02_r {parent=thumb_01_r, loc=<-5.9336 0.0448 -0.0401>, rot=<-0.0037 0.0213 0.298 -0.9543>}
 thumb_03_r {parent=thumb_02_r, loc=<-2.2223 -0.3682 0.0907>, rot=<0.0118 -0.013 0.0066 -0.9998>}
 index_metacarpal_r {parent=hand_r, loc=<-1.6535 -0.0183 2.96>, rot=<0.0158 -0.0124 -0.0173 -0.9996>}
 index_01_r {parent=index_metacarpal_r, loc=<-8.3008 -0.0357 0.0524>, rot=<-0.0044 0.0485 0.0924 -0.9945>}
 index_02_r {parent=index_01_r, loc=<-3.3587 0.3004 -0.0452>, rot=<-0.002 0.0027 0.3225 -0.9466>}
 index_03_r {parent=index_02_r, loc=<-2.5578 0.123 0.0252>, rot=<-0.002 0.0067 0.2558 -0.9667>}
 middle_metacarpal_r {parent=hand_r, loc=<-2.1856 0.1725 0.5984>, rot=<0.006 0.0168 -0.0329 -0.9993>}
 middle_01_r {parent=middle_metacarpal_r, loc=<-7.3275 0.2447 -0.0129>, rot=<-0.0208 0.0042 0.0858 -0.9961>}
 middle_02_r {parent=middle_01_r, loc=<-4.0123 -0.1401 -0.1175>, rot=<-0.0218 0.0258 0.3512 -0.9357>}
 middle_03_r {parent=middle_02_r, loc=<-2.6077 -0.0534 -0.1373>, rot=<0.0082 0.0188 0.1708 -0.9851>}
 ring_metacarpal_r {parent=hand_r, loc=<-2.2074 0.3625 -1.3638>, rot=<-0.0284 0.0722 -0.0458 -0.9959>}
 ring_01_r {parent=ring_metacarpal_r, loc=<-7.0295 0.6197 0.1446>, rot=<-0.0489 -0.0252 0.1892 -0.9804>}
 ring_02_r {parent=ring_01_r, loc=<-3.6174 -0.4647 0.1027>, rot=<0 0 0.2806 -0.9598>}
 ring_03_r {parent=ring_02_r, loc=<-2.4152 -0.114 -0.0199>, rot=<0 0 0.1189 -0.9929>}
 pinky_metacarpal_r {parent=hand_r, loc=<-1.7616 0.9753 -3.3814>, rot=<-0.0397 0.0814 -0.069 -0.9935>}
 pinky_01_r {parent=pinky_metacarpal_r, loc=<-7.1709 0.4831 0.1094>, rot=<-0.0654 -0.0035 0.2242 -0.9723>}
 pinky_02_r {parent=pinky_01_r, loc=<-2.8274 -0.0893 0.1043>, rot=<0.0037 0.0011 0.2854 -0.9584>}
 pinky_03_r {parent=pinky_02_r, loc=<-2.2046 -0.0307 0.0078>, rot=<-0.0002 0.0098 0.1445 -0.9895>}
 weapon_r {parent=hand_r, loc=<0.1974 3.4111 -0.3811>, rot=<0 0 0 1>}
 hand_attach_r {parent=hand_r, loc=<-8.5001 2.9999 0>, rot=<0 0 -0.7071 -0.7071>}
 deform_wrist_dn_base_r {parent=hand_r, loc=<-0.0001 1.5328 0>, rot=<0 0 0 -1>}
 deform_wrist_dn_r {parent=deform_wrist_dn_base_r, loc=<0 0 0>, rot=<0 0 0 1>}
 deform_wrist_up_base_r {parent=hand_r, loc=<-0.0002 -2.3007 -0.0001>, rot=<0 0 0 -1>}
 deform_wrist_up_r {parent=deform_wrist_up_base_r, loc=<0 0 0>, rot=<0 0 0 1>}
 lowerarm_twist_02_r {parent=lowerarm_r, loc=<-10.6707 0 0>, rot=<0 0 0 -1>}
 lowerarm_twist_01_r {parent=lowerarm_r, loc=<-18.316 0 0>, rot=<0 0 0 -1>}
 deform_elbow_in_base_r {parent=lowerarm_r, loc=<-0.1834 -3.6062 0.2572>, rot=<0 0 0 -1>}
 deform_elbow_in_r {parent=deform_elbow_in_base_r, loc=<0 0 0>, rot=<0 0 0 1>}
 deform_elbow_out_base_r {parent=lowerarm_r, loc=<0 4 -0.0001>, rot=<0 0 0 -1>}
 deform_elbow_out_r {parent=deform_elbow_out_base_r, loc=<0 0 0>, rot=<0 0 0 1>}
 upperarm_twist_01_r {parent=upperarm_r, loc=<-7.2135 0.1134 0>, rot=<0 0 0 -1>}
 deform_upperarm_bk_base_r {parent=upperarm_twist_01_r, loc=<6.4053 4.3746 -0.314>, rot=<0 0 0 -1>}
 deform_upperarm_bk_r {parent=deform_upperarm_bk_base_r, loc=<0 0 0>, rot=<0 0 0 1>}
 deform_upperarm_fr_base_r {parent=upperarm_twist_01_r, loc=<5.6312 -5.092 -0.3354>, rot=<0 0 0 -1>}
 deform_upperarm_fr_r {parent=deform_upperarm_fr_base_r, loc=<0 0 0>, rot=<0 0 0 1>}
 deform_upperarm_up_base_r {parent=upperarm_twist_01_r, loc=<8.891 -0.4825 -3.9517>, rot=<0 0 0 -1>}
 deform_upperarm_up_r {parent=deform_upperarm_up_base_r, loc=<0 0 0>, rot=<0 0 0 1>}
 upperarm_twist_02_r {parent=upperarm_r, loc=<-19.7998 0.1134 0>, rot=<0 0 0 -1>}
 attach_backpack {parent=spine_05, loc=<-0.123 -13.473 0>, rot=<-0.0093 -0.707 0.0093 -0.707>}
 Backpack_BR {parent=attach_backpack, type=Null, loc=<0 0 0>, rot=<0 0 0 1>}
 attach_cape {parent=spine_05, loc=<9.017 -11.714 0>, rot=<0.0616 -0.7044 -0.0616 -0.7044>}
 cape {parent=attach_cape, type=Null, loc=<0 0 0>, rot=<0 0 0 1>}
 base_Hair_l_01 {parent=spine_05, loc=<14.6205 3.1905 -9.1719>, rot=<-0.0336 0.4343 0.8951 0.095>}
 dyn_Hair_l_01 {parent=base_Hair_l_01, loc=<0 0 0>, rot=<0 0 0 1>}
 dyn_Hair_l_02 {parent=dyn_Hair_l_01, loc=<7.3206 0 0>, rot=<0 0.3897 0 -0.921>}
 dyn_Hair_l_03 {parent=dyn_Hair_l_02, loc=<7.0903 0 0>, rot=<0 0 0 1>}
 base_Hair_r_01 {parent=spine_05, loc=<13.9571 5.0284 8.8114>, rot=<-0.1195 -0.932 -0.335 0.0695>}
 dyn_Hair_r_01 {parent=base_Hair_r_01, loc=<0 0 0>, rot=<0 0 0 1>}
 dyn_Hair_r_02 {parent=dyn_Hair_r_01, loc=<7.0191 0 0>, rot=<0 0.3112 0 -0.9503>}
 dyn_Hair_r_03 {parent=dyn_Hair_r_02, loc=<4.7922 0 0>, rot=<0 0 0 1>}
 deform_pec_aim_base_l {parent=spine_04, loc=<3.2456 9.3309 -1.2237>, rot=<0.0603 -0.4033 0.3318 -0.8507>}
 deform_pec_aim_l {parent=deform_pec_aim_base_l, loc=<0 0 0>, rot=<0 0 0 1>}
 deform_pec_base_l {parent=deform_pec_aim_l, loc=<9.3665 0 0>, rot=<0 0 0 1>}
 deform_pec_l {parent=deform_pec_base_l, loc=<0 0 0>, rot=<0 0 0 1>}
 deform_pec_aim_base_r {parent=spine_04, loc=<3.2452 9.3309 1.2237>, rot=<-0.4033 -0.0603 0.8507 0.3318>}
 deform_pec_aim_r {parent=deform_pec_aim_base_r, loc=<0 0 0>, rot=<0 0 0 1>}
 deform_pec_base_r {parent=deform_pec_aim_r, loc=<-9.3667 -0.0003 -0.0004>, rot=<0 0 0 1>}
 deform_pec_r {parent=deform_pec_base_r, loc=<0 0 0>, rot=<0 0 0 1>}
 base_l_frontChest_flap_01 {parent=spine_04, loc=<4.6568 9.9107 -8.5828>, rot=<-0.1612 -0.8862 -0.3993 -0.1709>}
 dyn_l_frontChest_flap_01 {parent=base_l_frontChest_flap_01, loc=<0 0 0>, rot=<0 0 0 1>}
 dyn_l_frontChest_flap_02 {parent=dyn_l_frontChest_flap_01, loc=<3.6598 0 0>, rot=<0 0 0 1>}
 base_r_frontChest_flap_01 {parent=spine_04, loc=<4.6569 9.9107 8.5828>, rot=<0.1709 0.3993 0.8862 0.1612>}
 dyn_r_frontChest_flap_01 {parent=base_r_frontChest_flap_01, loc=<0 0 0>, rot=<0 0 0 1>}
 dyn_r_frontChest_flap_02 {parent=dyn_r_frontChest_flap_01, loc=<3.6598 0 0>, rot=<0 0 0 1>}
 thigh_l_null {parent=pelvis, type=Null, loc=<-1.0855 -0.2109 -8.7053>, rot=<0.0009 0.0355 0.0155 -0.9993>}
 thigh_r_null {parent=pelvis, type=Null, loc=<-1.0855 -0.2109 8.7053>, rot=<0.0355 -0.0009 0.9993 0.0155>}
 thigh_l {parent=pelvis, loc=<-1.0855 -0.2109 -8.7053>, rot=<0.0009 0.0355 0.0155 -0.9993>}
 calf_l {parent=thigh_l, loc=<-42.5284 -0.0019 -0.7297>, rot=<0.0006 -0.0161 -0.0391 -0.9991>}
 foot_l {parent=calf_l, loc=<-46.7401 -2.5014 -0.592>, rot=<0.0427 -0.0238 0.0353 -0.9982>}
 ball_l {parent=foot_l, loc=<-3.5658 10.8066 -0.1544>, rot=<0.0001 -0.0001 -0.7063 -0.7079>}
 calf_twist_02_l {parent=calf_l, loc=<-19.1546 -1.7943 -0.2638>, rot=<0.0237 0.0095 -0.0035 -0.9997>}
 dyn_l_leg_shoe_chain_01 {parent=calf_twist_02_l, loc=<-0.9719 5.1979 -0.4064>, rot=<-0.0537 -0.7157 -0.6946 -0.049>}
 calf_twist_01_l {parent=calf_l, loc=<-35.8483 -1.9197 -0.5794>, rot=<0.0237 0.0095 -0.0035 -0.9997>}
 deform_knee_in_base_l {parent=calf_l, loc=<0.8031 -5.0231 0.2456>, rot=<0 0 0 -1>}
 deform_knee_in_l {parent=deform_knee_in_base_l, loc=<0 0 0>, rot=<0 0 0 1>}
 deform_knee_out_base_l {parent=calf_l, loc=<0 4 0>, rot=<0 0 0 -1>}
 deform_knee_out_l {parent=deform_knee_out_base_l, loc=<0 0 0>, rot=<0 0 0 1>}
 thigh_twist_01_l {parent=thigh_l, loc=<-27.952 0 0>, rot=<0.0013 0 0 -1>}
 deform_groin_base_l {parent=thigh_l, loc=<0.4801 5.6776 -0.2486>, rot=<0 0 0 -1>}
 deform_groin_l {parent=deform_groin_base_l, loc=<0 0 0>, rot=<0 0 0 1>}
 dyn_l_leg_bag_01 {parent=thigh_l, loc=<-22.9131 0.1175 -7.1379>, rot=<-0.0018 -0.7566 0.0016 -0.6539>}
 dyn_l_leg_bag_chain_01 {parent=dyn_l_leg_bag_01, loc=<1.9422 5.0171 -6.2305>, rot=<0.5115 -0.4889 -0.4885 -0.5107>}
 thigh_r {parent=pelvis, loc=<-1.0855 -0.2109 8.7053>, rot=<0.0355 -0.0009 0.9993 0.0155>}
 calf_r {parent=thigh_r, loc=<42.5284 0.0019 0.7297>, rot=<0.0006 -0.0161 -0.0391 -0.9991>}
 foot_r {parent=calf_r, loc=<46.7401 2.5014 0.592>, rot=<0.0427 -0.0238 0.0353 -0.9982>}
 ball_r {parent=foot_r, loc=<3.5658 -10.8066 0.1544>, rot=<0.0001 -0.0001 -0.7063 -0.7079>}
 calf_twist_02_r {parent=calf_r, loc=<19.1546 1.7943 0.2638>, rot=<0.0237 0.0095 -0.0035 -0.9997>}
 dyn_r_leg_shoe_chain_01 {parent=calf_twist_02_r, loc=<2.6825 -4.4193 0.6399>, rot=<-0.7413 0.1226 0.1375 -0.6454>}
 calf_twist_01_r {parent=calf_r, loc=<35.8483 1.9197 0.5794>, rot=<0.0237 0.0095 -0.0035 -0.9997>}
 deform_knee_in_base_r {parent=calf_r, loc=<-0.8031 5.0231 -0.2457>, rot=<0 0 0 -1>}
 deform_knee_in_r {parent=deform_knee_in_base_r, loc=<0 0 0>, rot=<0 0 0 1>}
 deform_knee_out_base_r {parent=calf_r, loc=<0 -4 0>, rot=<0 0 0 -1>}
 deform_knee_out_r {parent=deform_knee_out_base_r, loc=<0 0 0>, rot=<0 0 0 1>}
 thigh_twist_01_r {parent=thigh_r, loc=<27.952 0 0>, rot=<0.0013 0 0 -1>}
 deform_groin_base_r {parent=thigh_r, loc=<-0.4801 -5.6776 0.2486>, rot=<0 0 0 -1>}
 deform_groin_r {parent=deform_groin_base_r, loc=<0 0 0>, rot=<0 0 0 1>}
 MSTR_frontSkirt_l {parent=pelvis, loc=<5.019 6.9439 -8.8195>, rot=<0.127 -0.8359 -0.534 0.002>}
 base_dyn_frontSkirt_l_01 {parent=MSTR_frontSkirt_l, loc=<0 0 0>, rot=<-0.0101 0.1245 -0.0072 -0.9921>}
 dyn_frontSkirt_l_01 {parent=base_dyn_frontSkirt_l_01, loc=<0 0 0>, rot=<0 0 0 1>}
 base_dyn_frontSkirt_l_02 {parent=dyn_frontSkirt_l_01, loc=<3.0407 0 0>, rot=<-0.0257 0.0171 0.0223 -0.9993>}
 dyn_frontSkirt_l_02 {parent=base_dyn_frontSkirt_l_02, loc=<0 0 0>, rot=<0 0 0 1>}
 base_dyn_frontSkirt_l_03 {parent=dyn_frontSkirt_l_02, loc=<9.8094 0 0>, rot=<0.0082 0.0959 0.0112 -0.9953>}
 dyn_frontSkirt_l_03 {parent=base_dyn_frontSkirt_l_03, loc=<0 0 0>, rot=<0 0 0 1>}
 base_dyn_frontSkirt_l_04 {parent=dyn_frontSkirt_l_03, loc=<10.3558 0 0>, rot=<0 0 0 1>}
 dyn_frontSkirt_l_04 {parent=base_dyn_frontSkirt_l_04, loc=<0 0 0>, rot=<0 0 0 1>}
 dyn_frontSkirt_l_05 {parent=dyn_frontSkirt_l_04, loc=<4.9045 0 0>, rot=<0 0 0 1>}
 MSTR_SideSkirt_l {parent=pelvis, loc=<7.0628 0.345 -12.5173>, rot=<0.0152 -0.9863 -0.0096 -0.1639>}
 dyn_SideSkirt_l_01 {parent=MSTR_SideSkirt_l, loc=<0 0 0>, rot=<0 0 0 1>}
 dyn_SideSkirt_l_02 {parent=dyn_SideSkirt_l_01, loc=<3.0084 0 0>, rot=<-0.0269 0.0071 -0.0077 -0.9996>}
 dyn_SideSkirt_l_03 {parent=dyn_SideSkirt_l_02, loc=<11.2436 0.1769 0.1007>, rot=<0.0172 0.0811 -0.0152 -0.9964>}
 dyn_SideSkirt_l_04 {parent=dyn_SideSkirt_l_03, loc=<10.1104 0 0>, rot=<0 0 0 1>}
 dyn_SideSkirt_l_05 {parent=dyn_SideSkirt_l_04, loc=<4.1251 0 0>, rot=<0 0 0 1>}
 MSTR_BackSkirt_m {parent=pelvis, loc=<7.5569 -8.8984 -0.0004>, rot=<0.1783 -0.6964 0.6708 -0.1824>}
 dyn_BackSkirt_m_01 {parent=MSTR_BackSkirt_m, loc=<0 0 0>, rot=<0 0 0 1>}
 dyn_BackSkirt_m_02 {parent=dyn_BackSkirt_m_01, loc=<3.2391 0.0127 -0.0044>, rot=<-0.035 -0.0895 -0.008 -0.9953>}
 dyn_BackSkirt_m_03 {parent=dyn_BackSkirt_m_02, loc=<11.9413 0 0>, rot=<0.0195 -0.0765 0.0166 -0.9967>}
 dyn_BackSkirt_m_04 {parent=dyn_BackSkirt_m_03, loc=<10.7954 0 0>, rot=<0 0.0462 0 -0.9989>}
 dyn_BackSkirt_m_05 {parent=dyn_BackSkirt_m_04, loc=<5.4438 0 0>, rot=<0 0 0 1>}
 MSTR_BackSkirt_l {parent=pelvis, loc=<7.5792 -7.4382 -7.8997>, rot=<0.1295 -0.8626 0.436 -0.2218>}
 dyn_BackSkirt_l_01 {parent=MSTR_BackSkirt_l, loc=<0 0 0>, rot=<0 0 0 1>}
 dyn_BackSkirt_l_02 {parent=dyn_BackSkirt_l_01, loc=<3.1864 0 0>, rot=<0.0172 -0.1109 -0.0159 -0.9936>}
 dyn_BackSkirt_l_03 {parent=dyn_BackSkirt_l_02, loc=<11.6555 -0.1226 0.1051>, rot=<-0.0158 0.0114 0.0129 -0.9997>}
 dyn_BackSkirt_l_04 {parent=dyn_BackSkirt_l_03, loc=<10.7697 0 0>, rot=<0 0 0 1>}
 dyn_BackSkirt_l_05 {parent=dyn_BackSkirt_l_04, loc=<4.8316 0 0>, rot=<0 0 0 1>}
 MSTR_frontSkirt_m {parent=pelvis, loc=<3.5584 8.9649 0>, rot=<0.0874 0.7008 0.7075 0.0275>}
 dyn_frontSkirt_m_01 {parent=MSTR_frontSkirt_m, loc=<0 0 0>, rot=<0 0 0 1>}
 dyn_frontSkirt_m_02 {parent=dyn_frontSkirt_m_01, loc=<3.3431 -0.3456 0.0826>, rot=<0.065 0.0169 0.0168 -0.9976>}
 dyn_frontSkirt_m_03 {parent=dyn_frontSkirt_m_02, loc=<7.8358 -0.3452 -0.269>, rot=<0.0551 -0.0144 0.0391 -0.9976>}
 dyn_frontSkirt_m_04 {parent=dyn_frontSkirt_m_03, loc=<10.173 0 0>, rot=<0 0 0 1>}
 dyn_frontSkirt_m_05 {parent=dyn_frontSkirt_m_04, loc=<7.0283 0 0>, rot=<0 0 0 1>}
 MSTR_frontSkirt_r {parent=pelvis, loc=<4.4222 7.8695 6.751>, rot=<0.1353 0.5595 0.8173 0.0257>}
 dyn_frontSkirt_r_01 {parent=MSTR_frontSkirt_r, loc=<0 0 0>, rot=<-0.0002 -0.0145 0.0163 -0.9998>}
 dyn_frontSkirt_r_02 {parent=dyn_frontSkirt_r_01, loc=<3.3202 0 0>, rot=<0.0303 -0.0811 -0.0175 -0.9961>}
 dyn_frontSkirt_r_03 {parent=dyn_frontSkirt_r_02, loc=<9.0726 0 0>, rot=<-0.0099 0.1374 -0.0188 -0.9903>}
 dyn_frontSkirt_r_04 {parent=dyn_frontSkirt_r_03, loc=<10.52 0 0>, rot=<0 0 0 1>}
 dyn_frontSkirt_r_05 {parent=dyn_frontSkirt_r_04, loc=<4.6664 0 0>, rot=<0 0 0 1>}
 MSTR_SideSkirt_r {parent=pelvis, loc=<7.0663 0.3454 12.5166>, rot=<0.2309 0.0106 0.9728 -0.0145>}
 dyn_SideSkirt_r_01 {parent=MSTR_SideSkirt_r, loc=<0 0 0>, rot=<0 0 0 1>}
 dyn_SideSkirt_r_02 {parent=dyn_SideSkirt_r_01, loc=<3.0904 0 0>, rot=<0.0261 -0.0838 0.0101 -0.9961>}
 dyn_SideSkirt_r_03 {parent=dyn_SideSkirt_r_02, loc=<11.1154 0 0>, rot=<-0.0174 0.0629 0.0149 -0.9978>}
 dyn_SideSkirt_r_04 {parent=dyn_SideSkirt_r_03, loc=<10.3904 0 0>, rot=<0 0 0 1>}
 dyn_SideSkirt_r_05 {parent=dyn_SideSkirt_r_04, loc=<4.2295 0 0>, rot=<0 0 0 1>}
 MSTR_BackSkirt_r {parent=pelvis, loc=<7.5827 -7.437 7.8984>, rot=<0.2762 -0.4009 0.8706 -0.0713>}
 dyn_BackSkirt_r_01 {parent=MSTR_BackSkirt_r, loc=<0 0 0>, rot=<0 0 0 1>}
 dyn_BackSkirt_r_02 {parent=dyn_BackSkirt_r_01, loc=<3.1619 0 0>, rot=<-0.018 -0.1113 0.0036 -0.9936>}
 dyn_BackSkirt_r_03 {parent=dyn_BackSkirt_r_02, loc=<12.1489 0 0>, rot=<0.0164 -0.0322 -0.0122 -0.9993>}
 dyn_BackSkirt_r_04 {parent=dyn_BackSkirt_r_03, loc=<10.1967 0 0>, rot=<0 0 0 -1>}
 dyn_BackSkirt_r_05 {parent=dyn_BackSkirt_r_04, loc=<4.2238 0 0>, rot=<0 0 0 1>}
 deform_glute_base_l {parent=pelvis, loc=<-7.3466 -9.0213 -9.1245>, rot=<0.0172 0.031 -0.4477 -0.8935>}
 deform_glute_l {parent=deform_glute_base_l, loc=<0 0 0>, rot=<0 0 0 1>}
 deform_glute_base_r {parent=pelvis, loc=<-7.3464 -9.0213 9.1245>, rot=<0.031 -0.0172 0.8935 -0.4477>}
 deform_glute_r {parent=deform_glute_base_r, loc=<0 0 0>, rot=<0 0 0 1>}
 base_strap_loop_01 {parent=pelvis, loc=<0.8882 2.9573 -12.1411>, rot=<0.0677 -0.912 -0.3083 -0.262>}
 dyn_strap_loop_01 {parent=base_strap_loop_01, loc=<0 0 0>, rot=<0 0 0 1>}
 dyn_strap_loop_02 {parent=dyn_strap_loop_01, loc=<7.7848 0 0>, rot=<0 0 0 1>}
 base_bag_01 {parent=pelvis, loc=<1.0973 8.9883 8.9841>, rot=<0.6274 0.2987 0.5976 0.4001>}
 dyn_bag_01 {parent=base_bag_01, loc=<0 0 0>, rot=<0 0 0 1>}
 dyn_bag_02 {parent=dyn_bag_01, loc=<5.0792 0 0>, rot=<0 0 0 1>}
 attach {parent=root, loc=<0 0 0>, rot=<0 0 0 1>}
 ik_hand_root {parent=root, loc=<0 0 0>, rot=<0 0 0 1>}
 ik_hand_gun {parent=ik_hand_root, loc=<-48.1743 3.3545 104.4381>, rot=<0.6276 0.1309 0.4183 -0.6434>}
 ik_hand_r {parent=ik_hand_gun, loc=<0 0 0>, rot=<0 0 0 1>}
 ik_hand_l {parent=ik_hand_gun, loc=<59.3271 67.6997 34.3524>, rot=<-0.6158 -0.7027 -0.3565 0>}
 ik_foot_root {parent=root, loc=<0 0 0>, rot=<0 0 0 1>}
 ik_foot_l {parent=ik_foot_root, loc=<14.8262 -7.3545 8.296>, rot=<0.0357 0.7045 0.0284 -0.7082>}
 ik_foot_r {parent=ik_foot_root, loc=<-14.8262 -7.3545 8.296>, rot=<0.7082 0.0284 -0.7045 0.0357>}
 root_fp {parent=root, loc=<0 0 0>, rot=<0 0 0 1>}
 attach_fp {parent=root_fp, loc=<0 0 0>, rot=<0 0 0 1>}
 camera_root {parent=root_fp, loc=<0 0 0>, rot=<0 0 0 1>}
 camera {parent=camera_root, loc=<0 0 0>, rot=<0 0 0 1>}
 attach_emote_01 {parent=root, loc=<0 0 0>, rot=<0 0 0 1>}
 attach_emote_02 {parent=root, loc=<0 0 0>, rot=<0 0 0 1>}
 attach_emote_03 {parent=root, loc=<0 0 0>, rot=<0 0 0 1>}
 warp_point_01 {parent=root, loc=<0 0 0>, rot=<0 0 0 1>}
 warp_point_02 {parent=root, loc=<0 0 0>, rot=<0 0 0 1>}
 warp_point_03 {parent=root, loc=<0 0 0>, rot=<0 0 0 1>}
 spine_01_ctrl {controlType=EULER_TRANSFORM, type=Control, loc=<0 0 0>, rot=<0 0 0 1>}
 spine_02_ctrl {controlType=EULER_TRANSFORM, parent=spine_01_ctrl, type=Control, loc=<0 0 0>, rot=<0 0 0 1>}
 spine_03_ctrl {controlType=EULER_TRANSFORM, parent=spine_02_ctrl, type=Control, loc=<0 0 0>, rot=<0 0 0 1>}
 spine_04_ctrl {controlType=EULER_TRANSFORM, parent=spine_03_ctrl, type=Control, loc=<0 0 0>, rot=<0 0 0 1>}
 spine_05_ctrl {controlType=EULER_TRANSFORM, parent=spine_04_ctrl, type=Control, loc=<0 0 0>, rot=<0 0 0 1>}
 clavicle_l_ctrl {controlType=EULER_TRANSFORM, parent=spine_05_ctrl, type=Control, loc=<0 0 0>, rot=<0 0 0 1>}
 upperarm_l_ctrl {controlType=EULER_TRANSFORM, parent=clavicle_l_ctrl, type=Control, loc=<0 0 0>, rot=<0 0 0 1>}
 lowerarm_l_ctrl {controlType=EULER_TRANSFORM, parent=upperarm_l_ctrl, type=Control, loc=<0 0 0>, rot=<0 0 0 1>}
 clavicle_r_ctrl {controlType=EULER_TRANSFORM, parent=spine_05_ctrl, type=Control, loc=<0 0 0>, rot=<0 0 0 1>}
 upperarm_r_ctrl {controlType=EULER_TRANSFORM, parent=clavicle_r_ctrl, type=Control, loc=<0 0 0>, rot=<0 0 0 1>}
 lowerarm_r_ctrl {controlType=EULER_TRANSFORM, parent=upperarm_r_ctrl, type=Control, loc=<0 0 0>, rot=<0 0 0 1>}
 pelvis_ctrl {controlType=EULER_TRANSFORM, type=Control, loc=<0 0 0>, rot=<0 0 0 1>}
 thigh_l_ctrl {controlType=EULER_TRANSFORM, parent=pelvis_ctrl, type=Control, loc=<0 0 0>, rot=<0 0 0 1>}
 thigh_r_ctrl {controlType=EULER_TRANSFORM, parent=pelvis_ctrl, type=Control, loc=<0 0 0>, rot=<0 0 0 1>}
]
variables [
 BaseballSlide {type=double, default=0}
 Crouching {type=bool, default=False}
 Crouching_Alpha {type=double, default=0}
 Crouching_Alpha_Bag {type=double, default=0}
 DBNO {type=bool, default=False}
 Idle {type=double, default=0}
 KneeSliding {type=double, default=0}
 Moving2D {type=bool, default=False}
 isCrouchStrafe {type=double, default=0}
]
