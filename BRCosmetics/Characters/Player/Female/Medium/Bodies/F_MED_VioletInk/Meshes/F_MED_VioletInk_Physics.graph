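# Physics asset for the F_MED_VioletInk female player body cosmetic, covering the pelvis and skirt bones.
# Eight tiny capsules on the MSTR front, side and back skirt roots attach to a pelvis capsule with free swing and twist.
# Only the front middle and left panels continue into three-capsule dyn chains, linked with locked swing and twist.
# All six chain capsules share identical geometry and the front-middle 02 shape name.
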PhysicsAsset /BRCosmetics/Characters/Player/Female/Medium/Bodies/F_MED_VioletInk/Meshes/F_MED_VioletInk_Physics
bodies {
 MSTR_BackSkirt_l [{shapeType=Capsule, length=0.505, radius=0.505, shapeName=MSTR_BackSkirt_l_capsule}]
 MSTR_BackSkirt_m [{shapeType=Capsule, length=0.505, radius=0.505, shapeName=MSTR_BackSkirt_m_capsule}]
 MSTR_BackSkirt_r [{shapeType=Capsule, length=0.505, radius=0.505, shapeName=MSTR_BackSkirt_r_capsule}]
 MSTR_SideSkirt_l [{shapeType=Capsule, length=0.505, radius=0.505, shapeName=MSTR_SideSkirt_l_capsule}]
 MSTR_SideSkirt_r [{shapeType=Capsule, length=0.505, radius=0.505, shapeName=MSTR_SideSkirt_r_capsule}]
 MSTR_frontSkirt_l [{shapeType=Capsule, length=0.505, radius=0.505, shapeName=MSTR_frontSkirt_l_capsule}]
 MSTR_frontSkirt_m [{shapeType=Capsule, length=0.505, radius=0.505, shapeName=MSTR_frontSkirt_m_capsule}]
 MSTR_frontSkirt_r [{shapeType=Capsule, length=0.505, radius=0.505, shapeName=MSTR_frontSkirt_r_capsule}]
 dyn_frontSkirt_l_02 [{shapeType=Capsule, length=11.125, radius=4.8207, shapeName=dyn_frontSkirt_m_02_capsule, rotation=<rot 0.0008 -0.0001 90>}]
 dyn_frontSkirt_l_03 [{shapeType=Capsule, length=11.125, radius=4.8207, shapeName=dyn_frontSkirt_m_02_capsule, rotation=<rot 0.0008 -0.0001 90>}]
 dyn_frontSkirt_l_04 [{shapeType=Capsule, length=11.125, radius=4.8207, shapeName=dyn_frontSkirt_m_02_capsule, rotation=<rot 0.0008 -0.0001 90>}]
 dyn_frontSkirt_m_02 [{shapeType=Capsule, length=11.125, radius=4.8207, shapeName=dyn_frontSkirt_m_02_capsule, rotation=<rot 0.0008 -0.0001 90>}]
 dyn_frontSkirt_m_03 [{shapeType=Capsule, length=11.125, radius=4.8207, shapeName=dyn_frontSkirt_m_02_capsule, rotation=<rot 0.0008 -0.0001 90>}]
 dyn_frontSkirt_m_04 [{shapeType=Capsule, length=11.125, radius=4.8207, shapeName=dyn_frontSkirt_m_02_capsule, rotation=<rot 0.0008 -0.0001 90>}]
 pelvis [{
   shapeType=Capsule
   length=6.3768
   radius=9.9978
   shapeName=pelvis_capsule
   center=<-0.3977 -1.413 -0.5657>
   rotation=<rot 0 3.2167 -149.7676>
  }]
}
constraints [
 {
  bone1Name=MSTR_frontSkirt_l
  bone2Name=pelvis
  swing1LimitDegrees=45
  swing1Motion=Free
  swing2LimitDegrees=45
  swing2Motion=Free
  twistLimitDegrees=45
  twistMotion=Free
 }
 {
  bone1Name=MSTR_SideSkirt_r
  bone2Name=pelvis
  swing1LimitDegrees=45
  swing1Motion=Free
  swing2LimitDegrees=45
  swing2Motion=Free
  twistLimitDegrees=45
  twistMotion=Free
 }
 {
  bone1Name=MSTR_SideSkirt_l
  bone2Name=pelvis
  swing1LimitDegrees=45
  swing1Motion=Free
  swing2LimitDegrees=45
  swing2Motion=Free
  twistLimitDegrees=45
  twistMotion=Free
 }
 {
  bone1Name=MSTR_BackSkirt_r
  bone2Name=pelvis
  swing1LimitDegrees=45
  swing1Motion=Free
  swing2LimitDegrees=45
  swing2Motion=Free
  twistLimitDegrees=45
  twistMotion=Free
 }
 {
  bone1Name=MSTR_BackSkirt_m
  bone2Name=pelvis
  swing1LimitDegrees=45
  swing1Motion=Free
  swing2LimitDegrees=45
  swing2Motion=Free
  twistLimitDegrees=45
  twistMotion=Free
 }
 {
  bone1Name=MSTR_BackSkirt_l
  bone2Name=pelvis
  swing1LimitDegrees=45
  swing1Motion=Free
  swing2LimitDegrees=45
  swing2Motion=Free
  twistLimitDegrees=45
  twistMotion=Free
 }
 {
  bone1Name=MSTR_frontSkirt_m
  bone2Name=pelvis
  swing1LimitDegrees=45
  swing1Motion=Free
  swing2LimitDegrees=45
  swing2Motion=Free
  twistLimitDegrees=45
  twistMotion=Free
 }
 {
  bone1Name=MSTR_frontSkirt_r
  bone2Name=pelvis
  swing1LimitDegrees=45
  swing1Motion=Free
  swing2LimitDegrees=45
  swing2Motion=Free
  twistLimitDegrees=45
  twistMotion=Free
 }
 {
  bone1Name=dyn_frontSkirt_m_02
  bone2Name=MSTR_frontSkirt_m
  swing1LimitDegrees=45
  swing1Motion=Locked
  swing2LimitDegrees=45
  swing2Motion=Locked
  twistLimitDegrees=45
  twistMotion=Locked
 }
 {
  bone1Name=dyn_frontSkirt_m_03
  bone2Name=dyn_frontSkirt_m_02
  swing1LimitDegrees=45
  swing1Motion=Locked
  swing2LimitDegrees=45
  swing2Motion=Locked
  twistLimitDegrees=45
  twistMotion=Locked
 }
 {
  bone1Name=dyn_frontSkirt_m_04
  bone2Name=dyn_frontSkirt_m_03
  swing1LimitDegrees=45
  swing1Motion=Locked
  swing2LimitDegrees=45
  swing2Motion=Locked
  twistLimitDegrees=45
  twistMotion=Locked
 }
 {
  bone1Name=dyn_frontSkirt_l_02
  bone2Name=MSTR_frontSkirt_l
  swing1LimitDegrees=45
  swing1Motion=Locked
  swing2LimitDegrees=45
  swing2Motion=Locked
  twistLimitDegrees=45
  twistMotion=Locked
 }
 {
  bone1Name=dyn_frontSkirt_l_03
  bone2Name=dyn_frontSkirt_l_02
  swing1LimitDegrees=45
  swing1Motion=Locked
  swing2LimitDegrees=45
  swing2Motion=Locked
  twistLimitDegrees=45
  twistMotion=Locked
 }
 {
  bone1Name=dyn_frontSkirt_l_04
  bone2Name=dyn_frontSkirt_l_03
  swing1LimitDegrees=45
  swing1Motion=Locked
  swing2LimitDegrees=45
  swing2Motion=Locked
  twistLimitDegrees=45
  twistMotion=Locked
 }
]
