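# Physics asset for the F_MED_VioletInk female medium player body, covering only the left front skirt chain, pelvis and left thigh.
# Capsule bodies sit on three skirt bones, the pelvis and the thigh.
# The skirt bodies form a chain from the pelvis with zero-degree swing and twist limits, while the thigh joins the pelvis with free swing and twist.
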PhysicsAsset /BRCosmetics/Characters/Player/Female/Medium/Bodies/F_MED_VioletInk/Meshes/F_MED_VioletInk_Physics1
bodies {
 base_dyn_frontSkirt_l_02 [{
   shapeType=Capsule
   length=11.125
   radius=4.8207
   shapeName=dyn_frontSkirt_m_02_capsule
   center=<0 0 1.1285>
   rotation=<rot 0.0008 -0.0001 90>
  }]
 base_dyn_frontSkirt_l_03 [{
   shapeType=Capsule
   length=11.125
   radius=4.8207
   shapeName=dyn_frontSkirt_m_02_capsule
   center=<0 0 2.2039>
   rotation=<rot 0.0008 -0.0001 90>
  }]
 base_dyn_frontSkirt_l_04 [{
   shapeType=Capsule
   length=11.125
   radius=4.8207
   shapeName=dyn_frontSkirt_m_02_capsule
   center=<0 0 0.1298>
   rotation=<rot 0.0008 -0.0001 90>
  }]
 pelvis [{shapeType=Capsule, length=13.1268, radius=7.3728, shapeName=pelvis_capsule}]
 thigh_l [{
   shapeType=Capsule
   length=36.87
   radius=9.7426
   shapeName=thigh_l_capsule
   center=<-22.4387 -0.662 -0.1095>
   rotation=<rot 0.1844 88.7724 -88.9543>
  }]
}
constraints [
 {bone1Name=base_dyn_frontSkirt_l_02, bone2Name=pelvis, swing1LimitDegrees=0, swing2LimitDegrees=0, twistLimitDegrees=0}
 {bone1Name=base_dyn_frontSkirt_l_03, bone2Name=base_dyn_frontSkirt_l_02, swing1LimitDegrees=0, swing2LimitDegrees=0, twistLimitDegrees=0}
 {bone1Name=base_dyn_frontSkirt_l_04, bone2Name=base_dyn_frontSkirt_l_03, swing1LimitDegrees=0, swing2LimitDegrees=0, twistLimitDegrees=0}
 {
  bone1Name=thigh_l
  bone2Name=pelvis
  swing1LimitDegrees=45
  swing1Motion=Free
  swing2LimitDegrees=45
  swing2Motion=Free
  twistLimitDegrees=45
  twistMotion=Free
 }
]
