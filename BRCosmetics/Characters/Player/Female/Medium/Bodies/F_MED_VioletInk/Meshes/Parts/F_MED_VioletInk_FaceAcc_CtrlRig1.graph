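# Control Rig for the VioletInk female character's face accessory part, correcting hair bone placement as the body moves.
# A sequence runs parent constraints blending hair bones between spine, head and other hair bones, with some offsets weighted by RifleFloat.
# A DeltaRotation helper turns right clavicle rotation into a weight, scaled by OnlyClav, for a back-hair offset.
# Spherical pose reader, aim, crouch and hat correctives also exist but appear disconnected from execution.
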ControlRigBlueprint /BRCosmetics/Characters/Player/Female/Medium/Bodies/F_MED_VioletInk/Meshes/Parts/F_MED_VioletInk_FaceAcc_CtrlRig1
graphs [
 Multiply_3_ContainedGraph {nodes [
   n1 = RigVMFunctionEntryNode()
   RigVMFunctionReturnNode(Result=&RigVMFunction_MathDoubleMul(B=@n1.C, A=&RigVMFunction_MathDoubleMul(A=@n1.A, B=@n1.B).Result).Result)
  ]}
 New Function_ContainedGraph {nodes [
   n1 = RigVMFunctionEntryNode()
   RigVMFunction_MathBoolToInteger = &RigVMFunction_MathBoolToInteger(Value=@n1.onlyTwist)
   RigVMFunction_MathQuaternionSwingTwist = &RigVMFunction_MathQuaternionSwingTwist(
    TwistAxis=@n1.twistAxis
    Input=&RigUnit_GetRelativeTransformForItem(bChildInitial=False, bParentInitial=False, Child=@n1.child, Parent=@n1.parent).RelativeTransform.Rotation
   )
   n2 = RigVMDispatchNode(
    Values=((),())
    Index=@RigVMFunction_MathBoolToInteger.Result
    Values.0=@RigVMFunction_MathQuaternionSwingTwist.Swing
    Values.1=@RigVMFunction_MathQuaternionSwingTwist.Twist
   ) {template="DISPATCH_RigVMDispatch_SelectInt32(in Index,in Values,out Result)"}
   RigVMFunction_MathQuaternionToEuler = &RigVMFunction_MathQuaternionToEuler(Value=@n2.Result, RotationOrder=@n1.Rotation Order)
   > DISPATCH_RigVMDispatch_Print = RigVMDispatchNode(ScreenColor=(R=1,G=1,B=1,A=1), ScreenDuration=1.5, Enabled=@n1.debugRotation, Value=@RigVMFunction_MathQuaternionToEuler.Result) {template="DISPATCH_RigVMDispatch_Print(in Prefix,in Value,in Enabled,in ScreenDuration,in ScreenColor)"}
   Cast = RigVMDispatchNode(Value=@n1.triggerAxis) {template="DISPATCH_RigVMDispatch_CastEnumToInt(in Value,out Result)"}
   DISPATCH_RigVMDispatch_SelectInt32 = RigVMDispatchNode(
    Values=(0,0,0)
    Index=@Cast.Result
    Values.0=@RigVMFunction_MathQuaternionToEuler.Result.X
    Values.1=@RigVMFunction_MathQuaternionToEuler.Result.Y
    Values.2=@RigVMFunction_MathQuaternionToEuler.Result.Z
   ) {template="DISPATCH_RigVMDispatch_SelectInt32(in Index,in Values,out Result)"}
   Remap = &RigVMFunction_MathDoubleRemap(
    bClamp=False
    Value=@DISPATCH_RigVMDispatch_SelectInt32.Result
    TargetMinimum=@n1.outputMinMax.X
    TargetMaximum=@n1.outputMinMax.Y
    SourceMinimum=@n1.rotationMinMax.X
    SourceMaximum=@n1.rotationMinMax.Y
   )
   Absolute = &RigVMFunction_MathDoubleAbs(Value=@Remap.Result)
   Multiply = &RigVMFunction_MathDoubleMul(
    A=&RigVMFunction_MathDoubleClamp(Value=RigVMDispatchNode(True=@Absolute.Result, Condition=@n1.absoluteValue, False=@Remap.Result) {template="DISPATCH_RigVMDispatch_If(in Condition,in True,in False,out Result)"}.Result).Result
    B=@n1.premult
   )
   > RigVMFunctionReturnNode(outputQuat=@n2.Result, outputWeight=@Multiply.Result)
  ]}
 New Function_ContainedGraph {nodes [
   n1 = RigVMFunctionEntryNode()
   RigVMFunction_MathBoolToInteger = &RigVMFunction_MathBoolToInteger(Value=@n1.onlyTwist)
   RigVMFunction_MathQuaternionSwingTwist = &RigVMFunction_MathQuaternionSwingTwist(
    TwistAxis=@n1.twistAxis
    Input=&RigUnit_GetRelativeTransformForItem(bChildInitial=False, bParentInitial=False, Child=@n1.child, Parent=@n1.parent).RelativeTransform.Rotation
   )
   n2 = RigVMDispatchNode(
    Values=((),())
    Index=@RigVMFunction_MathBoolToInteger.Result
    Values.0=@RigVMFunction_MathQuaternionSwingTwist.Swing
    Values.1=@RigVMFunction_MathQuaternionSwingTwist.Twist
   ) {template="DISPATCH_RigVMDispatch_SelectInt32(in Index,in Values,out Result)"}
   RigVMFunction_MathQuaternionToEuler = &RigVMFunction_MathQuaternionToEuler(Value=@n2.Result, RotationOrder=@n1.Rotation Order)
   > DISPATCH_RigVMDispatch_Print = RigVMDispatchNode(ScreenColor=(R=1,G=1,B=1,A=1), ScreenDuration=1.5, Enabled=@n1.debugRotation, Value=@RigVMFunction_MathQuaternionToEuler.Result) {template="DISPATCH_RigVMDispatch_Print(in Prefix,in Value,in Enabled,in ScreenDuration,in ScreenColor)"}
   Cast = RigVMDispatchNode(Value=@n1.triggerAxis) {template="DISPATCH_RigVMDispatch_CastEnumToInt(in Value,out Result)"}
   DISPATCH_RigVMDispatch_SelectInt32 = RigVMDispatchNode(
    Values=(0,0,0)
    Index=@Cast.Result
    Values.0=@RigVMFunction_MathQuaternionToEuler.Result.X
    Values.1=@RigVMFunction_MathQuaternionToEuler.Result.Y
    Values.2=@RigVMFunction_MathQuaternionToEuler.Result.Z
   ) {template="DISPATCH_RigVMDispatch_SelectInt32(in Index,in Values,out Result)"}
   Remap = &RigVMFunction_MathDoubleRemap(
    bClamp=False
    Value=@DISPATCH_RigVMDispatch_SelectInt32.Result
    TargetMinimum=@n1.outputMinMax.X
    TargetMaximum=@n1.outputMinMax.Y
    SourceMinimum=@n1.rotationMinMax.X
    SourceMaximum=@n1.rotationMinMax.Y
   )
   Absolute = &RigVMFunction_MathDoubleAbs(Value=@Remap.Result)
   Multiply = &RigVMFunction_MathDoubleMul(
    A=&RigVMFunction_MathDoubleClamp(Value=RigVMDispatchNode(True=@Absolute.Result, Condition=@n1.absoluteValue, False=@Remap.Result) {template="DISPATCH_RigVMDispatch_If(in Condition,in True,in False,out Result)"}.Result).Result
    B=@n1.premult
   )
   > RigVMFunctionReturnNode(outputQuat=@n2.Result, outputWeight=@Multiply.Result)
  ]}
 RigVMFunctionLibrary {links [], nodes []}
 RigVMFunction_Sequence_1_ContainedGraph {nodes [
   n1 = RigVMFunctionEntryNode()
   RigVMFunction_Sequence = &RigVMFunction_Sequence(ExecuteContext=@n1.ExecuteContext)
   n2 = &RigVMFunction_Sequence(ExecuteContext=@RigVMFunction_Sequence.B)
   n3 = &RigVMFunction_Sequence(ExecuteContext=@n2.B)
   n4 = &RigVMFunction_Sequence(ExecuteContext=@n3.B)
   RigVMFunctionReturnNode(A=@RigVMFunction_Sequence.A, B=@n2.A, C=@n3.A, D=@n4.A, E=@n4.B)
  ]}
 RigVMModel {nodes [
   &RigUnit_ParentConstraint(
    AdvancedSettings=(InterpolationType=Average,RotationOrderForFilter=XZY)
    Child=(Type=Bone,Name="base_Hair_l_01")
    ExecutePin=()
    Parents=((Item=(Type=Bone,Name="spine_05"),Weight=0.5),(Item=(Type=Bone,Name="clavicle_l"),Weight=0.5))
    Weight=1
   )
   > &RigUnit_ParentConstraint(
    AdvancedSettings=(InterpolationType=Average,RotationOrderForFilter=XZY)
    Child=(Type=Bone,Name="dyn_Hair_l_02")
    Parents=((Item=(Type=Bone,Name="spine_05"),Weight=0.5),(Item=(Type=Bone,Name="base_Hair_l_01"),Weight=0.5))
    Weight=1
   )
   RigVMDispatchNode(Enabled=True, ExecuteContext=(), ScreenColor=(R=1,G=1,B=1,A=1), ScreenDuration=0.05, Value=0) {template="DISPATCH_RigVMDispatch_Print(in Prefix,in Value,in Enabled,in ScreenDuration,in ScreenColor)"}
   AlphaInterp = &RigVMFunction_AlphaInterp(
    InterpSpeedDecreasing=2
    InterpSpeedIncreasing=50
    Scale=1
    bClampResult=False
    Value=&RigVMFunction_MathDoubleMul(A=1, B=RigVMVariableNode() {variable=aimFloat}.Value).Result
   )
   > &RigUnit_ModifyTransforms(
    ItemToModify=((Item=(Type=Bone,Name="base_Hair_Back_r_01"),Transform=(Rotation=(X=0,Y=0,Z=0,W=1),Translation=(X=0,Y=-2,Z=-3),Scale3D=(X=1,Y=1,Z=1))))
    WeightMinimum=0
    Weight=@AlphaInterp.Result
   )
   > n1 = &RigUnit_SphericalPoseReader(
    ActiveRegionScaleFactors=(PositiveWidth=1,NegativeWidth=1,PositiveHeight=1,NegativeHeight=1)
    ActiveRegionSize=0.08
    Debug=(bDrawDebug=True,bDraw2D=False,bDrawLocalAxes=False,DebugScale=25,DebugSegments=20,DebugThickness=0.25)
    DriverAxis=(X=1,Y=0,Z=0)
    DriverItem=(Type=Bone,Name="head")
    FalloffRegionScaleFactors=(PositiveWidth=1,NegativeWidth=1,PositiveHeight=1,NegativeHeight=1)
    FalloffSize=0.2
    FlipHeightScaling=False
    FlipWidthScaling=False
    OptionalParentItem=(Type=Bone,Name="None")
    RotationOffset=(X=95,Y=0,Z=-75)
   )
   Multiply_2 = RigVMAggregateNode(C=@n1.OutputParam, B=RigVMVariableNode() {variable=aimFloat}.Value, A=RigVMVariableNode() {variable=RifleFloat}.Value)
   > RigVMDispatchNode(Enabled=True, ScreenColor=(R=1,G=1,B=1,A=1), ScreenDuration=0.05, Value=@Multiply_2.Result) {template="DISPATCH_RigVMDispatch_Print(in Prefix,in Value,in Enabled,in ScreenDuration,in ScreenColor)"}
   > &RigUnit_ModifyTransforms(
    ItemToModify=((Item=(Type=Bone,Name="base_Hair_Back_r_01"),Transform=(Rotation=(X=0,Y=0,Z=-0.4226,W=0.9063),Translation=(X=0,Y=0,Z=3),Scale3D=(X=1,Y=1,Z=1))))
    WeightMinimum=0
    Weight=@Multiply_2.Result
   )
   RigUnit_BeginExecution = &RigUnit_BeginExecution()
   RigVMFunction_Sequence = RigVMAggregateNode(ExecuteContext=@RigUnit_BeginExecution.ExecutePin)
   &RigUnit_ParentConstraint(
    AdvancedSettings=(InterpolationType=Average,RotationOrderForFilter=XZY)
    Child=(Type=Bone,Name="base_Hair_Back_01")
    Parents=((Item=(Type=Bone,Name="spine_05"),Weight=0.9),(Item=(Type=Bone,Name="head"),Weight=0.1))
    Weight=1
    ExecutePin=@RigVMFunction_Sequence.A
   )
   > &RigUnit_ParentConstraint(
    AdvancedSettings=(InterpolationType=Average,RotationOrderForFilter=XZY)
    Child=(Type=Bone,Name="dyn_Hair_Back_02")
    Parents=((Item=(Type=Bone,Name="spine_05"),Weight=0.8),(Item=(Type=Bone,Name="head"),Weight=0.2))
    Weight=1
   )
   > &RigUnit_ParentConstraint(
    AdvancedSettings=(InterpolationType=Average,RotationOrderForFilter=XZY)
    Child=(Type=Bone,Name="dyn_Hair_Back_03")
    Parents=((Item=(Type=Bone,Name="spine_05"),Weight=0.8),(Item=(Type=Bone,Name="head"),Weight=0.2))
    Weight=1
   )
   > &RigUnit_ParentConstraint(
    AdvancedSettings=(InterpolationType=Average,RotationOrderForFilter=XZY)
    Child=(Type=Bone,Name="base_Hair_l_01")
    Filter=(TranslationFilter=(bX=true,bY=true,bZ=true),RotationFilter=(bX=true,bY=True,bZ=true),ScaleFilter=(bX=True,bY=True,bZ=True))
    Parents=((Item=(Type=Bone,Name="spine_05"),Weight=0.7),(Item=(Type=Bone,Name="head"),Weight=0.3))
    Weight=1
   )
   > &RigUnit_ParentConstraint(
    AdvancedSettings=(InterpolationType=Average,RotationOrderForFilter=XZY)
    Child=(Type=Bone,Name="base_Hair_r_01")
    Filter=(TranslationFilter=(bX=true,bY=true,bZ=true),RotationFilter=(bX=True,bY=True,bZ=true),ScaleFilter=(bX=True,bY=True,bZ=True))
    Parents=((Item=(Type=Bone,Name="spine_05"),Weight=0.7),(Item=(Type=Bone,Name="head"),Weight=0.3))
    Weight=1
   )
   n2 = RigVMVariableNode() {variable=RifleFloat}
   &RigUnit_ParentConstraint(
    AdvancedSettings=(InterpolationType=Average,RotationOrderForFilter=XZY)
    Child=(Type=Bone,Name="base_Hair_Secondry_Back_r_01")
    Filter=(TranslationFilter=(bX=true,bY=true,bZ=true),RotationFilter=(bX=true,bY=true,bZ=true),ScaleFilter=(bX=True,bY=True,bZ=True))
    Parents=((Item=(Type=Bone,Name="head"),Weight=1))
    ExecutePin=@RigVMFunction_Sequence.B
    Weight=@n2.Value
   )
   > &RigUnit_ModifyTransforms(
    ItemToModify=((Item=(Type=Bone,Name="base_Hair_Secondry_Back_r_01"),Transform=(Rotation=(X=0,Y=0,Z=0,W=1),Translation=(X=0,Y=-5,Z=0),Scale3D=(X=1,Y=1,Z=1))))
    WeightMinimum=0
    Weight=@n2.Value
   )
   > &RigUnit_ParentConstraint(
    AdvancedSettings=(InterpolationType=Average,RotationOrderForFilter=XZY)
    Child=(Type=Bone,Name="base_Hair_Back_r_01")
    Filter=(TranslationFilter=(bX=true,bY=true,bZ=true),RotationFilter=(bX=True,bY=True,bZ=True),ScaleFilter=(bX=True,bY=True,bZ=True))
    Parents=((Item=(Type=Bone,Name="spine_05"),Weight=0.8),(Item=(Type=Bone,Name="dyn_Hair_Back_02"),Weight=0.2))
    Weight=1
   )
   > &RigUnit_ParentConstraint(
    AdvancedSettings=(InterpolationType=Average,RotationOrderForFilter=XZY)
    Child=(Type=Bone,Name="dyn_Hair_Back_r_02")
    Parents=((Item=(Type=Bone,Name="spine_05"),Weight=0.7),(Item=(Type=Bone,Name="dyn_Hair_Back_02"),Weight=0.2),(Item=(Type=Bone,Name="base_Hair_Secondry_Back_r_01"),Weight=0.1))
    Weight=1
   )
   > &RigUnit_ParentConstraint(
    AdvancedSettings=(InterpolationType=Average,RotationOrderForFilter=XZY)
    Child=(Type=Bone,Name="dyn_Hair_Back_r_03")
    Parents=((Item=(Type=Bone,Name="spine_05"),Weight=0.8),(Item=(Type=Bone,Name="dyn_Hair_Back_02"),Weight=0.2))
    Weight=1
   )
   &RigUnit_ParentConstraint(
    AdvancedSettings=(InterpolationType=Average,RotationOrderForFilter=XZY)
    Child=(Type=Bone,Name="base_Hair_Back_l_01")
    Parents=((Item=(Type=Bone,Name="spine_05"),Weight=0.9),(Item=(Type=Bone,Name="base_Hair_Back_01"),Weight=0.1))
    Weight=1
    ExecutePin=@RigVMFunction_Sequence.C
   )
   > &RigUnit_ParentConstraint(
    AdvancedSettings=(InterpolationType=Average,RotationOrderForFilter=XZY)
    Child=(Type=Bone,Name="dyn_Hair_Back_l_02")
    Parents=((Item=(Type=Bone,Name="spine_05"),Weight=0.8),(Item=(Type=Bone,Name="dyn_Hair_Back_02"),Weight=0.2))
    Weight=1
   )
   > &RigUnit_ParentConstraint(
    AdvancedSettings=(InterpolationType=Average,RotationOrderForFilter=XZY)
    Child=(Type=Bone,Name="dyn_Hair_Back_l_03")
    Parents=((Item=(Type=Bone,Name="spine_05"),Weight=0.8),(Item=(Type=Bone,Name="dyn_Hair_Back_03"),Weight=0.2))
    Weight=1
   )
   > &RigUnit_ParentConstraint(
    AdvancedSettings=(InterpolationType=Average,RotationOrderForFilter=XZY)
    Child=(Type=Bone,Name="dyn_Hair_Back_Secondry_a_l_01")
    Parents=((Item=(Type=Bone,Name="head"),Weight=1))
    Weight=1
   )
   > &RigUnit_ParentConstraint(
    AdvancedSettings=(InterpolationType=Average,RotationOrderForFilter=XZY)
    Child=(Type=Bone,Name="base_Hair_Secondry_Back_r_01")
    Parents=((Item=(Type=Bone,Name="head"),Weight=1))
    Weight=1
   )
   > &RigUnit_ModifyTransforms(
    ItemToModify=((Item=(Type=Bone,Name="dyn_Hair_Back_Secondry_a_l_01"),Transform=(Rotation=(X=0,Y=0,Z=0,W=1),Translation=(X=0,Y=2,Z=2),Scale3D=(X=1,Y=1,Z=1))))
    Weight=1
    WeightMinimum=0
   )
   DeltaRotation v4_1 = RigVMCollapseNode(
    Rotation Order=ZYX
    child=(Type=Bone,Name="clavicle_r")
    outputMinMax=(X=0,Y=1)
    parent=(Type=Null,Name="clavicle_r_null")
    premult=1
    rotationMinMax=(X=0,Y=-45)
    triggerAxis=Y
    twistAxis=(X=1,Y=0,Z=0)
    ExecuteContext=@RigVMFunction_Sequence.D
   )
   n3 = &RigVMFunction_AlphaInterp(
    InterpSpeedDecreasing=5
    InterpSpeedIncreasing=5
    Scale=1
    bClampResult=False
    bInterpResult=False
    Value=&RigVMFunction_MathDoubleMul(A=@"DeltaRotation v4_1".outputWeight, B=RigVMVariableNode() {variable=OnlyClav}.Value).Result
   )
   > &RigUnit_ModifyTransforms(
    ItemToModify=((Item=(Type=Bone,Name="dyn_Hair_Back_r_01"),Transform=(Rotation=(X=0,Y=-0.0872,Z=0,W=0.9962),Translation=(X=-5,Y=0,Z=0),Scale3D=(X=1,Y=1,Z=1))))
    WeightMinimum=0
    Weight=@n3.Result
   )
   RigVMFunction_ControlFlowBranch = &RigVMFunction_ControlFlowBranch(ExecuteContext=(), Condition=RigVMVariableNode() {variable=Crocuh_aim}.Value)
   n4 = &RigUnit_SphericalPoseReader(
    ActiveRegionScaleFactors=(PositiveWidth=1.2,NegativeWidth=1.5,PositiveHeight=1.5,NegativeHeight=1.2)
    ActiveRegionSize=0.1
    Debug=(bDrawDebug=true,bDraw2D=False,bDrawLocalAxes=False,DebugScale=25,DebugSegments=20,DebugThickness=0.25)
    DriverAxis=(X=-1,Y=0,Z=0)
    DriverItem=(Type=Bone,Name="upperarm_r")
    FalloffRegionScaleFactors=(PositiveWidth=1,NegativeWidth=3,PositiveHeight=3,NegativeHeight=1)
    FalloffSize=0.3
    FlipHeightScaling=False
    FlipWidthScaling=False
    OptionalParentItem=(Type=Bone,Name="clavicle_r")
    RotationOffset=(X=-127.4586,Y=-25.649,Z=60.4541)
    ExecutePin=@RigVMFunction_ControlFlowBranch.True
   )
   > &RigUnit_ModifyTransforms(
    ItemToModify="((Item=(Type=Bone,Name=\"dyn_Hair_r_01\"),Transform=(Rotation=(X=0.0999,Y=0.4508,Z=0.192,W=-0.866),Translation=(X=-3.9377,Y=-3.0727,Z=-2.5809),Scale3D=(X=1,Y=1,Z=1))),(Item=(Type=Bone,Name=\"dyn_Hair_r_02\"),Transform=(Rotation=(X=-0.0227,Y=-0.1285,Z=0.1722,W= ...(truncated)"
    WeightMinimum=0
    Weight=@n4.OutputParam
   )
   > &RigUnit_ModifyTransforms(
    ItemToModify="((Item=(Type=Bone,Name=\"dyn_Hair_l_01\"),Transform=(Rotation=(X=0,Y=-0.1305,Z=0,W=0.9914),Translation=(X=-0.8695,Y=0,Z=3.2452),Scale3D=(X=1,Y=1,Z=1))),(Item=(Type=Bone,Name=\"dyn_Hair_l_02\"),Transform=(Rotation=(X=0,Y=0,Z=0,W=1),Translation=(X=0,Y=0,Z=0),Sca ...(truncated)"
    Weight=1
    WeightMinimum=0
   )
   n5 = &RigVMFunction_ControlFlowBranch(ExecuteContext=(), Condition=RigVMVariableNode() {variable=Crocuh_aim_not}.Value)
   n6 = &RigUnit_SphericalPoseReader(
    ActiveRegionScaleFactors=(PositiveWidth=1.2,NegativeWidth=1.5,PositiveHeight=1.5,NegativeHeight=1.2)
    ActiveRegionSize=0.1
    Debug=(bDrawDebug=true,bDraw2D=False,bDrawLocalAxes=False,DebugScale=25,DebugSegments=20,DebugThickness=0.25)
    DriverAxis=(X=-1,Y=0,Z=0)
    DriverItem=(Type=Bone,Name="upperarm_r")
    FalloffRegionScaleFactors=(PositiveWidth=1,NegativeWidth=5,PositiveHeight=5,NegativeHeight=1)
    FalloffSize=0.2
    FlipHeightScaling=False
    FlipWidthScaling=False
    OptionalParentItem=(Type=Bone,Name="clavicle_r")
    RotationOffset=(X=-117.1398,Y=-25.649,Z=30.3874)
    ExecutePin=@n5.True
   )
   > &RigUnit_ModifyTransforms(
    ItemToModify="((Item=(Type=Bone,Name=\"dyn_Hair_r_01\"),Transform=(Rotation=(X=-0.0983,Y=0.2991,Z=-0.031,W=-0.9486),Translation=(X=-0.8855,Y=-3.7174,Z=1.2646),Scale3D=(X=1,Y=1,Z=1))),(Item=(Type=Bone,Name=\"dyn_Hair_r_02\"),Transform=(Rotation=(X=0,Y=0,Z=0,W=1),Translation= ...(truncated)"
    WeightMinimum=0
    Weight=@n6.OutputParam
   )
   > &RigUnit_ModifyTransforms(
    ItemToModify="((Item=(Type=Bone,Name=\"dyn_Hair_l_01\"),Transform=(Rotation=(X=0,Y=0,Z=0,W=1),Translation=(X=0,Y=-1.7048,Z=4.2368),Scale3D=(X=1,Y=1,Z=1))),(Item=(Type=Bone,Name=\"dyn_Hair_l_02\"),Transform=(Rotation=(X=0,Y=0,Z=0,W=1),Translation=(X=0,Y=0,Z=0),Scale3D=(X=1,Y ...(truncated)"
    Weight=1
    WeightMinimum=0
   )
   n7 = RigVMVariableNode() {variable=Rifle}
   n8 = &RigVMFunction_ControlFlowBranch(ExecuteContext=(), Condition=@n7.Value)
   &RigUnit_ParentConstraint(
    AdvancedSettings=(InterpolationType=Average,RotationOrderForFilter=XZY)
    Child=(Type=Bone,Name="base_Hair_Back_r_01")
    Filter=(TranslationFilter=(bX=true,bY=true,bZ=false),RotationFilter=(bX=True,bY=True,bZ=True),ScaleFilter=(bX=True,bY=True,bZ=True))
    Parents=((Item=(Type=Bone,Name="spine_05"),Weight=1),(Item=(Type=Bone,Name="clavicle_r"),Weight=0))
    Weight=1
    ExecutePin=@n8.True
   )
   > &RigUnit_ParentConstraint(
    AdvancedSettings=(InterpolationType=Average,RotationOrderForFilter=XZY)
    Child=(Type=Bone,Name="dyn_Hair_Back_r_02")
    Filter=(TranslationFilter=(bX=true,bY=true,bZ=false),RotationFilter=(bX=True,bY=True,bZ=True),ScaleFilter=(bX=True,bY=True,bZ=True))
    Parents=((Item=(Type=Bone,Name="spine_05"),Weight=0.8),(Item=(Type=Bone,Name="dyn_Hair_Back_02"),Weight=0.2))
    Weight=1
   )
   > &RigUnit_ParentConstraint(
    AdvancedSettings=(InterpolationType=Average,RotationOrderForFilter=XZY)
    Child=(Type=Bone,Name="dyn_Hair_Back_r_03")
    Filter=(TranslationFilter=(bX=true,bY=true,bZ=false),RotationFilter=(bX=True,bY=True,bZ=True),ScaleFilter=(bX=True,bY=True,bZ=True))
    Parents=((Item=(Type=Bone,Name="spine_05"),Weight=0.8),(Item=(Type=Bone,Name="dyn_Hair_Back_03"),Weight=0.2))
    Weight=1
   )
   > RigVMDispatchNode(Enabled=True, ScreenColor=(R=1,G=1,B=1,A=1), ScreenDuration=0.05, Value=@n7.Value) {template="DISPATCH_RigVMDispatch_Print(in Prefix,in Value,in Enabled,in ScreenDuration,in ScreenColor)"}
   n9 = &RigUnit_SphericalPoseReader(
    ActiveRegionScaleFactors=(PositiveWidth=1.2,NegativeWidth=1.5,PositiveHeight=1.5,NegativeHeight=1.2)
    ActiveRegionSize=0.1
    Debug=(bDrawDebug=true,bDraw2D=False,bDrawLocalAxes=False,DebugScale=25,DebugSegments=20,DebugThickness=0.25)
    DriverAxis=(X=-1,Y=0,Z=0)
    DriverItem=(Type=Bone,Name="upperarm_r")
    ExecutePin=()
    FalloffRegionScaleFactors=(PositiveWidth=1,NegativeWidth=5,PositiveHeight=5,NegativeHeight=1)
    FalloffSize=0.2
    FlipHeightScaling=False
    FlipWidthScaling=False
    OptionalParentItem=(Type=Bone,Name="clavicle_r")
    RotationOffset=(X=-117.1398,Y=-25.649,Z=30.3874)
   )
   > &RigUnit_ModifyTransforms(
    ItemToModify="((Item=(Type=Bone,Name=\"dyn_Hair_r_01\"),Transform=(Rotation=(X=-0.0983,Y=0.2991,Z=-0.031,W=-0.9486),Translation=(X=-0.8855,Y=-3.7174,Z=1.2646),Scale3D=(X=1,Y=1,Z=1))),(Item=(Type=Bone,Name=\"dyn_Hair_r_02\"),Transform=(Rotation=(X=0,Y=0,Z=0,W=1),Translation= ...(truncated)"
    WeightMinimum=0
    Weight=@n9.OutputParam
   )
   > &RigUnit_ModifyTransforms(
    ItemToModify="((Item=(Type=Bone,Name=\"dyn_Hair_l_01\"),Transform=(Rotation=(X=0,Y=0,Z=0,W=1),Translation=(X=0,Y=-1.7048,Z=4.2368),Scale3D=(X=1,Y=1,Z=1))),(Item=(Type=Bone,Name=\"dyn_Hair_l_02\"),Transform=(Rotation=(X=0,Y=0,Z=0,W=1),Translation=(X=0,Y=0,Z=0),Scale3D=(X=1,Y ...(truncated)"
    Weight=1
    WeightMinimum=0
   )
   n10 = &RigUnit_SphericalPoseReader(
    ActiveRegionScaleFactors=(PositiveWidth=1.2,NegativeWidth=1.5,PositiveHeight=1.5,NegativeHeight=1.2)
    ActiveRegionSize=0.1
    Debug=(bDrawDebug=true,bDraw2D=False,bDrawLocalAxes=False,DebugScale=25,DebugSegments=20,DebugThickness=0.25)
    DriverAxis=(X=-1,Y=0,Z=0)
    DriverItem=(Type=Bone,Name="upperarm_r")
    ExecutePin=()
    FalloffRegionScaleFactors=(PositiveWidth=1,NegativeWidth=5,PositiveHeight=5,NegativeHeight=1)
    FalloffSize=0.2
    FlipHeightScaling=False
    FlipWidthScaling=False
    OptionalParentItem=(Type=Bone,Name="clavicle_r")
    RotationOffset=(X=-117.1398,Y=-25.649,Z=30.3874)
   )
   > &RigUnit_ModifyTransforms(
    ItemToModify="((Item=(Type=Bone,Name=\"dyn_Hair_r_01\"),Transform=(Rotation=(X=-0.0983,Y=0.2991,Z=-0.031,W=-0.9486),Translation=(X=-0.8855,Y=-5,Z=1.2646),Scale3D=(X=1,Y=1,Z=1))),(Item=(Type=Bone,Name=\"dyn_Hair_r_02\"),Transform=(Rotation=(X=0,Y=0,Z=0,W=1),Translation=(X=0, ...(truncated)"
    WeightMinimum=0
    Weight=@n10.OutputParam
   )
   > &RigUnit_ModifyTransforms(
    ItemToModify="((Item=(Type=Bone,Name=\"dyn_Hair_l_01\"),Transform=(Rotation=(X=0,Y=0,Z=0,W=1),Translation=(X=0,Y=-1.7048,Z=4.2368),Scale3D=(X=1,Y=1,Z=1))),(Item=(Type=Bone,Name=\"dyn_Hair_l_02\"),Transform=(Rotation=(X=0,Y=0,Z=0,W=1),Translation=(X=0,Y=0,Z=0),Scale3D=(X=1,Y ...(truncated)"
    Weight=1
    WeightMinimum=0
   )
   DeltaRotation v4 = RigVMCollapseNode(
    ExecuteContext=()
    Rotation Order=ZYX
    child=(Type=Bone,Name="head")
    outputMinMax=(X=0,Y=1)
    parent=(Type=Null,Name="head_null")
    premult=1
    rotationMinMax=(X=-75,Y=75)
    triggerAxis=Y
    twistAxis=(X=1,Y=0,Z=0)
   )
   &RigUnit_ModifyTransforms(
    ExecutePin=()
    ItemToModify="((Item=(Type=Bone,Name=\"dyn_Hair_r_01\"),Transform=(Rotation=(X=0,Y=0,Z=-0.3827,W=0.9239),Translation=(X=0,Y=0,Z=100),Scale3D=(X=1,Y=1,Z=1))),(Item=(Type=Bone,Name=\"dyn_Hair_r_02\"),Transform=(Rotation=(X=0,Y=0,Z=-0.2588,W=0.9659),Translation=(X=0,Y=0,Z=0),S ...(truncated)"
    Weight=1
    WeightMinimum=0
   )
   &RigUnit_ModifyTransforms(
    ExecutePin=()
    ItemToModify=((Item=(Type=Bone,Name="dyn_Hair_r_01"),Transform=(Rotation=(X=0,Y=0,Z=0,W=1),Translation=(X=0,Y=1,Z=0),Scale3D=(X=1,Y=1,Z=1))),(Item=(Type=Bone,Name="dyn_Hair_l_01"),Transform=(Rotation=(X=0,Y=0,Z=0,W=1),Translation=(X=0,Y=-1,Z=0),Scale3D=(X=1,Y=1,Z=1))))
    Weight=1
    WeightMinimum=0
   )
   n11 = &RigVMFunction_AnimEvalRichCurve(
    Curve=(EditorCurveData=(Keys=((),(Time=0.5,Value=1),(Time=0.8))))
    SourceMaximum=2
    SourceMinimum=0
    TargetMinimum=0
    Value=&RigVMFunction_MathDoubleRemap(
     SourceMaximum=75
     SourceMinimum=45
     TargetMaximum=1
     TargetMinimum=0
     bClamp=false
     Value=&RigUnit_GetRelativeTransformForItem(Child=(Type=Bone,Name="pelvis"), Parent=(Type=Bone,Name="root"), bChildInitial=False, bParentInitial=False).RelativeTransform.Translation.Z
    ).Result
   )
   Multiply = &RigVMFunction_MathDoubleMul(B=@n11.Result, A=&RigVMFunction_MathBoolToFloat(Value=RigVMVariableNode() {variable=CrouchWalk}.Value).Result)
   &RigUnit_ModifyTransforms(
    ExecutePin=()
    ItemToModify="((Item=(Type=Bone,Name=\"dyn_hat_Back_01\"),Transform=(Rotation=(X=0,Y=0,Z=0,W=1),Translation=(X=0,Y=0,Z=0),Scale3D=(X=1,Y=1,Z=1))),(Item=(Type=Bone,Name=\"dyn_hat_01\"),Transform=(Rotation=(X=0,Y=-0.866,Z=0,W=0.5),Translation=(X=0,Y=0,Z=0),Scale3D=(X=1,Y=1,Z= ...(truncated)"
    WeightMinimum=0
    Weight=@Multiply.Result
   )
   RigVMVariableNode() {variable=Rifle}
  ]}
]
hierarchy [
 root {loc=<0 0 0>, rot=<0 0 0 1>}
 pelvis {parent=root, loc=<0 -0.5301 98.0571>, rot=<0.0127 0.707 -0.0127 -0.707>}
 spine_01 {parent=pelvis, loc=<6.1208 -0.9328 0>, rot=<0 0 0.034 -0.9994>}
 spine_02 {parent=spine_01, loc=<6.5997 -0.2095 0>, rot=<0 0 0.0238 -0.9997>}
 spine_03 {parent=spine_02, loc=<6.4914 -0.1237 0>, rot=<0 0 0.0149 -0.9999>}
 spine_04 {parent=spine_03, loc=<6.9598 0.0794 0>, rot=<0 0 -0.0132 -0.9999>}
 spine_05 {parent=spine_04, loc=<6.9987 0 0>, rot=<0 0 -0.0015 -1>}
 neck_01 {parent=spine_05, loc=<11.3138 0.1529 0>, rot=<0 0 -0.2003 -0.9797>}
 neck_02 {parent=neck_01, loc=<4.5627 0 0>, rot=<0 0 0.0504 -0.9987>}
 head {parent=neck_02, loc=<5.3914 0 0>, rot=<0 0 0.115 -0.9934>}
 base_Hair_Front_01 {parent=head, loc=<9.039 10.0748 2.2053>, rot=<0.5294 -0.305 0.7856 0.0979>}
 dyn_Hair_Front_01 {parent=base_Hair_Front_01, loc=<0 0 0>, rot=<0 0 0 1>}
 dyn_Hair_Front_02 {parent=dyn_Hair_Front_01, loc=<5.3838 0.0266 0.3169>, rot=<0.0861 0.1159 -0.0311 -0.989>}
 base_Hair_Back_01 {parent=head, loc=<0.7046 -7.3333 1.5201>, rot=<0.191 -0.6808 0.6884 -0.1615>}
 dyn_Hair_Back_01 {parent=base_Hair_Back_01, loc=<0 0 0>, rot=<0 0 0 1>}
 dyn_Hair_Back_02 {parent=dyn_Hair_Back_01, loc=<9.4144 0 -0.4972>, rot=<0 -0.06 0 -0.9982>}
 dyn_Hair_Back_03 {parent=dyn_Hair_Back_02, loc=<9.1453 -0.1102 -0.653>, rot=<0.0895 0 0 -0.996>}
 base_Hair_Back_a_01 {parent=dyn_Hair_Back_03, loc=<11.2843 3.9032 3.8351>, rot=<0.0085 -0.1906 -0.2672 -0.9446>}
 dyn_Hair_Back_a_01 {parent=base_Hair_Back_a_01, loc=<0 0 0>, rot=<0 0 0 1>}
 dyn_Hair_Back_a_02 {parent=dyn_Hair_Back_a_01, loc=<10.8891 0 0>, rot=<0 -0.2714 0.2714 -0.9234>}
 dyn_Hair_Back_a_03 {parent=dyn_Hair_Back_a_02, loc=<10.9042 0 0>, rot=<0 0 0 1>}
 base_Hair_Back_b_01 {parent=dyn_Hair_Back_03, loc=<14.4441 -1.2879 -3.7162>, rot=<-0.1142 -0.0846 -0.1356 -0.9805>}
 dyn_Hair_Back_b_01 {parent=base_Hair_Back_b_01, loc=<0 0 0>, rot=<0 0 0 1>}
 dyn_Hair_Back_b_02 {parent=dyn_Hair_Back_b_01, loc=<9.8933 0 0>, rot=<-0.062 0.1847 -0.3123 -0.9298>}
 base_Hair_Back_r_01 {parent=head, loc=<2.2127 -4.6859 7.4284>, rot=<0.1763 -0.6848 0.6848 -0.1763>}
 dyn_Hair_Back_r_01 {parent=base_Hair_Back_r_01, loc=<0 0 0>, rot=<0 0 0 1>}
 dyn_Hair_Back_r_02 {parent=dyn_Hair_Back_r_01, loc=<9.4144 0 -0.4972>, rot=<0 0.0321 0 -0.9995>}
 dyn_Hair_Back_r_03 {parent=dyn_Hair_Back_r_02, loc=<9.1453 -0.1102 -0.653>, rot=<0.0895 0 0 -0.996>}
 dyn_Hair_Back_r_04 {parent=dyn_Hair_Back_r_03, loc=<6.0746 0.6349 -4.3148>, rot=<0 0 0.0243 -0.9997>}
 base_Hair_Back_a_r_01 {parent=dyn_Hair_Back_r_04, loc=<4.0186 -2.0884 -3.3267>, rot=<-0.1328 -0.3363 0.3165 -0.877>}
 dyn_Hair_Back_a_r_01 {parent=base_Hair_Back_a_r_01, loc=<0 0 0>, rot=<0 0 0 1>}
 dyn_Hair_Back_c_r_01 {parent=dyn_Hair_Back_a_r_01, loc=<5.5613 -4.0655 0.6853>, rot=<-0.0525 0.1224 0.3879 -0.912>}
 dyn_Hair_Back_b_r_01 {parent=dyn_Hair_Back_a_r_01, loc=<12.2052 5.2039 0.3365>, rot=<-0.0509 0.1187 -0.5998 -0.7897>}
 dyn_Hair_Back_b_r_02 {parent=dyn_Hair_Back_b_r_01, loc=<5.3864 1.6229 1.931>, rot=<0.6677 0.4422 -0.2153 -0.5589>}
 dyn_Hair_Back_d_r_01 {parent=dyn_Hair_Back_r_02, loc=<4.4045 -4.7776 -2.8023>, rot=<-0.2872 0.6598 0.4785 -0.5032>}
 dyn_Hair_Back_d_r_02 {parent=dyn_Hair_Back_d_r_01, loc=<3.2085 0 0>, rot=<0 0 0 1>}
 base_Hair_Secondry_Back_r_01 {parent=base_Hair_Back_r_01, loc=<4.9346 0 -0.2325>, rot=<0 0 0 1>}
 base_Hair_Back_l_01 {parent=head, loc=<0.9888 -5.2206 -5.2919>, rot=<0.1155 -0.7013 0.6865 -0.1537>}
 dyn_Hair_Back_l_01 {parent=base_Hair_Back_l_01, loc=<0 0 0>, rot=<0 0 0 1>}
 dyn_Hair_Back_l_02 {parent=dyn_Hair_Back_l_01, loc=<7.7051 -0.0387 -0.5557>, rot=<0.0002 0.0171 -0.0113 -0.9998>}
 dyn_Hair_Back_l_03 {parent=dyn_Hair_Back_l_02, loc=<9.1453 -0.1102 -0.653>, rot=<0.0894 -0.0519 -0.0047 -0.9946>}
 base_Hair_Back_b_l_01 {parent=dyn_Hair_Back_l_03, loc=<3.2802 1.5407 -0.1252>, rot=<0.4605 0.326 -0.3396 -0.7526>}
 dyn_Hair_Back_b_l_01 {parent=base_Hair_Back_b_l_01, loc=<0 0 0>, rot=<0 0 0 1>}
 dyn_Hair_Back_b_l_02 {parent=dyn_Hair_Back_b_l_01, loc=<7.752 0 0>, rot=<0 0.2356 0 -0.9718>}
 base_Hair_Back_c_l_01 {parent=dyn_Hair_Back_l_03, loc=<9.7094 0.2216 -0.3218>, rot=<0.1594 0.0527 -0.3635 -0.9164>}
 dyn_Hair_Back_c_l_01 {parent=base_Hair_Back_c_l_01, loc=<0 0 0>, rot=<0 0 0 1>}
 dyn_Hair_Back_c_l_02 {parent=dyn_Hair_Back_c_l_01, loc=<9.2457 0 0>, rot=<0 0 0.3864 -0.9223>}
 dyn_Hair_Back_c_l_03 {parent=dyn_Hair_Back_c_l_02, loc=<9.0351 0 0>, rot=<0 0 0 1>}
 base_Hair_Back_a_l_01 {parent=dyn_Hair_Back_l_01, loc=<1.71 2.061 -2.42>, rot=<0.3224 -0.1253 0.061 -0.9363>}
 dyn_Hair_Back_a_l_01 {parent=base_Hair_Back_a_l_01, loc=<0 0 0>, rot=<0 0 0 1>}
 dyn_Hair_Back_a_l_02 {parent=dyn_Hair_Back_a_l_01, loc=<6.2675 0.3161 0.5473>, rot=<0 0.4592 0 -0.8883>}
 dyn_Hair_Back_a_l_03 {parent=dyn_Hair_Back_a_l_02, loc=<6.3158 0 0>, rot=<0.0946 0.4714 -0.1205 -0.8685>}
 dyn_Hair_Back_Secondry_a_l_01 {parent=dyn_Hair_Back_a_l_01, loc=<-2.652 0 0>, rot=<0 0 0 1>}
 base_hat_01 {parent=head, loc=<10.7632 -12.1789 0>, rot=<0.0566 -0.7048 0.7048 -0.0566>}
 dyn_hat_01 {parent=base_hat_01, loc=<0 0 0>, rot=<0 0 0 1>}
 dyn_hat_02 {parent=dyn_hat_01, loc=<4.4116 0 0>, rot=<0 0 0 1>}
 head_null {parent=head, type=Null, loc=<0 0 0>, rot=<0 0 0 1>}
 base_hat_Back_01 {parent=head, loc=<3.4753 -9.4287 0>, rot=<0.0029 -0.7071 -0.0029 -0.7071>}
 dyn_hat_Back_01 {parent=base_hat_Back_01, loc=<0 0 0>, rot=<0 0 0 1>}
 FACIAL_C_FacialRoot {parent=head, loc=<2.1524 -0.746 -0.0002>, rot=<0.4879 -0.5118 -0.4879 -0.5118>}
 FACIAL_C_Forehead {parent=FACIAL_C_FacialRoot, loc=<0.0036 -2.4195 9.5993>, rot=<-0.0402 0 -0.0001 -0.9992>}
 FACIAL_L_ForeheadIn {parent=FACIAL_C_FacialRoot, loc=<1.9046 -2.4839 9.3934>, rot=<-0.0575 -0.0696 0.0044 -0.9959>}
 FACIAL_R_ForeheadIn {parent=FACIAL_C_FacialRoot, loc=<-1.8975 -2.487 9.3781>, rot=<-0.0575 0.0696 -0.0046 -0.9959>}
 FACIAL_L_ForeheadMid {parent=FACIAL_C_FacialRoot, loc=<3.57 -2.9008 8.7512>, rot=<-0.0678 -0.1757 -0.0201 -0.9819>}
 FACIAL_R_ForeheadMid {parent=FACIAL_C_FacialRoot, loc=<-3.5592 -2.8995 8.733>, rot=<-0.0679 0.1757 0.02 -0.9819>}
 FACIAL_L_ForeheadOut {parent=FACIAL_C_FacialRoot, loc=<5.1382 -2.9422 7.5932>, rot=<-0.1183 -0.2559 -0.0784 -0.9562>}
 FACIAL_R_ForeheadOut {parent=FACIAL_C_FacialRoot, loc=<-5.1142 -2.9404 7.5755>, rot=<-0.1183 0.2558 0.0783 -0.9563>}
 FACIAL_L_EyesackUpper {parent=FACIAL_C_FacialRoot, loc=<3.7507 -1.6307 9.223>, rot=<0.052 -0.0844 -0.0643 -0.993>}
 FACIAL_R_EyesackUpper {parent=FACIAL_C_FacialRoot, loc=<-3.7478 -1.6382 9.206>, rot=<0.052 0.0844 0.0642 -0.993>}
 FACIAL_L_EyelidUpperFurrow {parent=FACIAL_C_FacialRoot, loc=<3.7207 -1.3515 9.1163>, rot=<-0.0269 -0.1511 -0.0792 -0.985>}
 FACIAL_R_EyelidUpperFurrow {parent=FACIAL_C_FacialRoot, loc=<-3.7187 -1.3629 9.1002>, rot=<-0.0269 0.1511 0.0791 -0.985>}
 FACIAL_L_EyelidUpperA {parent=FACIAL_C_FacialRoot, loc=<3.5358 0.1261 6.4299>, rot=<-0.1447 0 -0.0001 -0.9895>}
 FACIAL_R_EyelidUpperA {parent=FACIAL_C_FacialRoot, loc=<-3.5342 0.1212 6.4291>, rot=<-0.1447 0 -0.0001 -0.9895>}
 FACIAL_L_Eye {parent=FACIAL_C_FacialRoot, loc=<3.5285 0.0925 6.3896>, rot=<0.0032 0 -0.0001 -1>}
 FACIAL_R_Eye {parent=FACIAL_C_FacialRoot, loc=<-3.5313 0.0919 6.39>, rot=<0.0032 0 -0.0001 -1>}
 FACIAL_L_EyelidLowerA {parent=FACIAL_C_FacialRoot, loc=<3.5358 0.1261 6.4299>, rot=<0.2871 0 -0.0001 -0.9579>}
 FACIAL_R_EyelidLowerA {parent=FACIAL_C_FacialRoot, loc=<-3.5342 0.1212 6.4291>, rot=<0.2871 0 -0.0001 -0.9579>}
 FACIAL_L_EyeCornerInner {parent=FACIAL_C_FacialRoot, loc=<1.9618 0.7791 8.0803>, rot=<0.0032 0 -0.0001 -1>}
 FACIAL_R_EyeCornerInner {parent=FACIAL_C_FacialRoot, loc=<-1.946 0.7632 8.0877>, rot=<0.0032 0 -0.0001 -1>}
 FACIAL_L_EyeCornerOuter {parent=FACIAL_C_FacialRoot, loc=<5.5632 0.2925 7.2837>, rot=<0.003 -0.3826 0.0012 -0.9239>}
 FACIAL_R_EyeCornerOuter {parent=FACIAL_C_FacialRoot, loc=<-5.568 0.2864 7.2903>, rot=<0.0029 0.3826 -0.0013 -0.9239>}
 FACIAL_L_EyesackLower {parent=FACIAL_C_FacialRoot, loc=<4.2386 1.9592 7.7188>, rot=<-0.0191 -0.1398 -0.0219 -0.9898>}
 FACIAL_R_EyesackLower {parent=FACIAL_C_FacialRoot, loc=<-4.2322 1.9442 7.7261>, rot=<-0.0191 0.1398 0.0217 -0.9898>}
 FACIAL_L_CheekInner {parent=FACIAL_C_FacialRoot, loc=<2.3888 2.5237 8.1172>, rot=<-0.1201 -0.0886 -0.1969 -0.969>}
 FACIAL_R_CheekInner {parent=FACIAL_C_FacialRoot, loc=<-2.3811 2.5209 8.1092>, rot=<-0.1201 0.0886 0.1968 -0.969>}
 FACIAL_L_CheekOuter {parent=FACIAL_C_FacialRoot, loc=<4.5751 2.3869 6.378>, rot=<0.0582 -0.3766 0.1015 -0.919>}
 FACIAL_R_CheekOuter {parent=FACIAL_C_FacialRoot, loc=<-4.5641 2.379 6.3725>, rot=<0.0581 0.3766 -0.1016 -0.919>}
 FACIAL_L_NasolabialBulge {parent=FACIAL_C_FacialRoot, loc=<3.6398 5.2096 6.2521>, rot=<0.0609 -0.3773 0.0249 -0.9238>}
 FACIAL_R_NasolabialBulge {parent=FACIAL_C_FacialRoot, loc=<-3.6285 5.2068 6.2485>, rot=<0.0608 0.3773 -0.0251 -0.9238>}
 FACIAL_L_NasolabialFurrow {parent=FACIAL_C_FacialRoot, loc=<3.1228 4.3417 8.9083>, rot=<-0.0303 -0.127 -0.3633 -0.9225>}
 FACIAL_R_NasolabialFurrow {parent=FACIAL_C_FacialRoot, loc=<-3.1174 4.3326 8.896>, rot=<-0.0303 0.127 0.3632 -0.9225>}
 FACIAL_L_CheekLower {parent=FACIAL_C_FacialRoot, loc=<4.025 5.0651 5.7361>, rot=<0.0375 -0.4267 0.059 -0.9017>}
 FACIAL_R_CheekLower {parent=FACIAL_C_FacialRoot, loc=<-4.0142 5.0607 5.7327>, rot=<0.0374 0.4267 -0.0592 -0.9017>}
 FACIAL_C_Nose {parent=FACIAL_C_FacialRoot, loc=<0.0044 1.5954 9.9003>, rot=<-0.1272 0 -0.0001 -0.9919>}
 FACIAL_C_NoseLower {parent=FACIAL_C_Nose, loc=<-0.0043 2.9391 -0.5241>, rot=<0.1305 0 0 -0.9914>}
 FACIAL_L_Nostril {parent=FACIAL_C_Nose, loc=<0.8242 2.2238 -0.4846>, rot=<0.0173 0.1305 0.0023 -0.9913>}
 FACIAL_R_Nostril {parent=FACIAL_C_Nose, loc=<-0.825 2.2229 -0.4904>, rot=<0.0173 -0.1305 -0.0023 -0.9913>}
 FACIAL_C_NoseTip {parent=FACIAL_C_Nose, loc=<-0.0045 2.199 0.3155>, rot=<0.0872 0 0 -0.9962>}
 FACIAL_C_MouthUpper {parent=FACIAL_C_FacialRoot, loc=<0.0045 5.4607 7.3594>, rot=<0.0032 0 -0.0001 -1>}
 FACIAL_C_LipUpper {parent=FACIAL_C_MouthUpper, loc=<-0.003 0.0674 2.2501>, rot=<-0.0872 0 0 -0.9962>}
 FACIAL_L_LipUpper {parent=FACIAL_C_MouthUpper, loc=<1.0847 0.0459 2.0654>, rot=<-0.0778 -0.1601 -0.0379 -0.9833>}
 FACIAL_R_LipUpper {parent=FACIAL_C_MouthUpper, loc=<-1.087 0.0471 2.0618>, rot=<-0.0778 0.1601 0.0379 -0.9833>}
 FACIAL_L_LipUpperOuter {parent=FACIAL_C_MouthUpper, loc=<1.7925 0.1475 1.6063>, rot=<-0.0984 -0.2293 -0.0983 -0.9634>}
 FACIAL_R_LipUpperOuter {parent=FACIAL_C_MouthUpper, loc=<-1.8017 0.1449 1.6062>, rot=<-0.0984 0.2293 0.0983 -0.9634>}
 FACIAL_L_LipCorner {parent=FACIAL_C_MouthUpper, loc=<2.2864 0.5212 0.9415>, rot=<0 -0.2078 0 -0.9782>}
 FACIAL_R_LipCorner {parent=FACIAL_C_MouthUpper, loc=<-2.294 0.5129 0.9236>, rot=<0 0.2078 0 -0.9782>}
 FACIAL_L_Masseter {parent=FACIAL_C_FacialRoot, loc=<6.5998 2.7513 4.1696>, rot=<0.0725 0.1129 0.0075 -0.9909>}
 FACIAL_R_Masseter {parent=FACIAL_C_FacialRoot, loc=<-6.5846 2.7466 4.1705>, rot=<0.0725 -0.1129 -0.0076 -0.9909>}
 FACIAL_C_UnderChin {parent=FACIAL_C_FacialRoot, loc=<0.0009 9.1194 3.9515>, rot=<0.5026 0 -0.0001 -0.8645>}
 FACIAL_L_UnderChin {parent=FACIAL_C_FacialRoot, loc=<2.698 7.8621 2.796>, rot=<0.4052 -0.3546 0.0834 -0.8385>}
 FACIAL_R_UnderChin {parent=FACIAL_C_FacialRoot, loc=<-2.6995 7.864 2.7969>, rot=<0.4052 0.3546 -0.0835 -0.8385>}
 FACIAL_C_TeethUpper {parent=FACIAL_C_FacialRoot, loc=<0.0041 5.6188 8.2374>, rot=<0.0032 0 -0.0001 -1>}
 FACIAL_C_LowerLipRotation {parent=FACIAL_C_FacialRoot, loc=<-0.0022 2.0764 1.9391>, rot=<0.345 0 -0.0001 -0.9386>}
 FACIAL_C_MouthLower {parent=FACIAL_C_LowerLipRotation, loc=<0.0071 -0.9323 6.3217>, rot=<-0.342 0 0 -0.9397>}
 FACIAL_C_LipLower {parent=FACIAL_C_MouthLower, loc=<0.0033 1.3176 1.8172>, rot=<0.0872 0 0 -0.9962>}
 FACIAL_L_LipLower {parent=FACIAL_C_MouthLower, loc=<0.8703 1.2951 1.6588>, rot=<0.0927 -0.1294 0.0272 -0.9869>}
 FACIAL_R_LipLower {parent=FACIAL_C_MouthLower, loc=<-0.865 1.2896 1.6599>, rot=<0.0927 0.1294 -0.0272 -0.9869>}
 FACIAL_L_LipLowerOuter {parent=FACIAL_C_MouthLower, loc=<1.5946 0.9899 1.3553>, rot=<0.0913 -0.2321 0.0318 -0.9679>}
 FACIAL_R_LipLowerOuter {parent=FACIAL_C_MouthLower, loc=<-1.5956 0.9929 1.3654>, rot=<0.0913 0.2321 -0.0318 -0.9679>}
 FACIAL_C_TeethLower {parent=FACIAL_C_LowerLipRotation, loc=<0.0068 -1.3668 7.0098>, rot=<-0.342 0 0 -0.9397>}
 FACIAL_C_Tongue1 {parent=FACIAL_C_TeethLower, loc=<-0.0014 -0.1736 -5.0757>, rot=<-0.2588 0 0 -0.9659>}
 FACIAL_C_Tongue2 {parent=FACIAL_C_Tongue1, loc=<0.0005 0.5829 1.4893>, rot=<0.2588 0 0 -0.9659>}
 FACIAL_C_Tongue3 {parent=FACIAL_C_Tongue2, loc=<0.0008 0.2866 1.3437>, rot=<0.1305 0 0 -0.9914>}
 FACIAL_C_Tongue4 {parent=FACIAL_C_Tongue3, loc=<0.0011 -0.1056 1.0762>, rot=<0.0436 0 0 -0.999>}
 FACIAL_C_TongueUpper3 {parent=FACIAL_C_Tongue3, loc=<-0.0006 -0.3765 -0.1549>, rot=<0 0 0 1>}
 FACIAL_C_TongueLower3 {parent=FACIAL_C_Tongue3, loc=<-0.0002 0.8453 0.0088>, rot=<0 0 0 1>}
 FACIAL_L_TongueSide3 {parent=FACIAL_C_Tongue3, loc=<1.6033 0.0352 0.1318>, rot=<0 0 0 1>}
 FACIAL_R_TongueSide3 {parent=FACIAL_C_Tongue3, loc=<-1.6031 0.0338 0.1309>, rot=<0 0 0 1>}
 FACIAL_C_TongueUpper2 {parent=FACIAL_C_Tongue2, loc=<-0.0018 -0.5447 -0.1181>, rot=<0 0 0 1>}
 FACIAL_L_TongueSide2 {parent=FACIAL_C_Tongue2, loc=<1.9215 0.0838 0.2388>, rot=<0 0 0 1>}
 FACIAL_R_TongueSide2 {parent=FACIAL_C_Tongue2, loc=<-1.9188 0.0838 0.2381>, rot=<0 0 0 1>}
 FACIAL_C_Jaw {parent=FACIAL_C_FacialRoot, loc=<-0.0022 2.0764 1.9391>, rot=<0.345 0 -0.0001 -0.9386>}
 FACIAL_C_Jawline {parent=FACIAL_C_Jaw, loc=<0.0038 2.7925 8.1236>, rot=<0.2586 0 0 -0.966>}
 FACIAL_L_ChinSide {parent=FACIAL_C_Jaw, loc=<2.6701 0.8222 7.7712>, rot=<-0.2208 -0.3738 0.0499 -0.8995>}
 FACIAL_R_ChinSide {parent=FACIAL_C_Jaw, loc=<-2.6612 0.8216 7.7719>, rot=<-0.2208 0.3738 -0.0499 -0.8995>}
 FACIAL_C_Chin {parent=FACIAL_C_Jaw, loc=<0.0046 1.3181 8.2065>, rot=<-0.2591 0 0 -0.9659>}
 FACIAL_L_Ear {parent=FACIAL_C_FacialRoot, loc=<7.164 0.7697 0.5306>, rot=<-0.0231 -0.0008 -0.0262 -0.9994>}
 FACIAL_R_Ear {parent=FACIAL_C_FacialRoot, loc=<-7.1611 0.7636 0.5122>, rot=<-0.0231 0.0008 0.0261 -0.9994>}
 FACIAL_L_Jawline {parent=FACIAL_C_FacialRoot, loc=<4.4567 6.3359 4.4471>, rot=<0.2124 0.1576 -0.0552 -0.9628>}
 FACIAL_R_Jawline {parent=FACIAL_C_FacialRoot, loc=<-4.4538 6.3342 4.4458>, rot=<0.2124 -0.1576 0.0551 -0.9628>}
 FACIAL_C_NoseBridge {parent=FACIAL_C_FacialRoot, loc=<0.001 -0.5657 10.3457>, rot=<0.047 0 -0.0001 -0.9989>}
 FACIAL_L_EyelidUpperB {parent=FACIAL_C_FacialRoot, loc=<3.5358 0.1261 6.4299>, rot=<-0.2218 0 -0.0001 -0.9751>}
 FACIAL_L_EyelidLowerB {parent=FACIAL_C_FacialRoot, loc=<3.5358 0.1261 6.4299>, rot=<0.3203 0 -0.0001 -0.9473>}
 FACIAL_R_EyelidLowerB {parent=FACIAL_C_FacialRoot, loc=<-3.5342 0.1212 6.4291>, rot=<0.3286 0 -0.0001 -0.9445>}
 FACIAL_R_EyelidUpperB {parent=FACIAL_C_FacialRoot, loc=<-3.5342 0.1212 6.4291>, rot=<-0.2219 0 -0.0001 -0.9751>}
 clavicle_l {parent=spine_05, loc=<5.4831 -0.1053 -3.2946>, rot=<0.0465 -0.7289 -0.009 -0.6829>}
 upperarm_l {parent=clavicle_l, loc=<9.8658 -1.2136 -0.4105>, rot=<0.0164 -0.3245 -0.0269 -0.9454>}
 upperarm_twist_01_l {parent=upperarm_l, loc=<7.2135 -0.1134 0>, rot=<0 0 0 1>}
 lowerarm_l {parent=upperarm_l, loc=<26.465 -0.8345 0.259>, rot=<0.0036 0.0161 -0.2298 -0.9731>}
 clavicle_r {parent=spine_05, loc=<5.4831 -0.1053 3.2946>, rot=<0.7289 0.0465 -0.6829 0.009>}
 upperarm_r {parent=clavicle_r, loc=<-9.8658 1.2136 0.4105>, rot=<0.0164 -0.3245 -0.0269 -0.9454>}
 upperarm_twist_01_r {parent=upperarm_r, loc=<-7.2135 0.1134 0>, rot=<0 0 0 -1>}
 lowerarm_r {parent=upperarm_r, loc=<-26.465 0.8345 -0.259>, rot=<0.0036 0.0161 -0.2298 -0.9731>}
 base_Hair_l_01 {parent=spine_05, loc=<14.6205 3.1905 -9.1719>, rot=<-0.0336 0.4343 0.8951 0.095>}
 dyn_Hair_l_01 {parent=base_Hair_l_01, loc=<0 0 0>, rot=<0 0 0 1>}
 dyn_Hair_l_02 {parent=dyn_Hair_l_01, loc=<7.3206 0 0>, rot=<0 0.3897 0 -0.921>}
 dyn_Hair_l_03 {parent=dyn_Hair_l_02, loc=<7.0903 0 0>, rot=<0 0 0 1>}
 base_Hair_r_01 {parent=spine_05, loc=<13.9571 5.0284 8.8114>, rot=<-0.1195 -0.932 -0.335 0.0695>}
 dyn_Hair_r_01 {parent=base_Hair_r_01, loc=<0 0 0>, rot=<0 0 0 1>}
 dyn_Hair_r_02 {parent=dyn_Hair_r_01, loc=<7.0191 0 0>, rot=<0 0.3112 0 -0.9503>}
 dyn_Hair_r_03 {parent=dyn_Hair_r_02, loc=<4.7922 0 0>, rot=<0 0 0 1>}
 clavicle_r_null {parent=spine_05, type=Null, loc=<5.483 -0.1053 3.2946>, rot=<0.7289 0.0465 -0.6829 0.009>}
 Hair_Aim {parent=spine_05, loc=<0 0 0>, rot=<0 0 0 1>}
]
variables [
 Crocuh_aim {type=bool, default=False}
 Crocuh_aim_not {type=bool, default=False}
 CrouchWalk {type=bool, default=False}
 Melee {type=double, default=0}
 OnlyClav {type=double, default=0}
 Rifle {type=bool, default=False}
 RifleFloat {type=double, default=0}
 aimFloat {type=double, default=0}
]
