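# Physics asset for the F_MED_VioletInk face accessory mesh, setting up capsule bodies and constraints for back hair.
# Center, left and right hair chains branch into strands with few-degree swing and zero twist limits; the first center segments are fully locked.
# Spine, left clavicle and both upper arms add body capsules.
# Free constraints link hair roots to spine_05 and cross-link some center and right strands.
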PhysicsAsset /BRCosmetics/Characters/Player/Female/Medium/Bodies/F_MED_VioletInk/Meshes/Parts/F_MED_VioletInk_FaceAcc_Physics1
bodies {
 base_Hair_Back_l_01 [{shapeType=Capsule, length=0.505, radius=0.505, shapeName=base_Hair_Back_l_01_capsule, rotation=<rot 0 -90 0>}]
 base_Hair_Back_r_01 [{shapeType=Capsule, length=0.505, radius=0.505, shapeName=base_Hair_Back_r_01_capsule, rotation=<rot 0 -90 0>}]
 clavicle_l [{
   shapeType=Capsule
   length=18.2681
   radius=6.71
   shapeName=clavicle_l_capsule
   center=<6.1834 -0.2307 -2.632>
   rotation=<rot -80 179.9985 -174.9987>
  }]
 dyn_Hair_Back_01 [{shapeType=Capsule, length=0.5, radius=0.5, shapeName=dyn_Hair_Back_01_capsule}]
 dyn_Hair_Back_02 [{
   shapeType=Capsule
   length=7.6663
   radius=6.7962
   shapeName=dyn_Hair_Back_01_capsule
   center=<8.2939 0.8077 2.2221>
   rotation=<rot 0.0008 -175 90>
  }]
 dyn_Hair_Back_03 [{
   shapeType=Capsule
   length=7.1147
   radius=6.6534
   shapeName=dyn_Hair_Back_03_capsule
   center=<7.3085 0.5642 3.8049>
   rotation=<rot 14.4237 -176.7759 104.0528>
  }]
 dyn_Hair_Back_a_01 [{
   shapeType=Capsule
   length=3.5132
   radius=3.3404
   shapeName=dyn_Hair_Back_a_01_capsule
   center=<1.5359 0.5094 0.1397>
   rotation=<rot 0.8474 86.7605 -85.189>
  }]
 dyn_Hair_Back_a_02 [{
   shapeType=Capsule
   length=4.0132
   radius=3.3404
   shapeName=dyn_Hair_Back_a_01_capsule
   center=<2.4465 0.4997 0.2169>
   rotation=<rot -0.4253 101.7068 -85.1335>
  }]
 dyn_Hair_Back_a_03 [{
   shapeType=Capsule
   length=2.5828
   radius=4.0892
   shapeName=dyn_Hair_Back_a_01_capsule
   center=<2.2244 -1.9733 -0.8626>
   rotation=<rot 11.9997 12.0965 -138.8354>
  }]
 dyn_Hair_Back_a_l_01 [{
   shapeType=Capsule
   length=6.0628
   radius=2.7261
   shapeName=dyn_Hair_Back_a_l_01_capsule
   center=<4.4192 0.0571 -0.1421>
   rotation=<rot 0 -88.1476 -107.5723>
  }]
 dyn_Hair_Back_a_l_02 [{
   shapeType=Capsule
   length=3.8162
   radius=2.7599
   shapeName=dyn_Hair_Back_a_l_02_capsule
   center=<5.9358 -0.3614 0.3264>
   rotation=<rot 0.0002 86.547 -45.4533>
  }]
 dyn_Hair_Back_a_r_01 [{
   shapeType=Capsule
   length=3.8785
   radius=2.8623
   shapeName=dyn_Hair_Back_a_r_01_capsule
   center=<6.0924 1.5573 2.4447>
   rotation=<rot -2.7235 -84.3578 -83.1396>
  }]
 dyn_Hair_Back_b_01 [{
   shapeType=Capsule
   length=6.0532
   radius=3.3696
   shapeName=dyn_Hair_Back_b_01_capsule
   center=<6.9001 -1.2775 -0.3296>
   rotation=<rot -0.2783 -70.522 -89.2355>
  }]
 dyn_Hair_Back_b_02 [{
   shapeType=Capsule
   length=4.1321
   radius=2.8132
   shapeName=dyn_Hair_Back_b_02_capsule
   center=<6.9171 3.1032 1.1011>
   rotation=<rot -6.6136 153.0811 -67.4625>
  }]
 dyn_Hair_Back_b_l_01 [{
   shapeType=Capsule
   length=5.591
   radius=2.4631
   shapeName=dyn_Hair_Back_b_l_01_capsule
   center=<0.9637 -0.0111 -0.4298>
   rotation=<rot 0 -85.9919 -69.9426>
  }]
 dyn_Hair_Back_b_l_02 [{
   shapeType=Capsule
   length=3.5638
   radius=2.5319
   shapeName=dyn_Hair_Back_b_l_02_capsule
   center=<1.6011 -0.293 0.9742>
   rotation=<rot -16.4003 -90 31.3255>
  }]
 dyn_Hair_Back_b_r_01 [{
   shapeType=Capsule
   length=3.943
   radius=2.611
   shapeName=dyn_Hair_Back_c_r_01_capsule
   center=<5.0217 2.3444 2.4294>
   rotation=<rot -56.6328 40.8965 15.658>
  }]
 dyn_Hair_Back_c_l_01 [{
   shapeType=Capsule
   length=4.7969
   radius=2.9876
   shapeName=dyn_Hair_Back_c_l_01_capsule
   center=<1.2827 -0.126 0.6191>
   rotation=<rot 0.0112 -92.5197 -89.8695>
  }]
 dyn_Hair_Back_c_l_02 [{
   shapeType=Capsule
   length=3.6719
   radius=2.6126
   shapeName=dyn_Hair_Back_c_l_01_capsule
   center=<2.4219 0.9082 -0.0194>
   rotation=<rot 15.8732 -79.4445 -87.516>
  }]
 dyn_Hair_Back_c_l_03 [{
   shapeType=Capsule
   length=4.0469
   radius=2.2376
   shapeName=dyn_Hair_Back_c_l_01_capsule
   center=<2.0176 -3.8531 0.6667>
   rotation=<rot -0.736 177.5563 -104.9739>
  }]
 dyn_Hair_Back_c_r_01 [{
   shapeType=Capsule
   length=3.943
   radius=3.111
   shapeName=dyn_Hair_Back_c_r_01_capsule
   center=<2.8162 -2.115 2.8105>
   rotation=<rot 3.2127 34.1869 -65.4069>
  }]
 dyn_Hair_Back_l_02 [{
   shapeType=Capsule
   length=4.6663
   radius=6.7962
   shapeName=dyn_Hair_Back_01_capsule
   center=<10.6002 0.8077 3.4509>
   rotation=<rot 0.0007 180 90>
  }]
 dyn_Hair_Back_r_02 [{
   shapeType=Capsule
   length=6.2913
   radius=6.7962
   shapeName=dyn_Hair_Back_01_capsule
   center=<11.4876 0.8077 1.0492>
   rotation=<rot 15.0009 180 90>
  }]
 dyn_Hair_Back_r_04 [{
   shapeType=Capsule
   length=5.5
   radius=5.005
   shapeName=dyn_Hair_Back_r_04_capsule
   center=<-0.2969 -0.3559 -0.1957>
   rotation=<rot 55.4435 -6.01 14.7642>
  }]
 spine_01 [{shapeType=Capsule, length=33.7301, radius=14.7799, shapeName=spine_01_capsule, center=<3.5705 6.9235 0>, rotation=<rot 0 -90 0>}]
 spine_03 [{shapeType=Capsule, length=36.88, radius=11.63, shapeName=spine_01_capsule, center=<6.8925 3.9467 0>, rotation=<rot 0 -90 0>}]
 spine_05 [{shapeType=Capsule, length=26.005, radius=11.88, shapeName=spine_01_capsule, center=<7.0537 6.7117 0>, rotation=<rot 0 -90 0>}]
 upperarm_l [{shapeType=Capsule, length=28.875, radius=5.13, shapeName=upperarm_l_capsule, center=<14.5013 0 0.4586>, rotation=<rot -0.0002 -89.9999 90>}]
 upperarm_r [{
   shapeType=Capsule
   length=28.875
   radius=5.13
   shapeName=upperarm_r_capsule
   center=<-14.5013 0 -0.4586>
   rotation=<rot -0.0002 -89.9999 -90>
  }]
}
constraints [
 {bone1Name=spine_03, bone2Name=spine_01, swing1LimitDegrees=2, swing2LimitDegrees=3, twistLimitDegrees=45}
 {
  bone1Name=spine_05
  bone2Name=spine_03
  swing1LimitDegrees=45
  swing1Motion=Free
  swing2LimitDegrees=45
  swing2Motion=Free
  twistLimitDegrees=45
  twistMotion=Free
 }
 {
  bone1Name=dyn_Hair_Back_02
  bone2Name=dyn_Hair_Back_01
  swing1LimitDegrees=0
  swing1Motion=Locked
  swing2LimitDegrees=0
  swing2Motion=Locked
  twistLimitDegrees=0
  twistMotion=Locked
 }
 {bone1Name=dyn_Hair_Back_l_02, bone2Name=base_Hair_Back_l_01, swing1LimitDegrees=5, swing2LimitDegrees=10, twistLimitDegrees=0}
 {bone1Name=dyn_Hair_Back_r_02, bone2Name=base_Hair_Back_r_01, swing1LimitDegrees=5, swing2LimitDegrees=10, twistLimitDegrees=0}
 {
  bone1Name=dyn_Hair_Back_03
  bone2Name=dyn_Hair_Back_02
  swing1LimitDegrees=0
  swing1Motion=Locked
  swing2LimitDegrees=0
  swing2Motion=Locked
  twistLimitDegrees=0
  twistMotion=Locked
 }
 {bone1Name=dyn_Hair_Back_a_01, bone2Name=dyn_Hair_Back_03, swing1LimitDegrees=2, swing2LimitDegrees=4, twistLimitDegrees=0}
 {bone1Name=dyn_Hair_Back_a_02, bone2Name=dyn_Hair_Back_a_01, swing1LimitDegrees=1, swing2LimitDegrees=5, twistLimitDegrees=0}
 {bone1Name=dyn_Hair_Back_a_03, bone2Name=dyn_Hair_Back_a_02, swing1LimitDegrees=2, swing2LimitDegrees=4, twistLimitDegrees=0}
 {bone1Name=dyn_Hair_Back_b_01, bone2Name=dyn_Hair_Back_03, swing1LimitDegrees=2, swing2LimitDegrees=4, twistLimitDegrees=0}
 {bone1Name=dyn_Hair_Back_b_02, bone2Name=dyn_Hair_Back_b_01, swing1LimitDegrees=1, swing2LimitDegrees=2, twistLimitDegrees=0}
 {bone1Name=dyn_Hair_Back_a_l_01, bone2Name=base_Hair_Back_l_01, swing1LimitDegrees=2, swing2LimitDegrees=4, twistLimitDegrees=0}
 {bone1Name=dyn_Hair_Back_a_l_02, bone2Name=dyn_Hair_Back_a_l_01, swing1LimitDegrees=2, swing2LimitDegrees=5, twistLimitDegrees=0}
 {
  bone1Name=clavicle_l
  bone2Name=spine_05
  swing1LimitDegrees=45
  swing1Motion=Free
  swing2LimitDegrees=45
  swing2Motion=Free
  twistLimitDegrees=45
  twistMotion=Free
 }
 {bone1Name=dyn_Hair_Back_b_l_01, bone2Name=dyn_Hair_Back_l_02, swing1LimitDegrees=1, swing2LimitDegrees=5, twistLimitDegrees=0}
 {bone1Name=dyn_Hair_Back_b_l_02, bone2Name=dyn_Hair_Back_b_l_01, swing1LimitDegrees=1, swing2LimitDegrees=5, twistLimitDegrees=0}
 {bone1Name=dyn_Hair_Back_r_04, bone2Name=dyn_Hair_Back_r_02, swing1LimitDegrees=2, swing2LimitDegrees=5, twistLimitDegrees=0}
 {bone1Name=dyn_Hair_Back_b_r_01, bone2Name=dyn_Hair_Back_a_r_01, swing1LimitDegrees=2, swing2LimitDegrees=2, twistLimitDegrees=0}
 {bone1Name=dyn_Hair_Back_c_r_01, bone2Name=dyn_Hair_Back_a_r_01, swing1LimitDegrees=2, swing2LimitDegrees=2, twistLimitDegrees=0}
 {bone1Name=dyn_Hair_Back_a_r_01, bone2Name=dyn_Hair_Back_r_04, swing1LimitDegrees=5, swing2LimitDegrees=5, twistLimitDegrees=0}
 {bone1Name=dyn_Hair_Back_c_l_01, bone2Name=dyn_Hair_Back_l_02, swing1LimitDegrees=3, swing2LimitDegrees=5, twistLimitDegrees=0}
 {bone1Name=dyn_Hair_Back_c_l_02, bone2Name=dyn_Hair_Back_c_l_01, swing1LimitDegrees=2, swing2LimitDegrees=4, twistLimitDegrees=0}
 {bone1Name=dyn_Hair_Back_c_l_03, bone2Name=dyn_Hair_Back_c_l_02, swing1LimitDegrees=2, swing2LimitDegrees=4, twistLimitDegrees=0}
 {
  bone1Name=dyn_Hair_Back_01
  bone2Name=spine_05
  swing1LimitDegrees=45
  swing1Motion=Free
  swing2LimitDegrees=45
  swing2Motion=Free
  twistLimitDegrees=45
  twistMotion=Free
 }
 {
  bone1Name=base_Hair_Back_l_01
  bone2Name=spine_05
  swing1LimitDegrees=45
  swing1Motion=Free
  swing2LimitDegrees=45
  swing2Motion=Free
  twistLimitDegrees=45
  twistMotion=Free
 }
 {
  bone1Name=base_Hair_Back_r_01
  bone2Name=spine_05
  swing1LimitDegrees=45
  swing1Motion=Free
  swing2LimitDegrees=45
  swing2Motion=Free
  twistLimitDegrees=45
  twistMotion=Free
 }
 {
  bone1Name=upperarm_l
  bone2Name=clavicle_l
  swing1LimitDegrees=45
  swing1Motion=Free
  swing2LimitDegrees=45
  swing2Motion=Free
  twistLimitDegrees=45
  twistMotion=Free
 }
 {
  bone1Name=upperarm_r
  bone2Name=spine_05
  swing1LimitDegrees=45
  swing1Motion=Free
  swing2LimitDegrees=45
  swing2Motion=Free
  twistLimitDegrees=45
  twistMotion=Free
 }
 {
  bone1Name=dyn_Hair_Back_b_01
  bone2Name=dyn_Hair_Back_a_02
  swing1LimitDegrees=45
  swing1Motion=Free
  swing2LimitDegrees=45
  swing2Motion=Free
  twistLimitDegrees=45
  twistMotion=Free
 }
 {
  bone1Name=dyn_Hair_Back_b_02
  bone2Name=dyn_Hair_Back_a_02
  swing1LimitDegrees=45
  swing1Motion=Free
  swing2LimitDegrees=45
  swing2Motion=Free
  twistLimitDegrees=45
  twistMotion=Free
 }
 {
  bone1Name=dyn_Hair_Back_a_r_01
  bone2Name=dyn_Hair_Back_a_01
  swing1LimitDegrees=45
  swing1Motion=Free
  swing2LimitDegrees=45
  swing2Motion=Free
  twistLimitDegrees=45
  twistMotion=Free
 }
 {
  bone1Name=dyn_Hair_Back_b_r_01
  bone2Name=dyn_Hair_Back_a_02
  swing1LimitDegrees=45
  swing1Motion=Free
  swing2LimitDegrees=45
  swing2Motion=Free
  twistLimitDegrees=45
  twistMotion=Free
 }
]
