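# Animation Blueprint for the VioletInk female character's face accessory mesh, posing and simulating its hair.
# From a cached FrontHair pose, one branch uses weighted bone constraints and pack, cape and targeting offsets.
# The other chains AnimDynamics, corrective rotations, constraints and a spine-relative RigidBody.
# Each branch feeds the same Control Rig class, driven by movement-state variables, before a pack-flag blend and further bone adjustments.
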AnimBlueprint /BRCosmetics/Characters/Player/Female/Medium/Bodies/F_MED_VioletInk/Meshes/Parts/F_MED_VioletInk_FaceAcc_AnimBP
target_skeleton=&/F_MED_VioletInk_FaceAcc_Skeleton
graphs [
 AnimGraph {nodes [
   n1 = RigidBody!(
    ComponentPose=Local To Component()
    Alpha
    OverrideWorldGravity:Vector="0,0,-980"
    ExternalForce
    SimSpaceSettings:Sim Space Settings Structure=(WorldAlpha=1,VelocityScaleZ=0.4,DampingAlpha=1,MaxLinearVelocity=700,MaxAngularVelocity=8,MaxLinearAcceleration=3500,MaxAngularAcceleration=40,ExternalLinearDragV=(X=0.2,Y=0.2,Z=0.2),ExternalLinearVelocity=(X=0.2,Y=0.2,Z=0.2),ExternalAngularVelocity=(X=0,Y=0,Z=0))
   ) -> (Pose) {
    bOverrideWorldGravity=true
    overridePhysicsAsset=&/F_MED_VioletInk_FaceAcc_Physics1
    simulationSpace=BaseBoneSpace
    baseBoneRef {boneName=spine_05}
    overrideWorldGravity {z=-980}
    simSpaceSettings {
     maxAngularAcceleration=40
     maxAngularVelocity=8
     maxLinearAcceleration=3500
     maxLinearVelocity=700
     velocityScaleZ=0.4
     worldAlpha=1
     externalLinearDragV=<0.2 0.2 0.2>
     externalLinearVelocity=<0.2 0.2 0.2>
    }
   }
   Component To Local(ComponentPose=@n1)
   n2 = "Transform (Modify) Bone - Bone: base_Hair_l_01"(Alpha=1, Translation="5,0,0", Rotation="0,0,0", Scale="1,1,1") {
    rotationMode=BMM_Additive
    rotationSpace=BCS_BoneSpace
    translationMode=BMM_Additive
    translationSpace=BCS_BoneSpace
    alphaScaleBiasClamp {interpSpeedDecreasing=5, interpSpeedIncreasing=5}
    boneToModify {boneName=base_Hair_l_01}
    translation {x=5}
   }
   "Transform (Modify) Bone - Bone: base_Hair_r_01"(ComponentPose=@n2, Alpha=1, Translation="5,0,0", Rotation="0,0,0", Scale="1,1,1") {
    rotationMode=BMM_Additive
    rotationSpace=BCS_BoneSpace
    translationMode=BMM_Additive
    translationSpace=BCS_BoneSpace
    alphaScaleBiasClamp {interpSpeedDecreasing=5, interpSpeedIncreasing=5}
    boneToModify {boneName=base_Hair_r_01}
    translation {x=5}
   }
   n3 = "Transform (Modify) Bone - Bone: dyn_Hair_Back_01"(Alpha=1, Translation="0,0,2", Rotation="20,0,0", Scale="1,1,1") {
    rotationMode=BMM_Additive
    rotationSpace=BCS_BoneSpace
    translationMode=BMM_Additive
    translationSpace=BCS_BoneSpace
    boneToModify {boneName=dyn_Hair_Back_01}
    rotation {pitch=20}
    translation {z=2}
   }
   "Transform (Modify) Bone - Bone: dyn_Hair_Back_02"(ComponentPose=@n3, Alpha=1, Translation="0,0,0", Rotation="-40,0,0", Scale="1,1,1") {
    rotationMode=BMM_Additive
    rotationSpace=BCS_BoneSpace
    translationMode=BMM_Additive
    translationSpace=BCS_BoneSpace
    boneToModify {boneName=dyn_Hair_Back_02}
    rotation {pitch=-40}
   }
   n4 = Get isWearingPack() -> (isWearingPack:bool)
   n5 = "Constraint - dyn_Hair_Back_02 (PPPPP)"!(
    ComponentPose=Local To Component(LocalPose=Use cached pose 'FrontHair'())
    Alpha=1
    ConstraintWeights_0=0.744
    ConstraintWeights_1:float=0.099
    ConstraintWeights_2:float=0.0583
    ConstraintWeights_3:float=0.0559
    ConstraintWeights_4:float=0.0428
   ) -> (Pose) {
    alphaScaleBiasClamp {bInterpResult=true, interpSpeedDecreasing=4, interpSpeedIncreasing=4}
    boneToModify {boneName=dyn_Hair_Back_02}
    constraintSetup [
     {targetBone {boneName=spine_05}}
     {targetBone {boneName=neck_01}}
     {targetBone {boneName=neck_02}}
     {targetBone {boneName=spine_03}}
     {targetBone {boneName=clavicle_r}}
    ]
    constraintWeights [1,1,1,1,1]
   }
   n6 = "Constraint - dyn_Hair_Back_03 (PPPP)"!(
    ComponentPose=@n5
    Alpha=1
    ConstraintWeights_0=0.5324
    ConstraintWeights_1:float=0.3669
    ConstraintWeights_2:float=0.0556
    ConstraintWeights_3:float=0.0451
   ) -> (Pose) {
    alphaScaleBiasClamp {bInterpResult=true, interpSpeedDecreasing=4, interpSpeedIncreasing=4}
    boneToModify {boneName=dyn_Hair_Back_03}
    constraintSetup [{targetBone {boneName=spine_05}}, {targetBone {boneName=clavicle_r}}, {targetBone {boneName=clavicle_l}}, {targetBone {boneName=spine_04}}]
    constraintWeights [1,1,1,1]
   }
   n7 = "Constraint - dyn_Hair_Back_r_02 (PPPP)"!(
    ComponentPose=@n6
    Alpha=1
    ConstraintWeights_0=0.6328
    ConstraintWeights_1:float=0.2963
    ConstraintWeights_2:float=0.0382
    ConstraintWeights_3:float=0.0327
   ) -> (Pose) {
    alphaScaleBiasClamp {bInterpResult=true, interpSpeedDecreasing=4, interpSpeedIncreasing=4}
    boneToModify {boneName=dyn_Hair_Back_r_02}
    constraintSetup [{targetBone {boneName=spine_05}}, {targetBone {boneName=clavicle_r}}, {targetBone {boneName=neck_01}}, {targetBone {boneName=spine_03}}]
    constraintWeights [1,1,1,1]
   }
   n8 = "Constraint - dyn_Hair_Back_r_03 (PPPP)"!(
    ComponentPose=@n7
    Alpha=1
    ConstraintWeights_0=0.6983
    ConstraintWeights_1:float=0.2664
    ConstraintWeights_2:float=0.0223
    ConstraintWeights_3:float=0.013
   ) -> (Pose) {
    alphaScaleBiasClamp {bInterpResult=true, interpSpeedDecreasing=4, interpSpeedIncreasing=4}
    boneToModify {boneName=dyn_Hair_Back_r_03}
    constraintSetup [
     {targetBone {boneName=clavicle_r}}
     {targetBone {boneName=spine_05}}
     {targetBone {boneName=upperarm_twist_01_r}}
     {targetBone {boneName=spine_04}}
    ]
    constraintWeights [1,1,1,1]
   }
   n9 = "Constraint - dyn_Hair_Back_r_04 (PPPP)"!(
    ComponentPose=@n8
    Alpha=1
    ConstraintWeights_0=0.5177
    ConstraintWeights_1:float=0.2433
    ConstraintWeights_2:float=0.1643
    ConstraintWeights_3:float=0.0748
   ) -> (Pose) {
    alphaScaleBiasClamp {bInterpResult=true, interpSpeedDecreasing=4, interpSpeedIncreasing=4}
    boneToModify {boneName=dyn_Hair_Back_r_04}
    constraintSetup [
     {targetBone {boneName=clavicle_r}}
     {targetBone {boneName=spine_05}}
     {targetBone {boneName=spine_04}}
     {targetBone {boneName=upperarm_twist_01_r}}
    ]
    constraintWeights [1,1,1,1]
   }
   n10 = "Constraint - dyn_Hair_Back_l_02 (PPPPP)"!(
    ComponentPose=@n9
    Alpha=1
    ConstraintWeights_0=0.6912
    ConstraintWeights_1:float=0.1108
    ConstraintWeights_2:float=0.0819
    ConstraintWeights_3:float=0.0806
    ConstraintWeights_4:float=0.0355
   ) -> (Pose) {
    alphaScaleBiasClamp {bInterpResult=true, interpSpeedDecreasing=4, interpSpeedIncreasing=4}
    boneToModify {boneName=dyn_Hair_Back_l_02}
    constraintSetup [
     {targetBone {boneName=spine_05}}
     {targetBone {boneName=neck_01}}
     {targetBone {boneName=clavicle_l}}
     {targetBone {boneName=neck_02}}
     {targetBone {boneName=spine_03}}
    ]
    constraintWeights [1,1,1,1,1]
   }
   n11 = "Constraint - dyn_Hair_Back_l_03 (PP)"!(ComponentPose=@n10, Alpha=1, ConstraintWeights_0=0.8, ConstraintWeights_1:float=0.2) -> (Pose) {
    alphaScaleBiasClamp {bInterpResult=true, interpSpeedDecreasing=4, interpSpeedIncreasing=4}
    boneToModify {boneName=dyn_Hair_Back_l_03}
    constraintSetup [{targetBone {boneName=spine_05}}, {targetBone {boneName=spine_04}}]
    constraintWeights [1,1]
   }
   n12 = "Transform (Modify) Bone - Bone: dyn_Hair_Back_01"!(ComponentPose=@n11, bAlphaBoolEnabled:bool=@n4, Translation="0,0,5", Rotation="-20,0,0", Scale="1,1,0.5") -> (Pose) {
    alphaInputType=Bool
    rotationMode=BMM_Additive
    rotationSpace=BCS_BoneSpace
    scaleMode=BMM_Additive
    translationMode=BMM_Additive
    translationSpace=BCS_BoneSpace
    boneToModify {boneName=dyn_Hair_Back_01}
    rotation {pitch=-20}
    scale {z=0.5}
    translation {z=5}
   }
   n13 = "Transform (Modify) Bone - Bone: dyn_Hair_Back_03"!(ComponentPose=@n12, bAlphaBoolEnabled:bool=@n4, Translation="0,0,-15", Rotation="30,0,0", Scale="1,1,0.5") -> (Pose) {
    alphaInputType=Bool
    rotationMode=BMM_Additive
    rotationSpace=BCS_BoneSpace
    scaleMode=BMM_Additive
    translationMode=BMM_Additive
    translationSpace=BCS_BoneSpace
    boneToModify {boneName=dyn_Hair_Back_03}
    rotation {pitch=30}
    scale {z=0.5}
    translation {z=-15}
   }
   n14 = "Transform (Modify) Bone - Bone: base_Hair_Back_b_01"!(ComponentPose=@n13, bAlphaBoolEnabled:bool=@n4, Translation="0,0,-6", Rotation="-30,0,0", Scale="1,1,0.5") -> (Pose) {
    alphaInputType=Bool
    rotationMode=BMM_Additive
    rotationSpace=BCS_BoneSpace
    scaleMode=BMM_Additive
    translationMode=BMM_Additive
    translationSpace=BCS_BoneSpace
    boneToModify {boneName=base_Hair_Back_b_01}
    rotation {pitch=-30}
    scale {z=0.5}
    translation {z=-6}
   }
   n15 = "Transform (Modify) Bone - Bone: dyn_Hair_Back_r_01"!(ComponentPose=@n14, bAlphaBoolEnabled:bool=@n4, Translation="0,0,0", Rotation="-10,0.5,0", Scale="1,1,0.5") -> (Pose) {
    alphaInputType=Bool
    rotationMode=BMM_Additive
    rotationSpace=BCS_BoneSpace
    scaleMode=BMM_Additive
    translationMode=BMM_Additive
    translationSpace=BCS_BoneSpace
    boneToModify {boneName=dyn_Hair_Back_r_01}
    rotation {pitch=-10, yaw=0.5}
    scale {z=0.5}
   }
   n16 = "Transform (Modify) Bone - Bone: dyn_Hair_Back_l_02"!(ComponentPose=@n15, bAlphaBoolEnabled:bool=@n4, Translation="0,0,3", Rotation="-20,0,0", Scale="1,1,0.5") -> (Pose) {
    alphaInputType=Bool
    rotationMode=BMM_Additive
    rotationSpace=BCS_BoneSpace
    scaleMode=BMM_Additive
    translationMode=BMM_Additive
    translationSpace=BCS_BoneSpace
    boneToModify {boneName=dyn_Hair_Back_l_02}
    rotation {pitch=-20}
    scale {z=0.5}
    translation {z=3}
   }
   n17 = "Transform (Modify) Bone - Bone: dyn_Hair_Back_a_02"!(ComponentPose=@n16, bAlphaBoolEnabled:bool=@n4, Translation="0,0,0", Rotation="0,0,0", Scale="1,1,1") -> (Pose) {
    alphaInputType=Bool
    rotationMode=BMM_Additive
    rotationSpace=BCS_BoneSpace
    scaleMode=BMM_Additive
    translationMode=BMM_Additive
    translationSpace=BCS_BoneSpace
    boneToModify {boneName=dyn_Hair_Back_a_02}
   }
   n18 = "Transform (Modify) Bone - Bone: dyn_Hair_Back_a_03"!(ComponentPose=@n17, bAlphaBoolEnabled:bool=@n4, Translation="0,0,0", Rotation="0,0,0", Scale="1,1,1") -> (Pose) {
    alphaInputType=Bool
    rotationMode=BMM_Additive
    rotationSpace=BCS_BoneSpace
    scaleMode=BMM_Additive
    translationMode=BMM_Additive
    translationSpace=BCS_BoneSpace
    boneToModify {boneName=dyn_Hair_Back_a_03}
   }
   n19 = Get isWearingCape() -> (isWearingCape:bool)
   n20 = "Transform (Modify) Bone - Bone: dyn_Hair_Back_c_l_01"!(ComponentPose=@n18, bAlphaBoolEnabled:bool=@n19, Translation="0,0,-2", Rotation="-25,0,0", Scale="1,1,0.5") -> (Pose) {
    alphaInputType=Bool
    rotationMode=BMM_Additive
    rotationSpace=BCS_BoneSpace
    scaleMode=BMM_Additive
    translationMode=BMM_Additive
    translationSpace=BCS_BoneSpace
    boneToModify {boneName=dyn_Hair_Back_c_l_01}
    rotation {pitch=-25}
    scale {z=0.5}
    translation {z=-2}
   }
   n21 = "Transform (Modify) Bone - Bone: dyn_Hair_Back_03"(ComponentPose=@n20, bAlphaBoolEnabled=@n19, Translation="0,0,-2", Rotation="-15,0,0", Scale="1,1,0.5") {
    alphaInputType=Bool
    rotationMode=BMM_Additive
    rotationSpace=BCS_BoneSpace
    scaleMode=BMM_Additive
    translationMode=BMM_Additive
    translationSpace=BCS_BoneSpace
    boneToModify {boneName=dyn_Hair_Back_03}
    rotation {pitch=-15}
    scale {z=0.5}
    translation {z=-2}
   }
   n22 = "Transform (Modify) Bone - Bone: base_Hair_Back_b_01"(ComponentPose=@n21, bAlphaBoolEnabled=@n19, Translation="0,0,0", Rotation="0,0,0", Scale="1,1,0.5") {
    alphaInputType=Bool
    rotationMode=BMM_Additive
    rotationSpace=BCS_BoneSpace
    scaleMode=BMM_Additive
    translationMode=BMM_Additive
    translationSpace=BCS_BoneSpace
    boneToModify {boneName=base_Hair_Back_b_01}
    scale {z=0.5}
   }
   n23 = "Transform (Modify) Bone - Bone: dyn_Hair_Back_r_03"!(ComponentPose=@n22, bAlphaBoolEnabled:bool=@n19, Translation="0,0,-2", Rotation="-10,0,0", Scale="1,1,0.5") -> (Pose) {
    alphaInputType=Bool
    rotationMode=BMM_Additive
    rotationSpace=BCS_BoneSpace
    scaleMode=BMM_Additive
    translationMode=BMM_Additive
    translationSpace=BCS_BoneSpace
    boneToModify {boneName=dyn_Hair_Back_r_03}
    rotation {pitch=-10}
    scale {z=0.5}
    translation {z=-2}
   }
   n24 = "Transform (Modify) Bone - Bone: dyn_Hair_Back_l_02"(ComponentPose=@n23, bAlphaBoolEnabled=@n19, Translation="0,0,0", Rotation="0,0,0", Scale="1,1,0.5") {
    alphaInputType=Bool
    rotationMode=BMM_Additive
    rotationSpace=BCS_BoneSpace
    scaleMode=BMM_Additive
    translationMode=BMM_Additive
    translationSpace=BCS_BoneSpace
    boneToModify {boneName=dyn_Hair_Back_l_02}
    scale {z=0.5}
   }
   n25 = "Transform (Modify) Bone - Bone: dyn_Hair_Back_a_02"(ComponentPose=@n24, bAlphaBoolEnabled=@n19, Translation="0,0,0", Rotation="50,0,0", Scale="1,1,0.5") {
    alphaInputType=Bool
    rotationMode=BMM_Additive
    rotationSpace=BCS_BoneSpace
    scaleMode=BMM_Additive
    translationMode=BMM_Additive
    translationSpace=BCS_BoneSpace
    boneToModify {boneName=dyn_Hair_Back_a_02}
    rotation {pitch=50}
    scale {z=0.5}
   }
   n26 = "Transform (Modify) Bone - Bone: dyn_Hair_Back_a_03"(ComponentPose=@n25, bAlphaBoolEnabled=@n19, Translation="0,0,0", Rotation="0,0,0", Scale="1,1,0.5") {
    alphaInputType=Bool
    rotationMode=BMM_Additive
    rotationSpace=BCS_BoneSpace
    scaleMode=BMM_Additive
    translationMode=BMM_Additive
    translationSpace=BCS_BoneSpace
    boneToModify {boneName=dyn_Hair_Back_a_03}
    scale {z=0.5}
   }
   n27 = "Transform (Modify) Bone - Bone: alt_hair_01"!(ComponentPose=@n26, bAlphaBoolEnabled:bool=True, Translation="0,0,0", Rotation="0,0,0", Scale="1,1,1") -> (Pose) {
    alphaInputType=Bool
    rotationMode=BMM_Additive
    rotationSpace=BCS_BoneSpace
    scaleMode=BMM_Additive
    translationMode=BMM_Additive
    translationSpace=BCS_BoneSpace
    boneToModify {boneName=alt_hair_01}
   }
   n28 = "Transform (Modify) Bone - Bone: alt_hair_02"!(ComponentPose=@n27, bAlphaBoolEnabled:bool=True, Translation="0,8,0", Rotation="0,0,0", Scale="1,1,1") -> (Pose) {
    alphaInputType=Bool
    rotationMode=BMM_Additive
    rotationSpace=BCS_BoneSpace
    scaleMode=BMM_Additive
    translationMode=BMM_Additive
    translationSpace=BCS_BoneSpace
    boneToModify {boneName=alt_hair_02}
    translation {y=8}
   }
   n29 = "Transform (Modify) Bone - Bone: dyn_Hair_Back_a_02"(ComponentPose=@n28, bAlphaBoolEnabled=Get bIsTargeting(), Translation="0,0,6", Rotation="20,0,0", Scale="1,1,1") {
    alphaInputType=Bool
    rotationMode=BMM_Additive
    rotationSpace=BCS_BoneSpace
    scaleMode=BMM_Additive
    translationMode=BMM_Additive
    translationSpace=BCS_BoneSpace
    alphaBoolBlend {blendInTime=0.5, blendOption=Cubic, blendOutTime=0.5}
    boneToModify {boneName=dyn_Hair_Back_a_02}
    rotation {pitch=20}
    translation {z=6}
   }
   n30 = Control Rig!(
    Source=Component To Local(ComponentPose=@n29)
    Alpha
    Crocuh_aim:bool=false
    Crocuh_aim_not:bool=false
    CrouchWalk:bool=false
    HeadUP_NotCrouch:bool=Get HeadUP_NotCrouch() -> (HeadUP_NotCrouch:bool)
    HeadUP_Hair_ADJ:double=0
    Emote:double=0
    SkyDive_Float:double=0
    SkyDive_Dive_Hair:double=0
    Melee:double=0
    SkyDive_All:double=0
    Sprint:double=0
    HeadL_Only:double=0
    HeadR_Only:double=0
    RifleAim_Hair:double=0
    Rifle:double=0
    TargetingFloat:double=0
    HeadUp_FHair:double=0
    Skydive_Float_Para:double=0
    Skydive_Hair_Right:double=0
    Skydive_Hair_Left:double=0
    CrouchAim_Float:double=0
    EmoteHeadUp:double=0
    FrontEnd:bool=Get bIsInFrontEnd()
    BackBiling:bool=Get isWearingPack()
   ) -> (Pose) {controlRigAssetReference {blueprintRigClass=&/F_MED_VioletInk_FaceAcc_FWD_CtrlRig.F_MED_VioletInk_FaceAcc_FWD_CtrlRig_C}}
   n31 = Anim Dynamics - Bone: dyn_Hair_Back_l_01!(ComponentPose=Local To Component(LocalPose=Use cached pose 'FrontHair'()), Alpha=1, GravityOverride:Vector="0,0,-980", ExternalForce="0,0,0") -> (Pose) {
    angularDampingOverride=0.9999
    angularSpringConstant=800
    bOverrideAngularDamping=true
    bUseGravityOverride=true
    gravityScale=0.35
    simulationSpace=RootRelative
    boundBone {boneName=dyn_Hair_Back_l_01}
    componentAppliedLinearAccClamp=<5000 5000 5000>
    componentLinearAccScale=<1 1 1>
    componentLinearVelScale=<1 1 1>
    gravityOverride {z=-980}
    physicsBodyDefinitions [{
      boundBone {boneName=dyn_Hair_Back_l_01}
      boxExtents=<40 40 40>
      constraintSetup {angularLimitsMax=<0 -1 5>, angularLimitsMin=<0 -10 -5>, angularTarget=<1 0 0>}
      localJointOffset=<4 0 0>
     }]
   }
   n32 = Anim Dynamics - Bone: dyn_Hair_Back_01!(ComponentPose=@n31, Alpha=1, GravityOverride:Vector="0,0,-980", ExternalForce="0,0,0") -> (Pose) {
    angularDampingOverride=0.9999
    angularSpringConstant=800
    bOverrideAngularDamping=true
    bUseGravityOverride=true
    gravityScale=0.35
    simulationSpace=RootRelative
    boundBone {boneName=dyn_Hair_Back_01}
    componentAppliedLinearAccClamp=<5000 5000 5000>
    componentLinearAccScale=<1 1 1>
    componentLinearVelScale=<1 1 1>
    gravityOverride {z=-980}
    physicsBodyDefinitions [{
      boundBone {boneName=dyn_Hair_Back_01}
      boxExtents=<40 40 40>
      constraintSetup {angularLimitsMax=<0 -1 5>, angularLimitsMin=<0 -10 -5>, angularTarget=<1 0 0>}
      localJointOffset=<4 0 0>
     }]
   }
   n33 = Anim Dynamics - Bone: dyn_Hair_Back_r_01!(ComponentPose=@n32, Alpha=1, GravityOverride:Vector="0,0,-980", ExternalForce="0,0,0") -> (Pose) {
    angularDampingOverride=0.9999
    angularSpringConstant=800
    bOverrideAngularDamping=true
    bUseGravityOverride=true
    gravityScale=0.35
    simulationSpace=RootRelative
    boundBone {boneName=dyn_Hair_Back_r_01}
    componentAppliedLinearAccClamp=<5000 5000 5000>
    componentLinearAccScale=<1 1 1>
    componentLinearVelScale=<1 1 1>
    gravityOverride {z=-980}
    physicsBodyDefinitions [{
      boundBone {boneName=dyn_Hair_Back_r_01}
      boxExtents=<40 40 40>
      constraintSetup {angularLimitsMax=<0 -2 5>, angularLimitsMin=<0 -10 -5>, angularTarget=<1 0 0>}
      localJointOffset=<4 0 0>
     }]
   }
   n34 = Anim Dynamics - Bone: dyn_Hair_Back_l_02!(ComponentPose=@n33, Alpha=1, GravityOverride:Vector="0,0,-980", ExternalForce="0,0,0") -> (Pose) {
    angularDampingOverride=0.9999
    angularSpringConstant=800
    bOverrideAngularDamping=true
    bUseGravityOverride=true
    gravityScale=0.35
    simulationSpace=RootRelative
    boundBone {boneName=dyn_Hair_Back_l_02}
    componentAppliedLinearAccClamp=<5000 5000 5000>
    componentLinearAccScale=<1 1 1>
    componentLinearVelScale=<1 1 1>
    gravityOverride {z=-980}
    physicsBodyDefinitions [{
      boundBone {boneName=dyn_Hair_Back_l_02}
      boxExtents=<40 40 40>
      constraintSetup {angularLimitsMax=<0 0 5>, angularLimitsMin=<0 -20 -5>, angularTarget=<1 0 0>}
      localJointOffset=<4 0 0>
     }]
   }
   n35 = Anim Dynamics - Bone: dyn_Hair_Back_02!(ComponentPose=@n34, Alpha=1, GravityOverride:Vector="0,0,-980", ExternalForce="0,0,0") -> (Pose) {
    angularDampingOverride=0.9999
    angularSpringConstant=800
    bOverrideAngularDamping=true
    bUseGravityOverride=true
    gravityScale=0.35
    simulationSpace=RootRelative
    boundBone {boneName=dyn_Hair_Back_02}
    componentAppliedLinearAccClamp=<5000 5000 5000>
    componentLinearAccScale=<1 1 1>
    componentLinearVelScale=<1 1 1>
    gravityOverride {z=-980}
    physicsBodyDefinitions [{
      boundBone {boneName=dyn_Hair_Back_02}
      boxExtents=<40 40 40>
      constraintSetup {angularLimitsMax=<0 0 5>, angularLimitsMin=<0 -20 -5>, angularTarget=<1 0 0>}
      localJointOffset=<4 0 0>
     }]
   }
   n36 = Anim Dynamics - Bone: dyn_Hair_Back_r_02!(ComponentPose=@n35, Alpha=1, GravityOverride:Vector="0,0,-980", ExternalForce="0,0,0") -> (Pose) {
    angularDampingOverride=0.9999
    angularSpringConstant=800
    bOverrideAngularDamping=true
    bUseGravityOverride=true
    gravityScale=0.35
    simulationSpace=RootRelative
    boundBone {boneName=dyn_Hair_Back_r_02}
    componentAppliedLinearAccClamp=<5000 5000 5000>
    componentLinearAccScale=<1 1 1>
    componentLinearVelScale=<1 1 1>
    gravityOverride {z=-980}
    physicsBodyDefinitions [{
      boundBone {boneName=dyn_Hair_Back_r_02}
      boxExtents=<40 40 40>
      constraintSetup {angularLimitsMax=<0 0 5>, angularLimitsMin=<0 -20 -5>, angularTarget=<1 0 0>}
      localJointOffset=<4 0 0>
     }]
   }
   n37 = Anim Dynamics - Bone: dyn_Hair_Back_l_03!(ComponentPose=@n36, Alpha=1, GravityOverride:Vector="0,0,-980", ExternalForce="0,0,0") -> (Pose) {
    angularDampingOverride=0.9999
    angularSpringConstant=800
    bAngularSpring=true
    bOverrideAngularDamping=true
    bUseGravityOverride=true
    gravityScale=0.35
    simulationSpace=RootRelative
    boundBone {boneName=dyn_Hair_Back_l_03}
    componentAppliedLinearAccClamp=<5000 5000 5000>
    componentLinearAccScale=<1 1 1>
    componentLinearVelScale=<1 1 1>
    gravityOverride {z=-980}
    physicsBodyDefinitions [{
      boundBone {boneName=dyn_Hair_Back_l_03}
      boxExtents=<50 50 50>
      constraintSetup {angularLimitsMax=<0 10 5>, angularLimitsMin=<0 -60 -5>, angularTarget=<1 0 0>}
      localJointOffset=<6 0 0>
     }]
   }
   n38 = Anim Dynamics - Bone: dyn_Hair_Back_03!(ComponentPose=@n37, Alpha=1, GravityOverride:Vector="0,0,-980", ExternalForce="0,0,0") -> (Pose) {
    angularDampingOverride=0.9999
    angularSpringConstant=800
    bAngularSpring=true
    bOverrideAngularDamping=true
    bUseGravityOverride=true
    gravityScale=0.35
    simulationSpace=RootRelative
    boundBone {boneName=dyn_Hair_Back_03}
    componentAppliedLinearAccClamp=<5000 5000 5000>
    componentLinearAccScale=<1 1 1>
    componentLinearVelScale=<1 1 1>
    gravityOverride {z=-980}
    physicsBodyDefinitions [{
      boundBone {boneName=dyn_Hair_Back_03}
      boxExtents=<50 50 50>
      constraintSetup {angularLimitsMax=<0 10 5>, angularLimitsMin=<0 -60 -5>, angularTarget=<1 0 0>}
      localJointOffset=<6 0 0>
     }]
   }
   n39 = Anim Dynamics - Bone: dyn_Hair_Back_r_03!(ComponentPose=@n38, Alpha=1, GravityOverride:Vector="0,0,-980", ExternalForce="0,0,0") -> (Pose) {
    angularDampingOverride=0.9999
    angularSpringConstant=800
    bAngularSpring=true
    bOverrideAngularDamping=true
    bUseGravityOverride=true
    gravityScale=0.35
    simulationSpace=RootRelative
    boundBone {boneName=dyn_Hair_Back_r_03}
    componentAppliedLinearAccClamp=<5000 5000 5000>
    componentLinearAccScale=<1 1 1>
    componentLinearVelScale=<1 1 1>
    gravityOverride {z=-980}
    physicsBodyDefinitions [{
      boundBone {boneName=dyn_Hair_Back_r_03}
      boxExtents=<50 50 50>
      constraintSetup {angularLimitsMax=<0 10 5>, angularLimitsMin=<0 -60 -5>, angularTarget=<1 0 0>}
      localJointOffset=<6 0 0>
     }]
   }
   n40 = "Transform (Modify) Bone - Bone: dyn_Hair_Back_01"!(ComponentPose=@n39, Alpha:float=1, Translation="0,0,0", Rotation="-3.600,0.000,-0.000", Scale="1,1,1") -> (Pose) {
    rotationMode=BMM_Additive
    rotationSpace=BCS_BoneSpace
    translationMode=BMM_Additive
    translationSpace=BCS_BoneSpace
    alphaScaleBiasClamp {bClampResult=true, bInterpResult=true, clampMax=0.8, interpSpeedIncreasing=4}
    boneToModify {boneName=dyn_Hair_Back_01}
    rotation {pitch=-3.6}
   }
   n41 = "Transform (Modify) Bone - Bone: dyn_Hair_Back_l_01"(ComponentPose=@n40, Alpha=1, Translation="0,0,0", Rotation="-1.490,0.000,0.000", Scale="1,1,1") {
    rotationMode=BMM_Additive
    rotationSpace=BCS_BoneSpace
    translationMode=BMM_Additive
    translationSpace=BCS_BoneSpace
    alphaScaleBiasClamp {bClampResult=true, bInterpResult=true, clampMax=0.8, interpSpeedIncreasing=4}
    boneToModify {boneName=dyn_Hair_Back_l_01}
    rotation {pitch=-1.49}
   }
   n42 = "Transform (Modify) Bone - Bone: dyn_Hair_Back_r_01"!(ComponentPose=@n41, Alpha:float=1, Translation="0,0,0", Rotation="0.620,0.000,-0.000", Scale="1,1,1") -> (Pose) {
    rotationMode=BMM_Additive
    rotationSpace=BCS_BoneSpace
    translationMode=BMM_Additive
    translationSpace=BCS_BoneSpace
    alphaScaleBiasClamp {bClampResult=true, bInterpResult=true, clampMax=0.8, interpSpeedIncreasing=4}
    boneToModify {boneName=dyn_Hair_Back_r_01}
    rotation {pitch=0.62}
   }
   n43 = "Transform (Modify) Bone - Bone: dyn_Hair_Back_02"(ComponentPose=@n42, Alpha=1, Translation="0,0,0", Rotation="-3.247,-0.045,0.062", Scale="1,1,1") {
    rotationMode=BMM_Additive
    rotationSpace=BCS_BoneSpace
    translationMode=BMM_Additive
    translationSpace=BCS_BoneSpace
    alphaScaleBiasClamp {bClampResult=true, bInterpResult=true, clampMax=0.8, interpSpeedIncreasing=4}
    boneToModify {boneName=dyn_Hair_Back_02}
    rotation=<rot -3.247 -0.045 0.062>
   }
   n44 = "Transform (Modify) Bone - Bone: dyn_Hair_Back_l_02"!(ComponentPose=@n43, Alpha:float=1, Translation="0,0,0", Rotation="1.840,6.344,0.973", Scale="1,1,1") -> (Pose) {
    rotationMode=BMM_Additive
    rotationSpace=BCS_BoneSpace
    translationMode=BMM_Additive
    translationSpace=BCS_BoneSpace
    alphaScaleBiasClamp {bClampResult=true, bInterpResult=true, clampMax=0.8, interpSpeedIncreasing=4}
    boneToModify {boneName=dyn_Hair_Back_l_02}
    rotation=<rot 1.84 6.344 0.973>
   }
   n45 = "Transform (Modify) Bone - Bone: dyn_Hair_Back_r_02"(ComponentPose=@n44, Alpha=1, Translation="0,0,0", Rotation="-9.696,-7.613,-0.190", Scale="1,1,1") {
    rotationMode=BMM_Additive
    rotationSpace=BCS_BoneSpace
    translationMode=BMM_Additive
    translationSpace=BCS_BoneSpace
    alphaScaleBiasClamp {bClampResult=true, bInterpResult=true, clampMax=0.8, interpSpeedIncreasing=4}
    boneToModify {boneName=dyn_Hair_Back_r_02}
    rotation=<rot -9.696 -7.613 -0.19>
   }
   n46 = "Transform (Modify) Bone - Bone: dyn_Hair_Back_03"!(ComponentPose=@n45, Alpha:float=1, Translation="0,0,0", Rotation="39.630,-4.634,-2.567", Scale="1,1,1") -> (Pose) {
    rotationMode=BMM_Additive
    rotationSpace=BCS_BoneSpace
    translationMode=BMM_Additive
    translationSpace=BCS_BoneSpace
    alphaScaleBiasClamp {bClampResult=true, bInterpResult=true, clampMax=0.8, interpSpeedIncreasing=4}
    boneToModify {boneName=dyn_Hair_Back_03}
    rotation=<rot 39.63 -4.634 -2.567>
   }
   n47 = "Transform (Modify) Bone - Bone: dyn_Hair_Back_l_03"(ComponentPose=@n46, Alpha=1, Translation="0,0,0", Rotation="23.472,0.014,0.017", Scale="1,1,1") {
    rotationMode=BMM_Additive
    rotationSpace=BCS_BoneSpace
    translationMode=BMM_Additive
    translationSpace=BCS_BoneSpace
    alphaScaleBiasClamp {bClampResult=true, bInterpResult=true, clampMax=0.8, interpSpeedIncreasing=4}
    boneToModify {boneName=dyn_Hair_Back_l_03}
    rotation=<rot 23.472 0.014 0.017>
   }
   n48 = "Transform (Modify) Bone - Bone: dyn_Hair_Back_r_03"!(ComponentPose=@n47, Alpha:float=1, Translation="0,0,0", Rotation="48.399,-7.705,-3.776", Scale="1,1,1") -> (Pose) {
    rotationMode=BMM_Additive
    rotationSpace=BCS_BoneSpace
    translationMode=BMM_Additive
    translationSpace=BCS_BoneSpace
    alphaScaleBiasClamp {bClampResult=true, bInterpResult=true, clampMax=0.8, interpSpeedIncreasing=4}
    boneToModify {boneName=dyn_Hair_Back_r_03}
    rotation=<rot 48.399 -7.705 -3.776>
   }
   n49 = "Transform (Modify) Bone - Bone: base_Hair_Back_a_01"(ComponentPose=@n48, Alpha=1, Translation="0,0,0", Rotation="63.551,-1.503,-0.909", Scale="1,1,1") {
    rotationMode=BMM_Additive
    rotationSpace=BCS_BoneSpace
    translationMode=BMM_Additive
    translationSpace=BCS_BoneSpace
    alphaScaleBiasClamp {bClampResult=true, bInterpResult=true, clampMax=0.8, interpSpeedIncreasing=4}
    boneToModify {boneName=base_Hair_Back_a_01}
    rotation=<rot 63.551 -1.503 -0.909>
   }
   n50 = "Transform (Modify) Bone - Bone: dyn_Hair_Back_b_01"(ComponentPose=@n49, Alpha=1, Translation="0,0,0", Rotation="27.157,-24.834,-15.988", Scale="1,1,1") {
    rotationMode=BMM_Additive
    rotationSpace=BCS_BoneSpace
    translationMode=BMM_Additive
    translationSpace=BCS_BoneSpace
    alphaScaleBiasClamp {bClampResult=true, bInterpResult=true, clampMax=0.8, interpSpeedIncreasing=4}
    boneToModify {boneName=dyn_Hair_Back_b_01}
    rotation=<rot 27.157 -24.834 -15.988>
   }
   n51 = "Transform (Modify) Bone - Bone: base_Hair_Back_b_l_01"(ComponentPose=@n50, Alpha=1, Translation="0,0,0", Rotation="32.228,-16.772,-23.104", Scale="1,1,1") {
    rotationMode=BMM_Additive
    rotationSpace=BCS_BoneSpace
    translationMode=BMM_Additive
    translationSpace=BCS_BoneSpace
    alphaScaleBiasClamp {bClampResult=true, bInterpResult=true, clampMax=0.8, interpSpeedIncreasing=4}
    boneToModify {boneName=base_Hair_Back_b_l_01}
    rotation=<rot 32.228 -16.772 -23.104>
   }
   n52 = "Transform (Modify) Bone - Bone: dyn_Hair_Back_r_04"(ComponentPose=@n51, Alpha=1, Translation="0,0,0", Rotation="44.187,-0.067,-0.038", Scale="1,1,1") {
    rotationMode=BMM_Additive
    rotationSpace=BCS_BoneSpace
    translationMode=BMM_Additive
    translationSpace=BCS_BoneSpace
    alphaScaleBiasClamp {bClampResult=true, bInterpResult=true, clampMax=0.8, interpSpeedIncreasing=4}
    boneToModify {boneName=dyn_Hair_Back_r_04}
    rotation=<rot 44.187 -0.067 -0.038>
   }
   n53 = "Transform (Modify) Bone - Bone: dyn_Hair_Back_c_l_01"!(ComponentPose=@n52, Alpha:float=1, Translation="0,0,0", Rotation="38.175,-0.812,-0.340", Scale="1,1,1") -> (Pose) {
    rotationMode=BMM_Additive
    rotationSpace=BCS_BoneSpace
    translationMode=BMM_Additive
    translationSpace=BCS_BoneSpace
    alphaScaleBiasClamp {bClampResult=true, bInterpResult=true, clampMax=0.8, interpSpeedIncreasing=4}
    boneToModify {boneName=dyn_Hair_Back_c_l_01}
    rotation=<rot 38.175 -0.812 -0.34>
   }
   n54 = "Transform (Modify) Bone - Bone: dyn_Hair_l_01"(ComponentPose=@n53, Alpha=1, Translation="0,0,2", Rotation="2,0,0", Scale="1,1,1") {
    rotationMode=BMM_Additive
    rotationSpace=BCS_BoneSpace
    translationMode=BMM_Additive
    translationSpace=BCS_BoneSpace
    alphaScaleBiasClamp {bInterpResult=true, interpSpeedDecreasing=3, interpSpeedIncreasing=3}
    boneToModify {boneName=dyn_Hair_l_01}
    rotation {pitch=2}
    translation {z=2}
   }
   n55 = "Constraint - dyn_Hair_Back_02 (PPPPP)"(
    ComponentPose=@n54
    Alpha=1
    ConstraintWeights_0=0.744
    ConstraintWeights_1=0.099
    ConstraintWeights_2=0.0583
    ConstraintWeights_3=0.0559
    ConstraintWeights_4=0.0428
   ) {
    alphaScaleBiasClamp {bInterpResult=true, interpSpeedDecreasing=4, interpSpeedIncreasing=4}
    boneToModify {boneName=dyn_Hair_Back_02}
    constraintSetup [
     {targetBone {boneName=spine_05}}
     {targetBone {boneName=neck_01}}
     {targetBone {boneName=neck_02}}
     {targetBone {boneName=spine_03}}
     {targetBone {boneName=clavicle_r}}
    ]
    constraintWeights [1,1,1,1,1]
   }
   n56 = "Constraint - dyn_Hair_Back_03 (PPPP)"(ComponentPose=@n55, Alpha=1, ConstraintWeights_0=0.5324, ConstraintWeights_1=0.3669, ConstraintWeights_2=0.0556, ConstraintWeights_3=0.0451) {
    alphaScaleBiasClamp {bInterpResult=true, interpSpeedDecreasing=4, interpSpeedIncreasing=4}
    boneToModify {boneName=dyn_Hair_Back_03}
    constraintSetup [{targetBone {boneName=spine_05}}, {targetBone {boneName=clavicle_r}}, {targetBone {boneName=clavicle_l}}, {targetBone {boneName=spine_04}}]
    constraintWeights [1,1,1,1]
   }
   n57 = "Constraint - dyn_Hair_Back_r_02 (PPPP)"(ComponentPose=@n56, Alpha=1, ConstraintWeights_0=0.6328, ConstraintWeights_1=0.2963, ConstraintWeights_2=0.0382, ConstraintWeights_3=0.0327) {
    alphaScaleBiasClamp {bInterpResult=true, interpSpeedDecreasing=4, interpSpeedIncreasing=4}
    boneToModify {boneName=dyn_Hair_Back_r_02}
    constraintSetup [{targetBone {boneName=spine_05}}, {targetBone {boneName=clavicle_r}}, {targetBone {boneName=neck_01}}, {targetBone {boneName=spine_03}}]
    constraintWeights [1,1,1,1]
   }
   n58 = "Constraint - dyn_Hair_Back_r_03 (PPPP)"(ComponentPose=@n57, Alpha=1, ConstraintWeights_0=0.6983, ConstraintWeights_1=0.2664, ConstraintWeights_2=0.0223, ConstraintWeights_3=0.013) {
    alphaScaleBiasClamp {bInterpResult=true, interpSpeedDecreasing=4, interpSpeedIncreasing=4}
    boneToModify {boneName=dyn_Hair_Back_r_03}
    constraintSetup [
     {targetBone {boneName=clavicle_r}}
     {targetBone {boneName=spine_05}}
     {targetBone {boneName=upperarm_twist_01_r}}
     {targetBone {boneName=spine_04}}
    ]
    constraintWeights [1,1,1,1]
   }
   n59 = "Constraint - dyn_Hair_Back_r_04 (PPPP)"(ComponentPose=@n58, Alpha=1, ConstraintWeights_0=0.5177, ConstraintWeights_1=0.2433, ConstraintWeights_2=0.1643, ConstraintWeights_3=0.0748) {
    alphaScaleBiasClamp {bInterpResult=true, interpSpeedDecreasing=4, interpSpeedIncreasing=4}
    boneToModify {boneName=dyn_Hair_Back_r_04}
    constraintSetup [
     {targetBone {boneName=clavicle_r}}
     {targetBone {boneName=spine_05}}
     {targetBone {boneName=spine_04}}
     {targetBone {boneName=upperarm_twist_01_r}}
    ]
    constraintWeights [1,1,1,1]
   }
   n60 = "Constraint - dyn_Hair_Back_l_02 (PPPPP)"(
    ComponentPose=@n59
    Alpha=1
    ConstraintWeights_0=0.6912
    ConstraintWeights_1=0.1108
    ConstraintWeights_2=0.0819
    ConstraintWeights_3=0.0806
    ConstraintWeights_4=0.0355
   ) {
    alphaScaleBiasClamp {bInterpResult=true, interpSpeedDecreasing=4, interpSpeedIncreasing=4}
    boneToModify {boneName=dyn_Hair_Back_l_02}
    constraintSetup [
     {targetBone {boneName=spine_05}}
     {targetBone {boneName=neck_01}}
     {targetBone {boneName=clavicle_l}}
     {targetBone {boneName=neck_02}}
     {targetBone {boneName=spine_03}}
    ]
    constraintWeights [1,1,1,1,1]
   }
   n61 = "Constraint - dyn_Hair_Back_l_03 (PPPP)"!(
    ComponentPose=@n60
    Alpha=1
    ConstraintWeights_0=0.6983
    ConstraintWeights_1:float=0.2664
    ConstraintWeights_2:float=0.0223
    ConstraintWeights_3:float=0.013
   ) -> (Pose) {
    alphaScaleBiasClamp {bInterpResult=true, interpSpeedDecreasing=4, interpSpeedIncreasing=4}
    boneToModify {boneName=dyn_Hair_Back_l_03}
    constraintSetup [
     {targetBone {boneName=clavicle_l}}
     {targetBone {boneName=spine_05}}
     {targetBone {boneName=upperarm_twist_01_l}}
     {targetBone {boneName=spine_04}}
    ]
    constraintWeights [1,1,1,1]
   }
   n62 = "Constraint - base_Hair_Back_b_l_01 (PPPP)"!(
    ComponentPose=@n61
    Alpha=1
    ConstraintWeights_0=0.6087
    ConstraintWeights_1:float=0.216
    ConstraintWeights_2:float=0.1008
    ConstraintWeights_3:float=0.0744
   ) -> (Pose) {
    alphaScaleBiasClamp {bInterpResult=true, interpSpeedDecreasing=4, interpSpeedIncreasing=4}
    boneToModify {boneName=base_Hair_Back_b_l_01}
    constraintSetup [
     {targetBone {boneName=clavicle_l}}
     {targetBone {boneName=spine_05}}
     {targetBone {boneName=spine_04}}
     {targetBone {boneName=upperarm_twist_01_l}}
    ]
    constraintWeights [1,1,1,1]
   }
   n63 = "Constraint - dyn_Hair_Back_r_04 (PPPPPPP)"!(
    ComponentPose=@n62
    Alpha=1
    ConstraintWeights_0=0.5238
    ConstraintWeights_1:float=0.2171
    ConstraintWeights_2:float=0.178
    ConstraintWeights_3:float=0.031
    ConstraintWeights_4:float=0.0288
    ConstraintWeights_5:float=0
    ConstraintWeights_6:float=0.0212
   ) -> (Pose) {
    boneToModify {boneName=dyn_Hair_Back_r_04}
    constraintSetup [
     {targetBone {boneName=spine_04}}
     {targetBone {boneName=clavicle_r}}
     {targetBone {boneName=spine_05}}
     {targetBone {boneName=spine_03}}
     {targetBone {boneName=spine_02}}
     {targetBone {boneName=upperarm_twist_01_r}}
     {targetBone {boneName=spine_01}}
    ]
    constraintWeights [1,1,1,1,1,1,1]
   }
   n64 = RigidBody(
    ComponentPose=@n63
    OverrideWorldGravity="0,0,-980"
    SimSpaceSettings=(WorldAlpha=1,VelocityScaleZ=0.4,DampingAlpha=1,MaxLinearVelocity=700,MaxAngularVelocity=8,MaxLinearAcceleration=3500,MaxAngularAcceleration=40,ExternalLinearDragV=(X=0.2,Y=0.2,Z=0.2),ExternalLinearVelocity=(X=0.2,Y=0.2,Z=0.2),ExternalAngularVelocity=(X=0,Y=0,Z=0))
   ) {
    bOverrideWorldGravity=true
    overridePhysicsAsset=&F_MED_VioletInk_FaceAcc_Physics1
    simulationSpace=BaseBoneSpace
    baseBoneRef {boneName=spine_05}
    overrideWorldGravity {z=-980}
    simSpaceSettings {
     maxAngularAcceleration=40
     maxAngularVelocity=8
     maxLinearAcceleration=3500
     maxLinearVelocity=700
     velocityScaleZ=0.4
     worldAlpha=1
     externalLinearDragV=<0.2 0.2 0.2>
     externalLinearVelocity=<0.2 0.2 0.2>
    }
   }
   n65 = "Transform (Modify) Bone - Bone: base_Hair_Back_b_01"!(ComponentPose=@n64, Alpha:float=1, Translation="0,0,5", Rotation="-20,0,0", Scale="1,1,1") -> (Pose) {
    rotationMode=BMM_Additive
    rotationSpace=BCS_BoneSpace
    translationMode=BMM_Additive
    translationSpace=BCS_BoneSpace
    alphaScaleBiasClamp {bInterpResult=true, interpSpeedDecreasing=3, interpSpeedIncreasing=3}
    boneToModify {boneName=base_Hair_Back_b_01}
    rotation {pitch=-20}
    translation {z=5}
   }
   n66 = "Transform (Modify) Bone - Bone: dyn_Hair_Back_a_l_01"(ComponentPose=@n65, Alpha=1, Translation="0,0,0", Rotation="50,-5,-30", Scale="1,1,1") {
    rotationMode=BMM_Additive
    rotationSpace=BCS_BoneSpace
    translationMode=BMM_Additive
    translationSpace=BCS_BoneSpace
    alphaScaleBiasClamp {bInterpResult=true, interpSpeedDecreasing=3, interpSpeedIncreasing=3}
    boneToModify {boneName=dyn_Hair_Back_a_l_01}
    rotation=<rot 50 -5 -30>
   }
   n67 = Control Rig!(
    Source=Component To Local(ComponentPose=@n66)
    Alpha
    Crocuh_aim=Get Crouch_aim() -> (Crouch_aim:bool)
    Crocuh_aim_not=Get Crouch_aim_not() -> (Crouch_aim_not:bool)
    CrouchWalk=Get Crouch_Walk() -> (Crouch_Walk:bool)
    HeadUP_NotCrouch=Get HeadUP_NotCrouch()
    HeadUP_Hair_ADJ=Get HeadUP_Hair_ADJ() -> (HeadUP_Hair_ADJ:double)
    Emote=Get Emote() -> (Emote:double)
    SkyDive_Float=Get SkyDive_Float() -> (SkyDive_Float:double)
    SkyDive_Dive_Hair=Get SkyDive_Dive_Hair() -> (SkyDive_Dive_Hair:double)
    Melee=0
    SkyDive_All=Get SkyDive_All() -> (SkyDive_All:double)
    Sprint=Get Sprint() -> (Sprint:double)
    HeadL_Only=Get HeadL_Only() -> (HeadL_Only:double)
    HeadR_Only=Get HeadR_Only() -> (HeadR_Only:double)
    RifleAim_Hair=Get Rifle_Only() -> (Rifle_Only:double)
    Rifle=Get Rifle() -> (Rifle:double)
    TargetingFloat=Get targetFloat() -> (targetFloat:double)
    HeadUp_FHair=Get HeadUP() -> (HeadUP:double)
    Skydive_Float_Para=Get SkyDive_Float_Para() -> (SkyDive_Float_Para:double)
    Skydive_Hair_Right=Get SkyDive_Hair_Right() -> (SkyDive_Hair_Right:double)
    Skydive_Hair_Left=Get SkyDive_Hair_Left() -> (SkyDive_Hair_Left:double)
    CrouchAim_Float=Get CrouchAimFloat() -> (CrouchAimFloat:double)
    EmoteHeadUp=Get Emote_Head_Up() -> (Emote_Head_Up:double)
    FrontEnd=Get bIsInFrontEnd()
    BackBiling=Get isWearingPack()
    FrontEnd_Idle:double=Get FrontEnd_Idle() -> (FrontEnd_Idle:double)
    CapeBackpack:bool=Get isWearingPackBool() -> (isWearingPackBool:bool)
   ) -> (Pose) {controlRigAssetReference {blueprintRigClass=&F_MED_VioletInk_FaceAcc_FWD_CtrlRig_C}}
   n68 = Get isWearingPackBool()
   n69 = "Transform (Modify) Bone - Bone: dyn_Hair_Back_01"(
    ComponentPose=Local To Component(LocalPose=Blend Poses by bool(BlendPose_0=@n30, BlendPose_1=@n67, bActiveValue=@n68))
    Alpha=1
    Translation="0,0,0"
    Rotation="10,-0.001,0"
    Scale="1,1,1"
   ) {
    rotationMode=BMM_Additive
    rotationSpace=BCS_BoneSpace
    translationMode=BMM_Additive
    translationSpace=BCS_BoneSpace
    alphaScaleBiasClamp {bInterpResult=true, interpSpeedDecreasing=3}
    boneToModify {boneName=dyn_Hair_Back_01}
    rotation {pitch=10, yaw=-0.001}
   }
   n70 = "Transform (Modify) Bone - Bone: dyn_Hair_Back_r_01"(ComponentPose=@n69, Alpha=1, Translation="0,0,0", Rotation="15,0,30", Scale="1,1,1") {
    rotationMode=BMM_Additive
    rotationSpace=BCS_BoneSpace
    translationMode=BMM_Additive
    translationSpace=BCS_BoneSpace
    alphaScaleBiasClamp {bInterpResult=true, interpSpeedDecreasing=3}
    boneToModify {boneName=dyn_Hair_Back_r_01}
    rotation {pitch=15, roll=30}
   }
   n71 = "Transform (Modify) Bone - Bone: dyn_Hair_Back_l_01"(ComponentPose=@n70, Alpha=1, Translation="0,0,0", Rotation="10,0,0", Scale="1,1,1") {
    rotationMode=BMM_Additive
    rotationSpace=BCS_BoneSpace
    translationMode=BMM_Additive
    translationSpace=BCS_BoneSpace
    alphaScaleBiasClamp {bInterpResult=true, interpSpeedDecreasing=3}
    boneToModify {boneName=dyn_Hair_Back_l_01}
    rotation {pitch=10}
   }
   n72 = "Transform (Modify) Bone - Bone: dyn_Hair_l_01"(ComponentPose=@n71, Alpha=1, Translation="-3,0.45,-1.213", Rotation="-35.464,22.440,-38.959", Scale="1,1,1") {
    rotationMode=BMM_Additive
    rotationSpace=BCS_BoneSpace
    translationMode=BMM_Additive
    translationSpace=BCS_BoneSpace
    alphaScaleBiasClamp {bInterpResult=true, interpSpeedDecreasing=5, interpSpeedIncreasing=5}
    boneToModify {boneName=dyn_Hair_l_01}
    rotation=<rot -35.464 22.44 -38.959>
    translation=<-3 0.45 -1.213>
   }
   n73 = "Transform (Modify) Bone - Bone: dyn_Hair_r_01"(ComponentPose=@n72, Alpha=1, Translation="-3,-2.377,0.347", Rotation="1.870,-9.273,29.799", Scale="1,1,1") {
    rotationMode=BMM_Additive
    rotationSpace=BCS_BoneSpace
    translationMode=BMM_Additive
    translationSpace=BCS_BoneSpace
    alphaScaleBiasClamp {bInterpResult=true, interpSpeedDecreasing=5, interpSpeedIncreasing=5}
    boneToModify {boneName=dyn_Hair_r_01}
    rotation=<rot 1.87 -9.273 29.799>
    translation=<-3 -2.377 0.347>
   }
   n74 = "Transform (Modify) Bone - Bone: dyn_Hair_Back_r_01"(ComponentPose=@n73, Alpha=1, Translation="4.385,0.000,1.356", Rotation="-15,0,0", Scale="1,1,1.5") {
    rotationMode=BMM_Additive
    rotationSpace=BCS_BoneSpace
    scaleMode=BMM_Additive
    scaleSpace=BCS_BoneSpace
    translationMode=BMM_Additive
    translationSpace=BCS_BoneSpace
    alphaScaleBiasClamp {bInterpResult=true, interpSpeedDecreasing=5, interpSpeedIncreasing=5}
    boneToModify {boneName=dyn_Hair_Back_r_01}
    rotation {pitch=-15}
    scale {z=1.5}
    translation {x=4.385, z=1.356}
   }
   n75 = "Transform (Modify) Bone - Bone: dyn_Hair_Back_r_02"(ComponentPose=@n74, Alpha=1, Translation="0,0,0", Rotation="-15,0,0", Scale="1,1,1") {
    rotationMode=BMM_Additive
    rotationSpace=BCS_BoneSpace
    translationMode=BMM_Additive
    translationSpace=BCS_BoneSpace
    alphaScaleBiasClamp {interpSpeedDecreasing=5, interpSpeedIncreasing=5}
    boneToModify {boneName=dyn_Hair_Back_r_02}
    rotation {pitch=-15}
   }
   n76 = "Transform (Modify) Bone - Bone: base_Hair_l_01"(ComponentPose=@n75, Alpha=1, Translation="0,0,5", Rotation="0,0,0", Scale="1,1,1") {
    rotationMode=BMM_Additive
    rotationSpace=BCS_BoneSpace
    translationMode=BMM_Additive
    translationSpace=BCS_BoneSpace
    alphaScaleBiasClamp {interpSpeedDecreasing=5, interpSpeedIncreasing=5}
    boneToModify {boneName=base_Hair_l_01}
    translation {z=5}
   }
   n77 = "Transform (Modify) Bone - Bone: base_Hair_r_01"(ComponentPose=@n76, Alpha=1, Translation="0,0,5", Rotation="0,0,0", Scale="1,1,1") {
    rotationMode=BMM_Additive
    rotationSpace=BCS_BoneSpace
    translationMode=BMM_Additive
    translationSpace=BCS_BoneSpace
    alphaScaleBiasClamp {interpSpeedDecreasing=5, interpSpeedIncreasing=5}
    boneToModify {boneName=base_Hair_r_01}
    translation {z=5}
   }
   n78 = "Transform (Modify) Bone - Bone: base_Hair_l_01"(ComponentPose=@n77, Alpha=1, Translation="0,0,5", Rotation="0,0,0", Scale="1,1,1") {
    rotationMode=BMM_Additive
    rotationSpace=BCS_BoneSpace
    translationMode=BMM_Additive
    translationSpace=BCS_BoneSpace
    alphaScaleBiasClamp {interpSpeedDecreasing=5, interpSpeedIncreasing=5}
    boneToModify {boneName=base_Hair_l_01}
    translation {z=5}
   }
   n79 = "Transform (Modify) Bone - Bone: base_Hair_r_01"(ComponentPose=@n78, Alpha=1, Translation="0,0,5", Rotation="0,0,0", Scale="1,1,1") {
    rotationMode=BMM_Additive
    rotationSpace=BCS_BoneSpace
    translationMode=BMM_Additive
    translationSpace=BCS_BoneSpace
    alphaScaleBiasClamp {interpSpeedDecreasing=5, interpSpeedIncreasing=5}
    boneToModify {boneName=base_Hair_r_01}
    translation {z=5}
   }
   n80 = "Transform (Modify) Bone - Bone: dyn_Hair_Back_r_01"(ComponentPose=@n79, Alpha=1, Translation="0,1,3", Rotation="0,0,0", Scale="1,1,1") {
    rotationMode=BMM_Additive
    rotationSpace=BCS_BoneSpace
    translationMode=BMM_Additive
    translationSpace=BCS_BoneSpace
    alphaScaleBiasClamp {interpSpeedDecreasing=5, interpSpeedIncreasing=5}
    boneToModify {boneName=dyn_Hair_Back_r_01}
    translation {y=1, z=3}
   }
   n81 = "Transform (Modify) Bone - Bone: dyn_Hair_Back_r_01"(ComponentPose=@n80, Alpha=1, Translation="0,2.02,-3", Rotation="0,0,0", Scale="1,1,1") {
    rotationMode=BMM_Additive
    rotationSpace=BCS_BoneSpace
    translationMode=BMM_Additive
    translationSpace=BCS_BoneSpace
    alphaScaleBiasClamp {interpSpeedDecreasing=5, interpSpeedIncreasing=5}
    boneToModify {boneName=dyn_Hair_Back_r_01}
    translation {y=2.02, z=-3}
   }
   n82 = "Transform (Modify) Bone - Bone: dyn_Hair_Back_r_01"(ComponentPose=@n81, Alpha=1, Translation="0,0,0", Rotation="15,0,-30", Scale="1,1,1") {
    rotationMode=BMM_Additive
    rotationSpace=BCS_BoneSpace
    translationMode=BMM_Additive
    translationSpace=BCS_BoneSpace
    alphaScaleBiasClamp {interpSpeedDecreasing=5, interpSpeedIncreasing=5}
    boneToModify {boneName=dyn_Hair_Back_r_01}
    rotation {pitch=15, roll=-30}
   }
   n83 = "Transform (Modify) Bone - Bone: dyn_Hair_Back_a_l_01"(ComponentPose=@n82, Alpha=1, Translation="1.535,2.873,-7", Rotation="30.000,0.000,-0.000", Scale="1,1,1") {
    rotationMode=BMM_Additive
    rotationSpace=BCS_BoneSpace
    translationMode=BMM_Additive
    translationSpace=BCS_BoneSpace
    alphaScaleBiasClamp {interpSpeedDecreasing=5, interpSpeedIncreasing=5}
    boneToModify {boneName=dyn_Hair_Back_a_l_01}
    rotation=<rot 30 0.0001 0>
    translation=<1.535 2.873 -7>
   }
   n84 = "Transform (Modify) Bone - Bone: dyn_Hair_Back_a_l_01"(ComponentPose=@n83, Alpha=1, Translation="0.707,-5.946,5.796", Rotation="-13.505,36.515,-34.311", Scale="1,1,1") {
    rotationMode=BMM_Additive
    rotationSpace=BCS_BoneSpace
    translationMode=BMM_Additive
    translationSpace=BCS_BoneSpace
    alphaScaleBiasClamp {interpSpeedDecreasing=5, interpSpeedIncreasing=5}
    boneToModify {boneName=dyn_Hair_Back_a_l_01}
    rotation=<rot -13.505 36.515 -34.311>
    translation=<0.707 -5.946 5.796>
   }
   Output Pose(Result=Component To Local(ComponentPose=@n84)) {layerGroup=None}
   n85 = Control Rig!(
    Source=Use cached pose 'InitialBlend'()
    Alpha
    Rifle:bool=Get Only_Rifle() -> (Only_Rifle:bool)
    aimFloat:double=Get targetFloat()
    RifleFloat:double=Get Rifle()
    OnlyClav:double=Get OnlyClav() -> (OnlyClav:double)
   ) -> (Pose) {controlRigAssetReference {blueprintRigClass=&/F_MED_VioletInk_FaceAcc_CtrlRig1.F_MED_VioletInk_FaceAcc_CtrlRig1_C}}
   Save cached pose 'InitialPose'(Pose=@n85) {cacheName=InitialPose}
   n86 = Layered blend per bone(
    BasePose=Copy Pose From Mesh() {bUseAttachedParent=true}
    BlendPoses_0=Copy Pose From Mesh(SourceMeshComponent=Get Body() -> (Body:Skeletal Mesh Component Budgeted Object Reference)) {bUseAttachedParent=true}
   ) {layerSetup [{branchFilters [
       {blendDepth=2, boneName=clavicle_r}
       {blendDepth=2, boneName=clavicle_l}
       {blendDepth=2, boneName=upperarm_l}
       {blendDepth=2, boneName=upperarm_r}
      ]}]}
   Save cached pose 'InitialBlend'(Pose=@n86) {cacheName=InitialBlend}
   n87 = Anim Dynamics - Bone: dyn_hat_01!(
    ComponentPose=Local To Component(LocalPose=Use cached pose 'InitialPose'())
    Alpha=0.5
    GravityOverride:Vector="0,0,-980"
    ExternalForce="0,0,0"
   ) -> (Pose) {
    angularDampingOverride=0.9999
    angularSpringConstant=10000
    bAngularSpring=true
    bOverrideAngularDamping=true
    bUseGravityOverride=true
    gravityScale=0.35
    simulationSpace=RootRelative
    boundBone {boneName=dyn_hat_01}
    componentAppliedLinearAccClamp=<2000 2000 2000>
    componentLinearAccScale=<0.5 0.5 0.5>
    componentLinearVelScale=<0.5 0.5 0.5>
    gravityOverride {z=-980}
    physicsBodyDefinitions [{
      boundBone {boneName=dyn_hat_01}
      boxExtents=<20 30 30>
      constraintSetup {angularLimitsMax=<0 0 5>, angularLimitsMin=<0 -15 -5>, angularTarget=<1 0 0>}
      localJointOffset=<2 0 0>
     }]
   }
   n88 = Anim Dynamics - Bone: dyn_Hair_Front_01!(ComponentPose=@n87, Alpha=1, GravityOverride:Vector="0,0,-980", ExternalForce="0,0,0") -> (Pose) {
    angularDampingOverride=0.9999
    angularSpringConstant=10000
    bAngularSpring=true
    bOverrideAngularDamping=true
    bUseGravityOverride=true
    gravityScale=0.35
    simulationSpace=RootRelative
    boundBone {boneName=dyn_Hair_Front_01}
    componentLinearAccScale=<1 1 1>
    componentLinearVelScale=<1 1 1>
    gravityOverride {z=-980}
    physicsBodyDefinitions [{
      boundBone {boneName=dyn_Hair_Front_01}
      boxExtents=<30 30 30>
      constraintSetup {angularLimitsMax=<0 0 0>, angularLimitsMin=<0 -5 -2>, angularTarget=<1 0 0>}
      localJointOffset=<4 0 0>
     }]
   }
   n89 = Anim Dynamics - Bone: dyn_Hair_Front_02!(ComponentPose=@n88, Alpha=1, GravityOverride:Vector="0,0,-980", ExternalForce="0,0,0") -> (Pose) {
    angularDampingOverride=0.9999
    angularSpringConstant=10000
    bAngularSpring=true
    bOverrideAngularDamping=true
    bUseGravityOverride=true
    gravityScale=0.35
    simulationSpace=RootRelative
    boundBone {boneName=dyn_Hair_Front_02}
    componentLinearAccScale=<1 1 1>
    componentLinearVelScale=<1 1 1>
    gravityOverride {z=-980}
    physicsBodyDefinitions [{
      boundBone {boneName=dyn_Hair_Front_02}
      boxExtents=<30 30 30>
      constraintSetup {angularLimitsMax=<0 0 0>, angularLimitsMin=<0 -10 -5>, angularTarget=<1 0 0>}
      localJointOffset=<4 0 0>
     }]
   }
   n90 = Anim Dynamics - Bone: dyn_Hair_l_01!(ComponentPose=@n89, Alpha=1, GravityOverride:Vector="0,0,-980", ExternalForce="0,0,0") -> (Pose) {
    angularDampingOverride=0.9999
    angularSpringConstant=10000
    bAngularSpring=true
    bOverrideAngularDamping=true
    bUseGravityOverride=true
    gravityScale=0.35
    simulationSpace=BoneRelative
    boundBone {boneName=dyn_Hair_l_01}
    componentLinearAccScale=<1 1 1>
    componentLinearVelScale=<1 1 1>
    gravityOverride {z=-980}
    physicsBodyDefinitions [{
      boundBone {boneName=dyn_Hair_l_01}
      boxExtents=<30 30 30>
      constraintSetup {angularLimitsMax=<0 -5 0>, angularLimitsMin=<0 -15 -10>, angularTarget=<1 0 0>}
      localJointOffset=<4 0 0>
     }]
    relativeSpaceBone {boneName=head}
   }
   n91 = Anim Dynamics - Bone: dyn_Hair_l_02!(ComponentPose=@n90, Alpha=1, GravityOverride:Vector="0,0,-980", ExternalForce="0,0,0") -> (Pose) {
    angularDampingOverride=0.9999
    angularSpringConstant=10000
    bAngularSpring=true
    bOverrideAngularDamping=true
    bUseGravityOverride=true
    gravityScale=0.35
    simulationSpace=RootRelative
    boundBone {boneName=dyn_Hair_l_02}
    componentLinearAccScale=<1 1 1>
    componentLinearVelScale=<1 1 1>
    gravityOverride {z=-980}
    physicsBodyDefinitions [{
      boundBone {boneName=dyn_Hair_l_02}
      boxExtents=<30 30 30>
      constraintSetup {angularLimitsMax=<0 0 0>, angularLimitsMin=<0 -10 -10>, angularTarget=<1 0 0>}
      localJointOffset=<4 0 0>
     }]
   }
   n92 = Anim Dynamics - Bone: dyn_Hair_l_03!(ComponentPose=@n91, Alpha=1, GravityOverride:Vector="0,0,-980", ExternalForce="0,0,0") -> (Pose) {
    angularDampingOverride=0.9999
    angularSpringConstant=10000
    bAngularSpring=true
    bOverrideAngularDamping=true
    bUseGravityOverride=true
    gravityScale=0.35
    simulationSpace=RootRelative
    boundBone {boneName=dyn_Hair_l_03}
    componentLinearAccScale=<1 1 1>
    componentLinearVelScale=<1 1 1>
    gravityOverride {z=-980}
    physicsBodyDefinitions [{
      boundBone {boneName=dyn_Hair_l_03}
      boxExtents=<30 30 30>
      constraintSetup {angularLimitsMax=<0 0 0>, angularLimitsMin=<0 -15 -10>, angularTarget=<1 0 0>}
      localJointOffset=<4 0 0>
     }]
   }
   n93 = Anim Dynamics - Bone: dyn_Hair_r_01!(ComponentPose=@n92, Alpha=1, GravityOverride:Vector="0,0,-980", ExternalForce="0,0,0") -> (Pose) {
    angularDampingOverride=0.9999
    angularSpringConstant=10000
    bAngularSpring=true
    bOverrideAngularDamping=true
    bUseGravityOverride=true
    gravityScale=0.35
    simulationSpace=BoneRelative
    boundBone {boneName=dyn_Hair_r_01}
    componentLinearAccScale=<1 1 1>
    componentLinearVelScale=<1 1 1>
    gravityOverride {z=-980}
    physicsBodyDefinitions [{
      boundBone {boneName=dyn_Hair_r_01}
      boxExtents=<30 30 30>
      constraintSetup {angularLimitsMax=<0 -5 10>, angularLimitsMin=<0 -15 0>, angularTarget=<1 0 0>}
      localJointOffset=<4 0 0>
     }]
    relativeSpaceBone {boneName=head}
   }
   n94 = Anim Dynamics - Bone: dyn_Hair_r_02!(ComponentPose=@n93, Alpha=1, GravityOverride:Vector="0,0,-980", ExternalForce="0,0,0") -> (Pose) {
    angularDampingOverride=0.9999
    angularSpringConstant=10000
    bAngularSpring=true
    bOverrideAngularDamping=true
    bUseGravityOverride=true
    gravityScale=0.35
    simulationSpace=RootRelative
    boundBone {boneName=dyn_Hair_r_02}
    componentLinearAccScale=<1 1 1>
    componentLinearVelScale=<1 1 1>
    gravityOverride {z=-980}
    physicsBodyDefinitions [{
      boundBone {boneName=dyn_Hair_r_02}
      boxExtents=<30 30 30>
      constraintSetup {angularLimitsMax=<0 10 10>, angularLimitsMin=<0 0 0>, angularTarget=<1 0 0>}
      localJointOffset=<4 0 0>
     }]
   }
   n95 = Anim Dynamics - Bone: dyn_Hair_r_03!(ComponentPose=@n94, Alpha=1, GravityOverride:Vector="0,0,-980", ExternalForce="0,0,0") -> (Pose) {
    angularDampingOverride=0.9999
    angularSpringConstant=10000
    bAngularSpring=true
    bOverrideAngularDamping=true
    bUseGravityOverride=true
    gravityScale=0.35
    simulationSpace=RootRelative
    boundBone {boneName=dyn_Hair_r_03}
    componentLinearAccScale=<1 1 1>
    componentLinearVelScale=<1 1 1>
    gravityOverride {z=-980}
    physicsBodyDefinitions [{
      boundBone {boneName=dyn_Hair_r_03}
      boxExtents=<30 30 30>
      constraintSetup {angularLimitsMax=<0 15 10>, angularLimitsMin=<0 0 0>, angularTarget=<1 0 0>}
      localJointOffset=<4 0 0>
     }]
   }
   Save cached pose 'FrontHair'(Pose=Component To Local(ComponentPose=@n95)) {cacheName=FrontHair}
   Get isWearingBackbling()
   Get isWearingPackBool()
  ]}
 EventGraph {nodes [
   Event BlueprintUpdateAnimation()
   > n1 = Sequence!(execute) -> (then_0, then_1, then_2:Exec, then_3:Exec, then_4:Exec, then_5:Exec, then_6:Exec)
   Set targetFloat(execute=@n1.then_1, targetFloat:double="To Float (Boolean)"(InBool=Get bIsTargeting())) -> (Output_Get:double)
   > n2 = Branch(Condition=OR Boolean!(A=Get bIsSkyDiving(), B=Get bIsSkydiveDiveMode(), C:bool=Get bIsSkydiveFloating()) -> (ReturnValue))
   > Set Skydive(Skydive:double=0.5) -> (Output_Get:double)
   Set Skydive(execute=@n2.else, Skydive=1)
   n3 = Branch(execute=@n1.then_2, Condition=AND Boolean(A=Get bIsInAir() -> (bIsInAir:bool), B=Get bIsTacticalSprinting()))
   > Set InNot_Ground_Misc(InNot_Ground_Misc:double=0.2) -> (Output_Get:double)
   Set InNot_Ground_Misc(execute=@n3.else, InNot_Ground_Misc=1)
   n4 = Branch(execute=@n1.then_3, Condition=OR Boolean!(A=Get bIsSurfaceSwimming(), B=Get bIsSwimSprinting() -> (bIsSwimSprinting:bool)) -> (ReturnValue))
   > Set Swim_Dyn(Swim_Dyn:double=0.2) -> (Output_Get:double)
   Set Swim_Dyn(execute=@n4.else, Swim_Dyn=1)
   n5 = "Equal (Float)"(A=Get InNot_Ground_Misc() -> (InNot_Ground_Misc:double), B=Get Skydive() -> (Skydive:double))
   n6 = Get Swim_Dyn() -> (Swim_Dyn:double)
   Set Skydive_NotGround(execute=@n1.then_4, Skydive_NotGround:double="To Float (Boolean)"(InBool="Equal (Float)"(A:bool=@n5, B=@n6))) -> (Output_Get:double)
   n7 = CheckForBackbling()
   n8 = NOT Boolean(A=@n7.bIsBackblingACape)
   n9 = OR Boolean(A=@n7, B=@n7.bIsBackblingLong)
   Set isWearingPack(execute=@n1.then_6, isWearingPack:bool=AND Boolean(A=@n9, B=@n8)) -> (Output_Get:bool)
   > Set isWearingCape(isWearingCape=@n7.bIsBackblingACape)
   n10 = "Gravity Override State Function (Parameter Asset) (Nativized)"(
    execute=@n1.then_0
    CharacterPartAnimInstance=Self-Reference()
    GameTimeInSeconds=GetGameTimeInSeconds()
    Parameters=&/M_MED_VioletInkFaceAcc_Parameters1
   )
   > Set Gravity Override(Gravity Override=@n10)
   > Set RBAN Settings Out(RBAN Settings Out=@n10.OutRigidBodyAnimNodeParameters)
   n11 = GetOwnerSkeletalMeshComponent()
   n12 = BreakRotator(InRot=BreakTransform(InTransform=GetDeltaTransformFromRefPose(self=@n11, BoneName=head, BaseName=spine_05)).Rotation)
   > Set HeadUP(HeadUP:double=MapRangeClamped(Value=@n12.Yaw, InRangeA=0, InRangeB=-75)) -> (Output_Get:double)
   > Set HeadDown(HeadDown:double=MapRangeClamped(Value=@n12.Yaw, InRangeA=0, InRangeB=75)) -> (Output_Get:double)
   > Set HeadL(HeadL:double=MapRangeClamped(Value=@n12, InRangeA=0, InRangeB=75)) -> (Output_Get:double)
   > Set HeadR(HeadR:double=MapRangeClamped(Value=@n12, InRangeA=0, InRangeB=-75)) -> (Output_Get:double)
   > Set HeadUP_Max(HeadUP_Max:double=MapRangeClamped(Value=@n12.Yaw, InRangeA=30, InRangeB=-60)) -> (Output_Get:double)
   > Set HeadDown_Max(HeadDown_Max:double=MapRangeClamped(Value=@n12.Yaw, InRangeA=-30, InRangeB=60)) -> (Output_Get:double)
   n13 = GetOwnerSkeletalMeshComponent()
   n14 = GetAlphaFromReferenceTransformDelta(SkeletalMeshComponent=@n13, BoneName=head, BaseName=spine_05, DeltaComponentType=EulerY, InputRange=(X=-5,Y=10), TwistAxis="1,0,0")
   n15 = GetAlphaFromReferenceTransformDelta(SkeletalMeshComponent=@n13, BoneName=ev, BaseName=spine_05, InputRange=(X=-16,Y=-50), TwistAxis="1,0,0")
   > n16 = Set Head look up alpha(Head look up alpha:double="Clamp (Float)"(Value=float + float(A=@n15, B=@n14), Min=0, Max=1)) -> (Output_Get:double)
   n17 = Get bIsCrouching()
   > Set Head look up crouch alpha(Head look up crouch alpha:double=SelectFloat(A=@n16, B=0, bPickA=@n17)) -> (Output_Get:double)
   > n18 = Set Crouch_aim(Crouch_aim:bool=AND Boolean(A=@n17, B=Get bIsTargeting())) -> (Output_Get:bool)
   > Set CrouchAimFloat(CrouchAimFloat:double="To Float (Boolean)"(InBool=@n18)) -> (Output_Get:double)
   n19 = Get bIsTargeting()
   > Set Crouch_aim_not(Crouch_aim_not:bool=AND Boolean(A=@n17, B=NOT Boolean(A=@n19))) -> (Output_Get:bool)
   n20 = Get HeadUP()
   > Set Crouch_Walk(Crouch_Walk:bool="Equal (Float)"(A:double=SelectFloat(A=@n20, B=0, bPickA=@n17), B=1)) -> (Output_Get:bool)
   > Set HeadUp_Aim(HeadUp_Aim:double=SelectFloat(A=Get Head look up alpha() -> (Head look up alpha:double), B=0, bPickA=@n19)) -> (Output_Get:double)
   > Set HeadUP_NotCrouch(HeadUP_NotCrouch:bool=NOT Boolean(A=@n17)) -> (Output_Get:bool)
   > Set Emote(Emote:double="To Float (Boolean)"(InBool=Get bIsPlayingEmote())) -> (Output_Get:double)
   n21 = Get bIsFalling()
   n22 = Get bIsJumping()
   > Set INAir(INAir:double="To Float (Boolean)"(InBool=OR Boolean(A=@n22, B=@n21))) -> (Output_Get:double)
   > Set SkyDive_Float(SkyDive_Float:double="To Float (Boolean)"(InBool=Get bIsSkydiveFloating())) -> (Output_Get:double)
   > Set SkyDive_Dive_Hair(SkyDive_Dive_Hair:double="To Float (Boolean)"(InBool=Get bIsSkydiveDiveMode())) -> (Output_Get:double)
   n23 = NOT Boolean(A=Get bIsMoving2D())
   n24 = NOT Boolean(A=Get bIsTargeting())
   n25 = Get bIsCrouching()
   > Set Crouch_Headup(Crouch_Headup:double=SelectFloat(A=Get Head look up alpha(), B=0, bPickA=AND Boolean!(A=@n25, B=@n24, C:bool=@n23) -> (ReturnValue))) -> (Output_Get:double)
   n26 = Get bIsSprinting()
   n27 = Get bIsTacticalSprinting()
   > Set Sprint(Sprint:double="To Float (Boolean)"(InBool=OR Boolean(A=@n26, B=@n27))) -> (Output_Get:double)
   > Set SkyDive_All(SkyDive_All:double="To Float (Boolean)"(InBool=Get bIsSkyDiving())) -> (Output_Get:double)
   n28 = OR Boolean!(A=Get bIsCrouching(), B=Get bIsPlayingEmote(), C:bool=Get bIsTargeting()) -> (ReturnValue)
   > Set HeadUP_Only(HeadUP_Only:double=SelectFloat(A=Get HeadUP(), B=0, bPickA=NOT Boolean(A=@n28))) -> (Output_Get:double)
   n29 = OR Boolean(A=Get bIsCrouching(), B=Get bIsPlayingEmote(), C=Get bIsTargeting())
   > Set HeadDown_Only(HeadDown_Only:double=SelectFloat(A=Get HeadDown() -> (HeadDown:double), B=0, bPickA=NOT Boolean(A=@n29))) -> (Output_Get:double)
   n30 = OR Boolean(A=Get bIsCrouching(), B=Get bIsPlayingEmote(), C=Get bIsTargeting())
   > Set HeadL_Only(HeadL_Only:double=SelectFloat(A=Get HeadL() -> (HeadL:double), B=0, bPickA=NOT Boolean(A=@n30))) -> (Output_Get:double)
   n31 = OR Boolean(A=Get bIsCrouching(), B=Get bIsPlayingEmote(), C=Get bIsTargeting())
   > Set HeadR_Only(HeadR_Only:double=SelectFloat(A=Get HeadR() -> (HeadR:double), B=0, bPickA=NOT Boolean(A=@n31))) -> (Output_Get:double)
   > Set HeadUp_Rifle(HeadUp_Rifle:double=SelectFloat(A=Get HeadUP(), B=0, bPickA=Get bIsTargeting())) -> (Output_Get:double)
   n32 = Get bIsTargeting()
   > Set HeadDown_Rifle(HeadDown_Rifle:double=SelectFloat(A=Get HeadDown(), B=0, bPickA=@n32)) -> (Output_Get:double)
   > Set Rifle_Only(Rifle_Only:double=SelectFloat(A=Get Rifle(), B=0, bPickA=@n32)) -> (Output_Get:double)
   n33 = Get Only_Rifle()
   > Set HeadUP_Rile(HeadUP_Rile:double=SelectFloat(A=Get HeadUP_Max() -> (HeadUP_Max:double), B=0, bPickA=@n33)) -> (Output_Get:double)
   > Set HeadDown_Rile(HeadDown_Rile:double=SelectFloat(A=Get HeadDown_Max() -> (HeadDown_Max:double), B=0, bPickA=@n33)) -> (Output_Get:double)
   > Set PickAxe(PickAxe:double=Get Melee() -> (Melee:double)) -> (Output_Get:double)
   n34 = OR Boolean!(A=Get bIsCrouching(), B=Get bIsTargeting()) -> (ReturnValue)
   > Set OnlyClav(OnlyClav:double="To Float (Boolean)"(InBool=NOT Boolean(A=@n34))) -> (Output_Get:double)
   > Set Head_Down_pistol(Head_Down_pistol:double=float * float(A=Get pistol() -> (pistol:double), B=Get HeadDown()) {is=title K2Node_PromotableOperator}) -> (Output_Get:double)
   > Set Head_Up_Pistol(Head_Up_Pistol:double=float * float(A=Get pistol(), B=Get HeadUP())) -> (Output_Get:double)
   > Set Emote_Head_Up(Emote_Head_Up:double=SelectFloat(A=Get HeadUP_Max(), B=0, bPickA=Get bIsPlayingEmote())) -> (Output_Get:double)
   n35 = OR Boolean(A=Get bIsParachuteOpened(), B=Get bIsSkydiveFloating())
   > Set SkyDive_Hair_Left(SkyDive_Hair_Left:double=SelectFloat(A=Get strafeLeft float() -> (strafeLeft float:double), B=0, bPickA=@n35)) -> (Output_Get:double)
   > Set SkyDive_Hair_Right(SkyDive_Hair_Right:double=SelectFloat(A=Get strafeRight float() -> (strafeRight float:double), B=0, bPickA=@n35)) -> (Output_Get:double)
   n36 = NOT Boolean(A=Get bIsMoving2D())
   > Set HeadL_NotMove(HeadL_NotMove:double=SelectFloat(A=Get HeadL_Only(), B=0, bPickA=@n36)) -> (Output_Get:double)
   > Set HeadR_NotMove(HeadR_NotMove:double=SelectFloat(A=Get HeadR_Only(), B=0, bPickA=@n36)) -> (Output_Get:double)
   > Set SkyDive_Float_Para(SkyDive_Float_Para:double="To Float (Boolean)"(InBool=@n35)) -> (Output_Get:double)
   n37 = NOT Boolean(A=Get bIsPlayingEmote())
   n38 = Get bIsInFrontEnd()
   > Set FrontEnd_Idle(FrontEnd_Idle:double="To Float (Boolean)"(InBool=AND Boolean!(A=@n38, B=@n37) -> (ReturnValue))) -> (Output_Get:double)
   n39 = PlayerMovement(execute=@n1.then_5, AnimInstance=Self-Reference(), XDotProductCoefficient=1, YDotProductCoefficient=1, ZDotProductCoefficient=1)
   n40 = BreakVector(InVec=MakeVector(X=@n39, Y=@n39.YMovement))
   > Set strafeLeft float(strafeLeft float:double="Absolute (Float)"(A="Clamp (Float)"(Value=@n40, Min=-1, Max=0))) -> (Output_Get:double)
   > Set strafeRight float(strafeRight float:double="Clamp (Float)"(Value=@n40, Min=0, Max=1)) -> (Output_Get:double)
   > Set moveForward(moveForward:double="Clamp (Float)"(Value=@n40.Y, Min=0, Max=1)) -> (Output_Get:double)
   > Set moveBackward(moveBackward:double="Absolute (Float)"(A="Clamp (Float)"(Value=@n40.Y, Min=-1, Max=0))) -> (Output_Get:double)
   n41 = Get bIsMoving2D()
   > Set IsIdle(IsIdle:double="To Float (Boolean)"(InBool=NOT Boolean(A=@n41))) -> (Output_Get:double)
   Event OnBeginPickaxeSwing() {is=title K2Node_Event}
   > Set Melee_Cons(Melee_Cons:double=1) -> (Output_Get:double)
   n42 = Event OnCharacterCustomizationLoaded()
   > Set Body(Body:Skeletal Mesh Component Budgeted Object Reference=Get Array Item with EFort Custom Part Type(Array=@n42.CustomMeshParts)) -> (Output_Get:Skeletal Mesh Component Budgeted Object Reference)
   n43 = CheckForBackbling()
   > Set isWearingBackbling(isWearingBackbling=@n43)
   > Set isWearingPackBool(isWearingPackBool:bool=OR Boolean!(A=@n43, B=@n43.bIsBackblingACape, C:bool=@n43.bIsBackblingLong) -> (ReturnValue)) -> (Output_Get:bool)
   n44 = OR Boolean!(A=@n43.bIsBackblingACape, B=@n43.bIsBackblingLong) -> (ReturnValue)
   > Set is_WearingCape(is_WearingCape:bool=@n44) -> (Output_Get:bool)
   > Set is_f_WearingCape(is_f_WearingCape:double="To Float (Boolean)"(InBool=@n44)) -> (Output_Get:double)
   > Set isWearingBP(isWearingBP:bool=AND Boolean(A=@n43, B=NOT Boolean(A=@n44))) -> (Output_Get:bool)
   > Set Backbling(Backbling:double="To Float (Boolean)"(InBool=OR Boolean!(A=@n43, B=@n43.bIsBackblingACape, C:bool=@n43.bIsBackblingLong) -> (ReturnValue))) -> (Output_Get:double)
   Event OnEndPickaxeSwing() {is=title K2Node_Event}
   > Set Melee_Cons(Melee_Cons=0)
   Event OnWeaponChanged()
   > n45 = Cast To FortPlayerAnimInstance(Object=GetAnimInstance(self=GetOwnerSkeletalMeshComponent()))
   > Set pistol(pistol:double=0) -> (Output_Get:double)
   > Set Melee(Melee:double=0) -> (Output_Get:double)
   > Set Rifle(Rifle:double=0) -> (Output_Get:double)
   > n46 = Switch on EFortWeaponCoreAnimation!(execute, Selection:EFortWeaponCoreAnimation Enum=Get WeaponCoreAnim(self=@n45)) -> (
    Melee:Exec
    Pistol:Exec
    Shotgun:Exec
    PaperBlueprint:Exec
    Rifle:Exec
    MeleeOneHand:Exec
    MachinePistol:Exec
    RocketLauncher:Exec
    GrenadeLauncher:Exec
    GoingCommando:Exec
    AssaultRifle:Exec
    TacticalShotgun:Exec
    SniperRifle:Exec
    TrapPlacement:Exec
    ShoulderLauncher:Exec
    AbilityDecoTool:Exec
    Crossbow:Exec
    C4:Exec
    RemoteControl:Exec
    DualWield:Exec
    AR_BullPup:Exec
    AR_ForwardGrip:Exec
    MedPackPaddles:Exec
    SMG_P90:Exec
    AR_DrumGun:Exec
    Consumable_Small:Exec
    Consumable_Large:Exec
    Balloon:Exec
    MountedTurret:Exec
    CreativeTool:Exec
    ExplosiveBow:Exec
    AshtonIndigo:Exec
    AshtonChicago:Exec
    MeleeDualWield:Exec
    Unarmed:Exec
   )
   Set Rifle(
    execute=[
     @n46.AR_BullPup
     @n46.AR_ForwardGrip
     @n46.AssaultRifle
     @n46.Crossbow
     @n46.GrenadeLauncher
     @n46.Rifle
     @n46.RocketLauncher
     @n46.Shotgun
     @n46.SniperRifle
     @n46.TacticalShotgun
    ]
    Rifle=1
   )
   > Set Only_Rifle(Only_Rifle:bool=1) -> (Output_Get:bool)
   > Set Only_Pistol(Only_Pistol:bool=false) -> (Output_Get:bool)
   Set Melee(execute=[@n46.Melee, @n46.MeleeDualWield, @n46.MeleeOneHand], Melee=1)
   > Set Only_Rifle(Only_Rifle=false)
   > Set Only_Pistol(Only_Pistol=false)
   > Set Rifle(Rifle=0)
   Set pistol(execute=@n46.Pistol, pistol=1)
   > Set Only_Rifle(Only_Rifle=false)
   > Set Only_Pistol(Only_Pistol=true)
   > Set Rifle(Rifle=0)
   Event BlueprintInitializeAnimation()
   > Set Skeletal_Mesh(Skeletal_Mesh:Skeletal Mesh Component Object Reference=GetOwnerSkeletalMeshComponent()) -> (Output_Get:Skeletal Mesh Component Object Reference)
   NOT Boolean(A=Get bIsSprinting())
   NOT Boolean(A=@n9)
   OR Boolean(A=false, B=NOT Boolean(A=Get bIsSurfaceSwimming()), C=NOT Boolean(A=Get bIsSwimSprinting()))
   OR Boolean!(A=@n8, B=@n9) -> (ReturnValue)
   OR Boolean(A=false, B=false)
   OR Boolean!(A=@n25, B=@n24, C:bool=@n23) -> (ReturnValue)
   K2Node_Event_2 = CustomEvent!() -> (
    OutputDelegate
    then
    CustomBaseMesh:Skeletal Mesh Component Object Reference
    CustomParts:Array of Custom Character Part Object References
    CustomMeshParts:Array of Skeletal Mesh Component Budgeted Object References
   ) {is=title K2Node_CustomEvent}
   Get HeadUP_Max()
  ]}
]
variables [
 Backbling {type=number}
 Body {type=SkeletalMeshComponentBudgeted}
 Crouch_aim {type=boolean}
 Crouch_aim_not {type=boolean}
 Crouch_Headup {type=number}
 Crouch_Walk {type=boolean}
 CrouchAimFloat {type=number}
 Emote {type=number}
 Emote_Head_Up {type=number}
 FrontEnd_Idle {type=number}
 Gravity Override {type=Vector}
 Head look up alpha {type=number}
 Head look up crouch alpha {type=number}
 Head_Down_pistol {type=number}
 Head_Up_Pistol {type=number}
 HeadDown {type=number}
 HeadDown_Max {type=number}
 HeadDown_Only {type=number}
 HeadDown_Rifle {type=number}
 HeadDown_Rile {type=number}
 HeadL {type=number}
 HeadL_NotMove {type=number}
 HeadL_Only {type=number}
 HeadR {type=number}
 HeadR_NotMove {type=number}
 HeadR_Only {type=number}
 HeadUP {type=number}
 HeadUp_Aim {type=number}
 HeadUP_Hair_ADJ {type=number}
 HeadUP_Max {type=number}
 HeadUP_NotCrouch {type=boolean}
 HeadUP_Only {type=number}
 HeadUp_Rifle {type=number}
 HeadUP_Rile {type=number}
 INAir {type=number}
 InNot_Ground_Misc {type=number}
 is_f_WearingCape {type=number}
 is_WearingCape {type=boolean}
 isGunOut {type=number}
 IsIdle {type=number}
 isWearingBackbling {type=boolean}
 isWearingBP {type=boolean}
 isWearingCape {type=boolean}
 isWearingPack {type=boolean}
 isWearingPackBool {type=boolean}
 Melee {type=number}
 Melee_Cons {type=number}
 Melee_idle {type=number}
 moveBackward {type=number}
 moveForward {type=number}
 NewVar {type=number}
 NewVar_0 {type=number}
 Only_Pistol {type=boolean}
 Only_Rifle {type=boolean}
 OnlyClav {type=number}
 PickAxe {type=number}
 pistol {type=number}
 RBAN Settings Out {type=FortRigidBodyAnimNodeParameters}
 Rifle {type=number}
 Rifle_Only {type=number}
 Skeletal mesh {type=SkeletalMeshComponent}
 Skeletal_Mesh {type=SkeletalMeshComponent}
 Skydive {type=number}
 SkyDive_All {type=number}
 SkyDive_Dive_Hair {type=number}
 SkyDive_Float {type=number}
 SkyDive_Float_Para {type=number}
 SkyDive_Hair_Left {type=number}
 SkyDive_Hair_Right {type=number}
 Skydive_NotGround {type=number}
 Sprint {type=number}
 strafeLeft float {type=number}
 strafeRight float {type=number}
 Swim_Dyn {type=number}
 targetFloat {type=number}
]
