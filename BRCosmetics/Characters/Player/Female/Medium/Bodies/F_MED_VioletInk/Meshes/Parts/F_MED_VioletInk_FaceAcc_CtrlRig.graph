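# Control Rig for the VioletInk female character's face accessory part, correcting hair and hat bones by body pose.
# A sequence parent-constrains back hair chains, blending mostly toward spine_05.
# Crouch-aim bool branches use spherical pose readers on upperarm_r to weight right-hair offsets, plus fixed left-hair offsets.
# Pelvis height, curve-shaped and gated by CrouchWalk, weights hat offsets; left-hair constraints and a head DeltaRotation node sit unwired.
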ControlRigBlueprint /BRCosmetics/Characters/Player/Female/Medium/Bodies/F_MED_VioletInk/Meshes/Parts/F_MED_VioletInk_FaceAcc_CtrlRig
graphs [
 New Function_ContainedGraph {nodes [
   n1 = RigVMFunctionEntryNode()
   RigVMFunction_MathBoolToInteger = &RigVMFunction_MathBoolToInteger(Value=@n1.onlyTwist)
   RigVMFunction_MathQuaternionSwingTwist = &RigVMFunction_MathQuaternionSwingTwist(
    TwistAxis=@n1.twistAxis
    Input=&RigUnit_GetRelativeTransformForItem(bChildInitial=False, bParentInitial=False, Child=@n1.child, Parent=@n1.parent).RelativeTransform.Rotation
   )
   n2 = RigVMDispatchNode(
    Values=((),())
    Index=@RigVMFunction_MathBoolToInteger.Result
    Values.0=@RigVMFunction_MathQuaternionSwingTwist.Swing
    Values.1=@RigVMFunction_MathQuaternionSwingTwist.Twist
   ) {template="DISPATCH_RigVMDispatch_SelectInt32(in Index,in Values,out Result)"}
   RigVMFunction_MathQuaternionToEuler = &RigVMFunction_MathQuaternionToEuler(Value=@n2.Result, RotationOrder=@n1.Rotation Order)
   > DISPATCH_RigVMDispatch_Print = RigVMDispatchNode(ScreenColor=(R=1,G=1,B=1,A=1), ScreenDuration=1.5, Enabled=@n1.debugRotation, Value=@RigVMFunction_MathQuaternionToEuler.Result) {template="DISPATCH_RigVMDispatch_Print(in Prefix,in Value,in Enabled,in ScreenDuration,in ScreenColor)"}
   Cast = RigVMDispatchNode(Value=@n1.triggerAxis) {template="DISPATCH_RigVMDispatch_CastEnumToInt(in Value,out Result)"}
   DISPATCH_RigVMDispatch_SelectInt32 = RigVMDispatchNode(
    Values=(0,0,0)
    Index=@Cast.Result
    Values.0=@RigVMFunction_MathQuaternionToEuler.Result.X
    Values.1=@RigVMFunction_MathQuaternionToEuler.Result.Y
    Values.2=@RigVMFunction_MathQuaternionToEuler.Result.Z
   ) {template="DISPATCH_RigVMDispatch_SelectInt32(in Index,in Values,out Result)"}
   Remap = &RigVMFunction_MathDoubleRemap(
    bClamp=False
    Value=@DISPATCH_RigVMDispatch_SelectInt32.Result
    TargetMinimum=@n1.outputMinMax.X
    TargetMaximum=@n1.outputMinMax.Y
    SourceMinimum=@n1.rotationMinMax.X
    SourceMaximum=@n1.rotationMinMax.Y
   )
   Absolute = &RigVMFunction_MathDoubleAbs(Value=@Remap.Result)
   Multiply = &RigVMFunction_MathDoubleMul(
    A=&RigVMFunction_MathDoubleClamp(Value=RigVMDispatchNode(True=@Absolute.Result, Condition=@n1.absoluteValue, False=@Remap.Result) {template="DISPATCH_RigVMDispatch_If(in Condition,in True,in False,out Result)"}.Result).Result
    B=@n1.premult
   )
   > RigVMFunctionReturnNode(outputQuat=@n2.Result, outputWeight=@Multiply.Result)
  ]}
 RigVMFunctionLibrary {links [], nodes []}
 RigVMFunction_Sequence_1_ContainedGraph {nodes [
   n1 = RigVMFunctionEntryNode()
   RigVMFunction_Sequence = &RigVMFunction_Sequence(ExecuteContext=@n1.ExecuteContext)
   n2 = &RigVMFunction_Sequence(ExecuteContext=@RigVMFunction_Sequence.B)
   n3 = &RigVMFunction_Sequence(ExecuteContext=@n2.B)
   n4 = &RigVMFunction_Sequence(ExecuteContext=@n3.B)
   RigVMFunctionReturnNode(A=@RigVMFunction_Sequence.A, B=@n2.A, C=@n3.A, D=@n4.A, E=@n4.B)
  ]}
 RigVMModel {nodes [
   &RigUnit_ParentConstraint(
    AdvancedSettings=(InterpolationType=Average,RotationOrderForFilter=XZY)
    Child=(Type=Bone,Name="base_Hair_l_01")
    ExecutePin=()
    Parents=((Item=(Type=Bone,Name="spine_05"),Weight=0.5),(Item=(Type=Bone,Name="clavicle_l"),Weight=0.5))
    Weight=1
   )
   > &RigUnit_ParentConstraint(
    AdvancedSettings=(InterpolationType=Average,RotationOrderForFilter=XZY)
    Child=(Type=Bone,Name="dyn_Hair_l_02")
    Parents=((Item=(Type=Bone,Name="spine_05"),Weight=0.5),(Item=(Type=Bone,Name="base_Hair_l_01"),Weight=0.5))
    Weight=1
   )
   RigUnit_BeginExecution = &RigUnit_BeginExecution()
   RigVMFunction_Sequence = RigVMAggregateNode(ExecuteContext=@RigUnit_BeginExecution.ExecutePin)
   &RigUnit_ParentConstraint(
    AdvancedSettings=(InterpolationType=Average,RotationOrderForFilter=XZY)
    Child=(Type=Bone,Name="base_Hair_Back_01")
    Parents=((Item=(Type=Bone,Name="spine_05"),Weight=0.9),(Item=(Type=Bone,Name="head"),Weight=0.1))
    Weight=1
    ExecutePin=@RigVMFunction_Sequence.A
   )
   > &RigUnit_ParentConstraint(
    AdvancedSettings=(InterpolationType=Average,RotationOrderForFilter=XZY)
    Child=(Type=Bone,Name="dyn_Hair_Back_02")
    Parents=((Item=(Type=Bone,Name="spine_05"),Weight=0.8),(Item=(Type=Bone,Name="head"),Weight=0.2))
    Weight=1
   )
   > &RigUnit_ParentConstraint(
    AdvancedSettings=(InterpolationType=Average,RotationOrderForFilter=XZY)
    Child=(Type=Bone,Name="dyn_Hair_Back_03")
    Parents=((Item=(Type=Bone,Name="spine_05"),Weight=0.8),(Item=(Type=Bone,Name="head"),Weight=0.2))
    Weight=1
   )
   &RigUnit_ParentConstraint(
    AdvancedSettings=(InterpolationType=Average,RotationOrderForFilter=XZY)
    Child=(Type=Bone,Name="base_Hair_Back_r_01")
    Parents=((Item=(Type=Bone,Name="spine_05"),Weight=0.9),(Item=(Type=Bone,Name="base_Hair_Back_01"),Weight=0.1))
    Weight=1
    ExecutePin=@RigVMFunction_Sequence.B
   )
   > &RigUnit_ParentConstraint(
    AdvancedSettings=(InterpolationType=Average,RotationOrderForFilter=XZY)
    Child=(Type=Bone,Name="dyn_Hair_Back_r_02")
    Parents=((Item=(Type=Bone,Name="spine_05"),Weight=0.8),(Item=(Type=Bone,Name="dyn_Hair_Back_02"),Weight=0.2))
    Weight=1
   )
   > &RigUnit_ParentConstraint(
    AdvancedSettings=(InterpolationType=Average,RotationOrderForFilter=XZY)
    Child=(Type=Bone,Name="dyn_Hair_Back_r_03")
    Parents=((Item=(Type=Bone,Name="spine_05"),Weight=0.8),(Item=(Type=Bone,Name="dyn_Hair_Back_03"),Weight=0.2))
    Weight=1
   )
   &RigUnit_ParentConstraint(
    AdvancedSettings=(InterpolationType=Average,RotationOrderForFilter=XZY)
    Child=(Type=Bone,Name="base_Hair_Back_l_01")
    Parents=((Item=(Type=Bone,Name="spine_05"),Weight=0.9),(Item=(Type=Bone,Name="base_Hair_Back_01"),Weight=0.1))
    Weight=1
    ExecutePin=@RigVMFunction_Sequence.C
   )
   > &RigUnit_ParentConstraint(
    AdvancedSettings=(InterpolationType=Average,RotationOrderForFilter=XZY)
    Child=(Type=Bone,Name="dyn_Hair_Back_l_02")
    Parents=((Item=(Type=Bone,Name="spine_05"),Weight=0.8),(Item=(Type=Bone,Name="dyn_Hair_Back_02"),Weight=0.2))
    Weight=1
   )
   > n1 = &RigUnit_ParentConstraint(
    AdvancedSettings=(InterpolationType=Average,RotationOrderForFilter=XZY)
    Child=(Type=Bone,Name="dyn_Hair_Back_l_03")
    Parents=((Item=(Type=Bone,Name="spine_05"),Weight=0.8),(Item=(Type=Bone,Name="dyn_Hair_Back_03"),Weight=0.2))
    Weight=1
   )
   n2 = &RigVMFunction_ControlFlowBranch(ExecuteContext=@n1.ExecutePin, Condition=RigVMVariableNode() {variable=Crocuh_aim_not}.Value)
   n3 = &RigUnit_SphericalPoseReader(
    ActiveRegionScaleFactors=(PositiveWidth=1.2,NegativeWidth=1.5,PositiveHeight=1.5,NegativeHeight=1.2)
    ActiveRegionSize=0.1
    Debug=(bDrawDebug=True,bDraw2D=False,bDrawLocalAxes=False,DebugScale=25,DebugSegments=20,DebugThickness=0.25)
    DriverAxis=(X=-1,Y=0,Z=0)
    DriverItem=(Type=Bone,Name="upperarm_r")
    FalloffRegionScaleFactors=(PositiveWidth=1,NegativeWidth=5,PositiveHeight=5,NegativeHeight=1)
    FalloffSize=0.2
    FlipHeightScaling=False
    FlipWidthScaling=False
    OptionalParentItem=(Type=Bone,Name="clavicle_r")
    RotationOffset=(X=-117.1398,Y=-25.649,Z=30.3874)
    ExecutePin=@n2.True
   )
   > &RigUnit_ModifyTransforms(
    ItemToModify="((Item=(Type=Bone,Name=\"dyn_Hair_r_01\"),Transform=(Rotation=(X=-0.0983,Y=0.2991,Z=-0.031,W=-0.9486),Translation=(X=-0.8855,Y=-3.7174,Z=1.2646),Scale3D=(X=1,Y=1,Z=1))),(Item=(Type=Bone,Name=\"dyn_Hair_r_02\"),Transform=(Rotation=(X=0,Y=0,Z=0,W=1),Translation= ...(truncated)"
    WeightMinimum=0
    Weight=@n3.OutputParam
   )
   > &RigUnit_ModifyTransforms(
    ItemToModify="((Item=(Type=Bone,Name=\"dyn_Hair_l_01\"),Transform=(Rotation=(X=0,Y=0,Z=0,W=1),Translation=(X=0,Y=-1.7048,Z=4.2368),Scale3D=(X=1,Y=1,Z=1))),(Item=(Type=Bone,Name=\"dyn_Hair_l_02\"),Transform=(Rotation=(X=0,Y=0,Z=0,W=1),Translation=(X=0,Y=0,Z=0),Scale3D=(X=1,Y ...(truncated)"
    Weight=1
    WeightMinimum=0
   )
   RigVMFunction_ControlFlowBranch = &RigVMFunction_ControlFlowBranch(ExecuteContext=@RigVMFunction_Sequence.D, Condition=RigVMVariableNode() {variable=Crocuh_aim}.Value)
   n4 = &RigUnit_SphericalPoseReader(
    ActiveRegionScaleFactors=(PositiveWidth=1.2,NegativeWidth=1.5,PositiveHeight=1.5,NegativeHeight=1.2)
    ActiveRegionSize=0.1
    Debug=(bDrawDebug=true,bDraw2D=False,bDrawLocalAxes=False,DebugScale=25,DebugSegments=20,DebugThickness=0.25)
    DriverAxis=(X=-1,Y=0,Z=0)
    DriverItem=(Type=Bone,Name="upperarm_r")
    FalloffRegionScaleFactors=(PositiveWidth=1,NegativeWidth=3,PositiveHeight=3,NegativeHeight=1)
    FalloffSize=0.3
    FlipHeightScaling=False
    FlipWidthScaling=False
    OptionalParentItem=(Type=Bone,Name="clavicle_r")
    RotationOffset=(X=-127.4586,Y=-25.649,Z=60.4541)
    ExecutePin=@RigVMFunction_ControlFlowBranch.True
   )
   > &RigUnit_ModifyTransforms(
    ItemToModify="((Item=(Type=Bone,Name=\"dyn_Hair_r_01\"),Transform=(Rotation=(X=0.0999,Y=0.4508,Z=0.192,W=-0.866),Translation=(X=-3.9377,Y=-3.0727,Z=-2.5809),Scale3D=(X=1,Y=1,Z=1))),(Item=(Type=Bone,Name=\"dyn_Hair_r_02\"),Transform=(Rotation=(X=-0.0227,Y=-0.1285,Z=0.1722,W= ...(truncated)"
    WeightMinimum=0
    Weight=@n4.OutputParam
   )
   > &RigUnit_ModifyTransforms(
    ItemToModify="((Item=(Type=Bone,Name=\"dyn_Hair_l_01\"),Transform=(Rotation=(X=0,Y=-0.1305,Z=0,W=0.9914),Translation=(X=-0.8695,Y=0,Z=3.2452),Scale3D=(X=1,Y=1,Z=1))),(Item=(Type=Bone,Name=\"dyn_Hair_l_02\"),Transform=(Rotation=(X=0,Y=0,Z=0,W=1),Translation=(X=0,Y=0,Z=0),Sca ...(truncated)"
    Weight=1
    WeightMinimum=0
   )
   n5 = &RigVMFunction_AnimEvalRichCurve(
    Curve=(EditorCurveData=(Keys=((),(Time=0.5,Value=1),(Time=0.8))))
    SourceMaximum=2
    SourceMinimum=0
    TargetMinimum=0
    Value=&RigVMFunction_MathDoubleRemap(
     SourceMaximum=75
     SourceMinimum=45
     TargetMaximum=1
     TargetMinimum=0
     bClamp=false
     Value=&RigUnit_GetRelativeTransformForItem(Child=(Type=Bone,Name="pelvis"), Parent=(Type=Bone,Name="root"), bChildInitial=False, bParentInitial=False).RelativeTransform.Translation.Z
    ).Result
   )
   Multiply = &RigVMFunction_MathDoubleMul(B=@n5.Result, A=&RigVMFunction_MathBoolToFloat(Value=RigVMVariableNode() {variable=CrouchWalk}.Value).Result)
   &RigUnit_ModifyTransforms(
    ItemToModify="((Item=(Type=Bone,Name=\"dyn_hat_Back_01\"),Transform=(Rotation=(X=0,Y=0,Z=0,W=1),Translation=(X=0,Y=0,Z=0),Scale3D=(X=1,Y=1,Z=1))),(Item=(Type=Bone,Name=\"dyn_hat_01\"),Transform=(Rotation=(X=0,Y=-0.866,Z=0,W=0.5),Translation=(X=0,Y=0,Z=0),Scale3D=(X=1,Y=1,Z= ...(truncated)"
    WeightMinimum=0
    Weight=@Multiply.Result
    ExecutePin=@RigVMFunction_Sequence.E
   )
   DeltaRotation v4 = RigVMCollapseNode(
    ExecuteContext=()
    Rotation Order=ZYX
    child=(Type=Bone,Name="head")
    outputMinMax=(X=0,Y=1)
    parent=(Type=Null,Name="head_null")
    premult=1
    rotationMinMax=(X=-75,Y=75)
    triggerAxis=Y
    twistAxis=(X=1,Y=0,Z=0)
   )
   &RigUnit_ModifyTransforms(
    ExecutePin=()
    ItemToModify="((Item=(Type=Bone,Name=\"dyn_Hair_r_01\"),Transform=(Rotation=(X=0,Y=0,Z=-0.3827,W=0.9239),Translation=(X=0,Y=0,Z=100),Scale3D=(X=1,Y=1,Z=1))),(Item=(Type=Bone,Name=\"dyn_Hair_r_02\"),Transform=(Rotation=(X=0,Y=0,Z=-0.2588,W=0.9659),Translation=(X=0,Y=0,Z=0),S ...(truncated)"
    Weight=1
    WeightMinimum=0
   )
  ]}
]
hierarchy [
 root {loc=<0 0 0>, rot=<0 0 0 1>}
 pelvis {parent=root, loc=<0 -0.5301 98.0571>, rot=<0.0127 0.707 -0.0127 -0.707>}
 spine_01 {parent=pelvis, loc=<6.1208 -0.9328 0>, rot=<0 0 0.034 -0.9994>}
 spine_02 {parent=spine_01, loc=<6.5997 -0.2095 0>, rot=<0 0 0.0238 -0.9997>}
 spine_03 {parent=spine_02, loc=<6.4914 -0.1237 0>, rot=<0 0 0.0149 -0.9999>}
 spine_04 {parent=spine_03, loc=<6.9598 0.0794 0>, rot=<0 0 -0.0132 -0.9999>}
 spine_05 {parent=spine_04, loc=<6.9987 0 0>, rot=<0 0 -0.0015 -1>}
 neck_01 {parent=spine_05, loc=<11.3138 0.1529 0>, rot=<0 0 -0.2003 -0.9797>}
 neck_02 {parent=neck_01, loc=<4.5627 0 0>, rot=<0 0 0.0504 -0.9987>}
 head {parent=neck_02, loc=<5.3914 0 0>, rot=<0 0 0.115 -0.9934>}
 base_Hair_Front_01 {parent=head, loc=<9.6449 9.4177 3.2702>, rot=<0.6384 -0.2519 0.7269 0.0251>}
 dyn_Hair_Front_01 {parent=base_Hair_Front_01, loc=<0 0 0>, rot=<0 0 0 1>}
 dyn_Hair_Front_02 {parent=dyn_Hair_Front_01, loc=<6.1091 0 0>, rot=<0.0845 0.1617 -0.035 -0.9826>}
 base_Hair_Back_01 {parent=head, loc=<0.7046 -7.3333 1.5201>, rot=<0.191 -0.6808 0.6884 -0.1615>}
 dyn_Hair_Back_01 {parent=base_Hair_Back_01, loc=<0 0 0>, rot=<0 0 0 1>}
 dyn_Hair_Back_02 {parent=dyn_Hair_Back_01, loc=<9.4144 0 -0.4972>, rot=<0 -0.06 0 -0.9982>}
 dyn_Hair_Back_03 {parent=dyn_Hair_Back_02, loc=<9.1453 -0.1102 -0.653>, rot=<0.0895 0 0 -0.996>}
 base_Hair_Back_a_01 {parent=dyn_Hair_Back_03, loc=<11.2843 3.9032 3.8351>, rot=<0.0085 -0.1906 -0.2672 -0.9446>}
 dyn_Hair_Back_a_01 {parent=base_Hair_Back_a_01, loc=<0 0 0>, rot=<0 0 0 1>}
 dyn_Hair_Back_a_02 {parent=dyn_Hair_Back_a_01, loc=<10.8891 0 0>, rot=<0 -0.2714 0.2714 -0.9234>}
 dyn_Hair_Back_a_03 {parent=dyn_Hair_Back_a_02, loc=<10.9042 0 0>, rot=<0 0 0 1>}
 base_Hair_Back_b_01 {parent=dyn_Hair_Back_03, loc=<14.4441 -1.2879 -3.7162>, rot=<-0.1142 -0.0846 -0.1356 -0.9805>}
 dyn_Hair_Back_b_01 {parent=base_Hair_Back_b_01, loc=<0 0 0>, rot=<0 0 0 1>}
 dyn_Hair_Back_b_02 {parent=dyn_Hair_Back_b_01, loc=<9.8933 0 0>, rot=<-0.062 0.1847 -0.3123 -0.9298>}
 base_Hair_Back_r_01 {parent=head, loc=<2.2127 -4.6859 7.4284>, rot=<0.1763 -0.6848 0.6848 -0.1763>}
 dyn_Hair_Back_r_01 {parent=base_Hair_Back_r_01, loc=<0 0 0>, rot=<0 0 0 1>}
 dyn_Hair_Back_r_02 {parent=dyn_Hair_Back_r_01, loc=<9.4144 0 -0.4972>, rot=<0 0.0321 0 -0.9995>}
 dyn_Hair_Back_r_03 {parent=dyn_Hair_Back_r_02, loc=<9.1453 -0.1102 -0.653>, rot=<0.0895 0 0 -0.996>}
 dyn_Hair_Back_r_04 {parent=dyn_Hair_Back_r_03, loc=<6.0746 0.6349 -4.3148>, rot=<0 0 0.0243 -0.9997>}
 base_Hair_Back_a_r_01 {parent=dyn_Hair_Back_r_04, loc=<4.0186 -2.0884 -3.3267>, rot=<-0.1328 -0.3363 0.3165 -0.877>}
 dyn_Hair_Back_a_r_01 {parent=base_Hair_Back_a_r_01, loc=<0 0 0>, rot=<0 0 0 1>}
 dyn_Hair_Back_c_r_01 {parent=dyn_Hair_Back_a_r_01, loc=<5.5613 -4.0655 0.6853>, rot=<-0.0525 0.1224 0.3879 -0.912>}
 dyn_Hair_Back_b_r_01 {parent=dyn_Hair_Back_a_r_01, loc=<12.2052 5.2039 0.3365>, rot=<-0.0509 0.1187 -0.5998 -0.7897>}
 dyn_Hair_Back_d_r_01 {parent=dyn_Hair_Back_r_02, loc=<4.4045 -4.7776 -2.8023>, rot=<-0.2872 0.6598 0.4785 -0.5032>}
 base_Hair_Back_l_01 {parent=head, loc=<0.9888 -5.2206 -5.2919>, rot=<0.1155 -0.7013 0.6865 -0.1537>}
 dyn_Hair_Back_l_01 {parent=base_Hair_Back_l_01, loc=<0 0 0>, rot=<0 0 0 1>}
 dyn_Hair_Back_l_02 {parent=dyn_Hair_Back_l_01, loc=<7.7051 -0.0387 -0.5557>, rot=<0.0002 0.0171 -0.0113 -0.9998>}
 dyn_Hair_Back_l_03 {parent=dyn_Hair_Back_l_02, loc=<9.1453 -0.1102 -0.653>, rot=<0.0894 -0.0519 -0.0047 -0.9946>}
 base_Hair_Back_b_l_01 {parent=dyn_Hair_Back_l_03, loc=<3.2802 1.5407 -0.1252>, rot=<0.4605 0.326 -0.3396 -0.7526>}
 dyn_Hair_Back_b_l_01 {parent=base_Hair_Back_b_l_01, loc=<0 0 0>, rot=<0 0 0 1>}
 dyn_Hair_Back_b_l_02 {parent=dyn_Hair_Back_b_l_01, loc=<7.752 0 0>, rot=<0 0.2356 0 -0.9718>}
 base_Hair_Back_c_l_01 {parent=dyn_Hair_Back_l_03, loc=<9.7094 0.2216 -0.3218>, rot=<0.1594 0.0527 -0.3635 -0.9164>}
 dyn_Hair_Back_c_l_01 {parent=base_Hair_Back_c_l_01, loc=<0 0 0>, rot=<0 0 0 1>}
 dyn_Hair_Back_c_l_02 {parent=dyn_Hair_Back_c_l_01, loc=<9.2457 0 0>, rot=<0 0 0.3864 -0.9223>}
 dyn_Hair_Back_c_l_03 {parent=dyn_Hair_Back_c_l_02, loc=<9.0351 0 0>, rot=<0 0 0 1>}
 base_Hair_Back_a_l_01 {parent=dyn_Hair_Back_l_01, loc=<1.71 2.061 -2.42>, rot=<0.3224 -0.1253 0.061 -0.9363>}
 dyn_Hair_Back_a_l_01 {parent=base_Hair_Back_a_l_01, loc=<0 0 0>, rot=<0 0 0 1>}
 dyn_Hair_Back_a_l_02 {parent=dyn_Hair_Back_a_l_01, loc=<6.2675 0.3161 0.5473>, rot=<0 0.4592 0 -0.8883>}
 dyn_Hair_Back_a_l_03 {parent=dyn_Hair_Back_a_l_02, loc=<6.3158 0 0>, rot=<0.0946 0.4714 -0.1205 -0.8685>}
 base_hat_01 {parent=head, loc=<10.7632 -12.1789 0>, rot=<0.0566 -0.7048 0.7048 -0.0566>}
 dyn_hat_01 {parent=base_hat_01, loc=<0 0 0>, rot=<0 0 0 1>}
 dyn_hat_02 {parent=dyn_hat_01, loc=<4.4116 0 0>, rot=<0 0 0 1>}
 head_null {parent=head, type=Null, loc=<0 0 0>, rot=<0 0 0 1>}
 base_hat_Back_01 {parent=head, loc=<3.4753 -9.4287 0>, rot=<0.0029 -0.7071 -0.0029 -0.7071>}
 dyn_hat_Back_01 {parent=base_hat_Back_01, loc=<0 0 0>, rot=<0 0 0 1>}
 clavicle_l {parent=spine_05, loc=<5.4831 -0.1053 -3.2946>, rot=<0.0465 -0.7289 -0.009 -0.6829>}
 upperarm_l {parent=clavicle_l, loc=<9.8658 -1.2136 -0.4105>, rot=<0.0164 -0.3245 -0.0269 -0.9454>}
 upperarm_twist_01_l {parent=upperarm_l, loc=<7.2135 -0.1134 0>, rot=<0 0 0 1>}
 clavicle_r {parent=spine_05, loc=<5.4831 -0.1053 3.2946>, rot=<0.7289 0.0465 -0.6829 0.009>}
 upperarm_r {parent=clavicle_r, loc=<-9.8658 1.2136 0.4105>, rot=<0.0164 -0.3245 -0.0269 -0.9454>}
 upperarm_twist_01_r {parent=upperarm_r, loc=<-7.2135 0.1134 0>, rot=<0 0 0 -1>}
 base_Hair_l_01 {parent=spine_05, loc=<14.6205 3.1905 -9.1719>, rot=<-0.0336 0.4343 0.8951 0.095>}
 dyn_Hair_l_01 {parent=base_Hair_l_01, loc=<0 0 0>, rot=<0 0 0 1>}
 dyn_Hair_l_02 {parent=dyn_Hair_l_01, loc=<7.3206 0 0>, rot=<0 0.3897 0 -0.921>}
 dyn_Hair_l_03 {parent=dyn_Hair_l_02, loc=<7.0903 0 0>, rot=<0 0 0 1>}
 base_Hair_r_01 {parent=spine_05, loc=<13.9571 5.0284 8.8114>, rot=<-0.1195 -0.932 -0.335 0.0695>}
 dyn_Hair_r_01 {parent=base_Hair_r_01, loc=<0 0 0>, rot=<0 0 0 1>}
 dyn_Hair_r_02 {parent=dyn_Hair_r_01, loc=<7.0191 0 0>, rot=<0 0.3112 0 -0.9503>}
 dyn_Hair_r_03 {parent=dyn_Hair_r_02, loc=<4.7922 0 0>, rot=<0 0 0 1>}
]
variables [
 Crocuh_aim {type=bool, default=False}
 Crocuh_aim_not {type=bool, default=False}
 CrouchWalk {type=bool, default=False}
]
