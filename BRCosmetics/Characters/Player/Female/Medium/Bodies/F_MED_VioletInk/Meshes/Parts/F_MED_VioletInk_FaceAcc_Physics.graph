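# Physics asset for the FaceAcc part of the VioletInk female medium player body, covering the dynamic back-hair bones.
# All bodies are capsules, with the spine and left clavicle forming the torso around the hair.
# Branching center, left and right hair chains trace back to spine_05 through narrow swing limits and zero-degree twist limits.
# The spine_05–spine_03 and left clavicle constraints are free on all angular axes.
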PhysicsAsset /BRCosmetics/Characters/Player/Female/Medium/Bodies/F_MED_VioletInk/Meshes/Parts/F_MED_VioletInk_FaceAcc_Physics
bodies {
 clavicle_l [{
   shapeType=Capsule
   length=18.2681
   radius=6.71
   shapeName=clavicle_l_capsule
   center=<6.1834 -0.2307 -2.632>
   rotation=<rot -80 179.9985 -174.9987>
  }]
 dyn_Hair_Back_02 [{
   shapeType=Capsule
   length=5.6663
   radius=6.7962
   shapeName=dyn_Hair_Back_01_capsule
   center=<8.2939 0.8077 1.7176>
   rotation=<rot 0 95 89.9994>
  }]
 dyn_Hair_Back_03 [{
   shapeType=Capsule
   length=7.1147
   radius=6.6534
   shapeName=dyn_Hair_Back_03_capsule
   center=<6.656 1.1495 1.4379>
   rotation=<rot 13.6012 89.6564 75.1509>
  }]
 dyn_Hair_Back_a_01 [{
   shapeType=Capsule
   length=7.5132
   radius=4.0904
   shapeName=dyn_Hair_Back_a_01_capsule
   center=<3.79 0.379 0.3297>
   rotation=<rot 0 96.7249 -85.1149>
  }]
 dyn_Hair_Back_a_02 [{
   shapeType=Capsule
   length=7.5132
   radius=4.0904
   shapeName=dyn_Hair_Back_a_01_capsule
   center=<3.79 0.379 0.3297>
   rotation=<rot 0 96.7249 -85.1149>
  }]
 dyn_Hair_Back_a_03 [{
   shapeType=Capsule
   length=8.5132
   radius=3.0904
   shapeName=dyn_Hair_Back_a_01_capsule
   center=<1.876 -1.5649 -0.991>
   rotation=<rot 14.2281 48.159 -110.7212>
  }]
 dyn_Hair_Back_a_l_01 [{
   shapeType=Capsule
   length=6.0628
   radius=2.7261
   shapeName=dyn_Hair_Back_a_l_01_capsule
   center=<4.4192 0.0571 -0.1421>
   rotation=<rot 0 -88.1476 -107.5723>
  }]
 dyn_Hair_Back_a_l_02 [{
   shapeType=Capsule
   length=3.8162
   radius=2.7599
   shapeName=dyn_Hair_Back_a_l_02_capsule
   center=<5.9358 -0.3614 0.3264>
   rotation=<rot 0.0002 86.547 -45.4533>
  }]
 dyn_Hair_Back_a_r_01 [{
   shapeType=Capsule
   length=5.0035
   radius=4.3623
   shapeName=dyn_Hair_Back_a_r_01_capsule
   center=<3.7045 0.9455 -0.1044>
   rotation=<rot 0 -74.4719 -98.2061>
  }]
 dyn_Hair_Back_b_01 [{
   shapeType=Capsule
   length=8.3032
   radius=3.4946
   shapeName=dyn_Hair_Back_b_01_capsule
   center=<3.8802 -0.2623 -0.2866>
   rotation=<rot 0 -90.5201 -89.1865>
  }]
 dyn_Hair_Back_b_02 [{
   shapeType=Capsule
   length=6.8821
   radius=3.3132
   shapeName=dyn_Hair_Back_b_02_capsule
   center=<4.5564 1.6941 1.84>
   rotation=<rot 2.3737 119.2644 -76.4055>
  }]
 dyn_Hair_Back_b_l_01 [{
   shapeType=Capsule
   length=5.216
   radius=2.8381
   shapeName=dyn_Hair_Back_b_l_01_capsule
   center=<1.1384 0.0012 0.0501>
   rotation=<rot 0 -85.9919 -69.9426>
  }]
 dyn_Hair_Back_b_l_02 [{
   shapeType=Capsule
   length=4.3138
   radius=2.7819
   shapeName=dyn_Hair_Back_b_l_02_capsule
   center=<0.1032 -0.3193 1.0633>
   rotation=<rot -16.4001 -90 51.3255>
  }]
 dyn_Hair_Back_b_r_01 [{
   shapeType=Capsule
   length=6.318
   radius=3.736
   shapeName=dyn_Hair_Back_c_r_01_capsule
   center=<1.6126 1.5843 0>
   rotation=<rot -90 19.9997 0>
  }]
 dyn_Hair_Back_c_l_01 [{
   shapeType=Capsule
   length=7.1719
   radius=3.9876
   shapeName=dyn_Hair_Back_c_l_01_capsule
   center=<1.3009 0.2605 1.1387>
   rotation=<rot 0 -87.5198 -89.8692>
  }]
 dyn_Hair_Back_c_l_02 [{
   shapeType=Capsule
   length=5.6719
   radius=3.6126
   shapeName=dyn_Hair_Back_c_l_01_capsule
   center=<1.2998 0.2604 0.6448>
   rotation=<rot 0 -87.5198 -89.8692>
  }]
 dyn_Hair_Back_c_l_03 [{
   shapeType=Capsule
   length=5.6719
   radius=3.6126
   shapeName=dyn_Hair_Back_c_l_01_capsule
   center=<1.6049 -1.4131 0.5103>
   rotation=<rot 0.123 -157.5197 -89.9552>
  }]
 dyn_Hair_Back_c_r_01 [{
   shapeType=Capsule
   length=6.318
   radius=3.736
   shapeName=dyn_Hair_Back_c_r_01_capsule
   center=<1.8232 -1.407 0.5845>
   rotation=<rot -5.1319 52.3816 -75.7152>
  }]
 dyn_Hair_Back_l_02 [{
   shapeType=Capsule
   length=4.6663
   radius=6.7962
   shapeName=dyn_Hair_Back_01_capsule
   center=<10.6002 0.8077 3.4509>
   rotation=<rot 0 90 89.9995>
  }]
 dyn_Hair_Back_r_02 [{
   shapeType=Capsule
   length=6.2913
   radius=6.7962
   shapeName=dyn_Hair_Back_01_capsule
   center=<11.9621 0.8077 0.9221>
   rotation=<rot 0 90 74.9993>
  }]
 dyn_Hair_Back_r_04 [{
   shapeType=Capsule
   length=5.5
   radius=5.005
   shapeName=dyn_Hair_Back_r_04_capsule
   center=<-0.2969 -0.3559 -0.1957>
   rotation=<rot 55.4435 -6.01 14.7642>
  }]
 spine_01 [{shapeType=Capsule, length=20.13, radius=13.88, shapeName=spine_01_capsule, center=<-3.1824 -0.2348 0>, rotation=<rot 0 -90 0>}]
 spine_03 [{shapeType=Capsule, length=22.63, radius=11.38, shapeName=spine_01_capsule, center=<-2.2296 2.3893 0>, rotation=<rot 0 -90 0>}]
 spine_05 [{shapeType=Capsule, length=26.505, radius=7.88, shapeName=spine_01_capsule, center=<0 2.0442 0>, rotation=<rot 0 -90 0>}]
}
constraints [
 {bone1Name=spine_03, bone2Name=spine_01, swing1LimitDegrees=2, swing2LimitDegrees=3, twistLimitDegrees=45}
 {
  bone1Name=spine_05
  bone2Name=spine_03
  swing1LimitDegrees=45
  swing1Motion=Free
  swing2LimitDegrees=45
  swing2Motion=Free
  twistLimitDegrees=45
  twistMotion=Free
 }
 {bone1Name=dyn_Hair_Back_02, bone2Name=spine_05, swing1LimitDegrees=5, swing2LimitDegrees=10, twistLimitDegrees=0}
 {bone1Name=dyn_Hair_Back_l_02, bone2Name=spine_05, swing1LimitDegrees=5, swing2LimitDegrees=10, twistLimitDegrees=0}
 {bone1Name=dyn_Hair_Back_r_02, bone2Name=spine_05, swing1LimitDegrees=5, swing2LimitDegrees=10, twistLimitDegrees=0}
 {bone1Name=dyn_Hair_Back_03, bone2Name=dyn_Hair_Back_02, swing1LimitDegrees=1, swing2LimitDegrees=2, twistLimitDegrees=0}
 {bone1Name=dyn_Hair_Back_a_01, bone2Name=dyn_Hair_Back_03, swing1LimitDegrees=2, swing2LimitDegrees=4, twistLimitDegrees=0}
 {bone1Name=dyn_Hair_Back_a_02, bone2Name=dyn_Hair_Back_a_01, swing1LimitDegrees=1, swing2LimitDegrees=5, twistLimitDegrees=0}
 {bone1Name=dyn_Hair_Back_a_03, bone2Name=dyn_Hair_Back_a_02, swing1LimitDegrees=2, swing2LimitDegrees=4, twistLimitDegrees=0}
 {bone1Name=dyn_Hair_Back_b_01, bone2Name=dyn_Hair_Back_03, swing1LimitDegrees=2, swing2LimitDegrees=4, twistLimitDegrees=0}
 {bone1Name=dyn_Hair_Back_b_02, bone2Name=dyn_Hair_Back_b_01, swing1LimitDegrees=1, swing2LimitDegrees=2, twistLimitDegrees=0}
 {bone1Name=dyn_Hair_Back_a_l_01, bone2Name=spine_05, swing1LimitDegrees=2, swing2LimitDegrees=4, twistLimitDegrees=0}
 {bone1Name=dyn_Hair_Back_a_l_02, bone2Name=dyn_Hair_Back_a_l_01, swing1LimitDegrees=2, swing2LimitDegrees=5, twistLimitDegrees=0}
 {
  bone1Name=clavicle_l
  bone2Name=spine_05
  swing1LimitDegrees=45
  swing1Motion=Free
  swing2LimitDegrees=45
  swing2Motion=Free
  twistLimitDegrees=45
  twistMotion=Free
 }
 {bone1Name=dyn_Hair_Back_b_l_01, bone2Name=dyn_Hair_Back_l_02, swing1LimitDegrees=1, swing2LimitDegrees=5, twistLimitDegrees=0}
 {bone1Name=dyn_Hair_Back_b_l_02, bone2Name=dyn_Hair_Back_b_l_01, swing1LimitDegrees=1, swing2LimitDegrees=5, twistLimitDegrees=0}
 {bone1Name=dyn_Hair_Back_r_04, bone2Name=dyn_Hair_Back_r_02, swing1LimitDegrees=2, swing2LimitDegrees=5, twistLimitDegrees=0}
 {bone1Name=dyn_Hair_Back_b_r_01, bone2Name=dyn_Hair_Back_a_r_01, swing1LimitDegrees=2, swing2LimitDegrees=2, twistLimitDegrees=0}
 {bone1Name=dyn_Hair_Back_c_r_01, bone2Name=dyn_Hair_Back_a_r_01, swing1LimitDegrees=2, swing2LimitDegrees=2, twistLimitDegrees=0}
 {bone1Name=dyn_Hair_Back_a_r_01, bone2Name=dyn_Hair_Back_r_04, swing1LimitDegrees=5, swing2LimitDegrees=5, twistLimitDegrees=0}
 {bone1Name=dyn_Hair_Back_c_l_01, bone2Name=dyn_Hair_Back_l_02, swing1LimitDegrees=3, swing2LimitDegrees=5, twistLimitDegrees=0}
 {bone1Name=dyn_Hair_Back_c_l_02, bone2Name=dyn_Hair_Back_c_l_01, swing1LimitDegrees=2, swing2LimitDegrees=4, twistLimitDegrees=0}
 {bone1Name=dyn_Hair_Back_c_l_03, bone2Name=dyn_Hair_Back_c_l_02, swing1LimitDegrees=2, swing2LimitDegrees=4, twistLimitDegrees=0}
]
