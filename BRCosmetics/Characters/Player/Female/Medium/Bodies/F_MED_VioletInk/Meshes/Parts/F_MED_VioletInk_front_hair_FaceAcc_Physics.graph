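# Physics asset for the front hair face-accessory mesh of the VioletInk medium female player body, defining capsule bodies for hair physics.
# Left and right three-bone dyn_Hair chains use identical capsules, with narrow swing limits and zero twist limits.
# Both chains attach to spine_05 rather than the head.
# Head, upper spine, clavicle and upper-arm capsules form the surrounding body, joined by constraints with free angular motion.
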PhysicsAsset /BRCosmetics/Characters/Player/Female/Medium/Bodies/F_MED_VioletInk/Meshes/Parts/F_MED_VioletInk_front_hair_FaceAcc_Physics
bodies {
 clavicle_l [{
   shapeType=Capsule
   length=19.8931
   radius=5.085
   shapeName=clavicle_l_capsule
   center=<6.1834 -0.2307 -2.632>
   rotation=<rot -80 179.9985 -174.9987>
  }]
 clavicle_r [{
   shapeType=Capsule
   length=20.0181
   radius=4.96
   shapeName=clavicle_r_capsule
   center=<-6.1834 0.2307 2.632>
   rotation=<rot -80 179.9985 5.0013>
  }]
 dyn_Hair_l_01 [{
   shapeType=Capsule
   length=4.0212
   radius=2.7264
   shapeName=dyn_Hair_l_01_capsule
   center=<2.0863 0.049 -0.4234>
   rotation=<rot -2.0686 -91.0827 -93.8539>
  }]
 dyn_Hair_l_02 [{
   shapeType=Capsule
   length=4.0371
   radius=2.5664
   shapeName=dyn_Hair_l_02_capsule
   center=<2.2967 -0.1118 -0.6131>
   rotation=<rot -0.834 -91.6469 -80.4218>
  }]
 dyn_Hair_l_03 [{
   shapeType=Capsule
   length=4.0241
   radius=1.8916
   shapeName=dyn_Hair_l_03_capsule
   center=<1.6266 0.1597 1.2155>
   rotation=<rot 3.0204 102.8061 -52.9055>
  }]
 dyn_Hair_r_01 [{
   shapeType=Capsule
   length=4.0212
   radius=2.7264
   shapeName=dyn_Hair_l_01_capsule
   center=<2.0863 0.049 -0.4234>
   rotation=<rot -2.0686 -91.0827 -93.8539>
  }]
 dyn_Hair_r_02 [{
   shapeType=Capsule
   length=4.0371
   radius=2.5664
   shapeName=dyn_Hair_l_02_capsule
   center=<2.2967 -0.1118 -0.6131>
   rotation=<rot -0.834 -91.6469 -80.4218>
  }]
 dyn_Hair_r_03 [{
   shapeType=Capsule
   length=4.0241
   radius=1.8916
   shapeName=dyn_Hair_l_03_capsule
   center=<1.6266 0.1597 1.2155>
   rotation=<rot 3.0204 102.8061 -52.9055>
  }]
 head [{
   shapeType=Capsule
   length=22.0285
   radius=6.8472
   shapeName=head_capsule
   center=<-8.695 0.9009 0.4185>
   rotation=<rot -0.819 -99.1102 -88.5812>
  }]
 spine_05 [{shapeType=Capsule, length=24.13, radius=10.255, shapeName=spine_01_capsule, center=<-3.9165 -2.8656 0>, rotation=<rot 0 -90 0>}]
 upperarm_l [{shapeType=Capsule, length=33.7477, radius=5.4519, shapeName=upperarm_l_capsule, center=<4.4318 -1.8669 -0.576>, rotation=<rot 90 5.0003 0>}]
 upperarm_r [{shapeType=Capsule, length=33.7477, radius=5.4519, shapeName=upperarm_r_capsule, center=<-4.4318 1.8669 0.576>, rotation=<rot 90 -174.9997 0>}]
}
constraints [
 {
  bone1Name=clavicle_l
  bone2Name=spine_05
  swing1LimitDegrees=45
  swing1Motion=Free
  swing2LimitDegrees=45
  swing2Motion=Free
  twistLimitDegrees=45
  twistMotion=Free
 }
 {
  bone1Name=clavicle_r
  bone2Name=spine_05
  swing1LimitDegrees=45
  swing1Motion=Free
  swing2LimitDegrees=45
  swing2Motion=Free
  twistLimitDegrees=45
  twistMotion=Free
 }
 {
  bone1Name=head
  bone2Name=spine_05
  swing1LimitDegrees=45
  swing1Motion=Free
  swing2LimitDegrees=45
  swing2Motion=Free
  twistLimitDegrees=45
  twistMotion=Free
 }
 {bone1Name=dyn_Hair_l_01, bone2Name=spine_05, swing1LimitDegrees=2, swing2LimitDegrees=5, twistLimitDegrees=0}
 {bone1Name=dyn_Hair_l_02, bone2Name=dyn_Hair_l_01, swing1LimitDegrees=2, swing2LimitDegrees=4, twistLimitDegrees=0}
 {bone1Name=dyn_Hair_l_03, bone2Name=dyn_Hair_l_02, swing1LimitDegrees=2, swing2LimitDegrees=3, twistLimitDegrees=0}
 {bone1Name=dyn_Hair_r_01, bone2Name=spine_05, swing1LimitDegrees=2, swing2LimitDegrees=5, twistLimitDegrees=0}
 {bone1Name=dyn_Hair_r_02, bone2Name=dyn_Hair_r_01, swing1LimitDegrees=2, swing2LimitDegrees=4, twistLimitDegrees=0}
 {bone1Name=dyn_Hair_r_03, bone2Name=dyn_Hair_r_02, swing1LimitDegrees=2, swing2LimitDegrees=3, twistLimitDegrees=0}
 {
  bone1Name=upperarm_l
  bone2Name=clavicle_l
  swing1LimitDegrees=45
  swing1Motion=Free
  swing2LimitDegrees=45
  swing2Motion=Free
  twistLimitDegrees=45
  twistMotion=Free
 }
 {
  bone1Name=upperarm_r
  bone2Name=clavicle_r
  swing1LimitDegrees=45
  swing1Motion=Free
  swing2LimitDegrees=45
  swing2Motion=Free
  twistLimitDegrees=45
  twistMotion=Free
 }
]
